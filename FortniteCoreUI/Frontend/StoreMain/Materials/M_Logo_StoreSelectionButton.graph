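# UI-domain translucent material for a logo on a store selection button in the Fortnite frontend.
# A single texture sample reads from a texture object parameter named Logo, which defaults to the Epic Games logo.
# Exposing the texture as a parameter lets material instances swap in other logos without changing the graph.
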Material /FortniteCoreUI/Frontend/StoreMain/Materials/M_Logo_StoreSelectionButton
graph {
 properties {
  bUsedWithStaticMesh=false
  blendMode=BLEND_Translucent
  materialDomain=MD_UI
  pixelDepthOffsetMode=PDOM_Legacy
  usedShadingModels=Default Lit
 }
 nodes [TextureSample(Tex=TextureObjectParameter() {parameterName=Logo, texture=/FortniteCoreUI/Frontend/StoreMain/Materials/Textures/T_Logo_EpicGames}) {texture=&T_Logo_EpicGames}]
}
