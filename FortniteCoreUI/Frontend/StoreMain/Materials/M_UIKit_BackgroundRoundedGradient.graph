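# UI material for a frontend store UIKit background: a translucent radial gradient with rounded corners.
# An exponential radial gradient, with adjustable center, radius and density over scaled UVs, blends two RGBA parameters into color and alpha.
# The blended alpha is multiplied by a CoreUI rounded-corner mask and the opacity from the shared modular button-state lerp function.
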Material /FortniteCoreUI/Frontend/StoreMain/Materials/M_UIKit_BackgroundRoundedGradient
graph {
 properties {
  bUsedWithStaticMesh=false
  blendMode=BLEND_Translucent
  materialDomain=MD_UI
  pixelDepthOffsetMode=PDOM_Legacy
  usedShadingModels=Default Lit
 }
 nodes [
  n1 = VectorParameter() {group=Color, parameterName=RGB_A, defaultValue=<rgba 0.2139 0 0.2812 1>}
  n2 = VectorParameter() {group=Color, parameterName=RGB_B, defaultValue=<rgba 1 0 0.9429 1>}
  n3 = MaterialFunctionCall(
   UVs=MaterialFunctionCall() {materialFunction=&MF_UI_UVScaling}.ScaleX
   CenterPosition=MaterialFunctionCall(X=ScalarParameter() {defaultValue=0.5, parameterName=Center_X}, Y=ScalarParameter() {defaultValue=0.5, parameterName=Center_Y}) {materialFunction=&MakeFloat2}.Result
   Radius=ScalarParameter() {defaultValue=0.5, parameterName=Radius}
   Density=ScalarParameter() {defaultValue=2, parameterName=Density}
  ) {materialFunction=&RadialGradientExponential}
  n4 = VectorParameter() {group=Corners, parameterName=Corner Radius, sortPriority=2, defaultValue=<rgba 0.25 0.25 0.25 0.25>}
  n5 = AppendVector(A=@n4.RGB, B=@n4.RGB)
  n6 = ScalarParameter() {group=Corners, parameterName=Radius in Pixels, sliderMax=1, sortPriority=0}
  Multiply(
   A=LinearInterpolate(A=@n1.A, B=@n2.A, Alpha=@n3.RadialGradientExponential)
   B=Multiply(
    A=MaterialFunctionCall!(Idle, Idle Opacity, Hovered, Hovered Opacity, Focused, Focused Opacity, Pressed, Pressed Opacity, Disabled, Disabled Opacity) {materialFunction=/FortUILibrary/Systems/ModularButton/MaterialFunctions/MF_UIKit_ButtonStatesBase_Lerp}.Opacity
    B=MaterialFunctionCall(Corner Radius=@n5, Radius in Pixels=@n6) {materialFunction=&MF_UI_CoreUI_RoundedMask}.Shape Mask
   )
  )
  Reroute(None=LinearInterpolate(A=@n1.RGB, B=@n2.RGB, Alpha=@n3.RadialGradientExponential))
 ]
}
