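# M_UIKit_StoreSelection is a translucent UI material for a rounded selection highlight in the frontend store.
# MF_UI_CoreUI_RoundedMask supplies shape and outline masks with a fixed corner radius, its outline thickness switched by the HasOutline parameter.
# A light gray is maxed with the outline mask to brighten the border.
# The Opacity parameter lowers the fill by subtraction from the shape mask, then a Max keeps the outline at full strength.
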Material /FortniteCoreUI/Frontend/StoreMain/Materials/M_UIKit_StoreSelection
graph {
 properties {
  bAutomaticallySetUsageInEditor=false
  bUsedWithStaticMesh=false
  blendMode=BLEND_Translucent
  materialDomain=MD_UI
  usedShadingModels=Default Lit
 }
 nodes [
  n1 = MaterialFunctionCall(
   Corner Radius=Constant() {r=0.08}
   Outline Thickness=LinearInterpolate(B=ScalarParameter() {defaultValue=10, parameterName=OutlineThickness}, Alpha=ScalarParameter() {defaultValue=1, parameterName=HasOutline}) {constB=0}
  ) {materialFunction=&MF_UI_CoreUI_RoundedMask}
  Max(A=Constant4Vector() {constant=<rgba 0.8333 0.8333 0.8333 0>}.RGBA, B=@n1.Outline Mask)
  n2 = ScalarParameter() {defaultValue=0.25, parameterName=Opacity}
  Max(A=Subtract(A=@n1.Shape Mask, B=OneMinus(None=@n2)) {constB=0.75}, B=@n1.Outline Mask)
 ]
}
