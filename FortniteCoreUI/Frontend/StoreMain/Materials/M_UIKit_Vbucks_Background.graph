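# Translucent UI material for a V-Bucks background in the store frontend, drawing an orange-gold backdrop that reacts to hover.
# An eased Hover parameter drives default and hover gradient variants and slides five procedural triangles, staggered by differing SmoothStep thresholds.
# Top and bottom triangle groups are tinted and composited with the gradient, an exponential radial glow and an optional vignette.
# Fitted UVs position the shapes, and a rounded-mask function handles the corners.
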Material /FortniteCoreUI/Frontend/StoreMain/Materials/M_UIKit_Vbucks_Background
graph {
 properties {
  bAutomaticallySetUsageInEditor=false
  bUsedWithStaticMesh=false
  blendMode=BLEND_Translucent
  materialDomain=MD_UI
  usedShadingModels=Default Lit
 }
 nodes [
  n1 = NamedRerouteUsage()
  n2 = NamedRerouteUsage()
  n3 = NamedRerouteUsage()
  n4 = ScalarParameter() {group=Vignette, parameterName=VignetteOpacity}
  LinearInterpolate(
   A=LinearInterpolate(
    A=LinearInterpolate(A=@n1, Alpha=@n2)
    B=VectorParameter() {group=Vignette, parameterName=VignetteColor, defaultValue=<rgba 0.0417 0.0098 0 1>}.RGB
    Alpha=Multiply(A=@n3, B=@n4) {constB=0}
   )
   Alpha=LinearInterpolate(A=NamedRerouteUsage(), Alpha=NamedRerouteUsage())
  )
  n5 = ScalarParameter() {defaultValue=20, group=Corners, parameterName=CornerRadius}
  MaterialFunctionCall(
   Corner Radius=MaterialFunctionCall(R=@n5, G=@n5, B=@n5, A=@n5) {materialFunction=&AppendMany}.RGBA
   Radius in Pixels=ScalarParameter() {defaultValue=1, group=Corners, parameterName=Radius in Pixels, sliderMax=1, sortPriority=0}
  ) {materialFunction=&MF_UI_CoreUI_RoundedMask}
  MaterialFunctionCall() {materialFunction=&DebugTimeSine}
  n6 = VectorParameter() {group=Gradient Base, parameterName=GradientBaseColor1 Hover, defaultValue=<rgba 0.7865 0.2047 0.0573 1>}
  n7 = VectorParameter() {group=Gradient Base, parameterName=GradientBaseColor2 Hover, defaultValue=<rgba 0.9792 0.4143 0.1836 1>}
  n8 = NamedRerouteUsage()
  n9 = ScalarParameter() {defaultValue=-0.3, group=Gradient Base, parameterName=GradientBaseRotation}
  n10 = ScalarParameter() {defaultValue=0.3, group=Gradient Base, parameterName=GradientBaseIntensity}
  NamedRerouteDeclaration(None=LinearInterpolate(
    A=@n6.RGB
    B=@n7.RGB
    Alpha=MaterialFunctionCall(UVs=@n8, RotationAngle=@n9, GradientOffset=@n10) {materialFunction=&MF_UI_LinearGradient}.Result
   )) {name=GradientBase Hover, nodeColor=<rgba 0 0 0 1>}
  NamedRerouteDeclaration(None=LinearInterpolate(A=NamedRerouteUsage(), B=NamedRerouteUsage(), Alpha=NamedRerouteUsage())) {name=Gradient Base Final, nodeColor=<rgba 0 0 0 1>}
  n11 = NamedRerouteUsage()
  n12 = ScalarParameter() {defaultValue=0.17, group=Triangle Shape, parameterName=Triangle Bottom 1 - Rotation}
  n13 = LinearInterpolate(
   A=ScalarParameter() {defaultValue=1.341, group=Triangle Shape, parameterName="Triangle Bottom 1  - Position Z - Default"}
   B=ScalarParameter() {defaultValue=1.14, group=Triangle Shape, parameterName="Triangle Bottom 1  - Position Z - Hover"}
   Alpha=SmoothStep(Value=NamedRerouteUsage()) {constMin=0.4}
  )
  n14 = ScalarParameter() {group=Triangle Shape, parameterName="Triangle Bottom 1  - Position X"}
  n15 = MaterialFunctionCall(
   UVs=Add(
    A=MaterialFunctionCall(UVs=@n11, "Rotation Angle (0-1)"=@n12) {materialFunction=&CustomRotator}.Rotated Values
    B=AppendVector(A=@n13, B=@n14)
   )
   Texture Scale=AppendVector(
    A=ScalarParameter() {defaultValue=2.8, group=Triangle Shape, parameterName="Triangle Bottom 1  - Scale X"}
    B=ScalarParameter() {defaultValue=0.28, group=Triangle Shape, parameterName="Triangle Bottom 1  - Scale Y"}
   )
  ) {materialFunction=&ScaleUVsByCenter}
  n16 = Constant() {r=3}
  NamedRerouteDeclaration(None=Multiply(
    A=MaterialFunctionCall!(UVs=@n15.UVs, Size, Sides=@n16, InnerAngle, OutlineThickness, OutlineAlignment, GlowMin, GlowMax) {materialFunction=/FortUILibrary/MaterialFunctions/Shapes/MF_UI_RegularPolygon}.Fill
    B=ScalarParameter() {defaultValue=0.15, group=Triangle Opacity, parameterName=Triangle Opacity}
   )) {name=Triangle BG1 Bottom_1, nodeColor=<rgba 0 0 0 1>}
  n17 = NamedRerouteUsage()
  n18 = ScalarParameter() {defaultValue=0.34, group=Triangle Shape, parameterName=Triangle Bottom 2 - Rotation}
  n19 = LinearInterpolate(
   A=ScalarParameter() {defaultValue=1.3258, group=Triangle Shape, parameterName=Triangle Bottom 2 - Position Z - Default}
   B=ScalarParameter() {defaultValue=1.1133, group=Triangle Shape, parameterName=Triangle Bottom 2 - Position Z - Hover}
   Alpha=SmoothStep(Value=NamedRerouteUsage()) {constMin=0.2}
  )
  n20 = ScalarParameter() {defaultValue=0.14, group=Triangle Shape, parameterName="Triangle Bottom 2  - Position X"}
  n21 = MaterialFunctionCall(
   UVs=Add(
    A=MaterialFunctionCall(UVs=@n17, "Rotation Angle (0-1)"=@n18) {materialFunction=&CustomRotator}.Rotated Values
    B=AppendVector(A=@n19, B=@n20)
   )
   Texture Scale=AppendVector(
    A=ScalarParameter() {defaultValue=2.52, group=Triangle Shape, parameterName=Triangle Bottom 2 - Scale X}
    B=ScalarParameter() {defaultValue=0.33, group=Triangle Shape, parameterName=Triangle Bottom 2 - Scale Y}
   )
  ) {materialFunction=&ScaleUVsByCenter}
  n22 = Constant() {r=3}
  NamedRerouteDeclaration(None=Multiply(
    A=MaterialFunctionCall(UVs=@n21.UVs, Sides=@n22) {materialFunction=&MF_UI_RegularPolygon}.Fill
    B=ScalarParameter() {defaultValue=0.15, group=Triangle Opacity, parameterName=Triangle Opacity}
   )) {name=Triangle_BG2 Bottom_1, nodeColor=<rgba 0 0 0 1>}
  NamedRerouteDeclaration(None=LinearInterpolate(A=NamedRerouteUsage(), Alpha=NamedRerouteUsage())) {name=Triangle Bottom Assembly_1, nodeColor=<rgba 0 0 0 1>}
  n23 = NamedRerouteUsage()
  n24 = ScalarParameter() {defaultValue=1.3, group=Triangle Shape, parameterName=Triangle Top 1 - Rotation}
  n25 = LinearInterpolate(
   A=ScalarParameter() {defaultValue=-1.3636, group=Triangle Shape, parameterName=Triangle Top 1 - Position Z - Default}
   B=ScalarParameter() {defaultValue=-1.22, group=Triangle Shape, parameterName=Triangle Top 1 - Position Z - Hover}
   Alpha=SmoothStep(Value=NamedRerouteUsage()) {constMin=0.7}
  )
  n26 = ScalarParameter() {defaultValue=0.088, group=Triangle Shape, parameterName=Triangle Top 1 - Position X}
  n27 = MaterialFunctionCall(
   UVs=Add(
    A=MaterialFunctionCall(UVs=@n23, "Rotation Angle (0-1)"=@n24) {materialFunction=&CustomRotator}.Rotated Values
    B=AppendVector(A=@n25, B=@n26)
   )
   Texture Scale=AppendVector(
    A=ScalarParameter() {defaultValue=2.82, group=Triangle Shape, parameterName=Triangle Top 1 - Scale X}
    B=ScalarParameter() {defaultValue=0.29, group=Triangle Shape, parameterName=Triangle Top 1 - Scale Y}
   )
  ) {materialFunction=&ScaleUVsByCenter}
  n28 = Constant() {r=3}
  NamedRerouteDeclaration(None=Multiply(
    A=MaterialFunctionCall(UVs=@n27.UVs, Sides=@n28) {materialFunction=&MF_UI_RegularPolygon}.Fill
    B=ScalarParameter() {defaultValue=0.15, group=Triangle Opacity, parameterName=Triangle Opacity}
   )) {name=Triangle BG1 Top_1, nodeColor=<rgba 0 0 0 1>}
  n29 = NamedRerouteUsage()
  n30 = ScalarParameter() {defaultValue=1.21, group=Triangle Shape, parameterName=Triangle Top 2 - Rotation}
  n31 = LinearInterpolate(
   A=ScalarParameter() {defaultValue=-1.4144, group=Triangle Shape, parameterName=Triangle Top 2 - Position Z - Default}
   B=ScalarParameter() {defaultValue=-1.1386, group=Triangle Shape, parameterName=Triangle Top 2 - Position Z - Hover}
   Alpha=SmoothStep(Value=NamedRerouteUsage()) {constMin=0.3}
  )
  n32 = ScalarParameter() {defaultValue=0.21, group=Triangle Shape, parameterName=Triangle Top 2 - Position X}
  n33 = MaterialFunctionCall(
   UVs=Add(
    A=MaterialFunctionCall(UVs=@n29, "Rotation Angle (0-1)"=@n30) {materialFunction=&CustomRotator}.Rotated Values
    B=AppendVector(A=@n31, B=@n32)
   )
   Texture Scale=AppendVector(
    A=ScalarParameter() {defaultValue=3, group=Triangle Shape, parameterName=Triangle Top 2 - Scale X}
    B=ScalarParameter() {defaultValue=0.396, group=Triangle Shape, parameterName=Triangle Top 2 - Scale Y}
   )
  ) {materialFunction=&ScaleUVsByCenter}
  n34 = Constant() {r=3}
  NamedRerouteDeclaration(None=Multiply(
    A=MaterialFunctionCall(UVs=@n33.UVs, Sides=@n34) {materialFunction=&MF_UI_RegularPolygon}.Fill
    B=ScalarParameter() {defaultValue=0.15, group=Triangle Opacity, parameterName=Triangle Opacity}
   )) {name=Triangle BG2 Top_1, nodeColor=<rgba 0 0 0 1>}
  n35 = NamedRerouteUsage()
  n36 = ScalarParameter() {defaultValue=3.69, group=Triangle Shape, parameterName=Triangle Top 3 - Rotation}
  n37 = LinearInterpolate(
   A=ScalarParameter() {defaultValue=1.3799, group=Triangle Shape, parameterName=Triangle Top 3 - Position Z - Default}
   B=ScalarParameter() {defaultValue=1.2547, group=Triangle Shape, parameterName=Triangle Top 3 - Position Z - Hover}
   Alpha=SmoothStep(Value=NamedRerouteUsage()) {constMin=0.5}
  )
  n38 = ScalarParameter() {defaultValue=0.058, group=Triangle Shape, parameterName=Triangle Top 3 - Position X}
  n39 = MaterialFunctionCall(
   UVs=Add(
    A=MaterialFunctionCall(UVs=@n35, "Rotation Angle (0-1)"=@n36) {materialFunction=&CustomRotator}.Rotated Values
    B=AppendVector(A=@n37, B=@n38)
   )
   Texture Scale=AppendVector(
    A=ScalarParameter() {defaultValue=2.86, group=Triangle Shape, parameterName=Triangle Top 3 - Scale X}
    B=ScalarParameter() {defaultValue=0.36, group=Triangle Shape, parameterName=Triangle Top 3 - Scale Y}
   )
  ) {materialFunction=&ScaleUVsByCenter}
  n40 = Constant() {r=3}
  NamedRerouteDeclaration(None=Multiply(
    A=MaterialFunctionCall(UVs=@n39.UVs, Sides=@n40) {materialFunction=&MF_UI_RegularPolygon}.Fill
    B=ScalarParameter() {defaultValue=0.15, group=Triangle Opacity, parameterName=Triangle Opacity}
   )) {name=Triangle BG3 Top_1, nodeColor=<rgba 0 0 0 1>}
  n41 = MaterialFunctionCall(GradientStart=Constant() {r=0.7}, GradientEnd=Constant() {r=-0.007}, RotationAngle=Constant() {r=0.05}) {materialFunction=&MF_UI_LinearGradient}
  NamedRerouteDeclaration(None=Multiply(A=MaterialFunctionCall("Value (0-1)"=@n41.Result) {materialFunction=&MF_UI_EaseCurves}.EaseInOutSine)) {name=VignetteMask, nodeColor=<rgba 0 0 0 0.5>}
  n42 = NamedRerouteUsage()
  n43 = NamedRerouteUsage()
  NamedRerouteDeclaration(None=LinearInterpolate(A=LinearInterpolate(A=@n42, Alpha=@n43), Alpha=NamedRerouteUsage())) {name=Triangle Top Assembly_1, nodeColor=<rgba 0 0 0 1>}
  NamedRerouteDeclaration(None=Multiply(
    A=NamedRerouteUsage()
    B=VectorParameter() {group=Triangle Color, parameterName=Triangle Top Color, defaultValue=<rgba 1 0.7913 0.2542 1>}.RGB
   )) {name=Triangle Top with Color_1, nodeColor=<rgba 0 0 0 1>}
  NamedRerouteDeclaration(None=Multiply(
    A=NamedRerouteUsage()
    B=VectorParameter() {group=Triangle Color, parameterName=Triangle Bottom Color, defaultValue=<rgba 1 0.7913 0.2542 1>}.RGB
   )) {name=Triangle Bottom with Color_1, nodeColor=<rgba 0 0 0 1>}
  n44 = VectorParameter() {group=Gradient Base, parameterName=GradientBaseColor1 Default, defaultValue=<rgba 0.7865 0.1687 0.0123 1>}
  n45 = VectorParameter() {group=Gradient Base, parameterName=GradientBaseColor2 Default, defaultValue=<rgba 0.9792 0.2944 0.0148 1>}
  n46 = NamedRerouteUsage()
  n47 = ScalarParameter() {defaultValue=-0.3, group=Gradient Base, parameterName=GradientBaseRotation}
  n48 = ScalarParameter() {defaultValue=0.3, group=Gradient Base, parameterName=GradientBaseIntensity}
  NamedRerouteDeclaration(None=LinearInterpolate(
    A=@n44.RGB
    B=@n45.RGB
    Alpha=MaterialFunctionCall(UVs=@n46, RotationAngle=@n47, GradientOffset=@n48) {materialFunction=&MF_UI_LinearGradient}.Result
   )) {name=GradientBase Default, nodeColor=<rgba 0 0 0 1>}
  n49 = MaterialFunctionCall(
   UVs=NamedRerouteUsage()
   Texture Scale=AppendVector(
    A=ScalarParameter() {defaultValue=1.5, group=Gradient Glow, parameterName=GradientGlowSizeX}
    B=ScalarParameter() {defaultValue=1, group=Gradient Glow, parameterName=GradientGlowSizeY}
   )
  ) {materialFunction=&ScaleUVsByCenter}
  n50 = MaterialFunctionCall(
   X=ScalarParameter() {defaultValue=0.5, group=Gradient Glow, parameterName=GradientGlowPositionX}
   Y=ScalarParameter() {defaultValue=0.4, group=Gradient Glow, parameterName=GradientGlowPositionY}
  ) {materialFunction=&MakeFloat2}
  n51 = ScalarParameter() {defaultValue=0.6, group=Gradient Glow, parameterName=GradientGlowRadius}
  n52 = ScalarParameter() {defaultValue=1.45, group=Gradient Glow, parameterName=GradientGlowDensity}
  NamedRerouteDeclaration(None=LinearInterpolate(
    B=VectorParameter() {group=Gradient Glow, parameterName=GradientGlowColor, defaultValue=<rgba 1 0.8035 0.0815 1>}.RGB
    Alpha=MaterialFunctionCall(UVs=@n49.UVs, CenterPosition=@n50.Result, Radius=@n51, Density=@n52) {materialFunction=&RadialGradientExponential}.RadialGradientExponential
   )) {name=Gradient Glow, nodeColor=<rgba 0 0 0 1>}
  n53 = VertexColor()
  n54 = NamedRerouteUsage()
  n55 = Multiply(A=Constant3Vector() {constant=<rgba 1 1 1 1>}, B=@n54)
  n56 = Constant() {r=0.9}
  NamedRerouteDeclaration(None=LinearInterpolate(A=Multiply(A=@n53.A, B=@n55), Alpha=Multiply(A=@n54, B=@n56))) {name=GradientGlowAlphaComposite, nodeColor=<rgba 0 0 0 1>}
  NamedRerouteDeclaration(None=MaterialFunctionCall(UV=TextureCoordinate()) {materialFunction=&MF_UI_UVScaling}.Scaled to Fit) {name=UV, nodeColor=<rgba 0 0 0 1>}
  NamedRerouteDeclaration(None=MaterialFunctionCall("Value (0-1)"=ScalarParameter() {group=State, parameterName=Hover}) {materialFunction=&MF_UI_EaseCurves}.EaseOutQuint) {name=Hover, nodeColor=<rgba 0 0 0 1>}
 ]
}
