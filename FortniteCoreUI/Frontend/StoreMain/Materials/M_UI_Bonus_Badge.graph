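# UI material for a bonus badge, apparently for the store screen, with a two-color gradient and animated glow.
# The badge SDF blends a rounded-box mask toward a constant using a squared, offset six-sided polygon field, with DPI-aware anti-aliasing available for edges.
# Glow multiplies rotating blurred noise by a smoothstepped, inverted distance field, with scale, speed, size and opacity parameters.
# The color gradient is angled and cubic-eased, and vertex alpha scales the output.
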Material /FortniteCoreUI/Frontend/StoreMain/Materials/M_UI_Bonus_Badge
graph {
 properties {bAutomaticallySetUsageInEditor=false, blendMode=BLEND_AlphaComposite, materialDomain=MD_UI, usedShadingModels=Default Lit}
 nodes [
  n1 = MaterialFunctionCall!(DistanceField=NamedRerouteUsage(), Threshold=Constant() {r=0.01}) {materialFunction=/FortUILibrary/MaterialFunctions/Shapes/Modifiers/MF_UI_AntiAliasing_DPI}
  n2 = NamedRerouteUsage()
  Multiply(A=VertexColor().A, B=Multiply(A=NamedRerouteUsage(), B=LinearInterpolate(A=@n1.Inner, Alpha=@n2)))
  n3 = MaterialFunctionCall(Texture Scale=ScalarParameter() {defaultValue=1.5, group=Glow, parameterName=Glow - Scale}) {materialFunction=&ScaleUVsByCenter}
  n4 = Constant2Vector() {g=0.5}
  n5 = Time()
  n6 = ScalarParameter() {defaultValue=0.5, group=Glow, parameterName=Glow - Movement Speed, sortPriority=10}
  n7 = TextureSample(UVs=Rotator(Coordinate=Add(A=@n3.UVs, B=@n4), Time=Multiply(A=@n5, B=@n6)) {desc=MoreRandomic, speed=1}) {samplerType=SAMPLERTYPE_Grayscale, texture=/Game/Global/Textures/Default/Noise/LowResBlurredNoiseNormalized}
  n8 = NamedRerouteUsage()
  n9 = Power(Base=SmoothStep(Min=ScalarParameter() {defaultValue=0.97, group=Glow, parameterName=Glow- Size}, Value=OneMinus(None=@n8)) {constMin=0.97})
  NamedRerouteDeclaration(None=Multiply(A=Multiply(A=@n7.R, B=@n9), B=ScalarParameter() {group=Glow, parameterName=Glow - Opacity})) {name=Glow, nodeColor=<rgba 0 0 0 1>}
  n10 = MaterialFunctionCall(
   UVs=MaterialFunctionCall() {materialFunction=&GetUserInterfaceUV}.Normalized UV
   Texture Scale=Constant2Vector() {g=1.5, r=1.5}
  ) {materialFunction=&ScaleUVsByCenter}
  n11 = Saturate(None=MaterialFunctionCall(UV=@n10.UVs, Size=Constant() {r=0.5}, Corner Radius=Constant() {r=0.05}) {materialFunction=&MF_UI_CoreUI_RoundedMask}.SDF)
  n12 = Constant2Vector() {g=0.25}
  n13 = SmoothStep(Value=MaterialFunctionCall!(UVs=Add(A=@n10.UVs, B=@n12), Radius=Constant() {r=1}, Sides=Constant() {r=6}, InnerAngle) {materialFunction=/FortUILibrary/MaterialFunctions/Shapes/SDFs/MF_SDF_RegularPolygon}.SDF) {constMax=0.4}
  NamedRerouteDeclaration(None=LinearInterpolate(A=Multiply(A=@n11, B=@n11), Alpha=Multiply(A=@n13, B=@n13))) {name=Badge SDF, nodeColor=<rgba 0 0 0 1>}
  n14 = MaterialFunctionCall(RotationAngle=Constant() {r=-0.2}) {materialFunction=&MF_UI_LinearGradient}
  NamedRerouteDeclaration(None=LinearInterpolate(
    A=VectorParameter() {group=Color, parameterName=Color - Gradient 1, defaultValue=<rgba 1 0.9551 0.901 1>}.RGB
    B=VectorParameter() {group=Color, parameterName=Color - Gradient 2, defaultValue=<rgba 1 0.7214 0.3854 1>}.RGB
    Alpha=MaterialFunctionCall("Value (0-1)"=@n14.Result) {materialFunction=&MF_UI_EaseCurves}.EaseInOutCubic
   )) {name=Color, nodeColor=<rgba 0 0 0 1>}
 ]
}
