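# M_UI_ModalBackground is a translucent UI-domain material in the frontend store folder, named as the background for modal screens.
# GetSlatePost0 samples the first Slate post buffer, and its LinearRGB output feeds the Post Buffer Blur input of MF_UI_CoreUI_BackplateDisabled.
# That shared CoreUI function holds the backplate look, so the material itself adds no other math or parameters.
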Material /FortniteCoreUI/Frontend/StoreMain/Materials/M_UI_ModalBackground
graph {
 properties {
  bAutomaticallySetUsageInEditor=false
  bUsedWithStaticMesh=false
  blendMode=BLEND_Translucent
  materialDomain=MD_UI
  usedShadingModels=Default Lit
 }
 nodes [MaterialFunctionCall!(Post Buffer Blur=MaterialFunctionCall!("UVs (Optional)") {materialFunction=/Engine/Functions/UserInterface/GetSlatePost0}.LinearRGB) {materialFunction=/FortUILibrary/MaterialFunctions/CoreUI/MF_UI_CoreUI_BackplateDisabled}]
}
