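# Translucent UI material for the V-Bucks header background in the store front end.
# Flipped, offset UVs pass through scaled-to-fit mapping into a rotated linear gradient, thresholded with anti-aliasing into a crisp Shape Mask edge tilted by Angle Amount.
# A second, smoothstepped angled gradient multiplies with the rerouted mask to form a soft fade, alongside a constant blue color.
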Material /FortniteCoreUI/Frontend/StoreMain/Materials/M_UI_Vbucks_Header_Background
graph {
 properties {
  bAutomaticallySetUsageInEditor=false
  bUsedWithStaticMesh=false
  blendMode=BLEND_Translucent
  materialDomain=MD_UI
  usedShadingModels=Default Lit
 }
 nodes [
  Constant3Vector() {constant=<rgba 0.0273 0.4852 0.8308 1>}
  n1 = Constant() {r=0.2}
  Multiply(
   A=SmoothStep(Value=MaterialFunctionCall(RotationAngle=@n1) {materialFunction=&MF_UI_LinearGradient}.Result) {constMin=0.35}
   B=NamedRerouteUsage()
  )
  n2 = TextureCoordinate()
  n3 = MaterialFunctionCall(UV=Add(A=OneMinus(None=@n2), B=Constant2Vector() {r=0.5})) {materialFunction=&MF_UI_UVScaling}
  n4 = ScalarParameter() {defaultValue=-0.025, parameterName=Angle Amount}
  n5 = Constant2Vector() {g=1, r=0.5}
  NamedRerouteDeclaration(None=MaterialFunctionCall!(
    Threshold=Constant() {r=0.5}
    Gradient=MaterialFunctionCall(UVs=@n3.Scaled to Fit, RotationAngle=@n4, RotationCenter=@n5) {materialFunction=&MF_UI_LinearGradient}.Result
    Filter Width Override
    Filter Width Fudge Factor
   ) {materialFunction=/Game/UI/Materials/AAThreshold}.Result) {name=Shape Mask, nodeColor=<rgba 0.5469 0 1 1>}
 ]
}
