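# Store price widget built on WBP_UIKit_Price, showing an exact price with currency icon, additional price text and a regular price.
# A horizontal box holds overlaid platform and V-Bucks icons, the price texts, and a regular-price group with a tilted strike-through image.
# The event graph is empty; getter functions only return these child widgets, apparently for the parent class to use.
# Main price texts share a PriceSwitcher font material and auto-collapse when empty.
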WidgetBlueprint /FortniteCoreUI/Frontend/StoreMain/WBP_Price_ExactAmount
graphs [
 EventGraph {links [], nodes []}
 GetAdditionalPriceTextRef {nodes [
   GetAdditionalPriceTextRef() {is=title K2Node_FunctionEntry}
   > Return Node!(execute, Widget:Common Text Object Reference=Get Text_AdditionalPrice() -> (Text_AdditionalPrice:Common Text Object Reference))
  ]}
 GetAdditionalRegularPriceTextRef_Optional {nodes [
   GetAdditionalRegularPriceTextRef_Optional() {is=title K2Node_FunctionEntry}
   > Return Node(Widget=Get Text_AdditionalRegularPrice() -> (Text_AdditionalRegularPrice:Common Text Object Reference))
  ]}
 GetPlatformIconRef {nodes [
   GetPlatformIconRef() {is=title K2Node_FunctionEntry}
   > Return Node(Widget:WBP Price Platform Store Icon Object Reference=Get PlatformIcon() -> (PlatformIcon:WBP Price Platform Store Icon Object Reference))
  ]}
 GetPriceTextRef {nodes [
   GetPriceTextRef() {is=title K2Node_FunctionEntry}
   > Return Node(Widget:Common Text Object Reference=Get Text_Price() -> (Text_Price:Common Text Object Reference))
  ]}
 GetRegularPriceContainerRef_Optional {nodes [
   GetRegularPriceContainerRef_Optional() {is=title K2Node_FunctionEntry}
   > Return Node(Widget:Widget Object Reference=Get HB_RegularPrice() -> (HB_RegularPrice:Horizontal Box Object Reference))
  ]}
 GetRegularPriceTextRef_Optional {nodes [
   GetRegularPriceTextRef_Optional() {is=title K2Node_FunctionEntry}
   > Return Node(Widget:Common Text Object Reference=Get Text_RegularPrice() -> (Text_RegularPrice:Common Text Object Reference))
  ]}
 GetVBuckIconRef {nodes [
   GetVBuckIconRef() {is=title K2Node_FunctionEntry}
   > Return Node(Widget:Widget Object Reference=Get Image_Currency() -> (Image_Currency:Image Object Reference))
  ]}
]
variables []
widgets {
 info {
  inheritedWidgetCount=0
  parentClass=/FortUILibrary/Systems/Price/WBP_UIKit_Price.WBP_UIKit_Price_C
  rootWidgetClass=&HorizontalBox
  widgetCount=13
 }
 widgets [
  HB_PriceContainer {widgetClassPath=&HorizontalBox, widget=HB_PriceContainer}
  Overlay_0 {
   widgetClassPath=&Overlay
   parent=HB_PriceContainer
   widget=Overlay_0
   slot {class=HorizontalBoxSlot, properties {horizontalAlignment=HAlign_Left, verticalAlignment=VAlign_Center}}
  }
  PlatformIcon {
   widgetClassPath=/FortUILibrary/Systems/ItemShopTile/WBP_Price_PlatformStoreIcon
   bIsVariable=true
   parent=Overlay_0
   widget=PlatformIcon
   properties {currency=EFortPriceCurrency_MAX, previewPlayStationPlatform=true}
   slot {class=OverlaySlot, properties {horizontalAlignment=HAlign_Center, verticalAlignment=VAlign_Center}}
  }
  Image_Currency {
   widgetClassPath=&Image
   bIsVariable=true
   parent=Overlay_0
   widget=Image_Currency
   properties {brush {
     bIsDynamicallyLoaded=false
     drawAs=Image
     imageType=FullColor
     mirroring=NoMirror
     resourceName=None
     resourceObject=/FortUILibrary/Textures/Icons/T_UI_Icon_UIKit_VBuck
     tiling=NoTile
     imageSize=<28 28>
     margin=<ltrb 0 0 0 0>
     outlineSettings {
      bUseBrushTransparency=false
      roundingType=HalfHeightRadius
      width=0
      color {colorUseRule=UseColor_Specified, specifiedColor=<rgba 0 0 0 0>}
      cornerRadii=<0 0 0 0>
     }
     tintColor {colorUseRule=UseColor_Specified, specifiedColor=<rgba 1 1 1 1>}
     uVRegion {bIsValid=false, max=<0 0>, min=<0 0>}
    }}
   slot {class=OverlaySlot, properties {horizontalAlignment=HAlign_Center, verticalAlignment=VAlign_Center}}
  }
  Spacer {
   widgetClassPath=&Spacer
   parent=HB_PriceContainer
   widget=Spacer
   slot {class=HorizontalBoxSlot, properties {horizontalAlignment=HAlign_Left}}
  }
  Text_Price {
   widgetClassPath=&CommonTextBlock
   bIsVariable=true
   parent=HB_PriceContainer
   widget=Text_Price
   properties {
    bAutoCollapseWithEmptyText=true
    justification=Center
    mobileFontSizeMultiplier=1.2
    text="1,800"
    textTransformPolicy=ToUpper
    font {
     bForceMonospaced=false
     bMaterialIsStencil=false
     fontMaterial=/FortUILibrary/Systems/ItemShopTile/Materials/PriceSwitcher/MI_UIKit_PriceSwitcher_Text
     fontObject=&HeadingNow
     letterSpacing=0
     monospacedWidth=1
     size=18
     skewAmount=0
     typefaceFontName=86 Bold
     outlineSettings {
      bApplyOutlineToDropShadows=false
      bMiteredCorners=false
      bSeparateFillAlpha=false
      outlineMaterial=null
      outlineSize=0
      outlineColor=<rgba 0 0 0 1>
     }
    }
    shadowOffset=<0 0>
   }
   slot {class=HorizontalBoxSlot, properties {horizontalAlignment=HAlign_Right, verticalAlignment=VAlign_Center, padding=<ltrb 4 0 4 0>}}
  }
  Text_AdditionalPrice {
   widgetClassPath=&CommonTextBlock
   bIsVariable=true
   parent=HB_PriceContainer
   widget=Text_AdditionalPrice
   properties {
    applyLineHeightToBottomLine=false
    bAutoCollapseWithEmptyText=true
    justification=Center
    mobileFontSizeMultiplier=1.2
    text=VAT incl.
    font {
     bForceMonospaced=false
     bMaterialIsStencil=false
     fontMaterial=&MI_UIKit_PriceSwitcher_Text
     fontObject=&HeadingNow
     letterSpacing=0
     monospacedWidth=1
     size=11.25
     skewAmount=0
     typefaceFontName=76 Bold
     outlineSettings {
      bApplyOutlineToDropShadows=false
      bMiteredCorners=false
      bSeparateFillAlpha=false
      outlineMaterial=null
      outlineSize=0
      outlineColor=<rgba 0 0 0 1>
     }
    }
    shadowOffset=<0 0>
   }
   slot {class=HorizontalBoxSlot, properties {horizontalAlignment=HAlign_Right, verticalAlignment=VAlign_Bottom, padding=<ltrb 4 0 4 0>}}
  }
  Spacer_1 {
   widgetClassPath=&Spacer
   parent=HB_PriceContainer
   widget=Spacer_1
   properties {size=<2 1>}
   slot {class=HorizontalBoxSlot, properties {horizontalAlignment=HAlign_Left}}
  }
  HB_RegularPrice {
   widgetClassPath=&HorizontalBox
   bIsVariable=true
   parent=HB_PriceContainer
   widget=HB_RegularPrice
   slot {class=HorizontalBoxSlot, properties {horizontalAlignment=HAlign_Right, verticalAlignment=VAlign_Center}}
  }
  Overlay_RegularPrice {widgetClassPath=&Overlay, parent=HB_RegularPrice, widget=Overlay_RegularPrice, slot {class=HorizontalBoxSlot}}
  Text_RegularPrice {
   widgetClassPath=&CommonTextBlock
   bIsVariable=true
   parent=Overlay_RegularPrice
   widget=Text_RegularPrice
   properties {
    bAutoCollapseWithEmptyText=true
    justification=Center
    mobileFontSizeMultiplier=1.2
    text="2,000"
    textTransformPolicy=ToUpper
    colorAndOpacity {colorUseRule=UseColor_Specified, specifiedColor=<rgba 1 1 1 0.5>}
    font {
     bForceMonospaced=false
     bMaterialIsStencil=false
     fontMaterial=&MI_UIKit_PriceSwitcher_Text
     fontObject=&HeadingNow
     letterSpacing=0
     monospacedWidth=1
     size=15.75
     skewAmount=0
     typefaceFontName=76 Bold
     outlineSettings {
      bApplyOutlineToDropShadows=false
      bMiteredCorners=false
      bSeparateFillAlpha=false
      outlineMaterial=null
      outlineSize=0
      outlineColor=<rgba 0 0 0 1>
     }
    }
    shadowOffset=<0 0>
   }
   slot {class=OverlaySlot, properties {horizontalAlignment=HAlign_Fill, verticalAlignment=VAlign_Center, padding=<ltrb 4 0 4 0>}}
  }
  Image_Strike {
   widgetClassPath=&Image
   parent=Overlay_RegularPrice
   widget=Image_Strike
   properties {
    visibility=SelfHitTestInvisible
    brush {
     bIsDynamicallyLoaded=false
     drawAs=Image
     imageType=NoImage
     mirroring=NoMirror
     resourceName=None
     resourceObject=/FortUILibrary/Systems/ItemShopTile/Materials/PriceSwitcher/MI_UIKit_ItemShopTile_RegularPriceStrike
     tiling=NoTile
     imageSize=<32 4>
     margin=<ltrb 0 0 0 0>
     outlineSettings {
      bUseBrushTransparency=false
      roundingType=HalfHeightRadius
      width=0
      color {colorUseRule=UseColor_Specified, specifiedColor=<rgba 0 0 0 0>}
      cornerRadii=<0 0 0 0>
     }
     tintColor {colorUseRule=UseColor_Specified, specifiedColor=<rgba 1 1 1 1>}
     uVRegion {bIsValid=false, max=<0 0>, min=<0 0>}
    }
    renderTransform {angle=-3, scale=<1 1>, shear=<0 0>, translation=<0 0>}
   }
   slot {class=OverlaySlot, properties {horizontalAlignment=HAlign_Fill, verticalAlignment=VAlign_Center}}
  }
  Text_AdditionalRegularPrice {
   widgetClassPath=&CommonTextBlock
   bIsVariable=true
   parent=HB_RegularPrice
   widget=Text_AdditionalRegularPrice
   properties {
    bAutoCollapseWithEmptyText=true
    justification=Center
    mobileFontSizeMultiplier=1.2
    text=*
    font {
     bForceMonospaced=false
     bMaterialIsStencil=false
     fontMaterial=null
     fontObject=&HeadingNow
     letterSpacing=0
     monospacedWidth=1
     size=15.75
     skewAmount=0
     typefaceFontName=75 Medium
     outlineSettings {
      bApplyOutlineToDropShadows=false
      bMiteredCorners=false
      bSeparateFillAlpha=false
      outlineMaterial=null
      outlineSize=0
      outlineColor=<rgba 0 0 0 1>
     }
    }
    shadowOffset=<0 0>
   }
   slot {class=HorizontalBoxSlot, properties {horizontalAlignment=HAlign_Right, verticalAlignment=VAlign_Center, padding=<ltrb 4 0 4 0>}}
  }
 ]
}
