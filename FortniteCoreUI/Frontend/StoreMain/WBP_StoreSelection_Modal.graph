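# Store selection modal on FortStoreSelectionPopupBase that lets the player pay for an offer through Epic or a platform app store.
# Setters fill store branding, prices and Epic Rewards text, showing one corner price when both prices match.
# Usability-testing parameters switch button order, wording, price visibility and layout.
# Clicks call base-class handlers and fire OfferID analytics, hover toggles material outlines, and gamepad input refreshes focus to the Epic button.
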WidgetBlueprint /FortniteCoreUI/Frontend/StoreMain/WBP_StoreSelection_Modal
graphs [
 BP_GetDesiredFocusTarget {nodes [
   Get Desired Focus Target()
   > Return Node!(execute, ReturnValue:Widget Object Reference=Get WBP_Button_Epic() -> (WBP_Button_Epic:WBP UIKit Button Generic Object Reference))
  ]}
 EventGraph {nodes [
   n1 = "On Clicked (Button_CloseX)"() {is=title K2Node_ComponentBoundEvent}
   "Fire Event Element Interacted (Non-Localized)"!(
    execute=[@n1.then, @n4.then]
    FortPC:Fort Player Controller Object Reference=GetFortPlayerControllerFromActor!(Actor:Actor Object Reference=GetOwningPlayer()) -> (ReturnValue:Fort Player Controller Object Reference)
    ElementName:String=Cancel Button
    AdditionalInfo:Flow Analytics Additional Info Structure="Add String (Non-Localized)"!(
     Info:"Flow Analytics Additional Info Structure (by ref)"=Make Flow Analytics Additional Info!() -> (FlowAnalyticsAdditionalInfo:Flow Analytics Additional Info Structure) {is=title K2Node_MakeStruct}
     Key:String=OfferID
     Value:String=Get Offer Id() -> (Offer Id:String)
    ) -> (ReturnValue:Flow Analytics Additional Info Structure)
   ) -> (then)
   > HandleClose(self:Fort Store Selection Popup Base Object Reference)
   "On Clicked (WBP_Button_Platform)"!() -> (OutputDelegate, then, Button:Common Button Base Object Reference) {is=title K2Node_ComponentBoundEvent}
   > HandleStoreSelectionClicked!(
    execute
    self:Fort Store Selection Popup Base Object Reference
    AppStore:EAppStore Enum=Get PlatformStore() -> (PlatformStore:EAppStore Enum=DebugStore)
   ) -> (then)
   n2 = GetFortPlayerControllerFromActor(Actor=GetOwningPlayer())
   n3 = "Add String (Non-Localized)"(Info=Make Flow Analytics Additional Info(), Key=OfferID, Value=Get Offer Id())
   > "Fire Event Element Interacted (Non-Localized)"(FortPC=@n2, ElementName=App Store Button, AdditionalInfo=@n3)
   "On Clicked (WBP_Button_Epic)"!() -> (OutputDelegate, then, Button:Common Button Base Object Reference) {is=title K2Node_ComponentBoundEvent}
   > HandleStoreSelectionClicked(AppStore=EpicPurchasingService)
   > "Fire Event Element Interacted (Non-Localized)"(FortPC=@n2, ElementName=Epic Store Button, AdditionalInfo=@n3)
   n4 = "On Clicked (WBP_Cancel_Button)"!() -> (OutputDelegate, then, Button:Common Button Base Object Reference) {is=title K2Node_ComponentBoundEvent}
   "On Hovered (WBP_Button_Epic)"!() -> (OutputDelegate, then, Button:Common Button Base Object Reference) {is=title K2Node_ComponentBoundEvent}
   > Set IsEpic(IsEpic:bool=true) -> (Output_Get:bool)
   > SetEpicButtonEnabled()
   "On Hovered (WBP_Button_Platform)"!() -> (OutputDelegate, then, Button:Common Button Base Object Reference) {is=title K2Node_ComponentBoundEvent}
   > Set IsEpic(IsEpic=false)
   > SetAppStoreButtonEnabled()
   n5 = HandleInputMethodChanged!() -> (OutputDelegate, then, NewInputType:ECommonInputType Enum) {is=title K2Node_CustomEvent}
   > n6 = Branch(Condition="Equal (Enum)"(A=@n5.NewInputType, B=Gamepad))
   > RequestRefreshFocus(self:Common Activatable Widget Object Reference)
   ResetButtonsOutlines(execute=@n6.else)
   Event Construct()
   > SetSidesToPad!(
    execute
    self:Safe Zone Object Reference=Get SafeZone_0!(self:Lightbox Object Reference=Get WBP_Lightbox() -> (WBP_Lightbox:Lightbox Object Reference)) -> (SafeZone_0:Safe Zone Object Reference)
    InPadLeft:bool=false
    InPadRight:bool=false
    InPadTop:bool=false
    InPadBottom:bool=false
   ) -> (then)
   > UpdatePricesVisibility()
   > Assign On Input Method Changed!(execute:Exec, self:Common Input Object Reference=Get CommonInputSubsystem(), Delegate:"Delegate (by ref)"=Create Event()) -> (then:Exec) {is=title K2Node_AssignDelegate}
   Event Destruct()
   > Unbind Event from On Input Method Changed(self=Get CommonInputSubsystem(), Delegate=Create Event())
   Event On Activated()
   > "Fire Event Sub View Opened (Non-Localized)"!(
    execute
    FortPC:Fort Player Controller Object Reference=GetFortPlayerControllerFromActor(Actor=GetOwningPlayer())
    ViewName:String=Store Selection Modal
    AdditionalInfo:Flow Analytics Additional Info Structure="Add String (Non-Localized)"(Info=Make Flow Analytics Additional Info(), Key=OfferID, Value=Get Offer Id())
    bIsOverlay:bool=true
   ) -> (then)
  ]}
 HandleVariantEvents {nodes [
   HandleVariantEvents() {is=title K2Node_FunctionEntry}
   n1 = Get IsPriceHidden() -> (IsPriceHidden:bool)
   > SwitchPriceVisibility!(execute, self, IsPriceHidden:bool=@n1) -> (then)
   > HideButtonPrices!(execute, self, IsPriceHidden:bool=@n1) -> (then)
   > SwitchButtonPlace!(execute, self, IsEpicButtonLeftSide:bool=Get IsEpicButtonLeftSide() -> (IsEpicButtonLeftSide:bool)) -> (then)
   > SwitchTextStyle!(execute, self, IsGreenTextLayout:bool=Get IsGreenTextLayout() -> (IsGreenTextLayout:bool)) -> (then)
  ]}
 ResetButtonsOutlines {nodes [
   ResetButtonsOutlines() {is=title K2Node_FunctionEntry}
   > n1 = GetDynamicMaterial(self=Get Image_PlatformStoreBG() -> (Image_PlatformStoreBG:Image Object Reference))
   > SetScalarParameterValue(self=@n1, ParameterName=HasOutline, Value=0)
   > n2 = GetDynamicMaterial(self=Get Image_EpicStoreBG() -> (Image_EpicStoreBG:Image Object Reference))
   > SetScalarParameterValue(self=@n2, ParameterName=HasOutline, Value=0)
  ]}
 SetAppStoreButtonEnabled {nodes [
   SetAppStoreButtonEnabled() {is=title K2Node_FunctionEntry}
   > n1 = GetDynamicMaterial(self=Get Image_PlatformStoreBG())
   > SetScalarParameterValue(self=@n1, ParameterName=HasOutline, Value=1)
   > n2 = GetDynamicMaterial(self=Get Image_EpicStoreBG())
   > SetScalarParameterValue(self=@n2, ParameterName=HasOutline, Value=0)
  ]}
 SetCompactUIStyle {nodes [
   SetCompactUIStyle()
   > Branch(Condition=IsUsingCompactUI())
   n1 = Get Text_EpicRewards_Green() -> (Text_EpicRewards_Green:Text Block Object Reference)
   > Set Justification!(execute, Justification:ETextJustify Enum=Left, self:Text Layout Widget Object Reference=@n1) -> (then, Output_Get:ETextJustify Enum=Left)
   > SetHorizontalAlignment(self:Vertical Box Slot Object Reference=SlotAsVerticalBoxSlot(Widget=@n1), InHorizontalAlignment=HAlign_Fill)
   n2 = Get Text_EpicRewards_GreenLayout() -> (Text_EpicRewards_GreenLayout:Text Block Object Reference)
   > Set Justification(Justification=Left, self=@n2) -> (Output_Get=Left)
   > SetHorizontalAlignment(self=SlotAsVerticalBoxSlot(Widget=@n2), InHorizontalAlignment=HAlign_Fill)
   n3 = Get SB_Buttons() -> (SB_Buttons:Size Box Object Reference)
   > ClearHeightOverride(self:Size Box Object Reference=@n3)
   > SetMinDesiredHeight!(execute, self:Size Box Object Reference=@n3, InMinDesiredHeight:float=300) -> (then)
   > SetPadding!(
    execute
    self:Overlay Slot Object Reference=SlotAsOverlaySlot(Widget=Get HorizontalBox_GreenText() -> (HorizontalBox_GreenText:Horizontal Box Object Reference))
    InPadding_Left:float=50
    InPadding_Top:float=50
    InPadding_Right:float=50
    InPadding_Bottom:float=50
   ) -> (then)
  ]}
 SetEpicButtonEnabled {nodes [
   SetEpicButtonEnabled() {is=title K2Node_FunctionEntry}
   > n1 = GetDynamicMaterial(self=Get Image_EpicStoreBG())
   > SetScalarParameterValue(self=@n1, ParameterName=HasOutline, Value=1)
   > n2 = GetDynamicMaterial(self=Get Image_PlatformStoreBG())
   > SetScalarParameterValue(self=@n2, ParameterName=HasOutline, Value=0)
  ]}
 SetEpicOfferPrice {nodes [
   n1 = SetEpicOfferPrice!() -> (then, InEpicOfferPrice:Text) {is=title K2Node_FunctionEntry}
   > n2 = Set Epic Price(Epic Price:Text=@n1) -> (Output_Get:Text)
   > "SetText (Text)"(self=Get Text_PricingEpicStoreButton() -> (Text_PricingEpicStoreButton:Text Block Object Reference), InText=@n2)
   > UpdatePricesVisibility()
  ]}
 SetEpicRewardsEarnedBackPercentage {nodes [
   n1 = SetEpicRewardsEarnedBackPercentage!() -> (then, InEpicRewardsEarnedBackPercentage:int) {is=title K2Node_FunctionEntry}
   > Set Epic Rewards Earned Back Percentage(Epic Rewards Earned Back Percentage:int=@n1) -> (Output_Get:int=0)
   > UpdateRewardsPriceText()
  ]}
 SetEpicRewardsEarnedBackPrice {nodes [
   n1 = SetEpicRewardsEarnedBackPrice!() -> (then, InEpicRewardsEarnedBackPrice:Text) {is=title K2Node_FunctionEntry}
   > Set Epic Rewards Earned Back Price Text(Epic Rewards Earned Back Price Text:Text=@n1) -> (Output_Get:Text)
   > UpdateRewardsPriceText()
  ]}
 SetPlatformStore {nodes [
   n1 = SetPlatformStore!() -> (then, InPlatformStore:EAppStore Enum=DebugStore) {is=title K2Node_FunctionEntry}
   > n2 = Set PlatformStore(PlatformStore:EAppStore Enum=@n1) -> (Output_Get:EAppStore Enum=DebugStore)
   n3 = Find!(
    TargetMap:Map of EAppStore Enums to App Store Info Structures=Get PlatformStoresInfo() -> (PlatformStoresInfo:Map of EAppStore Enums to App Store Info Structures=DebugStore)
    Key:"EAppStore Enum (by ref)"=@n2
   ) -> (Value_Logo:Slate Brush Structure, Value_DisplayName:Text, ReturnValue)
   > Branch(Condition=@n3.ReturnValue)
   > SetBrush(self=Get Image_PlatformStoreLogo() -> (Image_PlatformStoreLogo:Image Object Reference), InBrush=@n3)
   > Set Text(Text=@n3.Value_DisplayName, self=Get Text_PlatformStoreName() -> (Text_PlatformStoreName:Text Block Object Reference))
   > SetActiveWidget(
    self=Get Switcher_PlatformLogoName() -> (Switcher_PlatformLogoName:Common Visibility Switcher Object Reference)
    Widget=Select(
     Option 0:Widget Object Reference=Get Text_PlatformStoreName()
     Option 1:Widget Object Reference=Get Image_PlatformStoreLogo()
     Index=TextIsEmpty(InText=@n3.Value_DisplayName)
    ) -> (ReturnValue:Widget Object Reference)
   )
  ]}
 SetPlatformStoreOfferPrice {nodes [
   n1 = SetPlatformStoreOfferPrice!() -> (then, InPlatformStoreOfferPrice:Text) {is=title K2Node_FunctionEntry}
   > n2 = Set App Store Price(App Store Price:Text=@n1) -> (Output_Get:Text)
   > "SetText (Text)"(self=Get Text_PricingPlatformStoreButton() -> (Text_PricingPlatformStoreButton:Text Block Object Reference), InText=@n2)
   > "SetText (Text)"(self=Get Text_PriceCorner() -> (Text_PriceCorner:Text Block Object Reference), InText=@n2)
   > UpdatePricesVisibility()
  ]}
 SetUsabilityTestingParameters {nodes [
   n1 = SetUsabilityTestingParameters!() -> (then, InDisplayParameters:Fort Dual Billing Usability Testing Parameters Structure) {is=title K2Node_FunctionEntry}
   n2 = Break Fort Dual Billing Usability Testing Parameters!(FortDualBillingUsabilityTestingParameters:"Fort Dual Billing Usability Testing Parameters Structure (by ref)"=@n1) -> (bShouldUseGetInsteadOfEarn:bool=~, bPlaceEpicButtonOnTheLeft:bool=~, bUseEpicStyling:bool=~) {is=title K2Node_BreakStruct}
   > Set ShouldSayGet(ShouldSayGet:bool=@n2) -> (Output_Get:bool)
   > Set IsEpicButtonLeftSide(IsEpicButtonLeftSide:bool=@n2.bPlaceEpicButtonOnTheLeft) -> (Output_Get:bool)
   > Set IsPriceHidden(IsPriceHidden:bool=@n2.bUseEpicStyling) -> (Output_Get:bool)
   > Set IsGreenTextLayout(IsGreenTextLayout:bool=@n2.bUseEpicStyling) -> (Output_Get:bool)
   > HandleVariantEvents() {is=title K2Node_CallFunction}
   > UpdateRewardsPriceText()
  ]}
 UpdatePricesVisibility {nodes [
   UpdatePricesVisibility() {is=title K2Node_FunctionEntry}
   n1 = "Equal Exactly (Text)"!(A:"Text (by ref)"=Get App Store Price() -> (App Store Price:Text), B:"Text (by ref)"=Get Epic Price() -> (Epic Price:Text)) -> (ReturnValue:bool)
   > SetVisibility(
    self=Get Text_PriceCorner()
    InVisibility=Select(Option 0:ESlateVisibility Enum=Collapsed, Option 1:ESlateVisibility Enum=HitTestInvisible, Index=@n1) -> (ReturnValue:ESlateVisibility Enum)
   )
   > SetVisibility(
    self=[Get Text_PricingEpicStoreButton(), Get Text_PricingPlatformStoreButton()]
    InVisibility=Select(Option 0=HitTestInvisible, Option 1=Collapsed, Index=@n1)
   )
  ]}
 UpdateRewardsPriceText {nodes [
   UpdateRewardsPriceText() {is=title K2Node_FunctionEntry}
   n1 = Get Epic Rewards Earned Back Price Text() -> (Epic Rewards Earned Back Price Text:Text)
   > n2 = Branch(Condition=NOT Boolean(A=TextIsEmpty(InText=@n1)))
   n3 = Get Epic Rewards Earned Back Price Text()
   n4 = Get Epic Rewards Earned Back Percentage() -> (Epic Rewards Earned Back Percentage:int=0)
   > "SetText (Text)"(
    self=Get Text_EpicRewards() -> (Text_EpicRewards:Text Block Object Reference)
    InText=SelectText(
     A=Format Text!(
      Format="LOCTEXT(\"Get {Reward Percentage}% {Discounted Amount} back in Epic Rewards\")"
      Reward Percentage:int=@n4
      Discounted Amount:Text=@n3
     ) -> (Result)
     B=Format Text(Format="LOCTEXT(\"Earn {Reward Percentage}% {Discounted Amount} back in Epic Rewards\")", Reward Percentage=@n4, Discounted Amount=@n3)
     bPickA=Get ShouldSayGet() -> (ShouldSayGet:bool)
    )
   )
   n5 = Get Epic Rewards Earned Back Price Text()
   n6 = Get Epic Rewards Earned Back Percentage() -> (Epic Rewards Earned Back Percentage=0)
   > "SetText (Text)"(
    self=Get Text_EpicRewards_Green()
    InText=SelectText(
     A=Format Text(Format="LOCTEXT(\"Get {Reward Percentage}% {Discounted Amount} back\")", Reward Percentage=@n6, Discounted Amount=@n5)
     B=Format Text(Format="LOCTEXT(\"Earn {Reward Percentage}% {Discounted Amount} back\")", Reward Percentage=@n6, Discounted Amount=@n5)
     bPickA=Get ShouldSayGet()
    )
   )
   n7 = Get Epic Rewards Earned Back Percentage() -> (Epic Rewards Earned Back Percentage=0)
   "SetText (Text)"(
    execute=@n2.else
    self=Get Text_EpicRewards()
    InText=SelectText(
     A=Format Text!(Format="LOCTEXT(\"Get {Reward Percentage}% back in Epic Rewards\")", Reward Percentage=@n7) -> (Result)
     B=Format Text(Format="LOCTEXT(\"Earn {Reward Percentage}% back in Epic Rewards\")", Reward Percentage=@n7)
     bPickA=Get ShouldSayGet()
    )
   )
   n8 = Get Epic Rewards Earned Back Percentage() -> (Epic Rewards Earned Back Percentage=0)
   > "SetText (Text)"(
    self=Get Text_EpicRewards_Green()
    InText=SelectText(
     A=Format Text(Format="LOCTEXT(\"Get {Reward Percentage}% back\")", Reward Percentage=@n8)
     B=Format Text(Format="LOCTEXT(\"Earn {Reward Percentage}% back\")", Reward Percentage=@n8)
     bPickA=Get ShouldSayGet()
    )
   )
  ]}
 VariantsGraph {nodes [
   n1 = SwitchPriceVisibility!() -> (OutputDelegate, then, IsPriceHidden:bool=true) {is=title K2Node_CustomEvent}
   > SetVisibility(
    self=[
     Get HorizontalBox_Pricing() -> (HorizontalBox_Pricing:Horizontal Box Object Reference)
     Get Image_Divider() -> (Image_Divider:Image Object Reference)
    ]
    InVisibility=Select(Option 0=HitTestInvisible, Option 1=Collapsed, Index=@n1.IsPriceHidden)
   )
   > SetPadding!(
    execute
    self:Vertical Box Slot Object Reference=SlotAsVerticalBoxSlot(Widget=Get VerticalBox_ButtonArea() -> (VerticalBox_ButtonArea:Vertical Box Object Reference))
    InPadding:Margin Structure=Select!(
     Option 0_Left:float=60
     Option 0_Top:float=0
     Option 0_Right:float=60
     Option 0_Bottom:float=0
     Option 1_Left:float=60
     Option 1_Top:float=50
     Option 1_Right:float=60
     Option 1_Bottom:float=10
     Index=@n1.IsPriceHidden
    ) -> (ReturnValue:Margin Structure)
   ) -> (then)
   > SetHorizontalAlignment(
    self:Horizontal Box Slot Object Reference=SlotAsHorizontalBoxSlot!(Widget:Widget Object Reference=Get Text_Choose() -> (Text_Choose:Text Block Object Reference)) -> (ReturnValue:Horizontal Box Slot Object Reference)
    InHorizontalAlignment=Select!(Option 0:EHorizontalAlignment Enum=HAlign_Center, Option 1:EHorizontalAlignment Enum=HAlign_Left, Index=@n1.IsPriceHidden) -> (ReturnValue:EHorizontalAlignment Enum)
   )
   n2 = Get IsPriceHidden()
   > SetVisibility(
    self=Get Button_CloseX() -> (Button_CloseX:Button Object Reference)
    InVisibility=Select(Option 0:ESlateVisibility Enum=Collapsed, Option 1:ESlateVisibility Enum=Visible, Index=@n2) -> (ReturnValue:ESlateVisibility Enum)
   )
   > SetVisibility(
    self=Get WBP_Cancel_Button() -> (WBP_Cancel_Button:WBP UIKit Button Regular Object Reference)
    InVisibility=Select(Option 0=Visible, Option 1=Collapsed, Index=@n2)
   )
   n3 = SwitchButtonPlace!() -> (OutputDelegate, then, IsEpicButtonLeftSide:bool=~) {is=title K2Node_CustomEvent}
   n4 = SlotAsGridSlot(Widget=Get WBP_Button_Epic())
   > SetColumn!(execute, self:Grid Slot Object Reference=@n4, InColumn:int=SelectInt(A=0, B=1, bPickA=@n3.IsEpicButtonLeftSide)) -> (then)
   > SetPadding(
    self:Grid Slot Object Reference=@n4
    InPadding=Select!(
     Option 0_Left:float=20
     Option 0_Top:float=0
     Option 0_Right:float=0
     Option 0_Bottom:float=0
     Option 1_Left:float=0
     Option 1_Top:float=0
     Option 1_Right:float=20
     Option 1_Bottom:float=0
     Index=@n3.IsEpicButtonLeftSide
    ) -> (ReturnValue:Margin Structure)
   )
   n5 = SlotAsGridSlot(Widget=Get WBP_Button_Platform() -> (WBP_Button_Platform:WBP UIKit Button Generic Object Reference))
   n6 = Get IsEpicButtonLeftSide()
   > SetColumn(self=@n5, InColumn=SelectInt(A=1, B=0, bPickA=@n6))
   > SetPadding(
    self=@n5
    InPadding=Select(
     Option 0_Left=0
     Option 0_Top=0
     Option 0_Right=20
     Option 0_Bottom=0
     Option 1_Left=20
     Option 1_Top=0
     Option 1_Right=0
     Option 1_Bottom=0
     Index=@n6
    )
   )
   n7 = SwitchTextStyle!() -> (OutputDelegate, then, IsGreenTextLayout:bool=~) {is=title K2Node_CustomEvent}
   > SetActiveWidgetIndex(
    self:Common Visibility Switcher Object Reference=Get CommonVisibilitySwitcher_EpicButtonLayout() -> (CommonVisibilitySwitcher_EpicButtonLayout:Common Visibility Switcher Object Reference)
    Index=SelectInt(A=1, B=0, bPickA=@n7.IsGreenTextLayout)
   )
   > SetPadding(
    self:Overlay Slot Object Reference=SlotAsOverlaySlot(Widget=Get VerticalBox_AllContent() -> (VerticalBox_AllContent:Vertical Box Object Reference))
    InPadding=Select!(Option 0:Margin Structure, Option 1_Left=0, Option 1_Top=0, Option 1_Right=0, Option 1_Bottom=50, Index=@n7.IsGreenTextLayout) -> (ReturnValue)
   )
   n8 = HideButtonPrices!() -> (OutputDelegate, then, IsPriceHidden:bool=~) {is=title K2Node_CustomEvent}
   > Branch(Condition=@n8.IsPriceHidden)
   > SetVisibility(self=[Get Text_PricingPlatformStoreButton(), Get Text_PricingEpicStoreButton()], InVisibility=Collapsed)
   n9 = Event PreConstruct()
   > SetCompactUIStyle() {is=title K2Node_CallFunction}
   > Branch(Condition=@n9.IsDesignTime)
   > HandleVariantEvents()
  ]}
]
variables [
 App Store Price {type=string, default=""}
 Epic Price {type=string, default=""}
 Epic Rewards Earned Back Percentage {type=integer}
 Epic Rewards Earned Back Price Text {type=string, default=""}
 IsEpic {type=boolean, default=true}
 IsEpicButtonLeftSide {type=boolean, default=true}
 IsGreenTextLayout {type=boolean}
 IsPriceHidden {type=boolean}
 Offer Id {type=string, default=""}
 PlatformStore {type=EAppStore, default=DebugStore}
 ShouldSayGet {type=boolean, default=true}
]
widgets {
 info {
  inheritedWidgetCount=0
  namedSlotCount=5
  parentClass=/Script/FortniteUI.FortStoreSelectionPopupBase
  rootWidgetClass=&Overlay
  widgetCount=43
 }
 widgets [
  Overlay_30 {widgetClassPath=&Overlay, widget=Overlay_30}
  WBP_Lightbox {
   widgetClassPath=/Game/UI/Foundation/Lightbox/Lightbox
   bIsVariable=true
   parent=Overlay_30
   widget=WBP_Lightbox
   slot {class=OverlaySlot, properties {horizontalAlignment=HAlign_Fill, verticalAlignment=VAlign_Fill}}
  }
  Overlay_All {widgetClassPath=&Overlay, namedSlotHost=WBP_Lightbox, widget=Overlay_All}
  WBP_UIKit_Scrim {
   widgetClassPath=/FortUILibrary/Systems/Dialog/WBP_UIKit_Scrim
   parent=Overlay_All
   widget=WBP_UIKit_Scrim
   slot {class=OverlaySlot, properties {horizontalAlignment=HAlign_Fill, verticalAlignment=VAlign_Fill}}
  }
  HorizontalBox_ModalCentering {
   widgetClassPath=&HorizontalBox
   parent=Overlay_All
   widget=HorizontalBox_ModalCentering
   slot {class=OverlaySlot, properties {horizontalAlignment=HAlign_Fill, verticalAlignment=VAlign_Fill, padding=<ltrb 40 0 40 0>}}
  }
  Spacer_1 {
   widgetClassPath=&Spacer
   parent=HorizontalBox_ModalCentering
   widget=Spacer_1
   properties {visibility=HitTestInvisible}
   slot {class=HorizontalBoxSlot, properties {size {sizeRule=Fill, value=0.08}}}
  }
  Overlay_Modal {
   widgetClassPath=&Overlay
   parent=HorizontalBox_ModalCentering
   widget=Overlay_Modal
   slot {class=HorizontalBoxSlot, properties {verticalAlignment=VAlign_Center, size {sizeRule=Fill, value=1}}}
  }
  UIKit_Backplate {
   widgetClassPath=&WBP_UIKit_Backplate
   parent=Overlay_Modal
   widget=UIKit_Backplate
   properties {brightness=Obscure}
   slot {class=OverlaySlot, properties {horizontalAlignment=HAlign_Fill, verticalAlignment=VAlign_Fill}}
  }
  VerticalBox_AllContent {
   widgetClassPath=&VerticalBox
   bIsVariable=true
   parent=Overlay_Modal
   widget=VerticalBox_AllContent
   slot {class=OverlaySlot, properties {horizontalAlignment=HAlign_Fill, verticalAlignment=VAlign_Center}}
  }
  HorizontalBox_Pricing {
   widgetClassPath=&HorizontalBox
   bIsVariable=true
   parent=VerticalBox_AllContent
   widget=HorizontalBox_Pricing
   slot {class=VerticalBoxSlot, properties {padding=<ltrb 60 60 60 60>}}
  }
  Text_VBucks {
   widgetClassPath=&UEFN_TextBlock
   bIsVariable=true
   parent=HorizontalBox_Pricing
   widget=Text_VBucks
   properties {
    text="2,800 V-Bucks"
    visibility=HitTestInvisible
    font {
     bForceMonospaced=false
     bMaterialIsStencil=false
     fontMaterial=null
     fontObject=&HeadingNow
     letterSpacing=0
     monospacedWidth=1
     size=30
     skewAmount=0
     typefaceFontName=75 Medium
     outlineSettings {
      bApplyOutlineToDropShadows=false
      bMiteredCorners=false
      bSeparateFillAlpha=false
      outlineMaterial=null
      outlineSize=0
      outlineColor=<rgba 0 0 0 1>
     }
    }
   }
   slot {class=HorizontalBoxSlot, properties {horizontalAlignment=HAlign_Left, size {sizeRule=Fill, value=1}}}
  }
  Text_PriceCorner {
   widgetClassPath=&UEFN_TextBlock
   bIsVariable=true
   parent=HorizontalBox_Pricing
   widget=Text_PriceCorner
   properties {
    text=$XX.XX
    visibility=HitTestInvisible
    font {
     bForceMonospaced=false
     bMaterialIsStencil=false
     fontMaterial=null
     fontObject=&HeadingNow
     letterSpacing=0
     monospacedWidth=1
     size=30
     skewAmount=0
     typefaceFontName=75 Medium
     outlineSettings {
      bApplyOutlineToDropShadows=false
      bMiteredCorners=false
      bSeparateFillAlpha=false
      outlineMaterial=null
      outlineSize=0
      outlineColor=<rgba 0 0 0 1>
     }
    }
   }
   slot {class=HorizontalBoxSlot, properties {horizontalAlignment=HAlign_Right, size {sizeRule=Fill, value=1}}}
  }
  Image_Divider {
   widgetClassPath=&Image
   bIsVariable=true
   parent=VerticalBox_AllContent
   widget=Image_Divider
   properties {brush {
     bIsDynamicallyLoaded=false
     drawAs=Image
     imageType=NoImage
     mirroring=NoMirror
     resourceName=None
     resourceObject=null
     tiling=NoTile
     imageSize=<926 4>
     margin=<ltrb 0 0 0 0>
     outlineSettings {
      bUseBrushTransparency=false
      roundingType=HalfHeightRadius
      width=0
      color {colorUseRule=UseColor_Specified, specifiedColor=<rgba 0 0 0 0>}
      cornerRadii=<0 0 0 0>
     }
     tintColor {colorUseRule=UseColor_Specified, specifiedColor=<rgba 0.6939 0.6939 0.6939 0.3>}
     uVRegion {bIsValid=false, max=<0 0>, min=<0 0>}
    }}
   slot {class=VerticalBoxSlot, properties {verticalAlignment=VAlign_Center, padding=<ltrb 60 0 60 20>}}
  }
  VerticalBox_ButtonArea {
   widgetClassPath=&VerticalBox
   bIsVariable=true
   parent=VerticalBox_AllContent
   widget=VerticalBox_ButtonArea
   slot {class=VerticalBoxSlot, properties {verticalAlignment=VAlign_Center, padding=<ltrb 60 0 60 0>}}
  }
  HorizontalBox_0 {
   widgetClassPath=&HorizontalBox
   parent=VerticalBox_ButtonArea
   widget=HorizontalBox_0
   slot {class=VerticalBoxSlot, properties {padding=<ltrb 0 0 0 30>}}
  }
  Text_Choose {
   widgetClassPath=&UEFN_TextBlock
   bIsVariable=true
   parent=HorizontalBox_0
   widget=Text_Choose
   properties {
    text=Choose how to pay
    visibility=HitTestInvisible
    font {
     bForceMonospaced=false
     bMaterialIsStencil=false
     fontMaterial=null
     fontObject=&HeadingNow
     letterSpacing=0
     monospacedWidth=1
     size=25.5
     skewAmount=0
     typefaceFontName=74 Regular
     outlineSettings {
      bApplyOutlineToDropShadows=false
      bMiteredCorners=false
      bSeparateFillAlpha=false
      outlineMaterial=null
      outlineSize=0
      outlineColor=<rgba 0 0 0 1>
     }
    }
   }
   slot {class=HorizontalBoxSlot, properties {horizontalAlignment=HAlign_Center, verticalAlignment=VAlign_Center, size {sizeRule=Fill, value=1}}}
  }
  Button_CloseX {
   widgetClassPath=&Button
   parent=HorizontalBox_0
   widget=Button_CloseX
   properties {widgetStyle {
     clickedSlateSound {resourceObject=null}
     disabled {
      bIsDynamicallyLoaded=false
      drawAs=RoundedBox
      imageType=FullColor
      mirroring=NoMirror
      resourceName=None
      resourceObject=/FortniteCoreUI/Frontend/StoreMain/Materials/Textures/T_UI_Icon_XClose
      tiling=NoTile
      imageSize=<50 50>
      margin=<ltrb 0 0 0 0>
      outlineSettings {
       bUseBrushTransparency=true
       roundingType=FixedRadius
       width=0
       color {colorUseRule=UseColor_Specified, specifiedColor=<rgba 0.6951 0.6951 0.6951 0>}
       cornerRadii=<4 4 4 4>
      }
      tintColor {colorUseRule=UseColor_Specified, specifiedColor=<rgba 1 1 1 0.6>}
      uVRegion {bIsValid=false, max=<0 0>, min=<0 0>}
     }
     disabledForeground {colorUseRule=UseColor_Specified, specifiedColor=<rgba 0.5271 0.5271 0.5271 1>}
     hovered {
      bIsDynamicallyLoaded=false
      drawAs=RoundedBox
      imageType=FullColor
      mirroring=NoMirror
      resourceName=None
      resourceObject=&T_UI_Icon_XClose
      tiling=NoTile
      imageSize=<50 50>
      margin=<ltrb 0 0 0 0>
      outlineSettings {
       bUseBrushTransparency=true
       roundingType=FixedRadius
       width=0
       color {colorUseRule=UseColor_Specified, specifiedColor=<rgba 0.6951 0.6951 0.6951 0>}
       cornerRadii=<4 4 4 4>
      }
      tintColor {colorUseRule=UseColor_Specified, specifiedColor=<rgba 1 1 1 0.6>}
      uVRegion {bIsValid=false, max=<0 0>, min=<0 0>}
     }
     hoveredForeground {colorUseRule=UseColor_Specified, specifiedColor=<rgba 0 1 1 1>}
     hoveredSlateSound {resourceObject=null}
     normal {
      bIsDynamicallyLoaded=false
      drawAs=RoundedBox
      imageType=FullColor
      mirroring=NoMirror
      resourceName=None
      resourceObject=&T_UI_Icon_XClose
      tiling=NoTile
      imageSize=<50 50>
      margin=<ltrb 0 0 0 0>
      outlineSettings {
       bUseBrushTransparency=true
       roundingType=FixedRadius
       width=0
       color {colorUseRule=UseColor_Specified, specifiedColor=<rgba 0.6951 0.6951 0.6951 0>}
       cornerRadii=<4 4 4 4>
      }
      tintColor {colorUseRule=UseColor_Specified, specifiedColor=<rgba 1 1 1 0.6>}
      uVRegion {bIsValid=false, max=<0 0>, min=<0 0>}
     }
     normalForeground {colorUseRule=UseColor_Specified, specifiedColor=<rgba 0 1 1 1>}
     normalPadding=<ltrb 12 1.5 12 1.5>
     pressed {
      bIsDynamicallyLoaded=false
      drawAs=RoundedBox
      imageType=FullColor
      mirroring=NoMirror
      resourceName=None
      resourceObject=&T_UI_Icon_XClose
      tiling=NoTile
      imageSize=<50 50>
      margin=<ltrb 0 0 0 0>
      outlineSettings {
       bUseBrushTransparency=true
       roundingType=FixedRadius
       width=0
       color {colorUseRule=UseColor_Specified, specifiedColor=<rgba 0.6951 0.6951 0.6951 0>}
       cornerRadii=<4 4 4 4>
      }
      tintColor {colorUseRule=UseColor_Specified, specifiedColor=<rgba 1 1 1 0.6>}
      uVRegion {bIsValid=false, max=<0 0>, min=<0 0>}
     }
     pressedForeground {colorUseRule=UseColor_Specified, specifiedColor=<rgba 0 1 1 1>}
     pressedPadding=<ltrb 12 1.5 12 1.5>
     pressedSlateSound {resourceObject=null}
    }}
   slot {class=HorizontalBoxSlot, properties {horizontalAlignment=HAlign_Right, verticalAlignment=VAlign_Center, size {sizeRule=Fill, value=1}}}
  }
  SB_Buttons {
   widgetClassPath=&SizeBox
   bIsVariable=true
   parent=VerticalBox_ButtonArea
   widget=SB_Buttons
   properties {bOverride_HeightOverride=true, heightOverride=300}
   slot {class=VerticalBoxSlot, properties {verticalAlignment=VAlign_Center, size {sizeRule=Fill, value=1}}}
  }
  Grid_StoreOptions {
   widgetClassPath=&GridPanel
   parent=SB_Buttons
   widget=Grid_StoreOptions
   properties {
    bOverride_Cursor=true
    navigation=&self:WidgetTree.Grid_StoreOptions.WidgetNavigation_0
    columnFill [1,1]
   }
   slot {class=SizeBoxSlot, properties {verticalAlignment=VAlign_Center}}
  }
  WBP_Button_Epic {
   widgetClassPath=&WBP_UIKit_Button_Generic
   bIsVariable=true
   parent=Grid_StoreOptions
   widget=WBP_Button_Epic
   properties {
    bDisplayInputActionWhenNotInteractable=true
    bInteractableWhenSelected=true
    bSimulateHoverOnTouchInput=false
    bTriggerClickedAfterSelection=true
    backgroundClass=null
    backgroundMaterial=/FortniteCoreUI/Frontend/StoreMain/Materials/MI_UIKit_StoreSelection
    iconMaterial=/FortniteCoreUI/Frontend/StoreMain/Materials/MI_Logo_StoreSelectionButton_EpicGamesWhite
    minHeight=0
    text=Store
    useIconBlock=false
    useTextBlock=false
    visibility=Visible
    font {
     bForceMonospaced=false
     bMaterialIsStencil=false
     fontMaterial=/FortUILibrary/UIKit/Buttons/MaterialInstances/CTAButtons/MI_UIKit_TextQuiet_Prompt
     fontObject=&Burbank
     letterSpacing=48
     monospacedWidth=1
     size=21
     skewAmount=0
     typefaceFontName=Burbank-Black
     outlineSettings {
      bApplyOutlineToDropShadows=false
      bMiteredCorners=false
      bSeparateFillAlpha=false
      outlineMaterial=null
      outlineSize=0
      outlineColor=<rgba 0 0 0 1>
     }
    }
    iconPadding=<ltrb 0 50 0 90>
    iconSize=<170 60>
    textPadding=<ltrb 20 20 20 20>
   }
   slot {class=GridSlot, properties {padding=<ltrb 0 0 20 0>}}
  }
  Overlay_0 {widgetClassPath=&Overlay, namedSlotHost=WBP_Button_Epic, widget=Overlay_0}
  CommonVisibilitySwitcher_EpicButtonLayout {
   widgetClassPath=&CommonVisibilitySwitcher
   bIsVariable=true
   parent=Overlay_0
   widget=CommonVisibilitySwitcher_EpicButtonLayout
   slot {class=OverlaySlot, properties {horizontalAlignment=HAlign_Fill, verticalAlignment=VAlign_Center, padding=<ltrb 0 48 0 48>}}
  }
  VerticalBox_AllWhiteText {
   widgetClassPath=&VerticalBox
   parent=CommonVisibilitySwitcher_EpicButtonLayout
   widget=VerticalBox_AllWhiteText
   slot {class=CommonVisibilitySwitcherSlot}
  }
  Image_EpicLogo {
   widgetClassPath=&Image
   bIsVariable=true
   parent=VerticalBox_AllWhiteText
   widget=Image_EpicLogo
   properties {
    visibility=HitTestInvisible
    brush {
     bIsDynamicallyLoaded=false
     drawAs=Image
     imageType=FullColor
     mirroring=NoMirror
     resourceName=None
     resourceObject=&MI_Logo_StoreSelectionButton_EpicGamesWhite
     tiling=NoTile
     imageSize=<280 100>
     margin=<ltrb 0 0 0 0>
     outlineSettings {
      bUseBrushTransparency=false
      roundingType=HalfHeightRadius
      width=0
      color {colorUseRule=UseColor_Specified, specifiedColor=<rgba 0 0 0 0>}
      cornerRadii=<0 0 0 0>
     }
     tintColor {colorUseRule=UseColor_Specified, specifiedColor=<rgba 1 1 1 1>}
     uVRegion {bIsValid=false, max=<0 0>, min=<0 0>}
    }
   }
   slot {
    class=VerticalBoxSlot
    properties {horizontalAlignment=HAlign_Center, verticalAlignment=VAlign_Center, padding=<ltrb 0 0 0 20>, size {sizeRule=Automatic, value=2}}
   }
  }
  Text_PricingEpicStoreButton {
   widgetClassPath=&UEFN_TextBlock
   bIsVariable=true
   parent=VerticalBox_AllWhiteText
   widget=Text_PricingEpicStoreButton
   properties {
    text="CAD $XX,XX"
    visibility=HitTestInvisible
    font {
     bForceMonospaced=false
     bMaterialIsStencil=false
     fontMaterial=null
     fontObject=&HeadingNow
     letterSpacing=0
     monospacedWidth=1
     size=27
     skewAmount=0
     typefaceFontName=75 Medium
     outlineSettings {
      bApplyOutlineToDropShadows=false
      bMiteredCorners=false
      bSeparateFillAlpha=false
      outlineMaterial=null
      outlineSize=0
      outlineColor=<rgba 0 0 0 1>
     }
    }
   }
   slot {class=VerticalBoxSlot, properties {horizontalAlignment=HAlign_Center, padding=<ltrb 0 0 0 10>, size {sizeRule=Fill, value=1}}}
  }
  Text_EpicRewards {
   widgetClassPath=&UEFN_TextBlock
   bIsVariable=true
   parent=VerticalBox_AllWhiteText
   widget=Text_EpicRewards
   properties {
    autoWrapText=true
    justification=Center
    text="Earn 20% ($1.15) back in Epic Rewards"
    visibility=HitTestInvisible
    font {
     bForceMonospaced=false
     bMaterialIsStencil=false
     fontMaterial=null
     fontObject=&HeadingNow
     letterSpacing=0
     monospacedWidth=1
     size=18
     skewAmount=0
     typefaceFontName=74 Regular
     outlineSettings {
      bApplyOutlineToDropShadows=false
      bMiteredCorners=false
      bSeparateFillAlpha=false
      outlineMaterial=null
      outlineSize=0
      outlineColor=<rgba 0 0 0 1>
     }
    }
   }
   slot {class=VerticalBoxSlot, properties {padding=<ltrb 5 5 5 0>, size {sizeRule=Fill, value=1}}}
  }
  HorizontalBox_GreenText {
   widgetClassPath=&HorizontalBox
   bIsVariable=true
   parent=CommonVisibilitySwitcher_EpicButtonLayout
   widget=HorizontalBox_GreenText
   properties {bOverride_Cursor=true}
   slot {
    class=CommonVisibilitySwitcherSlot
    properties {horizontalAlignment=HAlign_Center, verticalAlignment=VAlign_Center, padding=<ltrb 0 50 0 50>}
   }
  }
  Image_EpicLogo_GreenLayout {
   widgetClassPath=&Image
   parent=HorizontalBox_GreenText
   widget=Image_EpicLogo_GreenLayout
   properties {
    visibility=HitTestInvisible
    brush {
     bIsDynamicallyLoaded=false
     drawAs=Image
     imageType=FullColor
     mirroring=NoMirror
     resourceName=None
     resourceObject=&MI_Logo_StoreSelectionButton_EpicGamesWhite
     tiling=NoTile
     imageSize=<280 100>
     margin=<ltrb 0 0 0 0>
     outlineSettings {
      bUseBrushTransparency=false
      roundingType=HalfHeightRadius
      width=0
      color {colorUseRule=UseColor_Specified, specifiedColor=<rgba 0 0 0 0>}
      cornerRadii=<0 0 0 0>
     }
     tintColor {colorUseRule=UseColor_Specified, specifiedColor=<rgba 1 1 1 1>}
     uVRegion {bIsValid=false, max=<0 0>, min=<0 0>}
    }
   }
   slot {class=HorizontalBoxSlot, properties {verticalAlignment=VAlign_Center, padding=<ltrb 20 0 20 0>}}
  }
  Image_GreenLayoutDivider {
   widgetClassPath=&Image
   parent=HorizontalBox_GreenText
   widget=Image_GreenLayoutDivider
   properties {brush {
     bIsDynamicallyLoaded=false
     drawAs=Image
     imageType=NoImage
     mirroring=NoMirror
     resourceName=None
     resourceObject=null
     tiling=NoTile
     imageSize=<4 50>
     margin=<ltrb 0 0 0 0>
     outlineSettings {
      bUseBrushTransparency=false
      roundingType=HalfHeightRadius
      width=0
      color {colorUseRule=UseColor_Specified, specifiedColor=<rgba 0 0 0 0>}
      cornerRadii=<0 0 0 0>
     }
     tintColor {colorUseRule=UseColor_Specified, specifiedColor=<rgba 1 1 1 0.2>}
     uVRegion {bIsValid=false, max=<0 0>, min=<0 0>}
    }}
   slot {class=HorizontalBoxSlot, properties {horizontalAlignment=HAlign_Center, verticalAlignment=VAlign_Center}}
  }
  VerticalBox_InsideEpicButton {
   widgetClassPath=&VerticalBox
   parent=HorizontalBox_GreenText
   widget=VerticalBox_InsideEpicButton
   slot {class=HorizontalBoxSlot, properties {verticalAlignment=VAlign_Center, padding=<ltrb 20 0 20 0>, size {sizeRule=Fill, value=1}}}
  }
  Text_EpicRewards_Green {
   widgetClassPath=&UEFN_TextBlock
   bIsVariable=true
   parent=VerticalBox_InsideEpicButton
   widget=Text_EpicRewards_Green
   properties {
    autoWrapText=true
    justification=Center
    text="Earn 20% ($1.15) back"
    visibility=HitTestInvisible
    colorAndOpacity {colorUseRule=UseColor_Specified, specifiedColor=<rgba 0.0395 1 0.2307 1>}
    font {
     bForceMonospaced=false
     bMaterialIsStencil=false
     fontMaterial=null
     fontObject=&HeadingNow
     letterSpacing=0
     monospacedWidth=1
     size=18
     skewAmount=0
     typefaceFontName=76 Bold
     outlineSettings {
      bApplyOutlineToDropShadows=false
      bMiteredCorners=false
      bSeparateFillAlpha=false
      outlineMaterial=null
      outlineSize=0
      outlineColor=<rgba 0 0 0 1>
     }
    }
   }
   slot {class=VerticalBoxSlot, properties {horizontalAlignment=HAlign_Left}}
  }
  Text_EpicRewards_GreenLayout {
   widgetClassPath=&UEFN_TextBlock
   bIsVariable=true
   parent=VerticalBox_InsideEpicButton
   widget=Text_EpicRewards_GreenLayout
   properties {
    autoWrapText=true
    justification=Center
    text=in Epic Rewards
    visibility=HitTestInvisible
    font {
     bForceMonospaced=false
     bMaterialIsStencil=false
     fontMaterial=null
     fontObject=&HeadingNow
     letterSpacing=0
     monospacedWidth=1
     size=18
     skewAmount=0
     typefaceFontName=74 Regular
     outlineSettings {
      bApplyOutlineToDropShadows=false
      bMiteredCorners=false
      bSeparateFillAlpha=false
      outlineMaterial=null
      outlineSize=0
      outlineColor=<rgba 0 0 0 1>
     }
    }
   }
   slot {class=VerticalBoxSlot, properties {horizontalAlignment=HAlign_Left}}
  }
  Image_EpicStoreBG {
   widgetClassPath=&Image
   bIsVariable=true
   namedSlotHost=WBP_Button_Epic
   widget=Image_EpicStoreBG
   properties {brush {
     bIsDynamicallyLoaded=false
     drawAs=Image
     imageType=FullColor
     mirroring=NoMirror
     resourceName=None
     resourceObject=&MI_UIKit_StoreSelection
     tiling=NoTile
     imageSize=<32 32>
     margin=<ltrb 0 0 0 0>
     outlineSettings {
      bUseBrushTransparency=false
      roundingType=HalfHeightRadius
      width=0
      color {colorUseRule=UseColor_Specified, specifiedColor=<rgba 0 0 0 0>}
      cornerRadii=<0 0 0 0>
     }
     tintColor {colorUseRule=UseColor_Specified, specifiedColor=<rgba 1 1 1 1>}
     uVRegion {bIsValid=false, max=<0 0>, min=<0 0>}
    }}
  }
  WBP_Button_Platform {
   widgetClassPath=&WBP_UIKit_Button_Generic
   bIsVariable=true
   parent=Grid_StoreOptions
   widget=WBP_Button_Platform
   properties {
    bDisplayInputActionWhenNotInteractable=true
    bInteractableWhenSelected=true
    bSimulateHoverOnTouchInput=false
    bTriggerClickedAfterSelection=true
    backgroundClass=null
    backgroundMaterial=/FortniteCoreUI/Frontend/StoreMain/Materials/MI_UIKit_StoreSelection_Apple
    text=in-app purchase
    textTransformPolicy=None
    useIconBlock=false
    useTextBlock=false
    visibility=Visible
    font {
     bForceMonospaced=false
     bMaterialIsStencil=false
     fontMaterial=&MI_UIKit_TextQuiet_Prompt
     fontObject=&HeadingNow
     letterSpacing=48
     monospacedWidth=1
     size=13.5
     skewAmount=0
     typefaceFontName=75 Medium
     outlineSettings {
      bApplyOutlineToDropShadows=false
      bMiteredCorners=false
      bSeparateFillAlpha=false
      outlineMaterial=null
      outlineSize=0
      outlineColor=<rgba 0 0 0 1>
     }
    }
   }
   slot {class=GridSlot, properties {column=1, padding=<ltrb 20 0 0 0>}}
  }
  Overlay {widgetClassPath=&Overlay, namedSlotHost=WBP_Button_Platform, widget=Overlay}
  VerticalBox_InsideAppButton {
   widgetClassPath=&VerticalBox
   parent=Overlay
   widget=VerticalBox_InsideAppButton
   slot {class=OverlaySlot, properties {horizontalAlignment=HAlign_Fill, verticalAlignment=VAlign_Fill, padding=<ltrb 0 50 0 50>}}
  }
  Switcher_PlatformLogoName {
   widgetClassPath=&CommonVisibilitySwitcher
   bIsVariable=true
   parent=VerticalBox_InsideAppButton
   widget=Switcher_PlatformLogoName
   slot {class=VerticalBoxSlot, properties {horizontalAlignment=HAlign_Center, verticalAlignment=VAlign_Center, size {sizeRule=Fill, value=1}}}
  }
  Text_PlatformStoreName {
   widgetClassPath=&UEFN_TextBlock
   bIsVariable=true
   parent=Switcher_PlatformLogoName
   widget=Text_PlatformStoreName
   properties {
    justification=Center
    text=in-app purchase
    visibility=HitTestInvisible
    wrapTextAt=650
    font {
     bForceMonospaced=false
     bMaterialIsStencil=false
     fontMaterial=null
     fontObject=&HeadingNow
     letterSpacing=0
     monospacedWidth=1
     size=25.5
     skewAmount=0
     typefaceFontName=75 Medium
     outlineSettings {
      bApplyOutlineToDropShadows=false
      bMiteredCorners=false
      bSeparateFillAlpha=false
      outlineMaterial=null
      outlineSize=0
      outlineColor=<rgba 0 0 0 1>
     }
    }
   }
   slot {class=CommonVisibilitySwitcherSlot, properties {horizontalAlignment=HAlign_Center, verticalAlignment=VAlign_Center}}
  }
  Image_PlatformStoreLogo {
   widgetClassPath=&Image
   bIsVariable=true
   parent=Switcher_PlatformLogoName
   widget=Image_PlatformStoreLogo
   properties {brush {
     bIsDynamicallyLoaded=false
     drawAs=Image
     imageType=FullColor
     mirroring=NoMirror
     resourceName=None
     resourceObject=/FortniteCoreUI/Frontend/StoreMain/Materials/MI_Logo_StoreSelectionButton_OneStore
     tiling=NoTile
     imageSize=<150 150>
     margin=<ltrb 0 0 0 0>
     outlineSettings {
      bUseBrushTransparency=false
      roundingType=HalfHeightRadius
      width=0
      color {colorUseRule=UseColor_Specified, specifiedColor=<rgba 0 0 0 0>}
      cornerRadii=<0 0 0 0>
     }
     tintColor {colorUseRule=UseColor_Specified, specifiedColor=<rgba 1 1 1 1>}
     uVRegion {bIsValid=false, max=<0 0>, min=<0 0>}
    }}
   slot {class=CommonVisibilitySwitcherSlot}
  }
  Text_PricingPlatformStoreButton {
   widgetClassPath=&UEFN_TextBlock
   bIsVariable=true
   parent=VerticalBox_InsideAppButton
   widget=Text_PricingPlatformStoreButton
   properties {
    justification=Center
    text="CAD $XX,XX"
    visibility=HitTestInvisible
    font {
     bForceMonospaced=false
     bMaterialIsStencil=false
     fontMaterial=null
     fontObject=&HeadingNow
     letterSpacing=0
     monospacedWidth=1
     size=27
     skewAmount=0
     typefaceFontName=75 Medium
     outlineSettings {
      bApplyOutlineToDropShadows=false
      bMiteredCorners=false
      bSeparateFillAlpha=false
      outlineMaterial=null
      outlineSize=0
      outlineColor=<rgba 0 0 0 1>
     }
    }
   }
   slot {class=VerticalBoxSlot, properties {horizontalAlignment=HAlign_Center, verticalAlignment=VAlign_Center, padding=<ltrb 0 10 0 0>}}
  }
  Image_PlatformStoreBG {
   widgetClassPath=&Image
   bIsVariable=true
   namedSlotHost=WBP_Button_Platform
   widget=Image_PlatformStoreBG
   properties {brush {
     bIsDynamicallyLoaded=false
     drawAs=Image
     imageType=FullColor
     mirroring=NoMirror
     resourceName=None
     resourceObject=&MI_UIKit_StoreSelection
     tiling=NoTile
     imageSize=<32 32>
     margin=<ltrb 0 0 0 0>
     outlineSettings {
      bUseBrushTransparency=false
      roundingType=HalfHeightRadius
      width=0
      color {colorUseRule=UseColor_Specified, specifiedColor=<rgba 0 0 0 0>}
      cornerRadii=<0 0 0 0>
     }
     tintColor {colorUseRule=UseColor_Specified, specifiedColor=<rgba 1 1 1 1>}
     uVRegion {bIsValid=false, max=<0 0>, min=<0 0>}
    }}
  }
  WBP_Cancel_Button {
   widgetClassPath=&WBP_UIKit_Button_Regular
   bIsVariable=true
   parent=VerticalBox_AllContent
   widget=WBP_Cancel_Button
   properties {
    iconBeforeText=false
    minHeight=100
    text=Cancel
    visibility=Visible
    renderTransform {angle=0, scale=<0.7 0.7>, shear=<0 0>, translation=<0 0>}
    renderTransformPivot=<1 0.5>
    triggeringInputAction {dataTable=&InputActionDataTable, rowName=Cancel}
   }
   slot {class=VerticalBoxSlot, properties {horizontalAlignment=HAlign_Right, verticalAlignment=VAlign_Bottom, padding=<ltrb 0 40 60 50>}}
  }
  Spacer {
   widgetClassPath=&Spacer
   parent=HorizontalBox_ModalCentering
   widget=Spacer
   properties {visibility=HitTestInvisible}
   slot {class=HorizontalBoxSlot, properties {size {sizeRule=Fill, value=0.08}}}
  }
 ]
}
