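# WBP_UIKit_StoreSelectionButton is a store selection button widget in the frontend UI, derived from the UIKit generic modular button base.
# It adds an overlay holding a centered, wrapping CommonTextBlock label in bold italic heading font, pushed down by a large top padding.
# PreConstruct calls the parent, then copies the ButtonLabel string variable into that text block.
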WidgetBlueprint /FortniteCoreUI/Frontend/StoreMain/WBP_UIKit_StoreSelectionButton
graphs [EventGraph {nodes [
   n1 = Event PreConstruct()
   > Parent: PreConstruct(IsDesignTime=@n1.IsDesignTime)
   > Set Text(
    Text=Get ButtonLabel() -> (ButtonLabel:Text)
    self=Get CommonTextBlock_ButtonLabel() -> (CommonTextBlock_ButtonLabel:Common Text Object Reference)
   )
  ]}]
variables [ButtonLabel {type=string, default=Placeholder Button Label}]
widgets {
 info {
  inheritedWidgetCount=0
  namedSlotCount=1
  parentClass=/FortUILibrary/Systems/ModularButton/WBP_UIKit_ButtonGeneric_Base.WBP_UIKit_ButtonGeneric_Base_C
  rootWidgetClass=null
  widgetCount=2
 }
 widgets [
  Overlay_37 {widgetClassPath=&Overlay, widget=Overlay_37}
  CommonTextBlock_ButtonLabel {
   widgetClassPath=&CommonTextBlock
   bIsVariable=true
   parent=Overlay_37
   widget=CommonTextBlock_ButtonLabel
   properties {
    justification=Center
    text=Placeholder Button Label
    wrapTextAt=368
    font {
     bForceMonospaced=false
     bMaterialIsStencil=false
     fontMaterial=null
     fontObject=&HeadingNow
     letterSpacing=-40
     monospacedWidth=1
     size=32.25
     skewAmount=0
     typefaceFontName=86 Bold Italic
     outlineSettings {
      bApplyOutlineToDropShadows=false
      bMiteredCorners=false
      bSeparateFillAlpha=false
      outlineMaterial=null
      outlineSize=0
      outlineColor=<rgba 0 0 0 1>
     }
    }
    shadowOffset=<0 0>
   }
   slot {class=OverlaySlot, properties {horizontalAlignment=HAlign_Center, padding=<ltrb 0 328 0 0>}}
  }
 ]
}
