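# Base activatable widget, with no layout of its own, for the content of the Fortnite V-Bucks store's offer tiles.
# It records the focused tile's offer ID, can look tiles up by ID, and stores an optional exact-pack offer view model with a validity flag.
# It relays purchasing state and completion through dispatchers and staggers tile intro animations by index.
# Offer setup and initial focus functions are empty stubs, apparently for child widgets to override.
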WidgetBlueprint /FortniteCoreUI/Frontend/StoreMain/WBP_VbucksStore_Content_Base
graphs [
 EventGraph {nodes [
   n1 = OnOfferTileFocused!() -> (OutputDelegate, then, Button:Common Button Base Object Reference) {is=title K2Node_CustomEvent}
   > n2 = Cast To WBP_VbucksStore_OfferTile!(execute, Object=@n1.Button) -> (then, CastFailed, AsWBP Vbucks Store Offer Tile:WBP Vbucks Store Offer Tile Object Reference) {is=title K2Node_DynamicCast}
   > Set LastFocusedOfferId(LastFocusedOfferId:String=Get OfferId!(self:Fort Mtx Store Offer VM Object Reference=Get FortMtxStoreOfferVM!(self:WBP Vbucks Store Offer Tile Object Reference=@n2) -> (FortMtxStoreOfferVM:Fort Mtx Store Offer VM Object Reference)) -> (OfferId:String)) -> (Output_Get:String)
   n3 = OnPurchaseCompleted!() -> (OutputDelegate, then, bSuccess:bool=~) {is=title K2Node_CustomEvent}
   > Call Purchase Completed!(execute:Exec, self:WBP Vbucks Store Content Base Object Reference, Success:bool=@n3.bSuccess) -> (then:Exec) {is=title K2Node_CallDelegate}
  ]}
 FindOfferTileById {nodes [
   n1 = FindOfferTileById!() -> (then, OfferId:String) {is=title K2Node_FunctionEntry}
   > n2 = Branch(Condition=IsEmpty(InString=@n1))
   > Return Node!(execute, OfferTile:WBP Vbucks Store Offer Tile Object Reference)
   n3 = For Each Loop(
    Exec=@n2.else
    Array:Array of WBP Vbucks Store Offer Tile Object References=Get OffersWidgets() -> (OffersWidgets:Array of WBP Vbucks Store Offer Tile Object References)
   ) -> (Array Element:WBP Vbucks Store Offer Tile Object Reference)
   Return Node(execute=@n3.Completed)
   n4 = Get FortMtxStoreOfferVM(self=@n3)
   > Branch(Condition="Equal Exactly (String)"(A=Get OfferId(self=@n4), B=Get OfferId!() -> (OfferId)))
   > Return Node(OfferTile=@n3)
  ]}
 GetDesiredFocusTile {nodes [
   GetDesiredFocusTile() {is=title K2Node_FunctionEntry}
   > Return Node!(execute, Tile:WBP Vbucks Store Offer Tile Object Reference)
  ]}
 InitializeFocusedOffer {nodes [InitializeFocusedOffer() {is=title K2Node_FunctionEntry}]}
 OfferTileFocused {nodes [OfferTileFocused!() -> (then, Button:Common Button Base Object Reference) {is=title K2Node_FunctionEntry}]}
 OnSetIsPurchasing {nodes [
   n1 = OnSetIsPurchasing!() -> (then, IsPurchasing:bool=~) {is=title K2Node_FunctionEntry}
   > Call Set Is Purchasing!(execute:Exec, self:WBP Vbucks Store Content Base Object Reference, IsPurchasing:bool=@n1) -> (then:Exec) {is=title K2Node_CallDelegate}
  ]}
 PurchaseCompleted {nodes [PurchaseCompleted!() -> (then, Success:bool=~) {is=title K2Node_FunctionEntry}]}
 SetExactPackOfferVM {nodes [
   n1 = SetExactPackOfferVM!() -> (then, ExactAmountOffer:Fort Mtx Store Offer VM Object Reference) {is=title K2Node_FunctionEntry}
   > n2 = Set ExactPackOfferVM(ExactPackOfferVM:Fort Mtx Store Offer VM Object Reference=@n1) -> (Output_Get:Fort Mtx Store Offer VM Object Reference)
   > Set HasExactPack(HasExactPack:bool=IsValid(Object=@n2)) -> (Output_Get:bool)
  ]}
 SetIsPurchasing {nodes [SetIsPurchasing!() -> (then, IsPurchasing:bool=~) {is=title K2Node_FunctionEntry}]}
 SetupCommonOffers {nodes [SetupCommonOffers() {is=title K2Node_FunctionEntry}]}
 SetupExactPackOffer {nodes [SetupExactPackOffer() {is=title K2Node_FunctionEntry}]}
 TriggerTilesStaggeredIntroAnimation {nodes [
   TriggerTilesStaggeredIntroAnimation() {is=title K2Node_FunctionEntry}
   > n1 = For Each Loop(Array=Get OffersWidgets())
   > PlayIntroAnimation!(
    execute
    self:WBP Vbucks Store Offer Tile Object Reference=@n1
    DelayTime:double=float * float(A:int=@n1.Array Index, B=Get OffersDelayTime() -> (OffersDelayTime:double)) {is=title K2Node_PromotableOperator}
   ) -> (then)
  ]}
]
variables [
 ExactPackOfferVM {type=FortMtxStoreOfferVM, default=null}
 HasExactPack {type=boolean}
 LastFocusedOfferId {type=string, default=""}
 OffersDelayTime {type=number, default=0.05}
 OffersWidgets {type="array<WBP_VbucksStore_OfferTile_C>", default []}
 OfferTileFocused {}
 PurchaseCompleted {}
 SetIsPurchasing {}
]
widgets {info {inheritedWidgetCount=0, parentClass=&CommonActivatableWidget, rootWidgetClass=null, widgetCount=0}, widgets []}
