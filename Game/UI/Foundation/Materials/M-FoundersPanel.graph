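# M-FoundersPanel is a translucent UI material that procedurally builds a soft, rounded-box falloff with a square cut out of its centre.
# Centred absolute UVs drive a SmoothStep distance falloff, from which a hard inner square made with Ceil and VectorLength is subtracted.
# The resulting mask blends between cyan-green and pink, and a heavily scaled-down copy gives a faint value, likely for opacity.
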Material /Game/UI/Foundation/Materials/M-FoundersPanel
graph {
 properties {
  bUsedWithStaticMesh=false
  blendMode=BLEND_Translucent
  materialDomain=MD_UI
  pixelDepthOffsetMode=PDOM_Legacy
  usedShadingModels=Default Lit
 }
 nodes [
  Constant3Vector()
  n1 = Abs(None=Add(A=TextureCoordinate()) {constB=-0.5})
  n2 = Subtract(A=@n1, B=Constant() {r=0.04})
  n3 = MaterialFunctionCall(
   Alpha=MaterialFunctionCall(Vector 2=Max(A=@n2) {constB=0}) {materialFunction=/Engine/Functions/Engine_MaterialFunctions02/Utility/VectorLength}.V2 Length
   Min=Constant()
   Max=Constant() {r=0.5}
  ) {materialFunction=&SmoothStep}
  n4 = Constant() {r=0.15}
  n5 = MaterialFunctionCall(Vector 2=Ceil(None=Subtract(A=@n1, B=@n4))) {materialFunction=&VectorLength}
  n6 = Subtract(A=Clamp(None=OneMinus(None=@n3.Result)), B=Clamp(None=OneMinus(None=@n5.V2 Length)))
  LinearInterpolate(A=Constant3Vector() {constant=<rgba 0 1 0.6042 1>}, B=Constant3Vector() {constant=<rgba 1 0 0.4885 1>}, Alpha=@n6)
  Multiply(A=@n6, B=Constant() {r=0.05})
 ]
}
