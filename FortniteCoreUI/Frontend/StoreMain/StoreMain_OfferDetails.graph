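# Store offer details screen on FortRealMoneyOfferDetails, showing an offer's key art, name, description, prices, attributes and V-Bucks quantities.
# Init runs on offer set, or on activation under bAltUpdateFlow, filling widgets and gradient materials from the offer's display asset.
# July 2019 experiment flags let currency packs substitute quantity-keyed A/B assets for art and colors.
# The purchase button reflects owned or purchasing states, and blocked purchases replace the info panel with a blocker.
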WidgetBlueprint /FortniteCoreUI/Frontend/StoreMain/StoreMain_OfferDetails
graphs [
 ContentBlocked {nodes [
   n1 = ContentBlocked!() -> (then, isContentBlocked:bool=~, Content Blocked Text:Text) {is=title K2Node_FunctionEntry}
   > Set Data!(
    execute
    self:WBP Content Purchase Blocker Object Reference=Get WBP_ContentPurchaseBlocker() -> (WBP_ContentPurchaseBlocker:WBP Content Purchase Blocker Object Reference)
    InText:Text=@n1.Content Blocked Text
   ) -> (then) {is=title K2Node_CallFunction}
   > SetVisibility(self=Get WBP_ContentPurchaseBlocker(), InVisibility=Select(Option 0=Collapsed, Option 1=SelfHitTestInvisible, Index=@n1))
   > SetVisibility(
    self=Get Overlay_VBucks() -> (Overlay_VBucks:Overlay Object Reference)
    InVisibility=Select(Option 0=SelfHitTestInvisible, Option 1=Collapsed, Index=@n1)
   )
  ]}
 EventGraph {nodes [
   n1 = Event SetLoadingState!() -> (OutputDelegate, then, bIsLoading:bool) {is=title K2Node_Event}
   > SetVisibility(
    self=Get Loading_Overlay() -> (Loading_Overlay:Overlay Object Reference)
    InVisibility=Select(Option 0=Collapsed, Option 1=HitTestInvisible, Index=@n1.bIsLoading)
   )
   Event On Offer Set() {is=title K2Node_Event}
   n2 = Get bAltUpdateFlow!(self:Fort Runtime Options Object Reference=GetRuntimeOptions()) -> (bAltUpdateFlow:bool)
   > n3 = Branch(Condition=@n2)
   Init(execute=[@n3.else, @n4.then])
   Event On Activated()
   > n4 = Branch(Condition=@n2)
   Event Construct()
   > Set ContentPurchaseFocusWidget(ContentPurchaseFocusWidget:Widget Object Reference=Get WBP_UIKit_Button_Loud!(self:WBP Content Purchase Blocker Object Reference=Get WBP_ContentPurchaseBlocker()) -> (WBP_UIKit_Button_Loud:WBP UIKit Button Loud Object Reference)) -> (Output_Get:Widget Object Reference)
   > SetMobileStyle() {is=title K2Node_CallFunction}
   Event ResetPurchaseButtonText() {is=title K2Node_Event}
   > UpdatePurchaseButton()
   n5 = Event UpdatePurchaseBlocked!() -> (OutputDelegate, then, bIsBlocked:bool, PurchaseNotAllowedReason:"Text (by ref)") {is=title K2Node_Event}
   > ContentBlocked!(execute, self, isContentBlocked:bool=@n5.bIsBlocked, Content Blocked Text:Text=@n5.PurchaseNotAllowedReason) -> (then) {is=title K2Node_CallFunction}
   Event OnPurchasingStarted() {is=title K2Node_Event}
   > SetPurchaseButtonText!(execute, self, ButtonText:Text="LOCTEXT(\"PURCHASING...\")") -> (then)
  ]}
 GetCurrencyABAssetFromQuantity {nodes [
   n1 = GetCurrencyABAssetFromQuantity!() -> (then, Quantity:int) {is=title K2Node_FunctionEntry}
   n2 = Find(
    TargetMap:Map of Integers to Fort Mtx Offer Data Object References=Get ABAssetCollection() -> (ABAssetCollection:Map of Integers to Fort Mtx Offer Data Object References)
    Key:"Integer (by ref)"=@n1
   ) -> (Value:Fort Mtx Offer Data Object Reference)
   > n3 = Branch(Condition=@n2.ReturnValue)
   > Return Node!(execute, DataAsset:Fort Mtx Offer Data Object Reference=@n2)
   Return Node(execute=@n3.else)
  ]}
 HandleBack {nodes [
   HandleBack() {is=title K2Node_FunctionEntry}
   > PopDetails()
   > Return Node!(execute, PassThrough:bool=false)
   IsRealMoneyPurchasePending!(self:Fort Mcp Context Object Reference=Get FortMcpContext!() -> (ReturnValue:Fort Mcp Context Object Reference) {is=title K2Node_GetBlueprintContextEx}) -> (ReturnValue:bool)
  ]}
 InitCurrency {nodes [
   InitCurrency() {is=title K2Node_FunctionEntry}
   > n1 = Sequence!(execute) -> (then_0, then_1, then_2:Exec)
   n2 = GetTotalQuantity!(self:Fort Store Front Offer Info Object Reference=GetOfferInfo!(self:Fort Store Front Offer Details Widget Base Object Reference) -> (ReturnValue:Fort Store Front Offer Info Object Reference)) -> (ReturnValue:int=0)
   > SetVisibility(
    self=Get MTXArea() -> (MTXArea:Size Box Object Reference)
    InVisibility=Select(Option 0=Collapsed, Option 1=SelfHitTestInvisible, Index="integer > integer"(A=@n2, B=0) {is=title K2Node_CallFunction})
   )
   n3 = "integer > integer"(A=GetBonusQuantity!(self:Fort Store Front Offer Info Object Reference=GetOfferInfo()) -> (ReturnValue:int=0), B=0)
   n4 = Get bEnableCurrencyInspectScreenBonusText!(self:Fort Runtime Options Object Reference=GetRuntimeOptions()) -> (bEnableCurrencyInspectScreenBonusText:bool)
   SetVisibility(
    execute=@n1.then_1
    self=Get BonusMTXArea() -> (BonusMTXArea:Horizontal Box Object Reference)
    InVisibility=Select(Option 0=Collapsed, Option 1=SelfHitTestInvisible, Index=AND Boolean(A=@n3, B=@n4))
   )
   n5 = GetOfferInfo()
   "SetText (Text)"(
    execute=@n1.then_2
    self=Get CommonText_BaseQuantity() -> (CommonText_BaseQuantity:Common Text Object Reference)
    InText="To Text (Integer)"(Value=GetBaseQuantity!(self:Fort Store Front Offer Info Object Reference=@n5) -> (ReturnValue:int=0))
   )
   > "SetText (Text)"(
    self=Get CommonText_BonusQuantity() -> (CommonText_BonusQuantity:Common Text Object Reference)
    InText="To Text (Integer)"(Value=GetBonusQuantity(self=@n5) -> (ReturnValue=0))
   )
   > "SetText (Text)"(
    self=Get TextDescription() -> (TextDescription:Common Text Object Reference)
    InText=GetDescription!(self:Fort Store Front Offer Info Object Reference=GetOfferInfo()) -> (ReturnValue:Text)
   )
   n6 = GetTotalQuantity(self=GetOfferInfo()) -> (ReturnValue=0)
   > SetVisibility(
    self=Get BorderDisclaimer() -> (BorderDisclaimer:Common Border Object Reference)
    InVisibility=Select(Option 0=Collapsed, Option 1=SelfHitTestInvisible, Index="integer > integer"(A=@n6, B=0))
   )
  ]}
 InitDetailAttributes {nodes [
   InitDetailAttributes() {is=title K2Node_FunctionEntry}
   n1 = GetOfferInfo()
   > n2 = For Each Loop(Array:Array of Fort Mtx Details Attribute Structures=Get DetailsAttributes!(self:Fort Mtx Offer Data Object Reference=GetOfferDisplayAsset!(self:Fort Store Front Offer Info Object Reference=@n1) -> (ReturnValue:Fort Mtx Offer Data Object Reference)) -> (DetailsAttributes:Array of Fort Mtx Details Attribute Structures)) -> (Array Element:Fort Mtx Details Attribute Structure)
   n3 = Break Fort Mtx Details Attribute!(FortMtxDetailsAttribute:"Fort Mtx Details Attribute Structure (by ref)"=@n2) -> (Name:Text, Value:Text) {is=title K2Node_BreakStruct}
   > n4 = Create Widget!(
    execute:Exec
    Class:User Widget Class Reference=&/StoreMain_OfferDetailsAttribute.StoreMain_OfferDetailsAttribute_C
    OwningPlayer:Player Controller Object Reference=GetOwningPlayer()
    AttributeDescription:Text=@n3
    AttributeValue:Text=@n3.Value
   ) -> (then:Exec, ReturnValue:Store Main Offer Details Attribute Object Reference)
   > AddChildToVerticalBox(self=Get VBoxOfferInfo() -> (VBoxOfferInfo:Vertical Box Object Reference), Content=@n4)
  ]}
 InitGraph {nodes [
   Init() {is=title K2Node_CustomEvent}
   > Set DisplayAsset(DisplayAsset:Fort Mtx Offer Data Object Reference=GetOfferDisplayAsset(self=GetOfferInfo())) -> (Output_Get:Fort Mtx Offer Data Object Reference)
   > ClearChildren(self=Get VBoxOfferInfo())
   n1 = Get FortExperimentContext!() -> (ReturnValue:Fort Experiment Context Object Reference) {is=title K2Node_GetBlueprintContextEx}
   > SetBrushFromMaterial(
    self=Get OverlayGradient() -> (OverlayGradient:Image Object Reference)
    Material=Select(
     Option 0:Material Interface Object Reference=/Game/UI/Foundation/Materials/M-Store-GradientOverlay
     Option 1:Material Interface Object Reference=/Game/UI/Foundation/Materials/M-Store-GradientOverlay_Radial
     Index=ShouldShowNewVBuckColorsJuly2019!(self:Fort Experiment Context Object Reference=@n1) -> (ReturnValue:bool)
    ) -> (ReturnValue:Material Interface Object Reference)
   )
   n2 = Get FortExperimentContext()
   n3 = GetOfferInfo()
   n4 = GetFortStoreOfferType!(self:Fort Store Front Offer Info Object Reference=@n3) -> (ReturnValue:EFortMtxStoreOfferType Enum=FoundersPack)
   > n5 = Branch(Condition=AND Boolean(
     A="Equal (Enum)"(A:EFortMtxStoreOfferType Enum=@n4, B:EFortMtxStoreOfferType Enum=CurrencyPack)
     B=ShouldShowNewVBuckCoinAssetsJuly2019!(self:Fort Experiment Context Object Reference=@n2) -> (ReturnValue:bool)
    ))
   > n6 = GetCurrencyABAssetFromQuantity!(execute, self, Quantity:int=GetTotalQuantity(self=@n3) -> (ReturnValue=0)) -> (then, DataAsset:Fort Mtx Offer Data Object Reference) {is=title K2Node_CallFunction}
   > n7 = Is Valid(InputObject=@n6)
   n8 = UpdateKeyArt!(
    execute=[@n5.else, @n7.Is Not Valid]
    self
    InDisplayAsset:Fort Mtx Offer Data Object Reference=Get DisplayAsset() -> (DisplayAsset:Fort Mtx Offer Data Object Reference)
   ) -> (then)
   n9 = Get FortExperimentContext()
   n10 = GetOfferInfo()
   n11 = GetFortStoreOfferType(self=@n10) -> (ReturnValue=FoundersPack)
   n12 = Branch(execute=[@n25.then, @n8.then], Condition=AND Boolean(A="Equal (Enum)"(A=@n11, B=CurrencyPack), B=ShouldShowNewVBuckColorsJuly2019(self=@n9)))
   > n13 = GetCurrencyABAssetFromQuantity(Quantity=GetTotalQuantity(self=@n10) -> (ReturnValue=0))
   > n14 = Is Valid(InputObject=@n13)
   SetColorAndOpacity(
    execute=[@n12.else, @n14.Is Not Valid]
    self:Image Object Reference=[Get BaseColor() -> (BaseColor:Image Object Reference), Get BaseColorDetails() -> (BaseColorDetails:Image Object Reference)]
    InColorAndOpacity:Linear Color Structure=Get Background!(self:Fort Mtx Offer Data Object Reference=Get DisplayAsset()) -> (Background:Linear Color Structure)
   )
   > SetOpacity!(execute, self:Image Object Reference=Get BaseColorDetails(), InOpacity:float=.75) -> (then)
   > n15 = GetDynamicMaterial(self=Get OverlayGradient())
   > n16 = GetDynamicMaterial(self=Get OverlayGradientDetails() -> (OverlayGradientDetails:Image Object Reference))
   n17 = Break Fort Mtx Gradient!(FortMtxGradient:"Fort Mtx Gradient Structure (by ref)"=Get Gradient!(self:Fort Mtx Offer Data Object Reference=Get DisplayAsset()) -> (Gradient:Fort Mtx Gradient Structure)) -> (Start:Linear Color Structure, Stop:Linear Color Structure) {is=title K2Node_BreakStruct}
   > SetVectorParameterValue(self=[@n16, @n15], ParameterName=RGBA1, Value=@n17)
   > n18 = SetVectorParameterValue(self=[@n16, @n15], ParameterName=RGBA0, Value=@n17.Stop)
   n19 = GetOfferInfo()
   "SetText (Text)"(
    execute=[@n24.then, @n18.then]
    self=Get TextName() -> (TextName:Common Text Object Reference)
    InText=GetName!(self:Fort Store Front Offer Info Object Reference=@n19) -> (ReturnValue:Text)
   )
   > "SetText (Text)"(
    self=Get TextOriginalPrice() -> (TextOriginalPrice:Common Text Object Reference)
    InText=GetRegularPrice!(self:Fort Store Front Offer Info Object Reference=@n19, PriceIndex:int=0, ItemQuantity:int=1) -> (ReturnValue:Text)
   )
   n20 = GetSalePriceFromLastAppStore!(self:Fort Real Money Offer Details Object Reference, PriceIndex:int=0, ItemQuantity:int=1) -> (SalePrice:Text, ReturnValue:bool)
   > "SetText (Text)"(self=Get TextOfferDisplayPrice() -> (TextOfferDisplayPrice:Common Text Object Reference), InText=@n20)
   > SetVisibility(
    self=Get OverlaySalePrice() -> (OverlaySalePrice:Overlay Object Reference)
    InVisibility=Select(Option 0:ESlateVisibility Enum=Collapsed, Option 1:ESlateVisibility Enum=SelfHitTestInvisible, Index=@n20.ReturnValue) -> (ReturnValue:ESlateVisibility Enum)
   )
   > Set Text(
    Text=GetAdditionalPriceText!(self:Fort Real Money Offer Details Object Reference) -> (ReturnValue:"Text (by ref)")
    self=Get Text_AdditionalPriceText() -> (Text_AdditionalPriceText:Common Text Object Reference)
   )
   > SetupCurrencyImage()
   > InitStoreTypeVisuals()
   > InitCurrency() {is=title K2Node_CallFunction}
   > InitDetailAttributes() {is=title K2Node_CallFunction}
   > UpdatePurchaseButton()
   SetColorAndOpacity(execute=@n14.Is Valid, self=[Get BaseColor(), Get BaseColorDetails()], InColorAndOpacity=Get Background(self=@n13))
   > SetOpacity(self=Get BaseColorDetails(), InOpacity=.75)
   > n21 = GetDynamicMaterial(self=Get OverlayGradient())
   > n22 = GetDynamicMaterial(self=Get OverlayGradientDetails())
   n23 = Break Fort Mtx Gradient(FortMtxGradient=Get Gradient(self=@n13))
   > SetVectorParameterValue(self=[@n22, @n21], ParameterName=RGBA1, Value=@n23)
   > n24 = SetVectorParameterValue(self=[@n21, @n22], ParameterName=RGBA0, Value=@n23.Stop)
   n25 = UpdateKeyArt(execute=@n7.Is Valid, InDisplayAsset=@n6)
  ]}
 InitStoreTypeVisuals {nodes [
   InitStoreTypeVisuals() {is=title K2Node_FunctionEntry}
   n1 = GetFortStoreOfferType(self=GetOfferInfo()) -> (ReturnValue=FoundersPack)
   > SetVisibility(
    self=Get VBuckHighlight() -> (VBuckHighlight:Image Object Reference)
    InVisibility=Select(Option 0=Collapsed, Option 1=SelfHitTestInvisible, Index="Equal (Enum)"(A=@n1, B=CurrencyPack))
   )
  ]}
 IsGiftingStorefrontOffer {nodes [
   IsGiftingStorefrontOffer() {is=title K2Node_FunctionEntry}
   > Return Node!(
    execute
    ReturnValue:bool=IsGiftingStorefrontOffer!(self:Fort Store Front Offer Info Object Reference=GetOfferInfo()) -> (ReturnValue:bool) {is=title K2Node_CallFunction}
   )
  ]}
 PopDetails {nodes [
   PopDetails() {is=title K2Node_FunctionEntry}
   > PopContentWidget!(execute) -> (then)
  ]}
 RegisterInput {nodes [
   RegisterInput() {is=title K2Node_FunctionEntry}
   > SetInputActionHandler!(
    execute
    self:Common Activatable Panel Legacy Object Reference
    InputActionRow:Data Table Row Handle Structure=Make Data Table Row Handle!(DataTable:Data Table Object Reference=&InputActionDataTable, RowName:Name=Back) -> (DataTableRowHandle:Data Table Row Handle Structure) {is=title K2Node_MakeStruct}
    CommitedEvent:Delegate=Create Event()
   ) -> (then)
  ]}
 SetMobileStyle {nodes [
   SetMobileStyle() {is=title K2Node_FunctionEntry}
   > Branch(Condition=IsMobileGame())
   > SetPadding!(
    execute
    self:Overlay Slot Object Reference=SlotAsOverlaySlot(Widget=Get SafeZonePlayerInfo() -> (SafeZonePlayerInfo:Safe Zone Object Reference))
    InPadding_Left:float=0
    InPadding_Top:float=24
    InPadding_Right:float=48
    InPadding_Bottom:float=0
   ) -> (then)
  ]}
 SetPurchaseButtonText {nodes [
   SetPurchaseButtonText!() -> (then, ButtonText:Text) {is=title K2Node_FunctionEntry}
   > n1 = Cast To WBP_UIKit_ButtonCTA_Base!(
    execute
    Object=GetPurchaseButtonCurrent!(self:Fort Real Money Offer Details Object Reference) -> (ReturnValue:Fort CTAButton Object Reference)
   ) -> (then, CastFailed, AsWBP UIKit Button CTA Base:WBP UIKit Button CTA Base Object Reference) {is=title K2Node_DynamicCast}
   > SetText!(execute, self:Fort CTAButton Object Reference=@n1, ButtonText:"Text (by ref)"=Get ButtonText() -> (ButtonText:Text)) -> (then)
  ]}
 SetupCurrencyImage {nodes [
   SetupCurrencyImage() {is=title K2Node_FunctionEntry}
   n1 = IsPriceInRealMoney!(self:Fort Store Front Offer Info Object Reference=GetOfferInfo(), PriceIndex:int=0) -> (ReturnValue:bool)
   > SetVisibility(
    self=Get Image_CurrencyIcon() -> (Image_CurrencyIcon:Fort Lazy Image Object Reference)
    InVisibility=Select(Option 0=Visible, Option 1=Collapsed, Index=@n1)
   )
   > UpdateCurrency!(
    execute
    self:WBP Price Platform Store Icon Object Reference=Get WBP_Price_PlatformStoreIcon() -> (WBP_Price_PlatformStoreIcon:WBP Price Platform Store Icon Object Reference)
    CurrencyType:EFortPriceCurrency Enum=Select(Option 0:EFortPriceCurrency Enum=Mtx, Option 1:EFortPriceCurrency Enum=RealMoney, Index=@n1) -> (ReturnValue:EFortPriceCurrency Enum)
   ) -> (then)
   > n2 = Set PriceItem(PriceItem:Fort Account Item Definition Object Reference=GetPriceItem!(self:Fort Store Front Offer Info Object Reference=GetOfferInfo(), PriceIndex:int=0) -> (RequiredItemCount:int=0, ReturnValue:Fort Account Item Definition Object Reference).ReturnValue) -> (Output_Get:Fort Account Item Definition Object Reference)
   > Is Valid(InputObject=@n2)
   > SetBrushFromLazyDisplayAsset!(
    execute
    self:Common Lazy Image Object Reference=Get Image_CurrencyIcon()
    LazyObject:"Soft Object Reference (by ref)"=GetSmallPreviewImage!(self:Fort Item Definition Object Reference=Get PriceItem() -> (PriceItem:Fort Account Item Definition Object Reference)) -> (ReturnValue:Texture 2D Soft Object Reference)
    bMatchTextureSize:bool=false
   ) -> (then)
  ]}
 UnregisterInput {nodes [
   UnregisterInput() {is=title K2Node_FunctionEntry}
   > RemoveAllInputActionHandlers(self:Common Activatable Panel Legacy Object Reference)
  ]}
 UpdateKeyArt {nodes [
   n1 = UpdateKeyArt!() -> (then, InDisplayAsset:Fort Mtx Offer Data Object Reference) {is=title K2Node_FunctionEntry}
   > SetDesiredSizeOverride!(
    execute
    self:Image Object Reference=Get KeyArt() -> (KeyArt:Fort Lazy Image Object Reference)
    DesiredSize:Vector 2D Structure=Get DetailsImageSize!(self:Fort Mtx Offer Data Object Reference=@n1) -> (DetailsImageSize:Vector 2D Structure)
   ) -> (then)
   n2 = GetSoftDetailsImageURL!(self:Fort Mtx Offer Data Object Reference=@n1, World:World Object Reference, TextureSize:int=1040) -> (ReturnValue:String)
   > n3 = Branch(Condition=NOT Boolean(A=IsEmpty(InString=@n2)))
   > SetBrushFromURL!(
    execute
    self:Fort Lazy Image Object Reference=Get KeyArt()
    CMSImageURL:String=@n2
    bMatchSize:bool=false
    Delay:float=0
    DownloadPriority:ECMSDownloadPriority Enum=Normal
   ) -> (then)
   SetBrushFromLazyDisplayAsset(
    execute=@n3.else
    self=Get KeyArt()
    LazyObject=Get SoftDetailsImage!(self:Fort Mtx Offer Data Object Reference=@n1) -> (SoftDetailsImage:Soft Object Reference)
    bMatchTextureSize=false
   )
  ]}
 UpdatePurchaseButton {nodes [
   UpdatePurchaseButton() {is=title K2Node_FunctionEntry}
   > n1 = Sequence!(execute) -> (then_0, then_1)
   n2 = AreGrantedItemsOwned!(self:Fort Store Front Offer Info Object Reference=GetOfferInfo()) -> (ReturnValue:bool)
   > SetPurchaseButtonText(ButtonText=Select(Option 0:Text="LOCTEXT(\"Purchase\")", Option 1:Text="LOCTEXT(\"Owned\")", Index=@n2) -> (ReturnValue:Text)) {is=title K2Node_CallFunction}
   > n3 = Cast To WBP_UIKit_ButtonCTA_Base(Object=GetPurchaseButtonCurrent())
   > SetIsInteractionEnabled!(execute, self:Common Button Base Object Reference=@n3, bInIsInteractionEnabled:bool=NOT Boolean(A=@n2)) -> (then)
   n4 = GetOfferInfo()
   Branch(execute=@n1.then_1, Condition=OR Boolean(A=IsMobileGame(), B=IsPriceInRealMoney(self=@n4, PriceIndex=0)))
   > SetRequiresHold!(
    execute
    self:Common Button Base Object Reference=Get Button_Purchase() -> (Button_Purchase:Fort CTAButton Object Reference)
    bInRequiresHold:bool=false
   ) -> (then)
   > SetRequiresHold(self=Get Button_Purchase_STW() -> (Button_Purchase_STW:Fort CTAButton Object Reference), bInRequiresHold=false)
  ]}
]
variables [
 ABAssetCollection {default {
   1000=/Game/Catalog/DisplayAssets/DA_MtxPack1000_ABTest
   13500=/Game/Catalog/DisplayAssets/DA_MtxPack13500_ABTest
   2800=/Game/Catalog/DisplayAssets/DA_MtxPack2800_ABTest
   5000=/Game/Catalog/DisplayAssets/DA_MtxPack5000_ABTest
  }}
 DisplayAsset {type=FortMtxOfferData, default=null}
 isContentBlocked {type=boolean}
 PreviewPlayStationPlatform {type=boolean, default=true}
]
widgets {
 info {inheritedWidgetCount=10, parentClass=/Script/FortniteUI.FortRealMoneyOfferDetails, rootWidgetClass=&Overlay, widgetCount=72}
 widgets [
  Overlay_2 {widgetClassPath=&Overlay, widget=Overlay_2}
  BaseColor {
   widgetClassPath=&Image
   bIsVariable=true
   parent=Overlay_2
   widget=BaseColor
   properties {colorAndOpacity=<rgba 0.5149 0.5333 0.6445 1>}
   slot {class=OverlaySlot, properties {horizontalAlignment=HAlign_Fill, verticalAlignment=VAlign_Fill}}
  }
  OverlayGradient {
   widgetClassPath=&Image
   bIsVariable=true
   parent=Overlay_2
   widget=OverlayGradient
   properties {brush {
     bIsDynamicallyLoaded=false
     drawAs=Image
     imageType=NoImage
     mirroring=NoMirror
     resourceName=None
     resourceObject=&M-Store-GradientOverlay
     tiling=NoTile
     imageSize=<32 32>
     margin=<ltrb 0 0 0 0>
     outlineSettings {
      bUseBrushTransparency=false
      roundingType=HalfHeightRadius
      width=0
      color {colorUseRule=UseColor_Specified, specifiedColor=<rgba 0 0 0 0>}
      cornerRadii=<0 0 0 0>
     }
     tintColor {colorUseRule=UseColor_Specified, specifiedColor=<rgba 1 1 1 1>}
     uVRegion {bIsValid=false, max=<0 0>, min=<0 0>}
    }}
   slot {class=OverlaySlot, properties {horizontalAlignment=HAlign_Fill, verticalAlignment=VAlign_Fill}}
  }
  WBP_CaptureForPostBufferUpdate {
   widgetClassPath=&WBP_CaptureForPostBufferUpdate
   bIsVariable=true
   parent=Overlay_2
   slot {class=OverlaySlot, properties {horizontalAlignment=HAlign_Fill, verticalAlignment=VAlign_Fill}}
  }
  SafeZone_0 {
   widgetClassPath=&SafeZone
   parent=Overlay_2
   widget=SafeZone_0
   properties {padBottom=false}
   slot {class=OverlaySlot, properties {verticalAlignment=VAlign_Center}}
  }
  SizeBox_12 {
   widgetClassPath=&SizeBox
   parent=SafeZone_0
   widget=SizeBox_12
   properties {heightOverride=900, widthOverride=1664}
   slot {class=SafeZoneSlot}
  }
  Overlay_1 {widgetClassPath=&Overlay, parent=SizeBox_12, widget=Overlay_1, slot {class=SizeBoxSlot}}
  VBuckHighlight {
   widgetClassPath=&Image
   bIsVariable=true
   parent=Overlay_1
   widget=VBuckHighlight
   properties {
    visibility=Collapsed
    brush {
     bIsDynamicallyLoaded=false
     drawAs=Image
     imageType=NoImage
     mirroring=NoMirror
     resourceName=None
     resourceObject=/Game/UI/Foundation/Materials/M-Store-Highlight
     tiling=NoTile
     imageSize=<1536 880>
     margin=<ltrb 0 0 0 0>
     outlineSettings {
      bUseBrushTransparency=false
      roundingType=HalfHeightRadius
      width=0
      color {colorUseRule=UseColor_Specified, specifiedColor=<rgba 0 0 0 0>}
      cornerRadii=<0 0 0 0>
     }
     tintColor {colorUseRule=UseColor_Specified, specifiedColor=<rgba 1 1 1 1>}
     uVRegion {bIsValid=false, max=<0 0>, min=<0 0>}
    }
    colorAndOpacity=<rgba 1 1 1 0.125>
   }
   slot {class=OverlaySlot, properties {horizontalAlignment=HAlign_Right}}
  }
  Highlight {
   widgetClassPath=&Image
   bIsVariable=true
   parent=Overlay_1
   widget=Highlight
   properties {
    brush {
     bIsDynamicallyLoaded=false
     drawAs=Image
     imageType=NoImage
     mirroring=NoMirror
     resourceName=None
     resourceObject=&M-Store-Highlight
     tiling=NoTile
     imageSize=<1152 768>
     margin=<ltrb 0 0 0 0>
     outlineSettings {
      bUseBrushTransparency=false
      roundingType=HalfHeightRadius
      width=0
      color {colorUseRule=UseColor_Specified, specifiedColor=<rgba 0 0 0 0>}
      cornerRadii=<0 0 0 0>
     }
     tintColor {colorUseRule=UseColor_Specified, specifiedColor=<rgba 1 1 1 1>}
     uVRegion {bIsValid=false, max=<0 0>, min=<0 0>}
    }
    colorAndOpacity=<rgba 1 1 1 0>
   }
   slot {class=OverlaySlot, properties {horizontalAlignment=HAlign_Right}}
  }
  HorizontalBox_0 {
   widgetClassPath=&HorizontalBox
   parent=Overlay_1
   widget=HorizontalBox_0
   slot {class=OverlaySlot, properties {horizontalAlignment=HAlign_Fill, verticalAlignment=VAlign_Center}}
  }
  Spacer_0 {
   widgetClassPath=&Spacer
   parent=HorizontalBox_0
   widget=Spacer_0
   properties {size=<528 1>}
   slot {class=HorizontalBoxSlot, properties {horizontalAlignment=HAlign_Left}}
  }
  ScaleBox_0 {
   widgetClassPath=&ScaleBox
   parent=HorizontalBox_0
   widget=ScaleBox_0
   slot {class=HorizontalBoxSlot, properties {size {sizeRule=Fill, value=1}}}
  }
  KeyArt {
   widgetClassPath=&FortLazyImage
   bIsVariable=true
   parent=ScaleBox_0
   widget=KeyArt
   properties {
    bOverride_Cursor=true
    visibility=HitTestInvisible
    brush {
     bIsDynamicallyLoaded=false
     drawAs=Image
     imageType=FullColor
     mirroring=NoMirror
     resourceName=None
     resourceObject=/Game/UI/ContentOnDemand/ItemDownloadIndicator/MI_ItemDownloadIndicator_SpinnerMaterial
     tiling=NoTile
     imageSize=<64 64>
     margin=<ltrb 0 0 0 0>
     outlineSettings {
      bUseBrushTransparency=false
      roundingType=HalfHeightRadius
      width=0
      color {colorUseRule=UseColor_Specified, specifiedColor=<rgba 0 0 0 0>}
      cornerRadii=<0 0 0 0>
     }
     tintColor {colorUseRule=UseColor_Specified, specifiedColor=<rgba 1 1 1 1>}
     uVRegion {bIsValid=false, max=<0 0>, min=<0 0>}
    }
    loadingBackgroundBrush {
     bIsDynamicallyLoaded=false
     drawAs=Image
     imageType=FullColor
     mirroring=NoMirror
     resourceName=None
     resourceObject=&MI_ItemDownloadIndicator_SpinnerMaterial
     tiling=NoTile
     imageSize=<64 64>
     margin=<ltrb 0 0 0 0>
     outlineSettings {
      bUseBrushTransparency=false
      roundingType=HalfHeightRadius
      width=0
      color {colorUseRule=UseColor_Specified, specifiedColor=<rgba 0 0 0 0>}
      cornerRadii=<0 0 0 0>
     }
     tintColor {colorUseRule=UseColor_Specified, specifiedColor=<rgba 1 1 1 1>}
     uVRegion {bIsValid=false, max=<0 0>, min=<0 0>}
    }
   }
   slot {class=ScaleBoxSlot}
  }
  CommonBorder_1 {
   widgetClassPath=&CommonBorder
   parent=Overlay_1
   widget=CommonBorder_1
   properties {
    horizontalAlignment=HAlign_Left
    style=/FortUILibrary/Styles/Border/Border-FloatingShadow.Border-FloatingShadow_C
    verticalAlignment=VAlign_Top
    background {
     bIsDynamicallyLoaded=false
     drawAs=Box
     imageType=NoImage
     mirroring=NoMirror
     resourceName=None
     resourceObject=/Game/UI/Foundation/Materials/M-FoundersPanel
     tiling=NoTile
     imageSize=<138 138>
     margin=<ltrb 0.5 0.5 0.5 0.5>
     outlineSettings {
      bUseBrushTransparency=false
      roundingType=HalfHeightRadius
      width=0
      color {colorUseRule=UseColor_Specified, specifiedColor=<rgba 0 0 0 0>}
      cornerRadii=<0 0 0 0>
     }
     tintColor {colorUseRule=UseColor_Specified, specifiedColor=<rgba 1 1 1 1>}
     uVRegion {bIsValid=false, max=<0 0>, min=<0 0>}
    }
    padding=<ltrb 48 48 48 48>
   }
   slot {class=OverlaySlot}
  }
  SizeBox_1 {
   widgetClassPath=&SizeBox
   parent=CommonBorder_1
   widget=SizeBox_1
   properties {bOverride_HeightOverride=true, bOverride_WidthOverride=true, heightOverride=932, widthOverride=544}
   slot {class=BorderSlot, properties {horizontalAlignment=HAlign_Left, verticalAlignment=VAlign_Top, padding=<ltrb 48 48 48 48>}}
  }
  PanelContainer {widgetClassPath=&Overlay, parent=SizeBox_1, widget=PanelContainer, slot {class=SizeBoxSlot}}
  Overlay_VBucks {
   widgetClassPath=&Overlay
   bIsVariable=true
   parent=PanelContainer
   widget=Overlay_VBucks
   slot {class=OverlaySlot, properties {horizontalAlignment=HAlign_Fill, verticalAlignment=VAlign_Fill}}
  }
  BaseColorDetails {
   widgetClassPath=&Image
   bIsVariable=true
   parent=Overlay_VBucks
   widget=BaseColorDetails
   properties {colorAndOpacity=<rgba 0.5149 0.5333 0.6445 1>}
   slot {class=OverlaySlot}
  }
  OverlayGradientDetails {
   widgetClassPath=&Image
   bIsVariable=true
   parent=Overlay_VBucks
   widget=OverlayGradientDetails
   properties {brush {
     bIsDynamicallyLoaded=false
     drawAs=Image
     imageType=NoImage
     mirroring=NoMirror
     resourceName=None
     resourceObject=&M-Store-GradientOverlay
     tiling=NoTile
     imageSize=<32 32>
     margin=<ltrb 0 0 0 0>
     outlineSettings {
      bUseBrushTransparency=false
      roundingType=HalfHeightRadius
      width=0
      color {colorUseRule=UseColor_Specified, specifiedColor=<rgba 0 0 0 0>}
      cornerRadii=<0 0 0 0>
     }
     tintColor {colorUseRule=UseColor_Specified, specifiedColor=<rgba 1 1 1 1>}
     uVRegion {bIsValid=false, max=<0 0>, min=<0 0>}
    }}
   slot {class=OverlaySlot, properties {horizontalAlignment=HAlign_Fill, verticalAlignment=VAlign_Fill}}
  }
  VerticalBox_0 {
   widgetClassPath=&VerticalBox
   parent=Overlay_VBucks
   widget=VerticalBox_0
   slot {class=OverlaySlot, properties {horizontalAlignment=HAlign_Fill, verticalAlignment=VAlign_Fill}}
  }
  VerticalBox_2 {
   widgetClassPath=&VerticalBox
   parent=VerticalBox_0
   widget=VerticalBox_2
   slot {class=VerticalBoxSlot, properties {padding=<ltrb 24 24 24 16>, size {sizeRule=Fill, value=1}}}
  }
  TextName {
   widgetClassPath=&CommonTextBlock
   bIsVariable=true
   parent=VerticalBox_2
   widget=TextName
   properties {
    autoWrapText=true
    style=/FortUILibrary/Styles/Typography/_TextStyle_SubH2._TextStyle_SubH2_C
    text=SUPER DELUXE FOUNDERS EDITION
    textTransformPolicy=ToUpper
    font {
     bForceMonospaced=false
     bMaterialIsStencil=false
     fontMaterial=null
     fontObject=&Burbank
     letterSpacing=0
     monospacedWidth=1
     size=33
     skewAmount=0
     typefaceFontName=BurbankRegular-Black
     outlineSettings {
      bApplyOutlineToDropShadows=false
      bMiteredCorners=false
      bSeparateFillAlpha=false
      outlineMaterial=null
      outlineSize=0
      outlineColor=<rgba 0 0 0 1>
     }
    }
    shadowOffset=<0 0>
    strikeBrush {
     bIsDynamicallyLoaded=false
     drawAs=NoDrawType
     imageType=NoImage
     mirroring=NoMirror
     resourceName=None
     resourceObject=null
     tiling=NoTile
     imageSize=<32 32>
     margin=<ltrb 0 0 0 0>
     outlineSettings {
      bUseBrushTransparency=false
      roundingType=HalfHeightRadius
      width=0
      color {colorUseRule=UseColor_Specified, specifiedColor=<rgba 0 0 0 0>}
      cornerRadii=<0 0 0 0>
     }
     tintColor {colorUseRule=UseColor_Specified, specifiedColor=<rgba 1 1 1 1>}
     uVRegion {bIsValid=false, max=<0 0>, min=<0 0>}
    }
   }
   slot {class=VerticalBoxSlot}
  }
  ScrollBox_Details {
   widgetClassPath=&ScrollBox
   bInherited=true
   bIsVariable=true
   parent=VerticalBox_2
   widget=ScrollBox_Details
   slot {class=VerticalBoxSlot, properties {padding=<ltrb 0 12 0 0>, size {sizeRule=Fill, value=1}}}
  }
  TextDescription {
   widgetClassPath=&CommonTextBlock
   bIsVariable=true
   parent=ScrollBox_Details
   widget=TextDescription
   properties {
    autoWrapText=true
    lineHeightPercentage=1.28
    mobileFontSizeMultiplier=1.25
    style=/FortUILibrary/Styles/Typography/_TextStyle_Body2._TextStyle_Body2_C
    text="ZZZZZZZZZZZZZZZZZZZZZZZZZZZZZZZZZZZZZZZZZZZZZZZZZZZZZZZZZZZZZZZZZZZZZZZZZZZZZZZZZZZZZZZZZZZZZZZZZZZZZZZZZZZZZZZZZZZZZZZZZZZZZZZZZ\r\nDescription\r\nText\r\nAppears\r\nHere"
    wrappingPolicy=AllowPerCharacterWrapping
    font {
     bForceMonospaced=false
     bMaterialIsStencil=false
     fontMaterial=null
     fontObject=&BurbankSmall
     letterSpacing=0
     monospacedWidth=1
     size=15
     skewAmount=0
     typefaceFontName=Bold
     outlineSettings {
      bApplyOutlineToDropShadows=false
      bMiteredCorners=false
      bSeparateFillAlpha=false
      outlineMaterial=null
      outlineSize=0
      outlineColor=<rgba 0 0 0 1>
     }
    }
    shadowOffset=<0 0>
    strikeBrush {
     bIsDynamicallyLoaded=false
     drawAs=NoDrawType
     imageType=NoImage
     mirroring=NoMirror
     resourceName=None
     resourceObject=null
     tiling=NoTile
     imageSize=<32 32>
     margin=<ltrb 0 0 0 0>
     outlineSettings {
      bUseBrushTransparency=false
      roundingType=HalfHeightRadius
      width=0
      color {colorUseRule=UseColor_Specified, specifiedColor=<rgba 0 0 0 0>}
      cornerRadii=<0 0 0 0>
     }
     tintColor {colorUseRule=UseColor_Specified, specifiedColor=<rgba 1 1 1 1>}
     uVRegion {bIsValid=false, max=<0 0>, min=<0 0>}
    }
   }
   slot {class=ScrollBoxSlot, properties {padding=<ltrb 0 0 0 8>}}
  }
  VBoxOfferInfo {widgetClassPath=&VerticalBox, bIsVariable=true, parent=ScrollBox_Details, widget=VBoxOfferInfo, slot {class=ScrollBoxSlot}}
  StoreMain_OfferDetailsAttribute1 {
   widgetClassPath=&/StoreMain_OfferDetailsAttribute
   bIsVariable=true
   parent=VBoxOfferInfo
   widget=StoreMain_OfferDetailsAttribute1
   slot {class=VerticalBoxSlot}
  }
  StoreMain_OfferDetailsAttribute2 {
   widgetClassPath=&StoreMain_OfferDetailsAttribute
   bIsVariable=true
   parent=VBoxOfferInfo
   widget=StoreMain_OfferDetailsAttribute2
   slot {class=VerticalBoxSlot}
  }
  VerticalBox_4 {
   widgetClassPath=&VerticalBox
   parent=VerticalBox_0
   widget=VerticalBox_4
   slot {class=VerticalBoxSlot, properties {verticalAlignment=VAlign_Bottom}}
  }
  MTXArea {
   widgetClassPath=&SizeBox
   bIsVariable=true
   parent=VerticalBox_4
   widget=MTXArea
   properties {bOverride_HeightOverride=true, heightOverride=80}
   slot {class=VerticalBoxSlot}
  }
  CommonBorder_3 {
   widgetClassPath=&CommonBorder
   parent=MTXArea
   widget=CommonBorder_3
   properties {
    horizontalAlignment=HAlign_Center
    style=/FortUILibrary/Styles/Border/Border-SolidBG-DkBlue.Border-SolidBG-DkBlue_C
    background {
     bIsDynamicallyLoaded=false
     drawAs=Box
     imageType=NoImage
     mirroring=NoMirror
     resourceName=None
     resourceObject=null
     tiling=NoTile
     imageSize=<32 32>
     margin=<ltrb 0 0 0 0>
     outlineSettings {
      bUseBrushTransparency=false
      roundingType=HalfHeightRadius
      width=0
      color {colorUseRule=UseColor_Specified, specifiedColor=<rgba 0 0 0 0>}
      cornerRadii=<0 0 0 0>
     }
     tintColor {colorUseRule=UseColor_Specified, specifiedColor=<rgba 1 1 1 1>}
     uVRegion {bIsValid=false, max=<0 0>, min=<0 0>}
    }
    brushColor=<rgba 1 1 1 0.3>
    padding=<ltrb 0 0 0 0>
   }
   slot {class=SizeBoxSlot}
  }
  HorizontalBox_3 {
   widgetClassPath=&HorizontalBox
   parent=CommonBorder_3
   widget=HorizontalBox_3
   slot {class=BorderSlot, properties {horizontalAlignment=HAlign_Center, padding=<ltrb 0 0 0 0>}}
  }
  VerticalBox_8 {
   widgetClassPath=&VerticalBox
   parent=HorizontalBox_3
   widget=VerticalBox_8
   slot {class=HorizontalBoxSlot, properties {horizontalAlignment=HAlign_Right, verticalAlignment=VAlign_Center}}
  }
  CommonText_BaseQuantity {
   widgetClassPath=&CommonTextBlock
   bIsVariable=true
   parent=VerticalBox_8
   widget=CommonText_BaseQuantity
   properties {
    justification=Center
    style=/FortUILibrary/Styles/Typography/_TextStyle_SubH4._TextStyle_SubH4_C
    text=0000
    font {
     bForceMonospaced=false
     bMaterialIsStencil=false
     fontMaterial=null
     fontObject=&BurbankSmall
     letterSpacing=0
     monospacedWidth=1
     size=24
     skewAmount=0
     typefaceFontName=Black
     outlineSettings {
      bApplyOutlineToDropShadows=false
      bMiteredCorners=false
      bSeparateFillAlpha=false
      outlineMaterial=null
      outlineSize=0
      outlineColor=<rgba 0 0 0 1>
     }
    }
    shadowOffset=<0 0>
    strikeBrush {
     bIsDynamicallyLoaded=false
     drawAs=NoDrawType
     imageType=NoImage
     mirroring=NoMirror
     resourceName=None
     resourceObject=null
     tiling=NoTile
     imageSize=<32 32>
     margin=<ltrb 0 0 0 0>
     outlineSettings {
      bUseBrushTransparency=false
      roundingType=HalfHeightRadius
      width=0
      color {colorUseRule=UseColor_Specified, specifiedColor=<rgba 0 0 0 0>}
      cornerRadii=<0 0 0 0>
     }
     tintColor {colorUseRule=UseColor_Specified, specifiedColor=<rgba 1 1 1 1>}
     uVRegion {bIsValid=false, max=<0 0>, min=<0 0>}
    }
   }
   slot {class=VerticalBoxSlot, properties {verticalAlignment=VAlign_Bottom, padding=<ltrb 0 4 0 0>}}
  }
  CommonText_Bucks {
   widgetClassPath=&CommonTextBlock
   parent=VerticalBox_8
   widget=CommonText_Bucks
   properties {
    justification=Center
    style=/FortUILibrary/Styles/Typography/_TextStyle_TagNormal._TextStyle_TagNormal_C
    text=V-Bucks
    textTransformPolicy=ToUpper
    font {
     bForceMonospaced=false
     bMaterialIsStencil=false
     fontMaterial=null
     fontObject=&BurbankSmall
     letterSpacing=0
     monospacedWidth=1
     size=15
     skewAmount=0
     typefaceFontName=Black
     outlineSettings {
      bApplyOutlineToDropShadows=false
      bMiteredCorners=false
      bSeparateFillAlpha=false
      outlineMaterial=null
      outlineSize=0
      outlineColor=<rgba 0 0 0 1>
     }
    }
    renderTransform {angle=0, scale=<1 1>, shear=<0 0>, translation=<0 -4>}
    shadowOffset=<0 0>
    strikeBrush {
     bIsDynamicallyLoaded=false
     drawAs=NoDrawType
     imageType=NoImage
     mirroring=NoMirror
     resourceName=None
     resourceObject=null
     tiling=NoTile
     imageSize=<32 32>
     margin=<ltrb 0 0 0 0>
     outlineSettings {
      bUseBrushTransparency=false
      roundingType=HalfHeightRadius
      width=0
      color {colorUseRule=UseColor_Specified, specifiedColor=<rgba 0 0 0 0>}
      cornerRadii=<0 0 0 0>
     }
     tintColor {colorUseRule=UseColor_Specified, specifiedColor=<rgba 1 1 1 1>}
     uVRegion {bIsValid=false, max=<0 0>, min=<0 0>}
    }
   }
   slot {class=VerticalBoxSlot, properties {verticalAlignment=VAlign_Top}}
  }
  BonusMTXArea {
   widgetClassPath=&HorizontalBox
   bIsVariable=true
   parent=HorizontalBox_3
   widget=BonusMTXArea
   slot {class=HorizontalBoxSlot, properties {verticalAlignment=VAlign_Center}}
  }
  CommonTextBlock_7 {
   widgetClassPath=&CommonTextBlock
   parent=BonusMTXArea
   widget=CommonTextBlock_7
   properties {
    style=/FortUILibrary/Styles/Typography/Foundation/TextStyle-Base-L-B.TextStyle-Base-L-B_C
    text=+
    font {
     bForceMonospaced=false
     bMaterialIsStencil=false
     fontMaterial=null
     fontObject=&NotoSans
     letterSpacing=0
     monospacedWidth=1
     size=32
     skewAmount=0
     typefaceFontName=NotoSans-Bold
     outlineSettings {
      bApplyOutlineToDropShadows=false
      bMiteredCorners=false
      bSeparateFillAlpha=false
      outlineMaterial=null
      outlineSize=0
      outlineColor=<rgba 0 0 0 1>
     }
    }
    shadowOffset=<0 0>
   }
   slot {class=HorizontalBoxSlot, properties {verticalAlignment=VAlign_Center, padding=<ltrb 8 0 8 0>}}
  }
  Border_1 {
   widgetClassPath=&Border
   parent=BonusMTXArea
   widget=Border_1
   properties {
    background {
     bIsDynamicallyLoaded=false
     drawAs=NoDrawType
     imageType=NoImage
     mirroring=NoMirror
     resourceName=None
     resourceObject=null
     tiling=NoTile
     imageSize=<32 32>
     margin=<ltrb 0 0 0 0>
     outlineSettings {
      bUseBrushTransparency=false
      roundingType=HalfHeightRadius
      width=0
      color {colorUseRule=UseColor_Specified, specifiedColor=<rgba 0 0 0 0>}
      cornerRadii=<0 0 0 0>
     }
     tintColor {colorUseRule=UseColor_Specified, specifiedColor=<rgba 1 1 1 1>}
     uVRegion {bIsValid=false, max=<0 0>, min=<0 0>}
    }
    contentColorAndOpacity=<rgba 0.9301 1 0.0097 1>
    padding=<ltrb 0 0 0 0>
   }
   slot {class=HorizontalBoxSlot, properties {horizontalAlignment=HAlign_Left, size {sizeRule=Fill, value=1}}}
  }
  VerticalBox_3 {widgetClassPath=&VerticalBox, parent=Border_1, widget=VerticalBox_3, slot {class=BorderSlot, properties {padding=<ltrb 0 0 0 0>}}}
  CommonText_BonusQuantity {
   widgetClassPath=&CommonTextBlock
   bIsVariable=true
   parent=VerticalBox_3
   widget=CommonText_BonusQuantity
   properties {
    justification=Center
    style=&_TextStyle_SubH4_C
    text=0000
    font {
     bForceMonospaced=false
     bMaterialIsStencil=false
     fontMaterial=null
     fontObject=&BurbankSmall
     letterSpacing=0
     monospacedWidth=1
     size=24
     skewAmount=0
     typefaceFontName=Black
     outlineSettings {
      bApplyOutlineToDropShadows=false
      bMiteredCorners=false
      bSeparateFillAlpha=false
      outlineMaterial=null
      outlineSize=0
      outlineColor=<rgba 0 0 0 1>
     }
    }
    shadowOffset=<0 0>
    strikeBrush {
     bIsDynamicallyLoaded=false
     drawAs=NoDrawType
     imageType=NoImage
     mirroring=NoMirror
     resourceName=None
     resourceObject=null
     tiling=NoTile
     imageSize=<32 32>
     margin=<ltrb 0 0 0 0>
     outlineSettings {
      bUseBrushTransparency=false
      roundingType=HalfHeightRadius
      width=0
      color {colorUseRule=UseColor_Specified, specifiedColor=<rgba 0 0 0 0>}
      cornerRadii=<0 0 0 0>
     }
     tintColor {colorUseRule=UseColor_Specified, specifiedColor=<rgba 1 1 1 1>}
     uVRegion {bIsValid=false, max=<0 0>, min=<0 0>}
    }
   }
   slot {class=VerticalBoxSlot, properties {verticalAlignment=VAlign_Bottom, padding=<ltrb 0 4 0 0>}}
  }
  CommonText_Bonus {
   widgetClassPath=&CommonTextBlock
   parent=VerticalBox_3
   widget=CommonText_Bonus
   properties {
    justification=Center
    style=&_TextStyle_TagNormal_C
    text=BONUS
    textTransformPolicy=ToUpper
    font {
     bForceMonospaced=false
     bMaterialIsStencil=false
     fontMaterial=null
     fontObject=&BurbankSmall
     letterSpacing=0
     monospacedWidth=1
     size=15
     skewAmount=0
     typefaceFontName=Black
     outlineSettings {
      bApplyOutlineToDropShadows=false
      bMiteredCorners=false
      bSeparateFillAlpha=false
      outlineMaterial=null
      outlineSize=0
      outlineColor=<rgba 0 0 0 1>
     }
    }
    renderTransform {angle=0, scale=<1 1>, shear=<0 0>, translation=<0 -4>}
    shadowOffset=<0 0>
    strikeBrush {
     bIsDynamicallyLoaded=false
     drawAs=NoDrawType
     imageType=NoImage
     mirroring=NoMirror
     resourceName=None
     resourceObject=null
     tiling=NoTile
     imageSize=<32 32>
     margin=<ltrb 0 0 0 0>
     outlineSettings {
      bUseBrushTransparency=false
      roundingType=HalfHeightRadius
      width=0
      color {colorUseRule=UseColor_Specified, specifiedColor=<rgba 0 0 0 0>}
      cornerRadii=<0 0 0 0>
     }
     tintColor {colorUseRule=UseColor_Specified, specifiedColor=<rgba 1 1 1 1>}
     uVRegion {bIsValid=false, max=<0 0>, min=<0 0>}
    }
   }
   slot {class=VerticalBoxSlot, properties {verticalAlignment=VAlign_Top}}
  }
  CommonBorder_0 {
   widgetClassPath=&CommonBorder
   parent=VerticalBox_4
   widget=CommonBorder_0
   properties {
    style=&Border-SolidBG-DkBlue_C
    verticalAlignment=VAlign_Center
    background {
     bIsDynamicallyLoaded=false
     drawAs=Box
     imageType=NoImage
     mirroring=NoMirror
     resourceName=None
     resourceObject=null
     tiling=NoTile
     imageSize=<32 32>
     margin=<ltrb 0 0 0 0>
     outlineSettings {
      bUseBrushTransparency=false
      roundingType=HalfHeightRadius
      width=0
      color {colorUseRule=UseColor_Specified, specifiedColor=<rgba 0 0 0 0>}
      cornerRadii=<0 0 0 0>
     }
     tintColor {colorUseRule=UseColor_Specified, specifiedColor=<rgba 1 1 1 1>}
     uVRegion {bIsValid=false, max=<0 0>, min=<0 0>}
    }
    brushColor=<rgba 1 1 1 0.15>
   }
   slot {class=VerticalBoxSlot}
  }
  VerticalBox_6 {
   widgetClassPath=&VerticalBox
   parent=CommonBorder_0
   widget=VerticalBox_6
   slot {class=BorderSlot, properties {verticalAlignment=VAlign_Center}}
  }
  HorizontalBox_1 {
   widgetClassPath=&HorizontalBox
   parent=VerticalBox_6
   widget=HorizontalBox_1
   slot {class=VerticalBoxSlot, properties {horizontalAlignment=HAlign_Center, padding=<ltrb 0 0 0 4>}}
  }
  WBP_Price_PlatformStoreIcon {
   widgetClassPath=/FortUILibrary/Systems/ItemShopTile/WBP_Price_PlatformStoreIcon
   bIsVariable=true
   parent=HorizontalBox_1
   widget=WBP_Price_PlatformStoreIcon
   properties {imageSize=<40 40>}
   slot {class=HorizontalBoxSlot, properties {horizontalAlignment=HAlign_Center, verticalAlignment=VAlign_Center, padding=<ltrb 0 0 4 0>}}
  }
  OverlaySalePrice {
   widgetClassPath=&Overlay
   bIsVariable=true
   parent=HorizontalBox_1
   widget=OverlaySalePrice
   slot {class=HorizontalBoxSlot, properties {horizontalAlignment=HAlign_Left, verticalAlignment=VAlign_Center}}
  }
  TextOriginalPrice {
   widgetClassPath=&CommonTextBlock
   bIsVariable=true
   parent=OverlaySalePrice
   widget=TextOriginalPrice
   properties {
    justification=Center
    style=&_TextStyle_SubH4_C
    text=$888.88
    font {
     bForceMonospaced=false
     bMaterialIsStencil=false
     fontMaterial=null
     fontObject=&BurbankSmall
     letterSpacing=0
     monospacedWidth=1
     size=24
     skewAmount=0
     typefaceFontName=Black
     outlineSettings {
      bApplyOutlineToDropShadows=false
      bMiteredCorners=false
      bSeparateFillAlpha=false
      outlineMaterial=null
      outlineSize=0
      outlineColor=<rgba 0 0 0 1>
     }
    }
    shadowOffset=<0 0>
    strikeBrush {
     bIsDynamicallyLoaded=false
     drawAs=NoDrawType
     imageType=NoImage
     mirroring=NoMirror
     resourceName=None
     resourceObject=null
     tiling=NoTile
     imageSize=<32 32>
     margin=<ltrb 0 0 0 0>
     outlineSettings {
      bUseBrushTransparency=false
      roundingType=HalfHeightRadius
      width=0
      color {colorUseRule=UseColor_Specified, specifiedColor=<rgba 0 0 0 0>}
      cornerRadii=<0 0 0 0>
     }
     tintColor {colorUseRule=UseColor_Specified, specifiedColor=<rgba 1 1 1 1>}
     uVRegion {bIsValid=false, max=<0 0>, min=<0 0>}
    }
   }
   slot {class=OverlaySlot, properties {horizontalAlignment=HAlign_Fill, verticalAlignment=VAlign_Fill, padding=<ltrb 12 0 12 0>}}
  }
  Image_1 {
   widgetClassPath=&Image
   parent=OverlaySalePrice
   widget=Image_1
   properties {
    brush {
     bIsDynamicallyLoaded=false
     drawAs=Image
     imageType=NoImage
     mirroring=NoMirror
     resourceName=None
     resourceObject=null
     tiling=NoTile
     imageSize=<32 4>
     margin=<ltrb 0 0 0 0>
     outlineSettings {
      bUseBrushTransparency=false
      roundingType=HalfHeightRadius
      width=0
      color {colorUseRule=UseColor_Specified, specifiedColor=<rgba 0 0 0 0>}
      cornerRadii=<0 0 0 0>
     }
     tintColor {colorUseRule=UseColor_Specified, specifiedColor=<rgba 1 1 1 1>}
     uVRegion {bIsValid=false, max=<0 0>, min=<0 0>}
    }
    colorAndOpacity=<rgba 1 1 1 0.66>
    renderTransform {angle=-5, scale=<1 1>, shear=<0 0>, translation=<0 0>}
   }
   slot {class=OverlaySlot, properties {horizontalAlignment=HAlign_Fill, verticalAlignment=VAlign_Center}}
  }
  HorizontalBox_2 {
   widgetClassPath=&HorizontalBox
   parent=HorizontalBox_1
   widget=HorizontalBox_2
   slot {class=HorizontalBoxSlot, properties {horizontalAlignment=HAlign_Center}}
  }
  Image_CurrencyIcon {
   widgetClassPath=&FortLazyImage
   bIsVariable=true
   parent=HorizontalBox_2
   widget=Image_CurrencyIcon
   slot {class=HorizontalBoxSlot, properties {verticalAlignment=VAlign_Center}}
  }
  Border_0 {
   widgetClassPath=&Border
   parent=HorizontalBox_2
   widget=Border_0
   properties {
    background {
     bIsDynamicallyLoaded=false
     drawAs=NoDrawType
     imageType=NoImage
     mirroring=NoMirror
     resourceName=None
     resourceObject=null
     tiling=NoTile
     imageSize=<32 32>
     margin=<ltrb 0 0 0 0>
     outlineSettings {
      bUseBrushTransparency=false
      roundingType=HalfHeightRadius
      width=0
      color {colorUseRule=UseColor_Specified, specifiedColor=<rgba 0 0 0 0>}
      cornerRadii=<0 0 0 0>
     }
     tintColor {colorUseRule=UseColor_Specified, specifiedColor=<rgba 1 1 1 1>}
     uVRegion {bIsValid=false, max=<0 0>, min=<0 0>}
    }
    contentColorAndOpacity=<rgba 0.9301 1 0.0097 1>
    padding=<ltrb 0 0 0 0>
   }
   slot {class=HorizontalBoxSlot}
  }
  TextOfferDisplayPrice {
   widgetClassPath=&CommonTextBlock
   bIsVariable=true
   parent=Border_0
   widget=TextOfferDisplayPrice
   properties {
    justification=Center
    style=&_TextStyle_SubH2_C
    text=$888.88
    font {
     bForceMonospaced=false
     bMaterialIsStencil=false
     fontMaterial=null
     fontObject=&Burbank
     letterSpacing=0
     monospacedWidth=1
     size=33
     skewAmount=0
     typefaceFontName=BurbankRegular-Black
     outlineSettings {
      bApplyOutlineToDropShadows=false
      bMiteredCorners=false
      bSeparateFillAlpha=false
      outlineMaterial=null
      outlineSize=0
      outlineColor=<rgba 0 0 0 1>
     }
    }
    shadowOffset=<0 0>
    strikeBrush {
     bIsDynamicallyLoaded=false
     drawAs=NoDrawType
     imageType=NoImage
     mirroring=NoMirror
     resourceName=None
     resourceObject=null
     tiling=NoTile
     imageSize=<32 32>
     margin=<ltrb 0 0 0 0>
     outlineSettings {
      bUseBrushTransparency=false
      roundingType=HalfHeightRadius
      width=0
      color {colorUseRule=UseColor_Specified, specifiedColor=<rgba 0 0 0 0>}
      cornerRadii=<0 0 0 0>
     }
     tintColor {colorUseRule=UseColor_Specified, specifiedColor=<rgba 1 1 1 1>}
     uVRegion {bIsValid=false, max=<0 0>, min=<0 0>}
    }
   }
   slot {class=BorderSlot, properties {padding=<ltrb 0 0 0 0>}}
  }
  Text_AdditionalPriceText {
   widgetClassPath=&CommonTextBlock
   bIsVariable=true
   parent=HorizontalBox_2
   widget=Text_AdditionalPriceText
   properties {
    bAutoCollapseWithEmptyText=true
    justification=Center
    text=VAT Incl.
    colorAndOpacity {colorUseRule=UseColor_Specified, specifiedColor=<rgba 0.9301 1 0.0097 1>}
    font {
     bForceMonospaced=false
     bMaterialIsStencil=false
     fontMaterial=null
     fontObject=&Burbank
     letterSpacing=0
     monospacedWidth=1
     size=16.5
     skewAmount=0
     typefaceFontName=BurbankRegular-Black
     outlineSettings {
      bApplyOutlineToDropShadows=false
      bMiteredCorners=false
      bSeparateFillAlpha=false
      outlineMaterial=null
      outlineSize=0
      outlineColor=<rgba 0 0 0 1>
     }
    }
    shadowOffset=<0 0>
    strikeBrush {
     bIsDynamicallyLoaded=false
     drawAs=NoDrawType
     imageType=NoImage
     mirroring=NoMirror
     resourceName=None
     resourceObject=null
     tiling=NoTile
     imageSize=<32 32>
     margin=<ltrb 0 0 0 0>
     outlineSettings {
      bUseBrushTransparency=false
      roundingType=HalfHeightRadius
      width=0
      color {colorUseRule=UseColor_Specified, specifiedColor=<rgba 0 0 0 0>}
      cornerRadii=<0 0 0 0>
     }
     tintColor {colorUseRule=UseColor_Specified, specifiedColor=<rgba 1 1 1 1>}
     uVRegion {bIsValid=false, max=<0 0>, min=<0 0>}
    }
   }
   slot {class=HorizontalBoxSlot, properties {verticalAlignment=VAlign_Center, padding=<ltrb 4 8 0 0>}}
  }
  VerticalBox_1 {
   widgetClassPath=&VerticalBox
   parent=VerticalBox_6
   widget=VerticalBox_1
   slot {class=VerticalBoxSlot, properties {padding=<ltrb 96 0 96 0>}}
  }
  Button_Purchase {
   widgetClassPath=/FortUILibrary/UIKit/Buttons/WBP_UIKit_Button_Loud
   bInherited=true
   bIsVariable=true
   parent=VerticalBox_1
   widget=Button_Purchase
   properties {bRequiresHold=true, buttonSize=Large, holdData=/FortUILibrary/Input/ShortInteraction_HoldData.ShortInteraction_HoldData_C, text=Purchase}
   slot {class=VerticalBoxSlot, properties {verticalAlignment=VAlign_Top, padding=<ltrb 0 0 0 24>}}
  }
  Button_Purchase_STW {
   widgetClassPath=&WBP_UIKit_Button_Loud
   bInherited=true
   bIsVariable=true
   parent=VerticalBox_1
   widget=Button_Purchase_STW
   properties {bRequiresHold=true, buttonSize=Large, holdData=&ShortInteraction_HoldData_C, text=Purchase}
   slot {class=VerticalBoxSlot, properties {verticalAlignment=VAlign_Top, padding=<ltrb 0 0 0 24>}}
  }
  Button_SendGift {
   widgetClassPath=&IconTextButton
   bInherited=true
   bIsVariable=true
   parent=VerticalBox_1
   widget=Button_SendGift
   properties {
    buttonText=Gift
    onButtonBaseClicked=(&self:WidgetTree.Button_SendGift.HandleOnButtonClicked)
    onButtonBaseDoubleClicked=(&self:WidgetTree.Button_SendGift.HandleOnButtonDoubleClicked)
    onButtonBaseHovered=(&self:WidgetTree.Button_SendGift.HandleOnButtonHovered)
    onButtonBaseUnhovered=(&self:WidgetTree.Button_SendGift.HandleOnButtonUnhovered)
    onSelectedChangedBase=(&self:WidgetTree.Button_SendGift.HandleOnSelectedChanged)
    style=/FortUILibrary/Foundation/Buttons/ButtonStyles/ButtonStyle_Feature_L_Yellow.ButtonStyle_Feature_L_Yellow_C
    hoveredSlateSoundOverride {resourceObject=/Game/Sounds/Fort_UI_Sounds/FrontEnd/ItemShop/Fort_UI_Shop_Menu_Hover_Cue}
    pressedSlateSoundOverride {resourceObject=/Game/Sounds/Fort_UI_Sounds/FrontEnd/ItemShop/Fort_UI_Shop_Menu_Click_Purchase_Cue}
    triggeringInputAction {dataTable=&InputActionDataTable, rowName=StoreMainSendAsGift}
   }
   slot {class=VerticalBoxSlot}
  }
  WBP_ContentPurchaseBlocker {
   widgetClassPath=/FortniteCoreUI/UI/ContentControls/WBP_ContentPurchaseBlocker
   bIsVariable=true
   parent=PanelContainer
   widget=WBP_ContentPurchaseBlocker
   properties {max Width=500, visibility=Collapsed}
   slot {class=OverlaySlot, properties {verticalAlignment=VAlign_Center, padding=<ltrb 40 0 40 0>}}
  }
  VerticalBox_5 {
   widgetClassPath=&VerticalBox
   parent=Overlay_1
   widget=VerticalBox_5
   slot {class=OverlaySlot, properties {horizontalAlignment=HAlign_Right, verticalAlignment=VAlign_Bottom, padding=<ltrb 6 6 6 6>}}
  }
  Container_SimpleMTXDisplay {
   widgetClassPath=/Game/UI/Frontend/ItemShop/SimpleMTXDisplay
   bIsVariable=true
   parent=VerticalBox_5
   widget=Container_SimpleMTXDisplay
   slot {class=VerticalBoxSlot, properties {horizontalAlignment=HAlign_Center}}
  }
  BorderDisclaimer {
   widgetClassPath=&CommonBorder
   bIsVariable=true
   parent=VerticalBox_5
   widget=BorderDisclaimer
   properties {padding=<ltrb 0 0 0 0>}
   slot {class=VerticalBoxSlot, properties {padding=<ltrb 0 0 48 48>}}
  }
  CommonTextBlock_0 {
   widgetClassPath=&CommonTextBlock
   parent=BorderDisclaimer
   widget=CommonTextBlock_0
   properties {
    style=/FortUILibrary/Styles/Typography/Foundation/TextStyle-Base-XS-70pc.TextStyle-Base-XS-70pc_C
    text=All V-Bucks purchased on one platform may not be redeemable on other platforms.
    wrapTextAt=800
    colorAndOpacity {colorUseRule=UseColor_Specified, specifiedColor=<rgba 1 1 1 0.7>}
    font {
     bForceMonospaced=false
     bMaterialIsStencil=false
     fontMaterial=null
     fontObject=&NotoSans
     letterSpacing=0
     monospacedWidth=1
     size=16
     skewAmount=0
     typefaceFontName=NotoSans-Regular
     outlineSettings {
      bApplyOutlineToDropShadows=false
      bMiteredCorners=false
      bSeparateFillAlpha=false
      outlineMaterial=null
      outlineSize=0
      outlineColor=<rgba 0 0 0 1>
     }
    }
    shadowOffset=<0 0>
   }
   slot {class=BorderSlot, properties {padding=<ltrb 0 0 0 0>}}
  }
  SafeZonePlayerInfo {
   widgetClassPath=&SafeZone
   bIsVariable=true
   parent=Overlay_2
   widget=SafeZonePlayerInfo
   properties {padBottom=false, padLeft=false}
   slot {class=OverlaySlot, properties {horizontalAlignment=HAlign_Right, padding=<ltrb 0 40 80 0>}}
  }
  WBP_PlayerInfo {
   widgetClassPath=/Game/UI/Frontend/ItemShop/MultiProduct/WBP_PlayerInfo
   bIsVariable=true
   parent=SafeZonePlayerInfo
   widget=WBP_PlayerInfo
   slot {class=SafeZoneSlot}
  }
  Loading_Overlay {
   widgetClassPath=&Overlay
   bIsVariable=true
   parent=Overlay_2
   widget=Loading_Overlay
   properties {visibility=Collapsed}
   slot {class=OverlaySlot, properties {horizontalAlignment=HAlign_Fill, verticalAlignment=VAlign_Fill}}
  }
  WBP_UIKit_Scrim {
   widgetClassPath=/FortUILibrary/Systems/Dialog/WBP_UIKit_Scrim
   parent=Loading_Overlay
   widget=WBP_UIKit_Scrim
   properties {blur=10}
   slot {class=OverlaySlot, properties {horizontalAlignment=HAlign_Fill, verticalAlignment=VAlign_Fill}}
  }
  LoadingIcon {
   widgetClassPath=/Script/CommonUI.CommonLazyImage
   parent=Loading_Overlay
   widget=LoadingIcon
   properties {
    visibility=HitTestInvisible
    brush {
     bIsDynamicallyLoaded=false
     drawAs=Image
     imageType=FullColor
     mirroring=NoMirror
     resourceName=None
     resourceObject=&MI_ItemDownloadIndicator_SpinnerMaterial
     tiling=NoTile
     imageSize=<64 64>
     margin=<ltrb 0 0 0 0>
     outlineSettings {
      bUseBrushTransparency=false
      roundingType=HalfHeightRadius
      width=0
      color {colorUseRule=UseColor_Specified, specifiedColor=<rgba 0 0 0 0>}
      cornerRadii=<0 0 0 0>
     }
     tintColor {colorUseRule=UseColor_Specified, specifiedColor=<rgba 1 1 1 1>}
     uVRegion {bIsValid=false, max=<0 0>, min=<0 0>}
    }
    loadingBackgroundBrush {
     bIsDynamicallyLoaded=false
     drawAs=Image
     imageType=FullColor
     mirroring=NoMirror
     resourceName=None
     resourceObject=&MI_ItemDownloadIndicator_SpinnerMaterial
     tiling=NoTile
     imageSize=<64 64>
     margin=<ltrb 0 0 0 0>
     outlineSettings {
      bUseBrushTransparency=false
      roundingType=HalfHeightRadius
      width=0
      color {colorUseRule=UseColor_Specified, specifiedColor=<rgba 0 0 0 0>}
      cornerRadii=<0 0 0 0>
     }
     tintColor {colorUseRule=UseColor_Specified, specifiedColor=<rgba 1 1 1 1>}
     uVRegion {bIsValid=false, max=<0 0>, min=<0 0>}
    }
   }
   slot {class=OverlaySlot, properties {horizontalAlignment=HAlign_Center, verticalAlignment=VAlign_Center}}
  }
  Button_Inspect {widgetClassPath=&CommonButtonBase, bInherited=true, widget=null}
  Picker_VariantSelector {widgetClassPath=/Script/FortniteUI.FortVariantPicker, bInherited=true, widget=null}
  TextUndoAvailability {widgetClassPath=&CommonTextBlock, bInherited=true, widget=null}
  Text_UndoBottomBarWarning {widgetClassPath=&CommonTextBlock, bInherited=true, widget=null}
  Button_Undo {widgetClassPath=&CommonButtonLegacy, bInherited=true, widget=null}
  Text_CosmeticWarning {widgetClassPath=&CommonTextBlock, bInherited=true, widget=null}
 ]
}
