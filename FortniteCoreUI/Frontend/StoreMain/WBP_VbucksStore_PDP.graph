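# V-Bucks store product detail page: an activatable widget showing one real-money offer and starting its purchase.
# FortMtxStoreOfferVM callbacks update key art (CMS URL or soft-image fallback), async-load the background and switch to a parental-controls blocker when required.
# Purchasing swaps content for a spinner until a result arrives; non-canceled results deactivate the page and broadcast OnPurchaseComplete.
# A per-offer UX trace ID feeds analytics and is cleared on exit unless purchase was clicked.
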WidgetBlueprint /FortniteCoreUI/Frontend/StoreMain/WBP_VbucksStore_PDP
animations [Intro {
  binding_count=2
  length_seconds=0.5
  object=&self:Intro
  sequence {bindings [
    Image_Background {
     id=#1
     possessed_class=Image
     tracks [{
       type=MovieSceneWidgetMaterialTrack
       sections [{
         type=MovieSceneParameterSection
         end_frame=10
         start_frame=0
         channels [
          Hover {type=MovieSceneScriptingFloatChannel, keys [[0,0],[10,1]]}
          GradientGlowDensity {type=MovieSceneScriptingFloatChannel, keys [[0,0],[4,1]]}
          VignetteOpacity {type=MovieSceneScriptingFloatChannel, keys [[0,0.95],[4,0.92]]}
         ]
        }]
      }]
    }
    Image_KeyArt {
     id=#2
     possessed_class=CommonLazyImage
     tracks [
      {
       type=MovieSceneColorTrack
       display_name=Color and Opacity
       sections [{
         type=MovieSceneColorSection
         end_frame=10
         start_frame=0
         channels [
          Color.R {type=MovieSceneScriptingFloatChannel, default=1}
          Color.G {type=MovieSceneScriptingFloatChannel, default=1}
          Color.B {type=MovieSceneScriptingFloatChannel, default=1}
          Color.A {type=MovieSceneScriptingFloatChannel, keys [[0,0],[3,1]]}
         ]
        }]
      }
      {
       type=MovieScene2DTransformTrack
       sections [{
         type=MovieScene2DTransformSection
         end_frame=10
         start_frame=0
         channels [
          Translation.X {type=MovieSceneScriptingFloatChannel, const [0,1]}
          Translation.Y {type=MovieSceneScriptingFloatChannel, const [3.4,1]}
          Angle {type=MovieSceneScriptingFloatChannel, const [0,1]}
          Scale.X {type=MovieSceneScriptingFloatChannel, keys [[0,0.2],[6,1]]}
          Scale.Y {type=MovieSceneScriptingFloatChannel, keys [[0,0.2],[6,1]]}
         ]
        }]
      }
     ]
    }
   ]}
 }]
graphs [
 BP_GetDesiredFocusTarget {nodes [
   Get Desired Focus Target()
   n1 = Get PurchaseAvailability!(self:Fort Mtx Store Offer VM Object Reference=Get FortMtxStoreOfferVM() -> (FortMtxStoreOfferVM:Fort Mtx Store Offer VM Object Reference)) -> (PurchaseAvailability:EFortMtxStoreOfferPurchaseAvailability Enum=Available)
   n2 = Get WBP_ContentPurchaseBlocker() -> (WBP_ContentPurchaseBlocker:WBP Content Purchase Blocker Object Reference)
   > Return Node!(
    execute
    ReturnValue:Widget Object Reference=Select(
     Option 0:Widget Object Reference=Get Button_Purchase() -> (Button_Purchase:WBP UIKit Button Loud Object Reference)
     Option 1:Widget Object Reference=Get WBP_UIKit_Button_Loud!(self:WBP Content Purchase Blocker Object Reference=@n2) -> (WBP_UIKit_Button_Loud:WBP UIKit Button Loud Object Reference)
     Index="Equal (Enum)"(A:EFortMtxStoreOfferPurchaseAvailability Enum=@n1, B:EFortMtxStoreOfferPurchaseAvailability Enum=Unavailable_ParentalControls)
    ) -> (ReturnValue:Widget Object Reference)
   )
  ]}
 BP_GetDesiredInputConfig {nodes [
   Get Desired Input Config() {is=title K2Node_FunctionEntry}
   n1 = GetOwningPlayer()
   > Return Node!(
    execute
    ReturnValue_bIgnoreMoveInput:bool=False
    ReturnValue_bIgnoreLookInput:bool=False
    ReturnValue_InputMode:ECommonInputMode Enum=Select(
     Option 0:ECommonInputMode Enum=All
     Option 1:ECommonInputMode Enum=Menu
     Index=Cast To FortPlayerControllerZone!(Object=@n1) -> (AsFort Player Controller Zone:Fort Player Controller Zone Object Reference, bSuccess:bool=~) {is=title K2Node_DynamicCast}.bSuccess
    ) -> (ReturnValue:ECommonInputMode Enum)
    ReturnValue_MouseCaptureMode:EMouseCaptureMode Enum=NoCapture
    ReturnValue_MouseLockMode:EMouseLockMode Enum=DoNotLock
    ReturnValue_bHideCursorDuringViewportCapture:bool=True
   )
  ]}
 BuildAnalyticsInfo {nodes [
   BuildAnalyticsInfo() {is=title K2Node_FunctionEntry}
   n1 = Get PurchaseUXTraceId() -> (PurchaseUXTraceId:String)
   > n2 = Branch(Condition=IsEmpty(InString=@n1))
   n3 = Make Flow Analytics Additional Info!() -> (FlowAnalyticsAdditionalInfo:Flow Analytics Additional Info Structure) {is=title K2Node_MakeStruct}
   > Return Node!(execute, ReturnValue:Flow Analytics Additional Info Structure=@n3)
   Return Node(
    execute=@n2.else
    ReturnValue="Add String (Non-Localized)"!(Info:"Flow Analytics Additional Info Structure (by ref)"=@n3, Key:String=PurchaseUXTraceId, Value:String=@n1) -> (ReturnValue:Flow Analytics Additional Info Structure)
   )
  ]}
 EventGraph {nodes [
   n1 = "On Clicked (DecoyButton_Back)"!() -> (OutputDelegate, then, Button:Common Button Base Object Reference) {is=title K2Node_ComponentBoundEvent}
   > Fire Event Element Interacted(
    Element=@n1.Button
    AdditionalInfo=BuildAnalyticsInfo!() -> (ReturnValue:Flow Analytics Additional Info Structure) {is=title K2Node_CallFunction}
   )
   > DeactivateWidget()
   n2 = "On Clicked (Button_Purchase)"!() -> (OutputDelegate, then, Button:Common Button Base Object Reference) {is=title K2Node_ComponentBoundEvent}
   > Fire Event Element Interacted(Element=@n2.Button, AdditionalInfo=BuildAnalyticsInfo())
   > Set ClearPurchaseUXTraceIdOnExit(ClearPurchaseUXTraceIdOnExit:bool=false) -> (Output_Get:bool)
   > n3 = SetIsPurchasing!(execute, self, bIsPurchasing:bool=true) -> (then)
   n4 = HandlePurchaseComplete!() -> (OutputDelegate, then, Result:EFortMtxStoreOfferPurchaseResult Enum=Succeeded) {is=title K2Node_CustomEvent}
   PurchaseOffer!(
    execute=@n3.then
    self:Fort Mtx Store Offer VM Object Reference=Get FortMtxStoreOfferVM()
    OnPurchaseCompleteCallback:"Delegate (by ref)"=@n4
   ) -> (then)
   n5 = VM_OnOfferInfoChanged!() -> (OutputDelegate, then, OfferInfo:Fort Store Front Offer Info Object Reference) {is=title K2Node_CustomEvent}
   n6 = GetOfferDisplayAsset!(self:Fort Store Front Offer Info Object Reference=@n5.OfferInfo) -> (ReturnValue:Fort Mtx Offer Data Object Reference)
   > UpdateKeyArt!(execute, self, InDisplayAsset:Fort Mtx Offer Data Object Reference=@n6) -> (then)
   n7 = NOT Boolean(A=Get MatchesExistingOffer() -> (MatchesExistingOffer:bool))
   n8 = Get IsExactAmount() -> (IsExactAmount:bool)
   > n9 = Async Load Asset(Asset=Select(
     Option 0:Soft Object Reference=Get SoftPDPBackgroundMaterial!(self:Fort Mtx Offer Data Object Reference=@n6) -> (SoftPDPBackgroundMaterial:Soft Object Reference)
     Option 1:Soft Object Reference=Get ExactPackBackground() -> (ExactPackBackground:Soft Object Reference)
     Index=AND Boolean(A=@n8, B=@n7)
    ) -> (ReturnValue:Soft Object Reference))
   n10 = Cast To MaterialInstance!(execute=@n9.Completed, Object=@n9) -> (then, CastFailed, AsMaterial Instance:Material Instance Object Reference) {is=title K2Node_DynamicCast}
   > SetBrushFromMaterial(self=Get Image_Background() -> (Image_Background:Image Object Reference), Material=@n10)
   n11 = VM_OnOfferIdChanged!() -> (OutputDelegate, then, EpicOfferId:String) {is=title K2Node_CustomEvent}
   > n12 = GeneratePurchaseUXTraceId!(
    execute
    self:Fort Mcp Context Object Reference=Get FortMcpContext!() -> (ReturnValue:Fort Mcp Context Object Reference) {is=title K2Node_GetBlueprintContextEx}
    EpicOfferId:String=@n11.EpicOfferId
    DebugSourceContext:String=GetDisplayName(Object=Self-Reference())
   ) -> (then, ReturnValue:String)
   > Set PurchaseUXTraceId(PurchaseUXTraceId:String=@n12) -> (Output_Get:String)
   > Set ClearPurchaseUXTraceIdOnExit(ClearPurchaseUXTraceIdOnExit=true)
   Event On Deactivated()
   > Set FortMtxStoreOfferVM(FortMtxStoreOfferVM:Fort Mtx Store Offer VM Object Reference) -> (Output_Get:Fort Mtx Store Offer VM Object Reference)
   > Branch(Condition=Get ClearPurchaseUXTraceIdOnExit() -> (ClearPurchaseUXTraceIdOnExit:bool))
   > ClearPurchaseUXTraceIdIfMatches!(
    execute
    self:Fort Mcp Context Object Reference=Get FortMcpContext()
    ExpectedUXTraceId:String=Get PurchaseUXTraceId()
    DebugSourceContext:String=GetDisplayName(Object=Self-Reference())
   ) -> (then, ReturnValue:bool)
   > Set PurchaseUXTraceId()
   Event On Activated()
   > Fire Event Sub View Opened!(
    execute
    View:Widget Object Reference=Self-Reference()
    AdditionalInfo:Flow Analytics Additional Info Structure=BuildAnalyticsInfo()
    bIsOverlay:bool=false
   ) -> (then)
   > PlayAnimationForward(InAnimation=Get Intro())
   > SetScrollWidget!(execute, ToScroll:Widget Object Reference=Get ScrollBox_Description() -> (ScrollBox_Description:Scroll Box Object Reference)) -> (then)
   Event Construct()
   > SetCompactStyle()
   > SetCurrency!(
    execute
    self:WBP Price Switcher Base Object Reference=Get WBP_PriceSwitcher_Base() -> (WBP_PriceSwitcher_Base:WBP Price Switcher Base Object Reference)
    Currency:EFortPriceCurrency Enum=RealMoney
   ) -> (then)
   Call On Purchase Complete!(
    execute:Exec=@n13.then
    self:WBP Vbucks Store PDP Object Reference
    bSuccess:bool="Equal (Enum)"(A:EFortMtxStoreOfferPurchaseResult Enum=@n4.Result, B:EFortMtxStoreOfferPurchaseResult Enum=Succeeded)
   ) -> (then:Exec) {is=title K2Node_CallDelegate}
   SetIsPurchasing(execute=@n4.then, bIsPurchasing=false)
   > Branch(Condition="Not Equal (Enum)"!(A:EFortMtxStoreOfferPurchaseResult Enum=@n4.Result, B:EFortMtxStoreOfferPurchaseResult Enum=Canceled) -> (ReturnValue:bool))
   > n13 = DeactivateWidget()
  ]}
 GetEpicRewardsText {nodes [
   n1 = GetEpicRewardsText!() -> (then, Percentage:int, RealMoney:Text) {is=title K2Node_FunctionEntry}
   n2 = Get Percentage() -> (Percentage:int=0)
   > Return Node!(
    execute
    Text:Text=Select(
     Option 0:Text
     Option 1:Text=Format Text!(
      Format="LOCTEXT(\"Earn <text color=\\\"38ff84\\\">{Percentage}% ({RealMoney}) back</> in Epic Rewards\")"
      Percentage:int=@n1
      RealMoney:Text=@n1.RealMoney
     ) -> (Result)
     Index="integer > integer"(A=@n2)
    ) -> (ReturnValue:Text)
   )
  ]}
 OnPurchaseComplete {nodes [OnPurchaseComplete!() -> (then, bSuccess:bool=~) {is=title K2Node_FunctionEntry}]}
 SetCompactStyle {nodes [
   SetCompactStyle() {is=title K2Node_FunctionEntry}
   > Branch(Condition=IsUsingCompactUI())
   > SetPadding!(
    execute
    self:Overlay Slot Object Reference=SlotAsOverlaySlot(Widget=Get WBP_PlayerInfo() -> (WBP_PlayerInfo:WBP Player Info Object Reference))
    InPadding_Left:float=0
    InPadding_Top:float=24
    InPadding_Right:float=48
    InPadding_Bottom:float=0
   ) -> (then)
   > SetPadding(
    self:Vertical Box Slot Object Reference=SlotAsVerticalBoxSlot(Widget=Get RichText_EpicRewards() -> (RichText_EpicRewards:Common Rich Text Block Object Reference))
    InPadding_Left=0
    InPadding_Top=12
    InPadding_Right=0
    InPadding_Bottom=0
   )
  ]}
 SetIsPurchasing {nodes [
   SetIsPurchasing!() -> (then, bIsPurchasing:bool=~) {is=title K2Node_FunctionEntry}
   > SetActiveWidget(
    self=Get Switcher() -> (Switcher:Common Visibility Switcher Object Reference)
    Widget=Select(
     Option 0:Widget Object Reference=Get Overlay_Content() -> (Overlay_Content:Overlay Object Reference)
     Option 1:Widget Object Reference=Get Loading_Image() -> (Loading_Image:Common Lazy Image Object Reference)
     Index=Get bIsPurchasing() -> (bIsPurchasing:bool)
    ) -> (ReturnValue:Widget Object Reference)
   )
   > SetText!(
    execute
    self:Fort CTAButton Object Reference=Get Button_Purchase()
    ButtonText:"Text (by ref)"=SelectText(A="LOCTEXT(\"PURCHASING...\")", B="LOCTEXT(\"Purchase\")", bPickA=Get bIsPurchasing())
   ) -> (then)
  ]}
 UpdateKeyArt {nodes [
   n1 = UpdateKeyArt!() -> (then, InDisplayAsset:Fort Mtx Offer Data Object Reference) {is=title K2Node_FunctionEntry}
   > SetDesiredSizeOverride!(
    execute
    self:Image Object Reference=Get Image_KeyArt() -> (Image_KeyArt:Fort Lazy Image Object Reference)
    DesiredSize:Vector 2D Structure=Get DetailsImageSize!(self:Fort Mtx Offer Data Object Reference=@n1) -> (DetailsImageSize:Vector 2D Structure)
   ) -> (then)
   n2 = GetSoftDetailsImageURL!(self:Fort Mtx Offer Data Object Reference=@n1, World:World Object Reference, TextureSize:int=1040) -> (ReturnValue:String)
   > n3 = Branch(Condition=NOT Boolean(A=IsEmpty(InString=@n2)))
   > SetBrushFromURL!(
    execute
    self:Fort Lazy Image Object Reference=Get Image_KeyArt()
    CMSImageURL:String=@n2
    bMatchSize:bool=false
    Delay:float=0
    DownloadPriority:ECMSDownloadPriority Enum=Normal
   ) -> (then)
   SetBrushFromLazyDisplayAsset!(
    execute=@n3.else
    self:Common Lazy Image Object Reference=Get Image_KeyArt()
    LazyObject:"Soft Object Reference (by ref)"=Get SoftDetailsImage!(self:Fort Mtx Offer Data Object Reference=@n1) -> (SoftDetailsImage:Soft Object Reference)
    bMatchTextureSize:bool=false
   ) -> (then)
  ]}
 VM_OnPurchaseAvailabilityChanged {nodes [
   VM_OnPurchaseAvailabilityChanged!() -> (then, PurchaseAvailability:EFortMtxStoreOfferPurchaseAvailability Enum=Available) {is=title K2Node_FunctionEntry}
   n1 = Get VerticalBox_VBucks() -> (VerticalBox_VBucks:Vertical Box Object Reference)
   > SetActiveWidget(
    self=Get Switcher_PurchaseBlocker() -> (Switcher_PurchaseBlocker:Common Visibility Switcher Object Reference)
    Widget=Select!(
     Available:Widget Object Reference=@n1
     Unavailable_RealMoney:Widget Object Reference=@n1
     Unavailable_ParentalControls:Widget Object Reference=Get WBP_ContentPurchaseBlocker()
     Index:EFortMtxStoreOfferPurchaseAvailability Enum=Get PurchaseAvailability!() -> (PurchaseAvailability=Available)
    ) -> (ReturnValue)
   )
  ]}
]
variables [
 ClearPurchaseUXTraceIdOnExit {type=boolean, default=true}
 ExactPackBackground {type=Object, default=/FortniteCoreUI/Frontend/StoreMain/Materials/MI_UIKit_Vbucks_Background_ExactAmount_V2}
 IsExactAmount {type=boolean, default=true}
 MatchesExistingOffer {type=boolean, default=true}
 OnPurchaseComplete {}
 PurchaseUXTraceId {type=string, default=""}
]
widgets {
 info {inheritedWidgetCount=0, parentClass=&CommonActivatableWidget, rootWidgetClass=&Overlay, widgetCount=22}
 widgets [
  Overlay_2 {widgetClassPath=&Overlay, widget=Overlay_2}
  Image_Background {
   widgetClassPath=&Image
   bIsVariable=true
   parent=Overlay_2
   widget=Image_Background
   properties {brush {
     bIsDynamicallyLoaded=false
     drawAs=Image
     imageType=FullColor
     mirroring=NoMirror
     resourceName=None
     resourceObject=/FortniteCoreUI/Frontend/StoreMain/Materials/MI_UIKit_Vbucks_Background_13500_V2
     tiling=NoTile
     imageSize=<32 32>
     margin=<ltrb 0 0 0 0>
     outlineSettings {
      bUseBrushTransparency=false
      roundingType=HalfHeightRadius
      width=0
      color {colorUseRule=UseColor_Specified, specifiedColor=<rgba 0 0 0 0>}
      cornerRadii=<0 0 0 0>
     }
     tintColor {colorUseRule=UseColor_Specified, specifiedColor=<rgba 1 1 1 1>}
     uVRegion {bIsValid=false, max=<0 0>, min=<0 0>}
    }}
   slot {class=OverlaySlot, properties {horizontalAlignment=HAlign_Fill, verticalAlignment=VAlign_Fill}}
  }
  SafeZone_0 {
   widgetClassPath=&SafeZone
   parent=Overlay_2
   widget=SafeZone_0
   slot {class=OverlaySlot, properties {horizontalAlignment=HAlign_Fill, verticalAlignment=VAlign_Fill}}
  }
  Switcher {widgetClassPath=&CommonVisibilitySwitcher, bIsVariable=true, parent=SafeZone_0, widget=Switcher, slot {class=SafeZoneSlot}}
  Overlay_Content {widgetClassPath=&Overlay, bIsVariable=true, parent=Switcher, widget=Overlay_Content, slot {class=CommonVisibilitySwitcherSlot}}
  WBP_PlayerInfo {
   widgetClassPath=/Game/UI/Frontend/ItemShop/MultiProduct/WBP_PlayerInfo
   bIsVariable=true
   parent=Overlay_Content
   widget=WBP_PlayerInfo
   slot {class=OverlaySlot, properties {horizontalAlignment=HAlign_Right, padding=<ltrb 0 40 80 0>}}
  }
  HorizontalBox_0 {
   widgetClassPath=&HorizontalBox
   parent=Overlay_Content
   widget=HorizontalBox_0
   slot {class=OverlaySlot, properties {horizontalAlignment=HAlign_Fill, verticalAlignment=VAlign_Fill, padding=<ltrb 80 40 80 40>}}
  }
  SizeBox_3 {
   widgetClassPath=&SizeBox
   parent=HorizontalBox_0
   widget=SizeBox_3
   properties {bOverride_WidthOverride=true, maxDesiredWidth=600, widthOverride=640}
   slot {class=HorizontalBoxSlot, properties {horizontalAlignment=HAlign_Left}}
  }
  Switcher_PurchaseBlocker {
   widgetClassPath=&CommonVisibilitySwitcher
   bIsVariable=true
   parent=SizeBox_3
   widget=Switcher_PurchaseBlocker
   slot {class=SizeBoxSlot}
  }
  VerticalBox_VBucks {
   widgetClassPath=&VerticalBox
   bIsVariable=true
   parent=Switcher_PurchaseBlocker
   widget=VerticalBox_VBucks
   slot {class=CommonVisibilitySwitcherSlot}
  }
  Text_ExactAmount {
   widgetClassPath=&CommonTextBlock
   bIsVariable=true
   parent=VerticalBox_VBucks
   widget=Text_ExactAmount
   properties {
    bIsScrollingEnabled=false
    mobileFontSizeMultiplier=1.25
    text=Exact Amount
    textTransformPolicy=ToUpper
    colorAndOpacity {colorUseRule=UseColor_Specified, specifiedColor=<rgba 1 1 1 0.7>}
    font {
     bForceMonospaced=false
     bMaterialIsStencil=false
     fontMaterial=null
     fontObject=&HeadingNow
     letterSpacing=0
     monospacedWidth=1
     size=15.75
     skewAmount=0
     typefaceFontName=86 Bold Italic
     outlineSettings {
      bApplyOutlineToDropShadows=false
      bMiteredCorners=false
      bSeparateFillAlpha=false
      outlineMaterial=null
      outlineSize=0
      outlineColor=<rgba 0 0 0 1>
     }
    }
    shadowOffset=<0 0>
   }
   slot {class=VerticalBoxSlot}
  }
  Text_VBucksQuantity {
   widgetClassPath=&CommonTextBlock
   bIsVariable=true
   parent=VerticalBox_VBucks
   widget=Text_VBucksQuantity
   properties {
    bIsScrollingEnabled=false
    justification=Center
    mobileFontSizeMultiplier=1.25
    text="[ph] 1000 V-BUCKS"
    textTransformPolicy=ToUpper
    font {
     bForceMonospaced=false
     bMaterialIsStencil=false
     fontMaterial=null
     fontObject=&HeadingNow
     letterSpacing=0
     monospacedWidth=1
     size=25.5
     skewAmount=0
     typefaceFontName=86 Bold Italic
     outlineSettings {
      bApplyOutlineToDropShadows=false
      bMiteredCorners=false
      bSeparateFillAlpha=false
      outlineMaterial=null
      outlineSize=0
      outlineColor=<rgba 0 0 0 1>
     }
    }
    shadowOffset=<0 0>
    strikeBrush {
     bIsDynamicallyLoaded=false
     drawAs=NoDrawType
     imageType=NoImage
     mirroring=NoMirror
     resourceName=None
     resourceObject=null
     tiling=NoTile
     imageSize=<32 32>
     margin=<ltrb 0 0 0 0>
     outlineSettings {
      bUseBrushTransparency=false
      roundingType=HalfHeightRadius
      width=0
      color {colorUseRule=UseColor_Specified, specifiedColor=<rgba 0 0 0 0>}
      cornerRadii=<0 0 0 0>
     }
     tintColor {colorUseRule=UseColor_Specified, specifiedColor=<rgba 1 1 1 1>}
     uVRegion {bIsValid=false, max=<0 0>, min=<0 0>}
    }
   }
   slot {class=VerticalBoxSlot, properties {horizontalAlignment=HAlign_Left, verticalAlignment=VAlign_Top, padding=<ltrb 0 0 0 8>}}
  }
  ScrollBox_Description {
   widgetClassPath=&ScrollBox
   bIsVariable=true
   parent=VerticalBox_VBucks
   widget=ScrollBox_Description
   properties {
    scrollbarPadding=<ltrb 8 0 0 0>
    scrollbarThickness=<12 12>
    widgetBarStyle {
     thickness=12
     draggedThumbImage {
      bIsDynamicallyLoaded=false
      drawAs=RoundedBox
      imageType=FullColor
      mirroring=NoMirror
      resourceName=../../../Engine/Content/Slate/Common/Scrollbar_Thumb.png
      resourceObject=null
      tiling=NoTile
      imageSize=<12 12>
      margin=<ltrb 0.25 0.25 0.25 0.25>
      outlineSettings {
       bUseBrushTransparency=false
       roundingType=FixedRadius
       width=2
       color {colorUseRule=UseColor_Specified, specifiedColor=<rgba 0 0 0 0>}
       cornerRadii=<6 6 6 6>
      }
      tintColor {colorUseRule=UseColor_Specified, specifiedColor=<rgba 1 1 1 1>}
      uVRegion {bIsValid=false, max=<0 0>, min=<0 0>}
     }
     horizontalBackgroundImage {
      bIsDynamicallyLoaded=false
      drawAs=NoDrawType
      imageType=NoImage
      mirroring=NoMirror
      resourceName=None
      resourceObject=null
      tiling=NoTile
      imageSize=<0 0>
      margin=<ltrb 0 0 0 0>
      outlineSettings {
       bUseBrushTransparency=false
       roundingType=HalfHeightRadius
       width=0
       color {colorUseRule=UseColor_Specified, specifiedColor=<rgba 0 0 0 0>}
       cornerRadii=<0 0 0 0>
      }
      tintColor {colorUseRule=UseColor_Specified, specifiedColor=<rgba 1 1 1 1>}
      uVRegion {bIsValid=false, max=<0 0>, min=<0 0>}
     }
     horizontalBottomSlotImage {
      bIsDynamicallyLoaded=false
      drawAs=NoDrawType
      imageType=FullColor
      mirroring=NoMirror
      resourceName=../../../Engine/Content/Slate/Common/Scrollbar_Background_Horizontal.png
      resourceObject=null
      tiling=NoTile
      imageSize=<8 8>
      margin=<ltrb 0 0 0 0>
      outlineSettings {
       bUseBrushTransparency=false
       roundingType=HalfHeightRadius
       width=0
       color {colorUseRule=UseColor_Specified, specifiedColor=<rgba 0 0 0 0>}
       cornerRadii=<0 0 0 0>
      }
      tintColor {colorUseRule=UseColor_Specified, specifiedColor=<rgba 1 1 1 1>}
      uVRegion {bIsValid=false, max=<0 0>, min=<0 0>}
     }
     horizontalTopSlotImage {
      bIsDynamicallyLoaded=false
      drawAs=NoDrawType
      imageType=FullColor
      mirroring=NoMirror
      resourceName=../../../Engine/Content/Slate/Common/Scrollbar_Background_Horizontal.png
      resourceObject=null
      tiling=NoTile
      imageSize=<8 8>
      margin=<ltrb 0 0 0 0>
      outlineSettings {
       bUseBrushTransparency=false
       roundingType=HalfHeightRadius
       width=0
       color {colorUseRule=UseColor_Specified, specifiedColor=<rgba 0 0 0 0>}
       cornerRadii=<0 0 0 0>
      }
      tintColor {colorUseRule=UseColor_Specified, specifiedColor=<rgba 1 1 1 1>}
      uVRegion {bIsValid=false, max=<0 0>, min=<0 0>}
     }
     hoveredThumbImage {
      bIsDynamicallyLoaded=false
      drawAs=RoundedBox
      imageType=FullColor
      mirroring=NoMirror
      resourceName=../../../Engine/Content/Slate/Common/Scrollbar_Thumb.png
      resourceObject=null
      tiling=NoTile
      imageSize=<12 12>
      margin=<ltrb 0.25 0.25 0.25 0.25>
      outlineSettings {
       bUseBrushTransparency=false
       roundingType=FixedRadius
       width=2
       color {colorUseRule=UseColor_Specified, specifiedColor=<rgba 0 0 0 0>}
       cornerRadii=<6 6 6 6>
      }
      tintColor {colorUseRule=UseColor_Specified, specifiedColor=<rgba 1 1 1 1>}
      uVRegion {bIsValid=false, max=<0 0>, min=<0 0>}
     }
     normalThumbImage {
      bIsDynamicallyLoaded=false
      drawAs=RoundedBox
      imageType=FullColor
      mirroring=NoMirror
      resourceName=../../../Engine/Content/Slate/Common/Scrollbar_Thumb.png
      resourceObject=null
      tiling=NoTile
      imageSize=<12 12>
      margin=<ltrb 0.25 0.25 0.25 0.25>
      outlineSettings {
       bUseBrushTransparency=false
       roundingType=FixedRadius
       width=2
       color {colorUseRule=UseColor_Specified, specifiedColor=<rgba 0 0 0 0>}
       cornerRadii=<6 6 6 6>
      }
      tintColor {colorUseRule=UseColor_Specified, specifiedColor=<rgba 1 1 1 0.749>}
      uVRegion {bIsValid=false, max=<0 0>, min=<0 0>}
     }
     verticalBackgroundImage {
      bIsDynamicallyLoaded=false
      drawAs=Box
      imageType=FullColor
      mirroring=NoMirror
      resourceName=None
      resourceObject=/FortUILibrary/UIKit/ScrollBar/MaterialInstances/MI_UIKit_ScrollbarBackground
      tiling=NoTile
      imageSize=<12 12>
      margin=<ltrb 0.5 0.5 0.5 0.5>
      outlineSettings {
       bUseBrushTransparency=false
       roundingType=HalfHeightRadius
       width=0
       color {colorUseRule=UseColor_Specified, specifiedColor=<rgba 0 0 0 0>}
       cornerRadii=<0 0 0 0>
      }
      tintColor {colorUseRule=UseColor_Specified, specifiedColor=<rgba 0.521 0.521 0.521 0.502>}
      uVRegion {bIsValid=false, max=<0 0>, min=<0 0>}
     }
     verticalBottomSlotImage {
      bIsDynamicallyLoaded=false
      drawAs=NoDrawType
      imageType=FullColor
      mirroring=NoMirror
      resourceName=../../../Engine/Content/Slate/Common/Scrollbar_Background_Vertical.png
      resourceObject=null
      tiling=NoTile
      imageSize=<8 8>
      margin=<ltrb 0 0 0 0>
      outlineSettings {
       bUseBrushTransparency=false
       roundingType=HalfHeightRadius
       width=0
       color {colorUseRule=UseColor_Specified, specifiedColor=<rgba 0 0 0 0>}
       cornerRadii=<0 0 0 0>
      }
      tintColor {colorUseRule=UseColor_Specified, specifiedColor=<rgba 1 1 1 1>}
      uVRegion {bIsValid=false, max=<0 0>, min=<0 0>}
     }
     verticalTopSlotImage {
      bIsDynamicallyLoaded=false
      drawAs=NoDrawType
      imageType=FullColor
      mirroring=NoMirror
      resourceName=../../../Engine/Content/Slate/Common/Scrollbar_Background_Vertical.png
      resourceObject=null
      tiling=NoTile
      imageSize=<8 8>
      margin=<ltrb 0 0 0 0>
      outlineSettings {
       bUseBrushTransparency=false
       roundingType=HalfHeightRadius
       width=0
       color {colorUseRule=UseColor_Specified, specifiedColor=<rgba 0 0 0 0>}
       cornerRadii=<0 0 0 0>
      }
      tintColor {colorUseRule=UseColor_Specified, specifiedColor=<rgba 1 1 1 1>}
      uVRegion {bIsValid=false, max=<0 0>, min=<0 0>}
     }
    }
    widgetStyle {
     barThickness=12
     bottomShadowBrush {
      bIsDynamicallyLoaded=false
      drawAs=NoDrawType
      imageType=FullColor
      mirroring=NoMirror
      resourceName=../../../Engine/Content/Slate/Common/ScrollBoxShadowBottom.png
      resourceObject=null
      tiling=NoTile
      imageSize=<16 8>
      margin=<ltrb 0.5 0 0.5 1>
      outlineSettings {
       bUseBrushTransparency=false
       roundingType=HalfHeightRadius
       width=0
       color {colorUseRule=UseColor_Specified, specifiedColor=<rgba 0 0 0 0>}
       cornerRadii=<0 0 0 0>
      }
      tintColor {colorUseRule=UseColor_Specified, specifiedColor=<rgba 1 1 1 1>}
      uVRegion {bIsValid=false, max=<0 0>, min=<0 0>}
     }
     horizontalScrolledContentPadding=<ltrb 0 0 1 0>
     leftShadowBrush {
      bIsDynamicallyLoaded=false
      drawAs=NoDrawType
      imageType=FullColor
      mirroring=NoMirror
      resourceName=../../../Engine/Content/Slate/Common/ScrollBoxShadowLeft.png
      resourceObject=null
      tiling=NoTile
      imageSize=<8 16>
      margin=<ltrb 1 0.5 0 0.5>
      outlineSettings {
       bUseBrushTransparency=false
       roundingType=HalfHeightRadius
       width=0
       color {colorUseRule=UseColor_Specified, specifiedColor=<rgba 0 0 0 0>}
       cornerRadii=<0 0 0 0>
      }
      tintColor {colorUseRule=UseColor_Specified, specifiedColor=<rgba 1 1 1 1>}
      uVRegion {bIsValid=false, max=<0 0>, min=<0 0>}
     }
     rightShadowBrush {
      bIsDynamicallyLoaded=false
      drawAs=NoDrawType
      imageType=FullColor
      mirroring=NoMirror
      resourceName=../../../Engine/Content/Slate/Common/ScrollBoxShadowRight.png
      resourceObject=null
      tiling=NoTile
      imageSize=<8 16>
      margin=<ltrb 0 0.5 1 0.5>
      outlineSettings {
       bUseBrushTransparency=false
       roundingType=HalfHeightRadius
       width=0
       color {colorUseRule=UseColor_Specified, specifiedColor=<rgba 0 0 0 0>}
       cornerRadii=<0 0 0 0>
      }
      tintColor {colorUseRule=UseColor_Specified, specifiedColor=<rgba 1 1 1 1>}
      uVRegion {bIsValid=false, max=<0 0>, min=<0 0>}
     }
     topShadowBrush {
      bIsDynamicallyLoaded=false
      drawAs=NoDrawType
      imageType=FullColor
      mirroring=NoMirror
      resourceName=../../../Engine/Content/Slate/Common/ScrollBoxShadowTop.png
      resourceObject=null
      tiling=NoTile
      imageSize=<16 8>
      margin=<ltrb 0.5 1 0.5 0>
      outlineSettings {
       bUseBrushTransparency=false
       roundingType=HalfHeightRadius
       width=0
       color {colorUseRule=UseColor_Specified, specifiedColor=<rgba 0 0 0 0>}
       cornerRadii=<0 0 0 0>
      }
      tintColor {colorUseRule=UseColor_Specified, specifiedColor=<rgba 1 1 1 1>}
      uVRegion {bIsValid=false, max=<0 0>, min=<0 0>}
     }
     verticalScrolledContentPadding=<ltrb 0 0 0 1>
    }
   }
   slot {class=VerticalBoxSlot, properties {padding=<ltrb 0 0 0 12>, size {sizeRule=Fill, value=1}}}
  }
  Text_Description {
   widgetClassPath=&CommonTextBlock
   bIsVariable=true
   parent=ScrollBox_Description
   widget=Text_Description
   properties {
    autoWrapText=true
    mobileFontSizeMultiplier=1.15
    text="[ph] Buy X,XXX Fortnite Vbucks, the in-game currency that can be spent in Fornite. You can purchase new customization items in the Shop for Fortnite Battle Royale, LEGO Fortnite, Fortnite Festival or Rocket Racing, as well as additional content like this season’s Battle Pass. "
    wrappingPolicy=AllowPerCharacterWrapping
    colorAndOpacity {colorUseRule=UseColor_Specified, specifiedColor=<rgba 1 1 1 0.7>}
    font {
     bForceMonospaced=false
     bMaterialIsStencil=false
     fontMaterial=null
     fontObject=&HeadingNow
     letterSpacing=0
     monospacedWidth=1
     size=15.75
     skewAmount=0
     typefaceFontName=74 Regular
     outlineSettings {
      bApplyOutlineToDropShadows=false
      bMiteredCorners=false
      bSeparateFillAlpha=false
      outlineMaterial=null
      outlineSize=0
      outlineColor=<rgba 0 0 0 1>
     }
    }
    shadowOffset=<0 0>
   }
   slot {class=ScrollBoxSlot}
  }
  WBP_PriceSwitcher_Base {
   widgetClassPath=/FortUILibrary/Systems/ItemShopTile/WBP_PriceSwitcher_Base
   bIsVariable=true
   parent=VerticalBox_VBucks
   widget=WBP_PriceSwitcher_Base
   properties {
    previewCurrency=RealMoney
    priceDisplayMode=PriceOnly
    widgetHeight=42
    additionalPriceTextFont {
     bForceMonospaced=false
     bMaterialIsStencil=false
     fontMaterial=/FortUILibrary/Systems/ItemShopTile/Materials/PriceSwitcher/MI_UIKit_PriceSwitcher_Text
     fontObject=&HeadingNow
     letterSpacing=0
     monospacedWidth=1
     size=11.25
     skewAmount=0
     typefaceFontName=74 Regular
     outlineSettings {
      bApplyOutlineToDropShadows=false
      bMiteredCorners=false
      bSeparateFillAlpha=false
      outlineMaterial=null
      outlineSize=0
      outlineColor=<rgba 0 0 0 1>
     }
    }
    currencyImageSize=<32 32>
    ownedTextFont {
     bForceMonospaced=false
     bMaterialIsStencil=false
     fontMaterial=&MI_UIKit_PriceSwitcher_Text
     fontObject=&HeadingNow
     letterSpacing=0
     monospacedWidth=1
     size=25.5
     skewAmount=0
     typefaceFontName=86 Bold Italic
     outlineSettings {
      bApplyOutlineToDropShadows=false
      bMiteredCorners=false
      bSeparateFillAlpha=false
      outlineMaterial=null
      outlineSize=0
      outlineColor=<rgba 0 0 0 1>
     }
    }
    priceTextFont {
     bForceMonospaced=false
     bMaterialIsStencil=false
     fontMaterial=&MI_UIKit_PriceSwitcher_Text
     fontObject=&HeadingNow
     letterSpacing=0
     monospacedWidth=1
     size=25.5
     skewAmount=0
     typefaceFontName=86 Bold Italic
     outlineSettings {
      bApplyOutlineToDropShadows=false
      bMiteredCorners=false
      bSeparateFillAlpha=false
      outlineMaterial=null
      outlineSize=0
      outlineColor=<rgba 0 0 0 1>
     }
    }
    regularPriceTextFont {
     bForceMonospaced=false
     bMaterialIsStencil=false
     fontMaterial=&MI_UIKit_PriceSwitcher_Text
     fontObject=&HeadingNow
     letterSpacing=0
     monospacedWidth=1
     size=25.5
     skewAmount=0
     typefaceFontName=86 Bold Italic
     outlineSettings {
      bApplyOutlineToDropShadows=false
      bMiteredCorners=false
      bSeparateFillAlpha=false
      outlineMaterial=null
      outlineSize=0
      outlineColor=<rgba 0 0 0 1>
     }
    }
    specialOfferTextFont {
     bForceMonospaced=false
     bMaterialIsStencil=false
     fontMaterial=&MI_UIKit_PriceSwitcher_Text
     fontObject=&HeadingNow
     letterSpacing=0
     monospacedWidth=1
     size=25.5
     skewAmount=0
     typefaceFontName=86 Bold Italic
     outlineSettings {
      bApplyOutlineToDropShadows=false
      bMiteredCorners=false
      bSeparateFillAlpha=false
      outlineMaterial=null
      outlineSize=0
      outlineColor=<rgba 0 0 0 1>
     }
    }
   }
   slot {class=VerticalBoxSlot}
  }
  RichText_EpicRewards {
   widgetClassPath=&CommonRichTextBlock
   bIsVariable=true
   parent=VerticalBox_VBucks
   widget=RichText_EpicRewards
   properties {
    autoWrapText=true
    bAutoCollapseWithEmptyText=true
    bIsScrollingEnabled=false
    bOverrideDefaultStyle=true
    mobileTextBlockScale=1.25
    text="[ph] Earn <text color=\"38FF84\">20% ($18.00) back</> in Epic Rewards"
    visibility=Collapsed
    defaultTextStyleOverride {
     overflowPolicy=Clip
     transformPolicy=None
     colorAndOpacity {colorUseRule=UseColor_Specified, specifiedColor=<rgba 1 1 1 1>}
     font {
      bForceMonospaced=false
      bMaterialIsStencil=false
      fontMaterial=null
      fontObject=&HeadingNow
      letterSpacing=0
      monospacedWidth=1
      size=20.25
      skewAmount=0
      typefaceFontName=75 Medium
      outlineSettings {
       bApplyOutlineToDropShadows=false
       bMiteredCorners=false
       bSeparateFillAlpha=false
       outlineMaterial=null
       outlineSize=0
       outlineColor=<rgba 0 0 0 1>
      }
     }
     highlightColor {colorUseRule=UseColor_Specified, specifiedColor=<rgba 0 0 0 1>}
     highlightShape {
      bIsDynamicallyLoaded=false
      drawAs=Image
      imageType=NoImage
      mirroring=NoMirror
      resourceName=None
      resourceObject=null
      tiling=NoTile
      imageSize=<32 32>
      margin=<ltrb 0 0 0 0>
      outlineSettings {
       bUseBrushTransparency=false
       roundingType=HalfHeightRadius
       width=0
       color {colorUseRule=UseColor_Specified, specifiedColor=<rgba 0 0 0 0>}
       cornerRadii=<0 0 0 0>
      }
      tintColor {colorUseRule=UseColor_Specified, specifiedColor=<rgba 1 1 1 1>}
      uVRegion {bIsValid=false, max=<0 0>, min=<0 0>}
     }
     selectedBackgroundColor {colorUseRule=UseColor_Foreground, specifiedColor=<rgba 1 0 1 1>}
     shadowColorAndOpacity=<rgba 0 0 0 1>
     shadowOffset=<0 0>
     strikeBrush {
      bIsDynamicallyLoaded=false
      drawAs=Image
      imageType=NoImage
      mirroring=NoMirror
      resourceName=None
      resourceObject=null
      tiling=NoTile
      imageSize=<32 32>
      margin=<ltrb 0 0 0 0>
      outlineSettings {
       bUseBrushTransparency=false
       roundingType=HalfHeightRadius
       width=0
       color {colorUseRule=UseColor_Specified, specifiedColor=<rgba 0 0 0 0>}
       cornerRadii=<0 0 0 0>
      }
      tintColor {colorUseRule=UseColor_Specified, specifiedColor=<rgba 1 1 1 1>}
      uVRegion {bIsValid=false, max=<0 0>, min=<0 0>}
     }
     underlineBrush {
      bIsDynamicallyLoaded=false
      drawAs=Image
      imageType=NoImage
      mirroring=NoMirror
      resourceName=None
      resourceObject=null
      tiling=NoTile
      imageSize=<32 32>
      margin=<ltrb 0 0 0 0>
      outlineSettings {
       bUseBrushTransparency=false
       roundingType=HalfHeightRadius
       width=0
       color {colorUseRule=UseColor_Specified, specifiedColor=<rgba 0 0 0 0>}
       cornerRadii=<0 0 0 0>
      }
      tintColor {colorUseRule=UseColor_Specified, specifiedColor=<rgba 1 1 1 1>}
      uVRegion {bIsValid=false, max=<0 0>, min=<0 0>}
     }
    }
   }
   slot {class=VerticalBoxSlot, properties {verticalAlignment=VAlign_Center, padding=<ltrb 0 8 0 0>}}
  }
  Text_Legal {
   widgetClassPath=&CommonTextBlock
   parent=VerticalBox_VBucks
   widget=Text_Legal
   properties {
    autoWrapText=true
    bIsScrollingEnabled=false
    mobileFontSizeMultiplier=1.4
    text=All V-Bucks purchased on one platform may not be redeemable on other platforms.
    colorAndOpacity {colorUseRule=UseColor_Specified, specifiedColor=<rgba 1 1 1 0.65>}
    font {
     bForceMonospaced=false
     bMaterialIsStencil=false
     fontMaterial=null
     fontObject=&HeadingNow
     letterSpacing=0
     monospacedWidth=1
     size=11.25
     skewAmount=0
     typefaceFontName=74 Regular
     outlineSettings {
      bApplyOutlineToDropShadows=false
      bMiteredCorners=false
      bSeparateFillAlpha=false
      outlineMaterial=null
      outlineSize=0
      outlineColor=<rgba 0 0 0 1>
     }
    }
    shadowOffset=<0 0>
   }
   slot {class=VerticalBoxSlot, properties {padding=<ltrb 0 16 0 0>}}
  }
  Button_Purchase {
   widgetClassPath=/FortUILibrary/UIKit/Buttons/WBP_UIKit_Button_Loud
   bIsVariable=true
   parent=VerticalBox_VBucks
   widget=Button_Purchase
   properties {holdData=/FortUILibrary/Input/ShortInteraction_HoldData.ShortInteraction_HoldData_C, minHeight=88, text=Purchase}
   slot {class=VerticalBoxSlot, properties {horizontalAlignment=HAlign_Left, verticalAlignment=VAlign_Bottom, padding=<ltrb 0 16 0 0>}}
  }
  WBP_ContentPurchaseBlocker {
   widgetClassPath=/FortniteCoreUI/UI/ContentControls/WBP_ContentPurchaseBlocker
   bIsVariable=true
   parent=Switcher_PurchaseBlocker
   widget=WBP_ContentPurchaseBlocker
   properties {max Width=500, text_Details="To purchase this V-Buck pack, your parent or guardian needs to set up your account from Parental Controls."}
   slot {class=CommonVisibilitySwitcherSlot, properties {horizontalAlignment=HAlign_Center, verticalAlignment=VAlign_Center}}
  }
  Image_KeyArt {
   widgetClassPath=&FortLazyImage
   bIsVariable=true
   parent=HorizontalBox_0
   widget=Image_KeyArt
   properties {
    bOverride_Cursor=true
    visibility=HitTestInvisible
    brush {
     bIsDynamicallyLoaded=false
     drawAs=Image
     imageType=FullColor
     mirroring=NoMirror
     resourceName=None
     resourceObject=/FortniteCoreUI/Frontend/StoreMain/Materials/Textures/T_UI_VBucks_Details_13500
     tiling=NoTile
     imageSize=<1040 1040>
     margin=<ltrb 0 0 0 0>
     outlineSettings {
      bUseBrushTransparency=false
      roundingType=HalfHeightRadius
      width=0
      color {colorUseRule=UseColor_Specified, specifiedColor=<rgba 0 0 0 0>}
      cornerRadii=<0 0 0 0>
     }
     tintColor {colorUseRule=UseColor_Specified, specifiedColor=<rgba 1 1 1 1>}
     uVRegion {bIsValid=false, max=<0 0>, min=<0 0>}
    }
    loadingBackgroundBrush {
     bIsDynamicallyLoaded=false
     drawAs=Image
     imageType=FullColor
     mirroring=NoMirror
     resourceName=None
     resourceObject=/Game/UI/ContentOnDemand/ItemDownloadIndicator/MI_ItemDownloadIndicator_SpinnerMaterial
     tiling=NoTile
     imageSize=<64 64>
     margin=<ltrb 0 0 0 0>
     outlineSettings {
      bUseBrushTransparency=false
      roundingType=HalfHeightRadius
      width=0
      color {colorUseRule=UseColor_Specified, specifiedColor=<rgba 0 0 0 0>}
      cornerRadii=<0 0 0 0>
     }
     tintColor {colorUseRule=UseColor_Specified, specifiedColor=<rgba 1 1 1 1>}
     uVRegion {bIsValid=false, max=<0 0>, min=<0 0>}
    }
   }
   slot {class=HorizontalBoxSlot, properties {horizontalAlignment=HAlign_Center, verticalAlignment=VAlign_Center, size {sizeRule=Fill, value=1}}}
  }
  Loading_Image {
   widgetClassPath=/Script/CommonUI.CommonLazyImage
   bIsVariable=true
   parent=Switcher
   widget=Loading_Image
   properties {
    visibility=HitTestInvisible
    brush {
     bIsDynamicallyLoaded=false
     drawAs=Image
     imageType=FullColor
     mirroring=NoMirror
     resourceName=None
     resourceObject=&MI_ItemDownloadIndicator_SpinnerMaterial
     tiling=NoTile
     imageSize=<64 64>
     margin=<ltrb 0 0 0 0>
     outlineSettings {
      bUseBrushTransparency=false
      roundingType=HalfHeightRadius
      width=0
      color {colorUseRule=UseColor_Specified, specifiedColor=<rgba 0 0 0 0>}
      cornerRadii=<0 0 0 0>
     }
     tintColor {colorUseRule=UseColor_Specified, specifiedColor=<rgba 1 1 1 1>}
     uVRegion {bIsValid=false, max=<0 0>, min=<0 0>}
    }
    loadingBackgroundBrush {
     bIsDynamicallyLoaded=false
     drawAs=Image
     imageType=FullColor
     mirroring=NoMirror
     resourceName=None
     resourceObject=&MI_ItemDownloadIndicator_SpinnerMaterial
     tiling=NoTile
     imageSize=<64 64>
     margin=<ltrb 0 0 0 0>
     outlineSettings {
      bUseBrushTransparency=false
      roundingType=HalfHeightRadius
      width=0
      color {colorUseRule=UseColor_Specified, specifiedColor=<rgba 0 0 0 0>}
      cornerRadii=<0 0 0 0>
     }
     tintColor {colorUseRule=UseColor_Specified, specifiedColor=<rgba 1 1 1 1>}
     uVRegion {bIsValid=false, max=<0 0>, min=<0 0>}
    }
   }
   slot {class=CommonVisibilitySwitcherSlot, properties {horizontalAlignment=HAlign_Center, verticalAlignment=VAlign_Center}}
  }
  DecoyButton_Back {
   widgetClassPath=/Game/UI/Frontend/BottomBar/WBP_BottomBarDecoyButton
   bIsVariable=true
   parent=Overlay_2
   widget=DecoyButton_Back
   properties {
    holdData=&CommonUI_HoldData_Base_C
    triggeringInputAction {dataTable=&InputActionDataTable, rowName=Back}
   }
   slot {class=OverlaySlot}
  }
 ]
}
