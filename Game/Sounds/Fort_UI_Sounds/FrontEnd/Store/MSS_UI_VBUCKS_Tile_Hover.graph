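# One-shot stereo MetaSound Source for the V-Bucks tile hover sound in the front-end store UI.
# On play, a shuffle node picks one of six hover wave variations and feeds it to a stereo wave player.
# Both channels are scaled by a MasterVolume input, and playback completion fires OnFinished.
# It uses the Fort_UI sound class and a hover-tick concurrency rule named for stopping the oldest instance.
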MetaSoundSource /Game/Sounds/Fort_UI_Sounds/FrontEnd/Store/MSS_UI_VBUCKS_Tile_Hover
inputs [
 MasterVolume {type=Float}
 Sounds {type=WaveAsset:Array}
 UE.Source.OnPlay {type=Trigger}
]
interfaces [UE.OutputFormat.Stereo:1.0, UE.Source.OneShot:1.0, UE.Source:1.1]
outputs [
 "UE.OutputFormat.Stereo.Audio:0" {type=Audio}
 "UE.OutputFormat.Stereo.Audio:1" {type=Audio}
 UE.Source.OneShot.OnFinished {type=Trigger}
]
pages [Default {
  input_defaults {
   MasterVolume=0.6743
   UE.Source.OnPlay=null
   Sounds [
    /Game/Sounds/Fort_UI_Sounds/FrontEnd/Store/wav/UI_VBucks_Hover_1
    /Game/Sounds/Fort_UI_Sounds/FrontEnd/Store/wav/UI_VBucks_Hover_2
    /Game/Sounds/Fort_UI_Sounds/FrontEnd/Store/wav/UI_VBucks_Hover_3
    /Game/Sounds/Fort_UI_Sounds/FrontEnd/Store/wav/UI_VBucks_Hover_4
    /Game/Sounds/Fort_UI_Sounds/FrontEnd/Store/wav/UI_VBucks_Hover_5
    /Game/Sounds/Fort_UI_Sounds/FrontEnd/Store/wav/UI_VBucks_Hover_6
   ]
  }
  nodes [
   n1 = Array.Shuffle:WaveAsset:Array(In Array=^Sounds, Next=^UE.Source.OnPlay) {also {defaults {Enable Shared State="<MetaSoundBuilderResult.SUCCEEDED: 0>"}}}
   WavePlayer_0 = UE.Wave Player:Stereo(Play=@n1.On Next, Wave Asset=@n1.Value) {also {version [1,1]}}
   Multiply_0 = UE.Multiply:Audio by Float(AdditionalOperands=^MasterVolume, PrimaryOperand=@WavePlayer_0.Out Left)
   Multiply_1 = UE.Multiply:Audio by Float(AdditionalOperands=^MasterVolume, PrimaryOperand=@WavePlayer_0.Out Right)
  ]
  links [
   {dst="@UE.OutputFormat.Stereo.Audio:0", src=Multiply_0.Out}
   {dst="@UE.OutputFormat.Stereo.Audio:1", src=Multiply_1.Out}
   {dst="@UE.Source.OneShot.OnFinished", src=WavePlayer_0.On Finished}
  ]
 }]
properties {
 assetImportData=&self:AssetImportData
 numChannels=2
 outputFormat=Stereo
 sampleRate=48000
 soundClassObject=&Fort_UI
 concurrencySet [/Game/Sounds/Concurrency_Rules/FrontEnd_UI_Menu_Max_01_Stop_Oldest_Hover_Tick]
}
