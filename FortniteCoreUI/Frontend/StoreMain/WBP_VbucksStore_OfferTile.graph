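# V-Bucks store offer tile button showing an offer's amount, bonus badge, art and price, with intro and hover animations.
# Offer viewmodel changes size the tile and async-load its background, image (URL or soft texture) and badge.
# Purchase availability switches the footer between price and "Learn More".
# Clicks send analytics, then open an unavailable modal, new or legacy details screen (config-gated), or purchase directly, broadcasting completion through dispatchers.
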WidgetBlueprint /FortniteCoreUI/Frontend/StoreMain/WBP_VbucksStore_OfferTile
animations [
 Intro {
  binding_count=4
  display_label=""
  length_seconds=0.5
  object=&self:Intro
  sequence {
   bindings [
    Image_Icon {
     id=#1
     possessed_class=Image
     tracks [
      {
       type=MovieScene2DTransformTrack
       sections [{
         type=MovieScene2DTransformSection
         end_frame=40
         start_frame=0
         channels [
          Translation.X {type=MovieSceneScriptingFloatChannel, default=0}
          Translation.Y {type=MovieSceneScriptingFloatChannel, default=0, keys [[5,258.199],[40,0]]}
          Scale.X {type=MovieSceneScriptingFloatChannel, default=1}
          Scale.Y {type=MovieSceneScriptingFloatChannel, default=1}
         ]
        }]
      }
      {
       type=MovieSceneFloatTrack
       display_name=Render Opacity
       sections [{
         type=MovieSceneFloatSection
         end_frame=11
         start_frame=0
         channels [None {type=MovieSceneScriptingFloatChannel, default=1, keys [[5,0],[11,1]]}]
        }]
      }
     ]
    }
    SizeBox {
     id=#2
     possessed_class=SizeBox
     tracks [
      {
       type=MovieSceneFloatTrack
       display_name=Render Opacity
       sections [{type=MovieSceneFloatSection, end_frame=40, start_frame=0, channels [None {type=MovieSceneScriptingFloatChannel, keys [[0,0],[10,1]]}]}]
      }
      {
       type=MovieScene2DTransformTrack
       sections [{
         type=MovieScene2DTransformSection
         end_frame=20
         start_frame=0
         channels [
          Translation.X {type=MovieSceneScriptingFloatChannel, default=1}
          Translation.Y {type=MovieSceneScriptingFloatChannel, keys [[0,450],[20,1]]}
          Scale.X {type=MovieSceneScriptingFloatChannel, default=1}
          Scale.Y {type=MovieSceneScriptingFloatChannel, keys [[0,0],[20,1]]}
         ]
        }]
      }
     ]
    }
    Switcher_Price {
     id=#3
     possessed_class=CommonWidgetSwitcherLegacy
     tracks [
      {
       type=MovieSceneFloatTrack
       display_name=Render Opacity
       sections [{type=MovieSceneFloatSection, end_frame=40, start_frame=0, channels [None {type=MovieSceneScriptingFloatChannel, keys [[15,0],[35,1]]}]}]
      }
      {
       type=MovieScene2DTransformTrack
       sections [{
         type=MovieScene2DTransformSection
         end_frame=50
         start_frame=0
         channels [
          Translation.X {type=MovieSceneScriptingFloatChannel}
          Translation.Y {type=MovieSceneScriptingFloatChannel}
          Angle {type=MovieSceneScriptingFloatChannel}
          Scale.X {type=MovieSceneScriptingFloatChannel, keys [[15,1.1],[50,1]]}
          Scale.Y {type=MovieSceneScriptingFloatChannel, keys [[15,0.8],[29.9983,1]]}
          Shear.X {type=MovieSceneScriptingFloatChannel}
          Shear.Y {type=MovieSceneScriptingFloatChannel}
         ]
        }]
      }
     ]
    }
    Grid_Amount {
     id=#4
     possessed_class=VerticalBox
     tracks [
      {
       type=MovieSceneFloatTrack
       display_name=Render Opacity
       sections [{type=MovieSceneFloatSection, end_frame=40, start_frame=0, channels [None {type=MovieSceneScriptingFloatChannel, keys [[10,0],[20,1]]}]}]
      }
      {
       type=MovieScene2DTransformTrack
       sections [{
         type=MovieScene2DTransformSection
         end_frame=40
         start_frame=0
         channels [
          Translation.X {type=MovieSceneScriptingFloatChannel}
          Translation.Y {type=MovieSceneScriptingFloatChannel, keys [[10,151.5209],[30,0]]}
          Angle {type=MovieSceneScriptingFloatChannel}
          Scale.X {type=MovieSceneScriptingFloatChannel}
          Scale.Y {type=MovieSceneScriptingFloatChannel}
          Shear.X {type=MovieSceneScriptingFloatChannel}
          Shear.Y {type=MovieSceneScriptingFloatChannel}
         ]
        }]
      }
     ]
    }
   ]
   display_rate {numerator=100}
  }
 }
 Hover {
  binding_count=5
  length_seconds=0.5
  object=&self:Hover
  sequence {
   bindings [
    Image_Background {
     id=#5
     possessed_class=Image
     tracks [{
       type=MovieSceneWidgetMaterialTrack
       sections [{type=MovieSceneParameterSection, end_frame=30, start_frame=0, channels [Hover {type=MovieSceneScriptingFloatChannel, keys [[0,0],[30,1]]}]}]
      }]
    }
    Border_Footer {
     id=#6
     possessed_class=Border
     tracks [
      {
       type=MovieSceneWidgetMaterialTrack
       sections [{
         type=MovieSceneParameterSection
         end_frame=12
         start_frame=0
         channels [TopCenterOffset {type=MovieSceneScriptingFloatChannel, keys [[0,-0.3],[12,-0.25]]}]
        }]
      }
      {
       type=MovieSceneColorTrack
       display_name=Brush Color
       sections [{
         type=MovieSceneColorSection
         end_frame=12
         start_frame=0
         channels [
          Color.R {type=MovieSceneScriptingFloatChannel, const [1,1]}
          Color.G {type=MovieSceneScriptingFloatChannel, keys [[0,0.9216],[12,1]]}
          Color.B {type=MovieSceneScriptingFloatChannel, keys [[0,0],[12,1]]}
          Color.A {type=MovieSceneScriptingFloatChannel, const [1,1]}
         ]
        }]
      }
     ]
    }
    Overlay {
     id=#7
     possessed_class=Overlay
     tracks [{
       type=MovieScene2DTransformTrack
       sections [{
         type=MovieScene2DTransformSection
         end_frame=12
         start_frame=0
         channels [
          Translation.X {type=MovieSceneScriptingFloatChannel}
          Translation.Y {type=MovieSceneScriptingFloatChannel, keys [[0,0],[12,-8]]}
         ]
        }]
      }]
    }
    Image_Icon {
     id=#8
     possessed_class=Image
     tracks [{
       type=MovieScene2DTransformTrack
       sections [{
         type=MovieScene2DTransformSection
         end_frame=12
         start_frame=0
         channels [
          Translation.X {type=MovieSceneScriptingFloatChannel, default=0}
          Translation.Y {type=MovieSceneScriptingFloatChannel, default=0}
          Angle {type=MovieSceneScriptingFloatChannel, default=0}
          Scale.X {type=MovieSceneScriptingFloatChannel, keys [[0,1],[12,1.05]]}
          Scale.Y {type=MovieSceneScriptingFloatChannel, keys [[0,1],[12,1.05]]}
         ]
        }]
      }]
    }
    ScaleBox_Icon {
     id=#9
     display_name=ScaleBoxKeyArt
     possessed_class=ScaleBox
     tracks [{
       type=MovieScene2DTransformTrack
       sections [{
         type=MovieScene2DTransformSection
         end_frame=12
         start_frame=0
         channels [
          Translation.X {type=MovieSceneScriptingFloatChannel}
          Translation.Y {type=MovieSceneScriptingFloatChannel, default=0, keys [[0,-24],[12,-32]]}
          Shear.X {type=MovieSceneScriptingFloatChannel}
          Shear.Y {type=MovieSceneScriptingFloatChannel}
         ]
        }]
      }]
    }
   ]
   display_rate {numerator=60}
  }
 }
]
graphs [
 EventGraph {nodes [
   n1 = PlayIntroAnimation!() -> (OutputDelegate, then, DelayTime:double=~) {is=title K2Node_CustomEvent}
   > StopAllAnimations(self:User Widget Object Reference)
   n2 = Get Intro()
   > PlayAnimation(InAnimation=@n2, StartAtTime=GetEndTime(self=@n2), PlayMode=Reverse, PlaybackSpeed=0)
   > Delay(Duration=@n1.DelayTime)
   > PlayAnimation(InAnimation=Get Intro())
   > PlaySound2D(Sound=Get MTXOfferTileIntroWhoosh() -> (MTXOfferTileIntroWhoosh:Sound Base Object Reference))
   n3 = VM_OfferInfoChanged!() -> (OutputDelegate, then, OfferInfo:Fort Store Front Offer Info Object Reference) {is=title K2Node_CustomEvent}
   > SetTileSize!(
    execute
    self
    DisplaySize:EFortMtxOfferDisplaySize Enum=GetOfferDisplaySize!(self:Fort Store Front Offer Info Object Reference=@n3.OfferInfo) -> (ReturnValue:EFortMtxOfferDisplaySize Enum=Small)
   ) -> (then)
   n4 = NOT Boolean(A=Get MatchesExistingTile() -> (MatchesExistingTile:bool))
   n5 = Get IsExactAmount() -> (IsExactAmount:bool)
   n6 = GetOfferDisplayAsset!(self:Fort Store Front Offer Info Object Reference=@n3.OfferInfo) -> (ReturnValue:Fort Mtx Offer Data Object Reference)
   > n7 = Async Load Asset(Asset=Select(
     Option 0:Soft Object Reference=Get SoftBackgroundImage!(self:Fort Mtx Offer Data Object Reference=@n6) -> (SoftBackgroundImage:Soft Object Reference)
     Option 1:Soft Object Reference=Get ExactPackBackground() -> (ExactPackBackground:Soft Object Reference)
     Index=AND Boolean(A=@n5, B=@n4)
    ) -> (ReturnValue:Soft Object Reference))
   n8 = Cast To MaterialInstance!(execute=@n7.Completed, Object=@n7) -> (then, CastFailed, AsMaterial Instance:Material Instance Object Reference) {is=title K2Node_DynamicCast}
   > SetBrushFromMaterial(self=Get Image_Background() -> (Image_Background:Image Object Reference), Material=@n8)
   > Set Background Glow Position!(execute:Exec) -> (then:Exec) {is=title K2Node_MacroInstance}
   > SetDisplayAssetImage!(execute, self, DisplayAsset:Fort Mtx Offer Data Object Reference=GetOfferDisplayAsset(self=@n3.OfferInfo)) -> (then)
   > n9 = Async Load Asset(Asset=Get SoftBadgeImage!(self:Fort Mtx Offer Data Object Reference=GetOfferDisplayAsset(self=@n3.OfferInfo)) -> (SoftBadgeImage:Soft Object Reference))
   > Set ContentColorAndOpacity!(
    execute
    ContentColorAndOpacity:Linear Color Structure=Get UpsellPrimaryColor!(self:Fort Mtx Offer Data Object Reference=GetOfferDisplayAsset(self=@n3.OfferInfo)) -> (UpsellPrimaryColor:Linear Color Structure)
    self:Border Object Reference=Get Border_HighlightText() -> (Border_HighlightText:Common Border Object Reference)
   ) -> (then, Output_Get:Linear Color Structure)
   n10 = Cast To MaterialInstance(execute=@n9.Completed, Object=@n9)
   > SetBrushFromMaterial(self=Get Image_Highlight_BG() -> (Image_Highlight_BG:Image Object Reference), Material=@n10)
   n11 = OnRealMoneyPurchaseComplete!() -> (OutputDelegate, then, bSuccess:bool=~, PurchasedItems:Array of Purchased Item Info Structures) {is=title K2Node_CustomEvent}
   > Call On Purchase Complete!(execute:Exec, self:WBP Vbucks Store Offer Tile Object Reference, bSuccess:bool=@n11.bSuccess) -> (then:Exec) {is=title K2Node_CallDelegate}
   n12 = OnPdpPurchaseComplete!() -> (OutputDelegate, then, bSuccess:bool=~) {is=title K2Node_CustomEvent}
   > Call On Purchase Complete(bSuccess=@n12.bSuccess)
   n13 = HandlePurchaseComplete!() -> (OutputDelegate, then, Result:EFortMtxStoreOfferPurchaseResult Enum=Succeeded) {is=title K2Node_CustomEvent}
   > Call Set Is Purchasing!(execute:Exec, self:WBP Vbucks Store Offer Tile Object Reference, IsPurchasing:bool) -> (then:Exec) {is=title K2Node_CallDelegate}
   > Call On Purchase Complete(bSuccess="Equal (Enum)"(A:EFortMtxStoreOfferPurchaseResult Enum=@n13.Result, B:EFortMtxStoreOfferPurchaseResult Enum=Succeeded))
   n14 = SetDisplayAssetImage!() -> (OutputDelegate, then, DisplayAsset:Fort Mtx Offer Data Object Reference) {is=title K2Node_CustomEvent}
   n15 = GetSoftTileImageURL!(self:Fort Mtx Offer Data Object Reference=@n14.DisplayAsset, World:World Object Reference, TextureSize:int=520) -> (ReturnValue:String)
   > n16 = Branch(Condition=IsEmpty(InString=@n15))
   > n17 = Async Load Asset(Asset=Get SoftTileImage!(self:Fort Mtx Offer Data Object Reference=@n14.DisplayAsset) -> (SoftTileImage:Soft Object Reference))
   n18 = Cast To Texture2D(execute=@n17.Completed, Object=@n17)
   > SetBrushFromTexture!(
    execute
    self:Image Object Reference=Get Image_Icon() -> (Image_Icon:Fort Lazy Image Object Reference)
    Texture:Texture 2D Object Reference=@n18
    bMatchSize:bool=false
   ) -> (then)
   SetBrushFromURL!(
    execute=@n16.else
    self:Fort Lazy Image Object Reference=Get Image_Icon()
    CMSImageURL:String=@n15
    bMatchSize:bool=false
    Delay:float=0
    DownloadPriority:ECMSDownloadPriority Enum=Normal
   ) -> (then)
   Event PreConstruct()
   > SetTileSize(DisplaySize=Medium)
   Event On Pressed() {is=title K2Node_Event}
   > Parent: On Pressed() {is=title K2Node_CallParentFunction}
   > PlaySound2D(Sound=Get PressedSound() -> (PressedSound:Sound Base Object Reference))
   Event On Hovered()
   > Parent: On Hovered() {is=title K2Node_CallParentFunction}
   n19 = Get Hover() -> (Hover:Widget Animation Object Reference)
   > PlayAnimationForward(InAnimation=@n19, PlaybackSpeed=2)
   > PlaySound2D(Sound=Get HoverSound() -> (HoverSound:Sound Base Object Reference))
   n20 = Get Block_Outline() -> (Block_Outline:WBP UIKit Block Outline Object Reference)
   > SetIsFocused!(execute, self:WBP UIKit Block Base Object Reference=@n20, IsFocused:bool=true) -> (then)
   > SetBlockVisibility!(execute, self:WBP UIKit Block Base Object Reference=@n20, Visible:bool=true) -> (then)
   Event On Clicked()
   n21 = Get FortMtxStoreOfferVM() -> (FortMtxStoreOfferVM:Fort Mtx Store Offer VM Object Reference)
   > FireEventElementInteractedByName!(
    execute
    Element Name:String=GetDisplayAssetPath!(self:Fort Store Front Offer Info Object Reference=Get OfferInfo!(self:Fort Mtx Store Offer VM Object Reference=@n21) -> (OfferInfo:Fort Store Front Offer Info Object Reference)) -> (ReturnValue:String)
    WidgetCallingFrom:Widget Object Reference=Self-Reference()
   ) -> (then, Success:bool)
   > n22 = Switch on EFortMtxStoreOfferPurchaseAvailability!(
    execute
    Selection:EFortMtxStoreOfferPurchaseAvailability Enum=Get PurchaseAvailability!(self:Fort Mtx Store Offer VM Object Reference=Get FortMtxStoreOfferVM()) -> (PurchaseAvailability:EFortMtxStoreOfferPurchaseAvailability Enum=Available)
   ) -> (Available:Exec, Unavailable_RealMoney:Exec, Unavailable_ParentalControls:Exec) {is=title K2Node_SwitchEnum}
   SendInteractionAnalytic!(
    execute=@n22.Unavailable_RealMoney
    self:Fort Mtx Store Offer VM Object Reference=Get FortMtxStoreOfferVM()
    Interaction:String=StoreTilePurchaseUnavailable
   ) -> (then)
   > n23 = PushPrimaryContentSoftClassAsync!(
    execute:Exec
    SoftWidgetClass:Common Activatable Widget Soft Class Reference=&/PurchaseUnavailable.PurchaseUnavailable_C
   ) -> (then:Exec, OnPrimaryContentPushed:Exec, ActivatableWidget:Common Activatable Widget Object Reference) {is=title K2Node_AsyncAction}
   > n24 = Cast To FortPurchaseUnavailableModal!(execute, Object=@n23) -> (then, CastFailed, AsFort Purchase Unavailable Modal:Fort Purchase Unavailable Modal Object Reference) {is=title K2Node_DynamicCast}
   > SetupOfferUnavailable!(
    execute
    self:Fort Purchase Unavailable Modal Object Reference=@n24
    OfferInfo:Fort Store Front Offer Info Object Reference=Get OfferInfo(self=Get FortMtxStoreOfferVM())
   ) -> (then)
   SendInteractionAnalytic(execute=[@n22.Available, @n22.Unavailable_ParentalControls], self=Get FortMtxStoreOfferVM(), Interaction=StoreTile)
   > n25 = Branch(Condition=ShouldOpenPdp!(self:Fort Mtx Store VM Object Reference=Get FortMtxStoreVM() -> (FortMtxStoreVM:Fort Mtx Store VM Object Reference)) -> (ReturnValue:bool))
   > n26 = Branch(Condition=GetConfigDataBoolWithDefault!(EntryName:String=show_new_vbucks_pdp, bDefaultValue:bool=false) -> (ReturnValue:bool))
   > n27 = PushPrimaryContentSoftClassAsync(SoftWidgetClass=&/WBP_VbucksStore_PDP.WBP_VbucksStore_PDP_C)
   n28 = Cast To WBP_VbucksStore_PDP!(execute=@n27.OnPrimaryContentPushed, Object=@n27) -> (then, CastFailed, AsWBP Vbucks Store PDP:WBP Vbucks Store PDP Object Reference) {is=title K2Node_DynamicCast}
   > Set with Broadcast IsExactAmount!(execute, IsExactAmount:bool=Get IsExactAmount(), self:WBP Vbucks Store PDP Object Reference=@n28) -> (then, Output_Get:bool)
   > Set MatchesExistingOffer!(execute, MatchesExistingOffer:bool=Get MatchesExistingTile(), self:WBP Vbucks Store PDP Object Reference=@n28) -> (then, Output_Get:bool)
   > SetFortMtxStoreOfferVM!(execute, self:WBP Vbucks Store PDP Object Reference=@n28, Viewmodel:Fort Mtx Store Offer VM Object Reference=Get FortMtxStoreOfferVM()) -> (then)
   > Unbind Event from On Purchase Complete(self:WBP Vbucks Store PDP Object Reference=@n28, Delegate=@n12) {is=title K2Node_RemoveDelegate}
   > Bind Event to On Purchase Complete(self:WBP Vbucks Store PDP Object Reference=@n28, Delegate=@n12) {is=title K2Node_AddDelegate}
   n29 = PushPrimaryContentSoftClassAsync(execute=@n26.else, SoftWidgetClass=&/StoreMain_OfferDetails.StoreMain_OfferDetails_C)
   n30 = Cast To FortRealMoneyOfferDetails!(execute=@n29.OnPrimaryContentPushed, Object=@n29) -> (then, CastFailed, AsFort Real Money Offer Details:Fort Real Money Offer Details Object Reference) {is=title K2Node_DynamicCast}
   > SetupOffer!(
    execute
    self:Fort Store Front Offer Details Widget Base Object Reference=@n30
    InOfferData:Fort Store Front Offer Info Object Reference=Get OfferInfo(self=Get FortMtxStoreOfferVM())
   ) -> (then)
   > Unbind Event from On Real Money Purchase Complete(self:Fort Store Front Offer Details Widget Base Object Reference=@n30, Delegate=@n11) {is=title K2Node_RemoveDelegate}
   > Bind Event to On Real Money Purchase Complete(self:Fort Store Front Offer Details Widget Base Object Reference=@n30, Delegate=@n11) {is=title K2Node_AddDelegate}
   Fire Event Element Interacted(execute=@n25.else, Element=Self-Reference())
   > Call Set Is Purchasing(IsPurchasing=true)
   > PurchaseOffer!(execute, self:Fort Mtx Store Offer VM Object Reference=Get FortMtxStoreOfferVM(), OnPurchaseCompleteCallback:"Delegate (by ref)"=@n13) -> (then)
   Event On Unhovered()
   > Parent: On Unhovered() {is=title K2Node_CallParentFunction}
   > PlayAnimationReverse(InAnimation=@n19, PlaybackSpeed=3)
   > SetIsFocused(self=@n20, IsFocused=false)
   > SetBlockVisibility(self=@n20, Visible=false)
  ]}
 GetBonusAmountText {nodes [
   n1 = GetBonusAmountText!() -> (then, BonusAmount:Text) {is=title K2Node_FunctionEntry}
   n2 = TextIsEmpty(InText=@n1)
   n3 = Format Text!(Format="LOCTEXT(\"Includes <img id=\\\"VBucksIcon\\\"/><text fontface=\\\"76 Bold\\\"> {BonusAmount} Bonus</>\")", BonusAmount:Text=@n1) -> (Result)
   > Return Node!(
    execute
    Return:Text=Select(Option 0:Text=Select(Option 0=@n3, Index=@n2), Option 1:Text="LOCTEXT(\"Exact Amount\")", Index=Get IsExactAmount()) -> (ReturnValue:Text)
   )
  ]}
 GetHighlightText {nodes [
   GetHighlightText!() -> (then, BonusPercentage:Text) {is=title K2Node_FunctionEntry}
   > n1 = Branch(Condition=TextIsEmpty(InText=Get BonusPercentage() -> (BonusPercentage:Text)))
   > Return Node!(execute, Text:Text)
   Return Node(execute=@n1.else, Text=Format Text!(Format="LOCTEXT(\"{BonusPercent}%\")", BonusPercent:Text=Get BonusPercentage()) -> (Result))
  ]}
 GetHighlightVisibility {nodes [
   n1 = GetHighlightVisibility!() -> (then, HighlightText:Text) {is=title K2Node_FunctionEntry}
   > Return Node!(
    execute
    Visibility:ESlateVisibility Enum=Select(Option 0:ESlateVisibility Enum=SelfHitTestInvisible, Option 1:ESlateVisibility Enum=Collapsed, Index=TextIsEmpty(InText=@n1)) -> (ReturnValue:ESlateVisibility Enum)
   )
  ]}
 OnPurchaseComplete {nodes [OnPurchaseComplete!() -> (then, bSuccess:bool=~) {is=title K2Node_FunctionEntry}]}
 SetBackgroundGlowPosition {nodes [
   Inputs()
   > n1 = GetDynamicMaterial(self=Get Image_Background())
   > SetScalarParameterValue(self=@n1, ParameterName=GradientGlowPositionY, Value=0.7)
   > Outputs!(then:Exec)
  ]}
 SetIsPurchasing {nodes [SetIsPurchasing!() -> (then, IsPurchasing:bool=~) {is=title K2Node_FunctionEntry}]}
 SetTileSize {nodes [
   n1 = SetTileSize!() -> (then, DisplaySize:EFortMtxOfferDisplaySize Enum=Medium) {is=title K2Node_FunctionEntry}
   > n2 = Switch on EFortMtxOfferDisplaySize!(execute, Selection:EFortMtxOfferDisplaySize Enum=@n1) -> (Small:Exec, Medium:Exec, Large:Exec) {is=title K2Node_SwitchEnum}
   n3 = Get SizeBox() -> (SizeBox:Size Box Object Reference)
   SetHeightOverride(execute=@n2.Large, self=@n3, InHeightOverride=600)
   > SetWidthOverride(self=@n3, InWidthOverride=690)
   n4 = Get SizeBox()
   SetHeightOverride(execute=@n2.Medium, self=@n4, InHeightOverride=623)
   > SetWidthOverride(self=@n4, InWidthOverride=320)
   n5 = Get SizeBox()
   SetHeightOverride(execute=@n2.Small, self=@n5, InHeightOverride=296)
   > SetWidthOverride(self=@n5, InWidthOverride=256)
  ]}
 VM_PurchaseAvailabilityChanged {nodes [
   VM_PurchaseAvailabilityChanged!() -> (then, PurchaseAvailability:EFortMtxStoreOfferPurchaseAvailability Enum=Available) {is=title K2Node_FunctionEntry}
   n1 = Get TextOfferPurchaseUnvailable() -> (TextOfferPurchaseUnvailable:Common Text Object Reference)
   > SetActiveWidget(
    self:Widget Switcher Object Reference=Get Switcher_Price() -> (Switcher_Price:Common Widget Switcher Legacy Object Reference)
    Widget=Select!(
     Available:Widget Object Reference=Get HBox_Price() -> (HBox_Price:Horizontal Box Object Reference)
     Unavailable_RealMoney:Widget Object Reference=@n1
     Unavailable_ParentalControls:Widget Object Reference=@n1
     Index:EFortMtxStoreOfferPurchaseAvailability Enum=Get PurchaseAvailability!() -> (PurchaseAvailability=Available)
    ) -> (ReturnValue:Widget Object Reference)
   )
  ]}
]
variables [
 ExactPackBackground {type=Object, default=/FortniteCoreUI/Frontend/StoreMain/Materials/MI_UIKit_Vbucks_Background_ExactAmount}
 HighlightText {type=string, default=""}
 HoverSound {type=SoundBase, default=/Game/Sounds/Fort_UI_Sounds/FrontEnd/Store/MSS_UI_VBUCKS_Tile_Hover}
 IsExactAmount {type=boolean}
 MatchesExistingTile {type=boolean}
 MTXOfferTileIntroWhoosh {type=SoundBase, default=null}
 OnPurchaseComplete {}
 PressedSound {type=SoundBase, default=/Game/Sounds/Fort_UI_Sounds/FrontEnd/Store/MSS_UI_VBucks_Store_Panel_Press}
 SetIsPurchasing {}
 ShowBonusPercentage {type=boolean}
 UseNewHighlight {type=boolean}
]
widgets {
 info {inheritedWidgetCount=1, parentClass=&CommonButtonBase, rootWidgetClass=&SizeBox, widgetCount=26}
 widgets [
  SizeBox {
   widgetClassPath=&SizeBox
   bIsVariable=true
   widget=SizeBox
   properties {bOverride_HeightOverride=true, bOverride_WidthOverride=true, heightOverride=624, widthOverride=320}
  }
  Overlay {widgetClassPath=&Overlay, parent=SizeBox, widget=Overlay, slot {class=SizeBoxSlot}}
  Image_Background {
   widgetClassPath=&Image
   bIsVariable=true
   parent=Overlay
   widget=Image_Background
   properties {
    visibility=SelfHitTestInvisible
    brush {
     bIsDynamicallyLoaded=false
     drawAs=Image
     imageType=FullColor
     mirroring=NoMirror
     resourceName=None
     resourceObject=&MI_UIKit_Vbucks_Background_ExactAmount
     tiling=NoTile
     imageSize=<32 32>
     margin=<ltrb 0 0 0 0>
     outlineSettings {
      bUseBrushTransparency=false
      roundingType=HalfHeightRadius
      width=0
      color {colorUseRule=UseColor_Specified, specifiedColor=<rgba 0 0 0 0>}
      cornerRadii=<0 0 0 0>
     }
     tintColor {colorUseRule=UseColor_Specified, specifiedColor=<rgba 1 1 1 1>}
     uVRegion {bIsValid=false, max=<0 0>, min=<0 0>}
    }
   }
   slot {class=OverlaySlot, properties {horizontalAlignment=HAlign_Fill, verticalAlignment=VAlign_Fill}}
  }
  Grid_Amount {
   widgetClassPath=&GridPanel
   bIsVariable=true
   parent=Overlay
   widget=Grid_Amount
   properties {bOverride_Cursor=true}
   slot {class=OverlaySlot, properties {horizontalAlignment=HAlign_Center, padding=<ltrb 16 100 16 0>}}
  }
  Text_CurrencyTotalQuantity {
   widgetClassPath=&CommonTextBlock
   parent=Grid_Amount
   widget=Text_CurrencyTotalQuantity
   properties {
    bAutoCollapseWithEmptyText=true
    bIsScrollingEnabled=false
    justification=Center
    mobileFontSizeMultiplier=1.1
    text="[PH] 9999"
    font {
     bForceMonospaced=false
     bMaterialIsStencil=false
     fontMaterial=null
     fontObject=&HeadingNow
     letterSpacing=-10
     monospacedWidth=1
     size=22.5
     skewAmount=0
     typefaceFontName=76 Bold
     outlineSettings {
      bApplyOutlineToDropShadows=false
      bMiteredCorners=false
      bSeparateFillAlpha=false
      outlineMaterial=null
      outlineSize=0
      outlineColor=<rgba 1 1 1 1>
     }
    }
    shadowOffset=<0 0>
   }
   slot {class=GridSlot}
  }
  Text_VBucks {
   widgetClassPath=&CommonTextBlock
   parent=Grid_Amount
   widget=Text_VBucks
   properties {
    bAutoCollapseWithEmptyText=true
    bIsScrollingEnabled=false
    justification=Center
    mobileFontSizeMultiplier=1.1
    text=V-Bucks
    textTransformPolicy=ToUpper
    font {
     bForceMonospaced=false
     bMaterialIsStencil=false
     fontMaterial=null
     fontObject=&HeadingNow
     letterSpacing=-10
     monospacedWidth=1
     size=18
     skewAmount=0
     typefaceFontName=76 Bold
     outlineSettings {
      bApplyOutlineToDropShadows=false
      bMiteredCorners=false
      bSeparateFillAlpha=false
      outlineMaterial=null
      outlineSize=0
      outlineColor=<rgba 1 1 1 1>
     }
    }
    renderTransform {angle=0, scale=<1 1>, shear=<0 0>, translation=<0 -4>}
    shadowOffset=<0 0>
   }
   slot {class=GridSlot, properties {row=1}}
  }
  RichText_Bonus {
   widgetClassPath=&CommonRichTextBlock
   bIsVariable=true
   parent=Grid_Amount
   widget=RichText_Bonus
   properties {
    bIsScrollingEnabled=false
    bOverrideDefaultStyle=true
    inlineIconDisplayMode=IconAndText
    justification=Center
    text="[PH] Includes <img id=\"VBucksIcon\"/><text fontface=\"76 Bold\"> 1000 Bonus</>"
    wrapTextAt=288
    decoratorClasses [&/RichTextDecorator_VBucksStore.RichTextDecorator_VBucksStore_C]
    defaultTextStyleOverride {
     overflowPolicy=Clip
     transformPolicy=None
     colorAndOpacity {colorUseRule=UseColor_Specified, specifiedColor=<rgba 1 1 1 0.8>}
     font {
      bForceMonospaced=false
      bMaterialIsStencil=false
      fontMaterial=null
      fontObject=&HeadingNow
      letterSpacing=0
      monospacedWidth=1
      size=13.5
      skewAmount=0
      typefaceFontName=75 Medium
      outlineSettings {
       bApplyOutlineToDropShadows=false
       bMiteredCorners=false
       bSeparateFillAlpha=false
       outlineMaterial=null
       outlineSize=0
       outlineColor=<rgba 0 0 0 1>
      }
     }
     highlightColor {colorUseRule=UseColor_Specified, specifiedColor=<rgba 0 0 0 1>}
     highlightShape {
      bIsDynamicallyLoaded=false
      drawAs=Image
      imageType=NoImage
      mirroring=NoMirror
      resourceName=None
      resourceObject=null
      tiling=NoTile
      imageSize=<32 32>
      margin=<ltrb 0 0 0 0>
      outlineSettings {
       bUseBrushTransparency=false
       roundingType=HalfHeightRadius
       width=0
       color {colorUseRule=UseColor_Specified, specifiedColor=<rgba 0 0 0 0>}
       cornerRadii=<0 0 0 0>
      }
      tintColor {colorUseRule=UseColor_Specified, specifiedColor=<rgba 1 1 1 1>}
      uVRegion {bIsValid=false, max=<0 0>, min=<0 0>}
     }
     selectedBackgroundColor {colorUseRule=UseColor_Foreground, specifiedColor=<rgba 1 0 1 1>}
     shadowColorAndOpacity=<rgba 0 0 0 1>
     shadowOffset=<0 0>
     strikeBrush {
      bIsDynamicallyLoaded=false
      drawAs=Image
      imageType=NoImage
      mirroring=NoMirror
      resourceName=None
      resourceObject=null
      tiling=NoTile
      imageSize=<32 32>
      margin=<ltrb 0 0 0 0>
      outlineSettings {
       bUseBrushTransparency=false
       roundingType=HalfHeightRadius
       width=0
       color {colorUseRule=UseColor_Specified, specifiedColor=<rgba 0 0 0 0>}
       cornerRadii=<0 0 0 0>
      }
      tintColor {colorUseRule=UseColor_Specified, specifiedColor=<rgba 1 1 1 1>}
      uVRegion {bIsValid=false, max=<0 0>, min=<0 0>}
     }
     underlineBrush {
      bIsDynamicallyLoaded=false
      drawAs=Image
      imageType=NoImage
      mirroring=NoMirror
      resourceName=None
      resourceObject=null
      tiling=NoTile
      imageSize=<32 32>
      margin=<ltrb 0 0 0 0>
      outlineSettings {
       bUseBrushTransparency=false
       roundingType=HalfHeightRadius
       width=0
       color {colorUseRule=UseColor_Specified, specifiedColor=<rgba 0 0 0 0>}
       cornerRadii=<0 0 0 0>
      }
      tintColor {colorUseRule=UseColor_Specified, specifiedColor=<rgba 1 1 1 1>}
      uVRegion {bIsValid=false, max=<0 0>, min=<0 0>}
     }
    }
   }
   slot {class=GridSlot, properties {horizontalAlignment=HAlign_Center, row=2}}
  }
  ScaleBox_Icon {
   widgetClassPath=&ScaleBox
   bIsVariable=true
   parent=Overlay
   widget=ScaleBox_Icon
   properties {clipping=ClipToBoundsWithoutIntersecting, stretch=ScaleToFitY, renderTransform {angle=0, scale=<1 1>, shear=<0 0>, translation=<0 -24>}}
   slot {class=OverlaySlot, properties {horizontalAlignment=HAlign_Center, verticalAlignment=VAlign_Center, padding=<ltrb 0 80 0 0>}}
  }
  Image_Icon {
   widgetClassPath=&FortLazyImage
   bIsVariable=true
   parent=ScaleBox_Icon
   widget=Image_Icon
   properties {
    bOverride_Cursor=true
    visibility=Visible
    brush {
     bIsDynamicallyLoaded=false
     drawAs=Image
     imageType=FullColor
     mirroring=NoMirror
     resourceName=None
     resourceObject=/FortniteCoreUI/Frontend/StoreMain/Materials/Textures/T_UI_VBucks_Tile_Exact_Amount
     tiling=NoTile
     imageSize=<520 520>
     margin=<ltrb 0 0 0 0>
     outlineSettings {
      bUseBrushTransparency=false
      roundingType=HalfHeightRadius
      width=0
      color {colorUseRule=UseColor_Specified, specifiedColor=<rgba 0 0 0 0>}
      cornerRadii=<0 0 0 0>
     }
     tintColor {colorUseRule=UseColor_Specified, specifiedColor=<rgba 1 1 1 1>}
     uVRegion {bIsValid=false, max=<0 0>, min=<0 0>}
    }
   }
   slot {class=ScaleBoxSlot}
  }
  Grid_Details {
   widgetClassPath=&GridPanel
   bIsVariable=true
   parent=Overlay
   widget=Grid_Details
   properties {bOverride_Cursor=true}
   slot {class=OverlaySlot, properties {horizontalAlignment=HAlign_Center, verticalAlignment=VAlign_Bottom, padding=<ltrb 16 0 16 16>}}
  }
  SizeBox_Price {
   widgetClassPath=&SizeBox
   parent=Grid_Details
   widget=SizeBox_Price
   properties {bOverride_HeightOverride=true, bOverride_WidthOverride=true, heightOverride=58, minDesiredHeight=58, minDesiredWidth=200, widthOverride=288}
   slot {class=GridSlot, properties {row=1, padding=<ltrb 0 16 0 0>}}
  }
  Border_Footer {
   widgetClassPath=&Border
   parent=SizeBox_Price
   widget=Border_Footer
   properties {
    horizontalAlignment=HAlign_Center
    verticalAlignment=VAlign_Center
    background {
     bIsDynamicallyLoaded=false
     drawAs=Image
     imageType=FullColor
     mirroring=NoMirror
     resourceName=None
     resourceObject=/FortniteCoreUI/Frontend/StoreMain/Materials/MI_UIKit_Price_Background
     tiling=NoTile
     imageSize=<32 32>
     margin=<ltrb 0 0 0 0>
     outlineSettings {
      bUseBrushTransparency=false
      roundingType=HalfHeightRadius
      width=0
      color {colorUseRule=UseColor_Specified, specifiedColor=<rgba 0 0 0 0>}
      cornerRadii=<0 0 0 0>
     }
     tintColor {colorUseRule=UseColor_Specified, specifiedColor=<rgba 1 1 1 1>}
     uVRegion {bIsValid=false, max=<0 0>, min=<0 0>}
    }
    padding=<ltrb 0 0 0 0>
   }
   slot {class=SizeBoxSlot}
  }
  Switcher_Price {
   widgetClassPath=&CommonWidgetSwitcherLegacy
   bIsVariable=true
   parent=Border_Footer
   widget=Switcher_Price
   properties {bOutroPanelBelow=false, bWidgetActivationEnabled=false, transitionDuration=0}
   slot {class=BorderSlot, properties {horizontalAlignment=HAlign_Center, verticalAlignment=VAlign_Center, padding=<ltrb 0 0 0 0>}}
  }
  HBox_Price {
   widgetClassPath=&HorizontalBox
   bIsVariable=true
   parent=Switcher_Price
   widget=HBox_Price
   properties {flowDirectionPreference=Culture}
   slot {class=WidgetSwitcherSlot, properties {horizontalAlignment=HAlign_Center, verticalAlignment=VAlign_Center}}
  }
  TextOfferDisplayPrice {
   widgetClassPath=&CommonTextBlock
   parent=HBox_Price
   widget=TextOfferDisplayPrice
   properties {
    justification=Center
    mobileFontSizeMultiplier=1.1
    text="[PH] $888.88"
    colorAndOpacity {colorUseRule=UseColor_Specified, specifiedColor=<rgba 0.0075 0.0091 0.0242 1>}
    font {
     bForceMonospaced=false
     bMaterialIsStencil=false
     fontMaterial=null
     fontObject=&HeadingNow
     letterSpacing=0
     monospacedWidth=1
     size=18
     skewAmount=0
     typefaceFontName=76 Bold
     outlineSettings {
      bApplyOutlineToDropShadows=false
      bMiteredCorners=false
      bSeparateFillAlpha=false
      outlineMaterial=null
      outlineSize=0
      outlineColor=<rgba 0 0 0 1>
     }
    }
    shadowOffset=<0 0>
   }
   slot {class=HorizontalBoxSlot, properties {verticalAlignment=VAlign_Center}}
  }
  WBP_Price_PlatformStoreIcon {
   widgetClassPath=/FortUILibrary/Systems/ItemShopTile/WBP_Price_PlatformStoreIcon
   parent=HBox_Price
   widget=WBP_Price_PlatformStoreIcon
   properties {imageSize=<40 40>}
   slot {class=HorizontalBoxSlot, properties {horizontalAlignment=HAlign_Center, verticalAlignment=VAlign_Center}}
  }
  TextOfferPurchaseUnvailable {
   widgetClassPath=&CommonTextBlock
   bIsVariable=true
   parent=Switcher_Price
   widget=TextOfferPurchaseUnvailable
   properties {
    justification=Center
    text=Learn More
    textTransformPolicy=ToUpper
    colorAndOpacity {colorUseRule=UseColor_Specified, specifiedColor=<rgba 0.0075 0.0091 0.0242 1>}
    font {
     bForceMonospaced=false
     bMaterialIsStencil=false
     fontMaterial=null
     fontObject=&HeadingNow
     letterSpacing=0
     monospacedWidth=1
     size=18
     skewAmount=0
     typefaceFontName=76 Bold
     outlineSettings {
      bApplyOutlineToDropShadows=false
      bMiteredCorners=false
      bSeparateFillAlpha=false
      outlineMaterial=null
      outlineSize=0
      outlineColor=<rgba 0 0 0 1>
     }
    }
    shadowOffset=<0 0>
   }
   slot {class=WidgetSwitcherSlot, properties {verticalAlignment=VAlign_Center}}
  }
  Block_Outline {
   widgetClassPath=/FortUILibrary/UIKit/Buttons/Blocks/WBP_UIKit_Block_Outline
   bIsVariable=true
   parent=Overlay
   widget=Block_Outline
   properties {bottomLeftCornerRadius=0.08, bottomRightCornerRadius=0.08, topLeftCornerRadius=0.08, topRightCornerRadius=0.08}
   slot {class=OverlaySlot, properties {horizontalAlignment=HAlign_Fill, verticalAlignment=VAlign_Fill}}
  }
  Overlay_Highlight_Style2 {
   widgetClassPath=&Overlay
   bIsVariable=true
   parent=Overlay
   widget=Overlay_Highlight_Style2
   slot {class=OverlaySlot, properties {horizontalAlignment=HAlign_Center, padding=<ltrb 0 -8 0 0>}}
  }
  Image_Highlight_BG {
   widgetClassPath=&Image
   bIsVariable=true
   parent=Overlay_Highlight_Style2
   widget=Image_Highlight_BG
   properties {brush {
     bIsDynamicallyLoaded=false
     drawAs=Image
     imageType=FullColor
     mirroring=NoMirror
     resourceName=None
     resourceObject=/FortniteCoreUI/Frontend/StoreMain/Materials/MI_UI_Bonus_Badge_1000
     tiling=NoTile
     imageSize=<129 76>
     margin=<ltrb 0 0 0 0>
     outlineSettings {
      bUseBrushTransparency=false
      roundingType=HalfHeightRadius
      width=0
      color {colorUseRule=UseColor_Specified, specifiedColor=<rgba 0 0 0 0>}
      cornerRadii=<0 0 0 0>
     }
     tintColor {colorUseRule=UseColor_Specified, specifiedColor=<rgba 1 1 1 1>}
     uVRegion {bIsValid=false, max=<0 0>, min=<0 0>}
    }}
   slot {class=OverlaySlot, properties {horizontalAlignment=HAlign_Fill, verticalAlignment=VAlign_Center}}
  }
  Border_HighlightText {
   widgetClassPath=&CommonBorder
   bIsVariable=true
   parent=Overlay_Highlight_Style2
   widget=Border_HighlightText
   properties {padding=<ltrb 25 0 25 0>}
   slot {class=OverlaySlot, properties {horizontalAlignment=HAlign_Center, verticalAlignment=VAlign_Center, padding=<ltrb 0 0 0 6>}}
  }
  VerticalBox_0 {
   widgetClassPath=&VerticalBox
   parent=Border_HighlightText
   widget=VerticalBox_0
   properties {bOverride_Cursor=true}
   slot {class=BorderSlot, properties {padding=<ltrb 25 0 25 0>}}
  }
  HorizontalBox_0 {
   widgetClassPath=&HorizontalBox
   parent=VerticalBox_0
   widget=HorizontalBox_0
   slot {class=VerticalBoxSlot, properties {horizontalAlignment=HAlign_Center}}
  }
  Text_Highlight_Amount {
   widgetClassPath=&CommonTextBlock
   parent=HorizontalBox_0
   widget=Text_Highlight_Amount
   properties {
    bIsScrollingEnabled=false
    justification=Center
    text=1000
    font {
     bForceMonospaced=false
     bMaterialIsStencil=false
     fontMaterial=null
     fontObject=&HeadingNow
     letterSpacing=0
     monospacedWidth=1
     size=18
     skewAmount=0
     typefaceFontName=76 Bold Italic
     outlineSettings {
      bApplyOutlineToDropShadows=false
      bMiteredCorners=false
      bSeparateFillAlpha=false
      outlineMaterial=null
      outlineSize=0
      outlineColor=<rgba 0 0 0 1>
     }
    }
    shadowOffset=<0 0>
   }
   slot {class=HorizontalBoxSlot, properties {verticalAlignment=VAlign_Center, size {sizeRule=Fill, value=1}}}
  }
  Text_Highlight_Amount_Bonus {
   widgetClassPath=&CommonTextBlock
   parent=VerticalBox_0
   widget=Text_Highlight_Amount_Bonus
   properties {
    bIsScrollingEnabled=false
    justification=Center
    text=Extra*
    textTransformPolicy=ToUpper
    font {
     bForceMonospaced=false
     bMaterialIsStencil=false
     fontMaterial=null
     fontObject=&HeadingNow
     letterSpacing=0
     monospacedWidth=1
     size=12
     skewAmount=0
     typefaceFontName=76 Bold Italic
     outlineSettings {
      bApplyOutlineToDropShadows=false
      bMiteredCorners=false
      bSeparateFillAlpha=false
      outlineMaterial=null
      outlineSize=0
      outlineColor=<rgba 0 0 0 1>
     }
    }
    shadowOffset=<0 0>
   }
   slot {class=VerticalBoxSlot, properties {horizontalAlignment=HAlign_Center, verticalAlignment=VAlign_Top, padding=<ltrb 10 -6 10 0>}}
  }
  InputActionWidget {widgetClassPath=&CommonActionWidget, bInherited=true, widget=null}
 ]
}
