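# V-Bucks store widget that shows the default packs as a row of offer tiles in a dynamic entry box.
# CreateDefaultPacks resets the box, creates a tile for each offer viewmodel passed in, and returns the tile array.
# SetUpOffer assigns each tile its viewmodel and rebinds its delegates, unbinding first to avoid duplicate bindings.
# Tile focus, purchase completion and purchasing-state events are re-broadcast through the widget's own dispatchers for its owner.
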WidgetBlueprint /FortniteCoreUI/Frontend/StoreMain/WBP_VbucksStore_DefaultPacks
graphs [
 CreateDefaultPacks {nodes [
   n1 = CreateDefaultPacks!() -> (then, Offers:Array of Fort Mtx Store Offer VM Object References) {is=title K2Node_FunctionEntry}
   > Reset!(
    execute
    self:Dynamic Entry Box Object Reference=Get EntryBox_OfferTiles() -> (EntryBox_OfferTiles:Dynamic Entry Box Object Reference)
    bDeleteWidgets:bool=false
   ) -> (then)
   > Clear(TargetArray:Array of WBP Vbucks Store Offer Tile Object References=Get OffersWidgets() -> (OffersWidgets:Array of WBP Vbucks Store Offer Tile Object References))
   > n2 = For Each Loop(Array:Array of Fort Mtx Store Offer VM Object References=@n1) -> (Array Element:Fort Mtx Store Offer VM Object Reference)
   Return Node!(execute=@n2.Completed, OffersWidgets:Array of WBP Vbucks Store Offer Tile Object References=Get OffersWidgets())
   > Set Offer(Offer:Fort Mtx Store Offer VM Object Reference=@n2) -> (Output_Get:Fort Mtx Store Offer VM Object Reference)
   > n3 = Create Entry of Class!(
    execute
    self:Dynamic Entry Box Object Reference=Get EntryBox_OfferTiles()
    EntryClass:User Widget Class Reference=&/WBP_VbucksStore_OfferTile.WBP_VbucksStore_OfferTile_C
   ) -> (then, ReturnValue:WBP Vbucks Store Offer Tile Object Reference)
   > SetUpOffer!(
    execute
    self
    OfferWidget:WBP Vbucks Store Offer Tile Object Reference=@n3
    OfferVM:Fort Mtx Store Offer VM Object Reference=Get Offer() -> (Offer:Fort Mtx Store Offer VM Object Reference)
   ) -> (then)
   > Add(
    TargetArray:Array of WBP Vbucks Store Offer Tile Object References=Get OffersWidgets()
    NewItem:"WBP Vbucks Store Offer Tile Object Reference (by ref)"=@n3
   ) {is=title K2Node_CallArrayFunction}
  ]}
 EventGraph {nodes [
   n1 = OnOfferTileFocused!() -> (OutputDelegate, then, Button:Common Button Base Object Reference) {is=title K2Node_CustomEvent}
   > Call Offer Tile Focused!(execute:Exec, self:WBP Vbucks Store Default Packs Object Reference, Button:Common Button Base Object Reference=@n1.Button) -> (then:Exec) {is=title K2Node_CallDelegate}
   n2 = OnPurchaseCompleted!() -> (OutputDelegate, then, bSuccess:bool=~) {is=title K2Node_CustomEvent}
   > Call Purchase Completed!(execute:Exec, self:WBP Vbucks Store Default Packs Object Reference, bSuccess:bool=@n2.bSuccess) -> (then:Exec) {is=title K2Node_CallDelegate}
   n3 = OnSetIsPurchasing!() -> (OutputDelegate, then, IsPurchasing:bool=~) {is=title K2Node_CustomEvent}
   > Call Set Is Purchasing!(execute:Exec, self:WBP Vbucks Store Default Packs Object Reference, IsPurchasing:bool=@n3.IsPurchasing) -> (then:Exec) {is=title K2Node_CallDelegate}
  ]}
 OfferTileFocused {nodes [OfferTileFocused!() -> (then, Button:Common Button Base Object Reference) {is=title K2Node_FunctionEntry}]}
 PurchaseCompleted {nodes [PurchaseCompleted!() -> (then, bSuccess:bool=~) {is=title K2Node_FunctionEntry}]}
 SetIsPurchasing {nodes [SetIsPurchasing!() -> (then, IsPurchasing:bool=~) {is=title K2Node_FunctionEntry}]}
 SetUpOffer {nodes [
   SetUpOffer!() -> (then, OfferWidget:WBP Vbucks Store Offer Tile Object Reference, OfferVM:Fort Mtx Store Offer VM Object Reference) {is=title K2Node_FunctionEntry}
   > SetFortMtxStoreOfferVM!(
    execute
    self:WBP Vbucks Store Offer Tile Object Reference=Get OfferWidget() -> (OfferWidget:WBP Vbucks Store Offer Tile Object Reference)
    Viewmodel:Fort Mtx Store Offer VM Object Reference=Get OfferVM() -> (OfferVM:Fort Mtx Store Offer VM Object Reference)
   ) -> (then)
   n1 = Create Event()
   > Unbind Event from On Focused(self:Common Button Base Object Reference=Get OfferWidget(), Delegate=@n1) {is=title K2Node_RemoveDelegate}
   > Bind Event to On Focused(self:Common Button Base Object Reference=Get OfferWidget(), Delegate=@n1) {is=title K2Node_AddDelegate}
   n2 = Create Event()
   n3 = Get OfferWidget()
   > Unbind Event from On Purchase Complete(self:WBP Vbucks Store Offer Tile Object Reference=@n3, Delegate=@n2) {is=title K2Node_RemoveDelegate}
   > Bind Event to On Purchase Complete(self:WBP Vbucks Store Offer Tile Object Reference=@n3, Delegate=@n2) {is=title K2Node_AddDelegate}
   n4 = Create Event()
   n5 = Get OfferWidget()
   > Unbind Event from Set Is Purchasing(self:WBP Vbucks Store Offer Tile Object Reference=@n5, Delegate=@n4) {is=title K2Node_RemoveDelegate}
   > Bind Event to Set Is Purchasing(self:WBP Vbucks Store Offer Tile Object Reference=@n5, Delegate=@n4) {is=title K2Node_AddDelegate}
  ]}
]
variables [
 HasExactPack {type=boolean}
 IsExactAmountMatching {type=boolean}
 IsExactPackExperimentEnabled {type=boolean}
 OffersWidgets {type="array<WBP_VbucksStore_OfferTile_C>", default []}
 OfferTileFocused {}
 PurchaseCompleted {}
 SetIsPurchasing {}
]
widgets {
 info {inheritedWidgetCount=0, parentClass=&UserWidget, rootWidgetClass=&DynamicEntryBox, widgetCount=1}
 widgets [EntryBox_OfferTiles {
   widgetClassPath=&DynamicEntryBox
   bIsVariable=true
   widget=EntryBox_OfferTiles
   properties {
    entryWidgetClass=&WBP_VbucksStore_OfferTile_C
    numDesignerPreviewEntries=4
    entrySpacing=<24 0>
   }
  }]
}
