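# V-Bucks store content widget that lays out an optional exact-amount offer tile, a divider, the default pack tiles and legal text.
# SetupCommonOffers sets the first offer as conversion base and builds DefaultPacks from the store viewmodel.
# SetupExactPackOffer shows the exact tile only when no regular pack grants the same amount, inserting it first.
# Tile events feed base-class handlers, focus restores the last offer, and PreConstruct adjusts padding for Save the World.
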WidgetBlueprint /FortniteCoreUI/Frontend/StoreMain/WBP_VbucksStore_Content
graphs [
 BP_GetDesiredFocusTarget {nodes [
   Get Desired Focus Target()
   > Return Node!(
    execute
    ReturnValue:Widget Object Reference=GetDesiredFocusTile!(self:WBP Vbucks Store Content Base Object Reference) -> (Tile:WBP Vbucks Store Offer Tile Object Reference)
   )
  ]}
 EventGraph {nodes [
   n1 = "Offer Tile Focused (DefaultPacks)"!() -> (OutputDelegate, then, Button:Common Button Base Object Reference) {is=title K2Node_ComponentBoundEvent}
   > OnOfferTileFocused!(execute, self:WBP Vbucks Store Content Base Object Reference, Button:Common Button Base Object Reference=@n1.Button) -> (then)
   n2 = "Purchase Completed (DefaultPacks)"!() -> (OutputDelegate, then, bSuccess:bool) {is=title K2Node_ComponentBoundEvent}
   > OnPurchaseCompleted!(execute, self:WBP Vbucks Store Content Base Object Reference, bSuccess:bool=@n2.bSuccess) -> (then)
   n3 = "Set Is Purchasing (DefaultPacks)"!() -> (OutputDelegate, then, IsPurchasing:bool) {is=title K2Node_ComponentBoundEvent}
   > OnSetIsPurchasing!(execute, self:WBP Vbucks Store Content Base Object Reference, IsPurchasing:bool=@n3.IsPurchasing) -> (then)
   Event PreConstruct()
   > SetPadding!(
    execute
    self:Overlay Slot Object Reference=SlotAsOverlaySlot(Widget=Get GridPanel_Offers() -> (GridPanel_Offers:Grid Panel Object Reference))
    InPadding:Margin Structure=Select(
     Option 0:Margin Structure=Get DefaultPadding() -> (DefaultPadding:Margin Structure)
     Option 1:Margin Structure=Get STWPading() -> (STWPading:Margin Structure)
     Index=IsInSubGame!(SubGame:ESubGame Enum=Campaign) -> (ReturnValue:bool)
    ) -> (ReturnValue:Margin Structure)
   ) -> (then)
  ]}
 GetDesiredFocusTile {nodes [
   GetDesiredFocusTile() {is=title K2Node_FunctionEntry}
   > n1 = FindOfferTileById!(execute, self:WBP Vbucks Store Content Base Object Reference, OfferId:String=Get LastFocusedOfferId() -> (LastFocusedOfferId:String)) -> (then, OfferTile:WBP Vbucks Store Offer Tile Object Reference)
   > Return Node!(execute, Tile:WBP Vbucks Store Offer Tile Object Reference=@n1)
  ]}
 InitializeFocusedOffer {nodes [
   InitializeFocusedOffer() {is=title K2Node_FunctionEntry}
   > Parent: InitializeFocusedOffer() {is=title K2Node_CallParentFunction}
   > n1 = FindOfferTileById(OfferId=Get LastFocusedOfferId())
   > n2 = Is Valid(InputObject=@n1)
   n3 = Get OffersWidgets() -> (OffersWidgets:Array of WBP Vbucks Store Offer Tile Object References)
   n4 = Branch(execute=@n2.Is Not Valid, Condition=Is Empty!(TargetArray:Array of WBP Vbucks Store Offer Tile Object References=@n3) -> (ReturnValue:bool))
   > Set LastFocusedOfferId(LastFocusedOfferId:String) -> (Output_Get:String)
   n5 = "Get (a copy)"(Array:Array of WBP Vbucks Store Offer Tile Object References=@n3) -> (Output:WBP Vbucks Store Offer Tile Object Reference)
   Set LastFocusedOfferId(
    execute=@n4.else
    LastFocusedOfferId=Get OfferId!(self:Fort Mtx Store Offer VM Object Reference=Get FortMtxStoreOfferVM!(self:WBP Vbucks Store Offer Tile Object Reference=@n5) -> (FortMtxStoreOfferVM:Fort Mtx Store Offer VM Object Reference)) -> (OfferId:String)
   )
  ]}
 SetupCommonOffers {nodes [
   SetupCommonOffers() {is=title K2Node_FunctionEntry}
   > Parent: SetupCommonOffers() {is=title K2Node_CallParentFunction}
   > Set IsExactAmountMatching(IsExactAmountMatching:bool=false) -> (Output_Get:bool)
   > n1 = For Each Loop(Array:Array of Fort Mtx Store Offer VM Object References=Get Offers!(self:Fort Mtx Store VM Object Reference=Get FortMtxStoreVM() -> (FortMtxStoreVM:Fort Mtx Store VM Object Reference)) -> (Offers:Array of Fort Mtx Store Offer VM Object References)) -> (Array Element:Fort Mtx Store Offer VM Object Reference)
   n2 = Get FortMtxStoreVM()
   SetBaseOfferForMtxConversion!(
    execute=@n1.Completed
    self:Fort Mtx Store VM Object Reference=Get FortMtxStoreVM()
    Offer:Fort Mtx Store Offer VM Object Reference="Get (a copy)"(Array:Array of Fort Mtx Store Offer VM Object References=Get Offers(self=@n2)) -> (Output:Fort Mtx Store Offer VM Object Reference)
   ) -> (then)
   n3 = Get FortMtxStoreVM()
   > n4 = CreateDefaultPacks!(
    execute
    self:WBP Vbucks Store Default Packs Object Reference=Get DefaultPacks() -> (DefaultPacks:WBP Vbucks Store Default Packs Object Reference)
    Offers:Array of Fort Mtx Store Offer VM Object References=Select(
     Option 0:Array of Fort Mtx Store Offer VM Object References=Get PacksToCreate() -> (PacksToCreate:Array of Fort Mtx Store Offer VM Object References)
     Option 1:Array of Fort Mtx Store Offer VM Object References=Get Offers(self=@n3)
     Index=Get IsExactAmountMatching() -> (IsExactAmountMatching:bool)
    ) -> (ReturnValue:Array of Fort Mtx Store Offer VM Object References)
   ) -> (then, OffersWidgets:Array of WBP Vbucks Store Offer Tile Object References)
   > Set OffersWidgets(OffersWidgets:Array of WBP Vbucks Store Offer Tile Object References=@n4) -> (Output_Get:Array of WBP Vbucks Store Offer Tile Object References)
   Set Offer(execute=@n1.LoopBody, Offer:Fort Mtx Store Offer VM Object Reference=@n1) -> (Output_Get:Fort Mtx Store Offer VM Object Reference)
   > n5 = Branch(Condition=Get HasExactPack() -> (HasExactPack:bool))
   n6 = Get Offer() -> (Offer:Fort Mtx Store Offer VM Object Reference)
   n7 = Get ExactPackOfferVM() -> (ExactPackOfferVM:Fort Mtx Store Offer VM Object Reference)
   > n8 = Branch(Condition="Equal (Integer)"(A=Get GrantedMtx!(self:Fort Mtx Store Offer VM Object Reference=@n7) -> (GrantedMtx:int=0), B=Get GrantedMtx(self=@n6) -> (GrantedMtx=0)))
   > Set IsExactAmountMatching(IsExactAmountMatching=true)
   Add(
    execute=[@n8.else, @n5.else]
    TargetArray:Array of Fort Mtx Store Offer VM Object References=Get PacksToCreate()
    NewItem:"Fort Mtx Store Offer VM Object Reference (by ref)"=Get Offer()
   ) {is=title K2Node_CallArrayFunction}
  ]}
 SetupExactPackOffer {nodes [
   SetupExactPackOffer() {is=title K2Node_FunctionEntry}
   > Parent: SetupExactPackOffer() {is=title K2Node_CallParentFunction}
   > n1 = Branch(Condition=Show Exact Pack Tile!() -> (ReturnValue:bool=~) {is=title K2Node_MacroInstance})
   > SetVisibility(
    self=[
     Get Divider() -> (Divider:WBP UIKit Divider Base Object Reference)
     Get Offer_ExactPack() -> (Offer_ExactPack:WBP Vbucks Store Offer Tile Object Reference)
    ]
    InVisibility=SelfHitTestInvisible
   )
   > Set MatchesExistingTile!(execute, MatchesExistingTile:bool=false, self:WBP Vbucks Store Offer Tile Object Reference=Get Offer_ExactPack()) -> (then, Output_Get:bool)
   > SetFortMtxStoreOfferVM!(
    execute
    self:WBP Vbucks Store Offer Tile Object Reference=Get Offer_ExactPack()
    Viewmodel:Fort Mtx Store Offer VM Object Reference=Get ExactPackOfferVM()
   ) -> (then)
   n2 = Create Event()
   n3 = Get Offer_ExactPack()
   > Unbind Event from On Focused(self:Common Button Base Object Reference=@n3, Delegate=@n2) {is=title K2Node_RemoveDelegate}
   > Bind Event to On Focused(self:Common Button Base Object Reference=@n3, Delegate=@n2) {is=title K2Node_AddDelegate}
   n4 = Create Event()
   n5 = Get Offer_ExactPack()
   > Unbind Event from On Purchase Complete(self:WBP Vbucks Store Offer Tile Object Reference=@n5, Delegate=@n4) {is=title K2Node_RemoveDelegate}
   > Bind Event to On Purchase Complete(self:WBP Vbucks Store Offer Tile Object Reference=@n5, Delegate=@n4) {is=title K2Node_AddDelegate}
   n6 = Create Event()
   n7 = Get Offer_ExactPack()
   > Unbind Event from Set Is Purchasing(self:WBP Vbucks Store Offer Tile Object Reference=@n7, Delegate=@n6) {is=title K2Node_RemoveDelegate}
   > Bind Event to Set Is Purchasing(self:WBP Vbucks Store Offer Tile Object Reference=@n7, Delegate=@n6) {is=title K2Node_AddDelegate}
   > Insert!(
    execute
    TargetArray:Array of WBP Vbucks Store Offer Tile Object References=Get OffersWidgets()
    NewItem:"WBP Vbucks Store Offer Tile Object Reference (by ref)"=Get Offer_ExactPack()
    Index:int=0
   ) -> (then) {is=title K2Node_CallArrayFunction}
   SetVisibility(execute=@n1.else, self=[Get Offer_ExactPack(), Get Divider()], InVisibility=Collapsed)
  ]}
 ShowExactPackTile {nodes [
   Inputs!()
   n1 = Get IsExactAmountMatching()
   Outputs!(ReturnValue:bool=AND Boolean(A=Get HasExactPack(), B=NOT Boolean(A=@n1)))
  ]}
]
variables [
 DefaultPadding {type=Margin, default=<ltrb 0 32 0 0>}
 IsExactAmountMatching {type=boolean}
 STWPading {type=Margin, default=<ltrb 0 -32 0 0>}
]
widgets {
 info {
  inheritedWidgetCount=0
  parentClass=&/WBP_VbucksStore_Content_Base.WBP_VbucksStore_Content_Base_C
  rootWidgetClass=&Overlay
  widgetCount=8
 }
 widgets [
  Overlay_0 {widgetClassPath=&Overlay, widget=Overlay_0, properties {bOverride_Cursor=true}}
  GridPanel_Offers {
   widgetClassPath=&GridPanel
   bIsVariable=true
   parent=Overlay_0
   widget=GridPanel_Offers
   properties {bOverride_Cursor=true, visibility=Visible}
   slot {class=OverlaySlot, properties {horizontalAlignment=HAlign_Center, verticalAlignment=VAlign_Center, padding=<ltrb 0 32 0 0>}}
  }
  DefaultPacks {
   widgetClassPath=&/WBP_VbucksStore_DefaultPacks
   bIsVariable=true
   parent=GridPanel_Offers
   widget=DefaultPacks
   slot {class=GridSlot, properties {column=2}}
  }
  Divider {
   widgetClassPath=/FortUILibrary/Systems/Tabs/WBP_UIKit_Divider_Base
   bIsVariable=true
   parent=GridPanel_Offers
   widget=Divider
   properties {div Width=4, div Color and Opacity=<rgba 1 1 1 0.1>}
   slot {class=GridSlot, properties {column=1, padding=<ltrb 40 250 40 250>}}
  }
  Offer_ExactPack {
   widgetClassPath=&/WBP_VbucksStore_OfferTile
   bIsVariable=true
   parent=GridPanel_Offers
   widget=Offer_ExactPack
   properties {
    flowDirectionPreference=Inherit
    holdData=&CommonUI_HoldData_Base_C
    isExactAmount=true
    visibility=Visible
   }
   slot {class=GridSlot}
  }
  VerticalBox_0 {
   widgetClassPath=&VerticalBox
   parent=Overlay_0
   widget=VerticalBox_0
   slot {class=OverlaySlot, properties {verticalAlignment=VAlign_Bottom, padding=<ltrb 80 0 0 40>}}
  }
  Text_Legal {
   widgetClassPath=&CommonRichTextBlock
   bIsVariable=true
   parent=VerticalBox_0
   widget=Text_Legal
   properties {
    bAutoCollapseWithEmptyText=true
    bOverrideDefaultStyle=true
    lineHeightPercentage=1.15
    mobileTextBlockScale=1.5
    text="Extra V-Bucks percentage shown as compared to price of 1,000 V-Bucks pack."
    textStyleSet=/Game/Athena/Items/Cosmetics/Metadata/ItemDescriptionTextStyle
    wrapTextAt=800
    defaultTextStyleOverride {
     overflowPolicy=Clip
     transformPolicy=None
     colorAndOpacity {colorUseRule=UseColor_Specified, specifiedColor=<rgba 1 1 1 0.7>}
     font {
      bForceMonospaced=false
      bMaterialIsStencil=false
      fontMaterial=null
      fontObject=&HeadingNow
      letterSpacing=0
      monospacedWidth=1
      size=11.25
      skewAmount=0
      typefaceFontName=74 Regular
      outlineSettings {
       bApplyOutlineToDropShadows=true
       bMiteredCorners=false
       bSeparateFillAlpha=false
       outlineMaterial=null
       outlineSize=0
       outlineColor=<rgba 0 0 0 1>
      }
     }
     highlightColor {colorUseRule=UseColor_Specified, specifiedColor=<rgba 0 0 0 0>}
     highlightShape {
      bIsDynamicallyLoaded=false
      drawAs=Image
      imageType=NoImage
      mirroring=NoMirror
      resourceName=None
      resourceObject=null
      tiling=NoTile
      imageSize=<32 32>
      margin=<ltrb 0 0 0 0>
      outlineSettings {
       bUseBrushTransparency=false
       roundingType=HalfHeightRadius
       width=0
       color {colorUseRule=UseColor_Specified, specifiedColor=<rgba 0 0 0 0>}
       cornerRadii=<0 0 0 0>
      }
      tintColor {colorUseRule=UseColor_Specified, specifiedColor=<rgba 1 1 1 1>}
      uVRegion {bIsValid=false, max=<0 0>, min=<0 0>}
     }
     selectedBackgroundColor {colorUseRule=UseColor_Foreground, specifiedColor=<rgba 1 0 1 1>}
     shadowColorAndOpacity=<rgba 0 0 0 1>
     shadowOffset=<0 0>
     strikeBrush {
      bIsDynamicallyLoaded=false
      drawAs=Image
      imageType=NoImage
      mirroring=NoMirror
      resourceName=None
      resourceObject=null
      tiling=NoTile
      imageSize=<32 32>
      margin=<ltrb 0 0 0 0>
      outlineSettings {
       bUseBrushTransparency=false
       roundingType=HalfHeightRadius
       width=0
       color {colorUseRule=UseColor_Specified, specifiedColor=<rgba 0 0 0 0>}
       cornerRadii=<0 0 0 0>
      }
      tintColor {colorUseRule=UseColor_Specified, specifiedColor=<rgba 1 1 1 1>}
      uVRegion {bIsValid=false, max=<0 0>, min=<0 0>}
     }
     underlineBrush {
      bIsDynamicallyLoaded=false
      drawAs=Image
      imageType=NoImage
      mirroring=NoMirror
      resourceName=None
      resourceObject=null
      tiling=NoTile
      imageSize=<32 32>
      margin=<ltrb 0 0 0 0>
      outlineSettings {
       bUseBrushTransparency=false
       roundingType=HalfHeightRadius
       width=0
       color {colorUseRule=UseColor_Specified, specifiedColor=<rgba 0 0 0 0>}
       cornerRadii=<0 0 0 0>
      }
      tintColor {colorUseRule=UseColor_Specified, specifiedColor=<rgba 1 1 1 1>}
      uVRegion {bIsValid=false, max=<0 0>, min=<0 0>}
     }
    }
   }
   slot {class=VerticalBoxSlot}
  }
  Text_Legal_Platforms {
   widgetClassPath=&CommonTextBlock
   bIsVariable=true
   parent=VerticalBox_0
   widget=Text_Legal_Platforms
   properties {
    autoWrapText=true
    bIsScrollingEnabled=false
    mobileFontSizeMultiplier=1.4
    text=All V-Bucks purchased on one platform may not be redeemable on other platforms.
    colorAndOpacity {colorUseRule=UseColor_Specified, specifiedColor=<rgba 1 1 1 0.65>}
    font {
     bForceMonospaced=false
     bMaterialIsStencil=false
     fontMaterial=null
     fontObject=&HeadingNow
     letterSpacing=0
     monospacedWidth=1
     size=11.25
     skewAmount=0
     typefaceFontName=74 Regular
     outlineSettings {
      bApplyOutlineToDropShadows=false
      bMiteredCorners=false
      bSeparateFillAlpha=false
      outlineMaterial=null
      outlineSize=0
      outlineColor=<rgba 0 0 0 1>
     }
    }
    shadowOffset=<0 0>
   }
   slot {class=VerticalBoxSlot}
  }
 ]
}
