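# V-Bucks store screen that shows purchasable pack offers, an optional Epic Rewards banner, and loading, empty or purchasing-disabled states.
# PreConstruct creates the content widget in a named slot and binds its purchase events.
# While active it monitors the MTX store view model, switches views by loading state, and sets up common and exact-amount offers.
# Standalone mode shows a back button and, outside Campaign, player info, and closes on successful purchase.
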WidgetBlueprint /FortniteCoreUI/Frontend/StoreMain/WBP_VbucksStore_Screen
graphs [
 BP_GetDesiredFocusTarget {nodes [
   Get Desired Focus Target()
   > Return Node!(
    execute
    ReturnValue:Widget Object Reference=GetDesiredFocusTile!(self:WBP Vbucks Store Content Base Object Reference=Get CurrentContentWidget() -> (CurrentContentWidget:WBP Vbucks Store Content Base Object Reference)) -> (Tile:WBP Vbucks Store Offer Tile Object Reference)
   )
  ]}
 BP_GetDesiredInputConfig {nodes [
   Get Desired Input Config() {is=title K2Node_FunctionEntry}
   > Return Node(ReturnValue:UIInput Config Structure=Make UIInput Config!(
     bIgnoreMoveInput:bool=False
     bIgnoreLookInput:bool=False
     InputMode:ECommonInputMode Enum=All
     MouseCaptureMode:EMouseCaptureMode Enum=CaptureDuringMouseDown
     MouseLockMode:EMouseLockMode Enum=DoNotLock
     bHideCursorDuringViewportCapture:bool=True
    ) -> (UIInputConfig:UIInput Config Structure) {is=title K2Node_MakeStruct})
  ]}
 EventGraph {nodes [
   "On Clicked (DecoyButton_Back)"!() -> (OutputDelegate, then, Button:Common Button Base Object Reference) {is=title K2Node_ComponentBoundEvent}
   > CloseStore()
   CloseStore() {is=title K2Node_CustomEvent}
   > Set FortBaseOfferVM(FortBaseOfferVM:Fort Base Offer VM Object Reference) -> (Output_Get:Fort Base Offer VM Object Reference)
   > DeactivateWidget()
   > Call On Store Closed!(execute:Exec, self:WBP Vbucks Store Screen Object Reference) -> (then:Exec) {is=title K2Node_CallDelegate}
   n1 = OnSetIsPurchasing!() -> (OutputDelegate, then, IsPurchasing:bool=~) {is=title K2Node_CustomEvent}
   > n2 = Set Is Purchasing(Is Purchasing:bool=@n1.IsPurchasing) -> (Output_Get:bool)
   > n3 = Branch(Condition=@n2)
   > SetActiveWidget(
    self=Get Switcher_MainContent() -> (Switcher_MainContent:Common Visibility Switcher Object Reference)
    Widget=Get Loading() -> (Loading:Image Object Reference)
   )
   n4 = Get Loading()
   SetActiveWidget(
    execute=@n3.else
    self=Get Switcher_MainContent()
    Widget=Select!(
     Uninitialized:Widget Object Reference=@n4
     Loading:Widget Object Reference=@n4
     Ready:Widget Object Reference=Get Overlay_Offers() -> (Overlay_Offers:Overlay Object Reference)
     Empty:Widget Object Reference=Get Text_NoOffers() -> (Text_NoOffers:Common Text Object Reference)
     PurchasingDisabled:Widget Object Reference=Get Text_PurchasingDisabled() -> (Text_PurchasingDisabled:Common Text Object Reference)
     Index:EFortMtxStoreLoadingState Enum=Get CachedLoadingState() -> (CachedLoadingState:EFortMtxStoreLoadingState Enum=Uninitialized)
    ) -> (ReturnValue:Widget Object Reference)
   )
   Event On Deactivated()
   > StopMonitoringForUpdates!(
    execute
    self:Fort Mcp Catalog VM Object Reference=Get FortMtxStoreVM() -> (FortMtxStoreVM:Fort Mtx Store VM Object Reference)
    RequestingObject:Object Reference
   ) -> (then)
   > DeactivateWidget(self=Get CurrentContentWidget())
   Event PreConstruct()
   > SetPadding!(
    execute
    self:Overlay Slot Object Reference=SlotAsOverlaySlot(Widget=Get Overlay_EpicRewards() -> (Overlay_EpicRewards:Overlay Object Reference))
    InPadding:Margin Structure=Select!(
     Option 0:Margin Structure=Get DefaultPadding() -> (DefaultPadding:Margin Structure)
     Option 1:Margin Structure=Get CampaingPadding() -> (CampaingPadding:Margin Structure)
     Index:bool=IsInSubGame!(SubGame:ESubGame Enum=Campaign) -> (ReturnValue:bool)
    ) -> (ReturnValue:Margin Structure)
   ) -> (then)
   > ClearChildren(self=Get NamedSlot_Content() -> (NamedSlot_Content:Named Slot Object Reference))
   > n5 = Create Widget!(
    execute:Exec
    Class:User Widget Class Reference=&/WBP_VbucksStore_Content.WBP_VbucksStore_Content_C
    OwningPlayer:Player Controller Object Reference=GetOwningPlayer()
    FortItemShopOfferVM:Fort Item Shop Offer VM Object Reference
   ) -> (then:Exec, ReturnValue:WBP Vbucks Store Content Object Reference)
   > n6 = Cast To WBP_VbucksStore_Content_Base!(execute, Object=@n5) -> (then, CastFailed, AsWBP Vbucks Store Content Base:WBP Vbucks Store Content Base Object Reference) {is=title K2Node_DynamicCast}
   > n7 = Set CurrentContentWidget(CurrentContentWidget:WBP Vbucks Store Content Base Object Reference=@n6) -> (Output_Get:WBP Vbucks Store Content Base Object Reference)
   > AddChild(self=Get NamedSlot_Content(), Content=@n7)
   n8 = Create Event()
   n9 = Get CurrentContentWidget()
   > Unbind Event from Purchase Completed(self:WBP Vbucks Store Content Base Object Reference=@n9, Delegate=@n8) {is=title K2Node_RemoveDelegate}
   > Bind Event to Purchase Completed(self:WBP Vbucks Store Content Base Object Reference=@n9, Delegate=@n8) {is=title K2Node_AddDelegate}
   n10 = Create Event()
   n11 = Get CurrentContentWidget()
   > Unbind Event from Set Is Purchasing(self:WBP Vbucks Store Content Base Object Reference=@n11, Delegate=@n10) {is=title K2Node_RemoveDelegate}
   > Bind Event to Set Is Purchasing(self:WBP Vbucks Store Content Base Object Reference=@n11, Delegate=@n10) {is=title K2Node_AddDelegate}
   Event On Activated()
   > Fire Event Main View Opened!(execute, View:Widget Object Reference=Self-Reference(), AdditionalInfo:Flow Analytics Additional Info Structure, bIsOverlay:bool=false) -> (then)
   > UpdateStandaloneLayout()
   > PlaySound2D(Sound=/Game/Sounds/Fort_UI_Sounds/FrontEnd/Store/MSS_UI_VBucks_Store_Screen_In)
   > StartMonitoringForUpdates!(execute, self:Fort Mcp Catalog VM Object Reference=Get FortMtxStoreVM(), RequestingObject:Object Reference) -> (then)
   > LoadingStateChanged!(
    execute
    self
    LoadingState:EFortMtxStoreLoadingState Enum=Get LoadingState!(self:Fort Mtx Store VM Object Reference=Get FortMtxStoreVM()) -> (LoadingState:EFortMtxStoreLoadingState Enum=Uninitialized)
   ) -> (then)
   > ActivateWidget(self=Get CurrentContentWidget())
  ]}
 GetPrimaryContentSetup {nodes [
   GetPrimaryContentSetup() {is=title K2Node_FunctionEntry}
   > Return Node!(
    execute
    ReturnValue_bShowBottomBar:bool=True
    ReturnValue_bShowChatWidget:bool=True
    ReturnValue_bEnableNewTextChat:bool=False
    ReturnValue_bShowTopBar:bool=True
    ReturnValue_bTabSupportsInterface:bool=False
   )
  ]}
 LoadingStateChanged {nodes [
   n1 = LoadingStateChanged!() -> (then, LoadingState:EFortMtxStoreLoadingState Enum=Uninitialized) {is=title K2Node_FunctionEntry}
   > Set CachedLoadingState(CachedLoadingState:EFortMtxStoreLoadingState Enum=@n1) -> (Output_Get:EFortMtxStoreLoadingState Enum=Uninitialized)
   > Branch(Condition=IsActivated!(self:Common Activatable Widget Object Reference) -> (ReturnValue:bool))
   > n2 = Branch(Condition=Get Is Purchasing() -> (Is Purchasing:bool))
   n3 = Switch on EFortMtxStoreLoadingState!(execute=[@n5.then, @n2.then], Selection:EFortMtxStoreLoadingState Enum=Get LoadingState!() -> (LoadingState=Uninitialized)) -> (Uninitialized:Exec, Loading:Exec, Ready:Exec, Empty:Exec, PurchasingDisabled:Exec) {is=title K2Node_SwitchEnum}
   DisplayEmptyStoreDialog(
    execute=@n3.Empty
    self:Fort Mcp Context Object Reference=Get FortMcpContext!() -> (ReturnValue:Fort Mcp Context Object Reference) {is=title K2Node_GetBlueprintContextEx}
   )
   UpdateExactPackOfferVM(execute=@n3.Ready)
   > SetupCommonOffers(self:WBP Vbucks Store Content Base Object Reference=Get CurrentContentWidget())
   > SetupExactPackOffer(self:WBP Vbucks Store Content Base Object Reference=Get CurrentContentWidget())
   > InitializeFocusedOffer(self:WBP Vbucks Store Content Base Object Reference=Get CurrentContentWidget())
   > RequestRefreshFocus(self:Common Activatable Widget Object Reference=Get CurrentContentWidget())
   > TriggerTilesStaggeredIntroAnimation(self:WBP Vbucks Store Content Base Object Reference=Get CurrentContentWidget())
   n4 = Get Loading()
   n5 = SetActiveWidget(
    execute=@n2.else
    self=Get Switcher_MainContent()
    Widget=Select!(
     Uninitialized:Widget Object Reference=@n4
     Loading:Widget Object Reference=@n4
     Ready:Widget Object Reference=Get Overlay_Offers()
     Empty:Widget Object Reference=Get Text_NoOffers()
     PurchasingDisabled:Widget Object Reference=Get Text_PurchasingDisabled()
     Index:EFortMtxStoreLoadingState Enum=Get LoadingState() -> (LoadingState=Uninitialized)
    ) -> (ReturnValue:Widget Object Reference)
   )
  ]}
 OnPurchaseCompleted {nodes [
   n1 = OnPurchaseCompleted!() -> (then, bSuccess:bool=~) {is=title K2Node_FunctionEntry}
   > Branch(Condition=AND Boolean(A=@n1, B=Get bIsStandalone() -> (bIsStandalone:bool)))
   > CloseStore() {is=title K2Node_CallFunction}
   > Return Node!(execute)
  ]}
 OnSetIsStandalone {nodes [
   n1 = OnSetIsStandalone!() -> (then, bIsStandalone:bool) {is=title K2Node_FunctionEntry}
   > Set bIsStandalone(bIsStandalone:bool=@n1) -> (Output_Get:bool)
   > UpdateStandaloneLayout()
  ]}
 OnStoreClosed {nodes [OnStoreClosed() {is=title K2Node_FunctionEntry}]}
 OnStoreSelected {nodes [
   OnStoreSelected!() -> (then, SelectedStore:EAppStore Enum=DebugStore) {is=title K2Node_FunctionEntry}
   > RefreshCatalogCache(self:Fort Mcp Catalog VM Object Reference=Get FortMtxStoreVM())
  ]}
 SetEpicRewardsText {nodes [
   n1 = SetEpicRewardsText!() -> (then, Percentage:int) {is=title K2Node_FunctionEntry}
   > n2 = Branch(Condition="integer > integer"(A=@n1))
   > Set Visibility(Visibility=SelfHitTestInvisible, self=Get Overlay_EpicRewards())
   > SetText(
    self=Get RichText_EpicRewards() -> (RichText_EpicRewards:Common Rich Text Block Object Reference)
    InText=Format Text!(
     Format="LOCTEXT(\"Earn <text color=\\\"38ff84\\\">{Quantity}% back</> in Epic Rewards on purchases of V-Bucks packs\")"
     Quantity:int=Get Percentage() -> (Percentage:int=0)
    ) -> (Result)
   )
   Set Visibility(execute=@n2.else, Visibility=Collapsed, self=Get Overlay_EpicRewards())
  ]}
 UpdateExactPackOfferVM {nodes [
   UpdateExactPackOfferVM() {is=title K2Node_FunctionEntry}
   > n1 = Get!(execute:Exec) -> (then:Exec, else:Exec, FortBaseOfferVM:Fort Base Offer VM Object Reference)
   > n2 = Set OfferPrice(OfferPrice:int="To Integer (Integer64)"!(InInt:Integer64=GetPrice!(self:Fort Base Offer VM Object Reference=@n1) -> (ReturnValue:Integer64=0)) -> (ReturnValue:int=0)) -> (Output_Get:int=0)
   n3 = Set ExactAmountOffer(
    execute=[@n2.then, @n1.else]
    ExactAmountOffer:Fort Mtx Store Offer VM Object Reference=GetExactAmountOffer!(self:Fort Mtx Store VM Object Reference=Get FortMtxStoreVM(), RequiredMtx:int=Get OfferPrice() -> (OfferPrice:int=0)) -> (ReturnValue:Fort Mtx Store Offer VM Object Reference)
   ) -> (Output_Get:Fort Mtx Store Offer VM Object Reference)
   > SetExactPackOfferVM!(
    execute
    self:WBP Vbucks Store Content Base Object Reference=Get CurrentContentWidget()
    ExactAmountOffer:Fort Mtx Store Offer VM Object Reference=@n3
   ) -> (then)
  ]}
 UpdateStandaloneLayout {nodes [
   UpdateStandaloneLayout() {is=title K2Node_FunctionEntry}
   n1 = NOT Boolean(A=IsInSubGame(SubGame=Campaign))
   n2 = Get bIsStandalone()
   > SetVisibility(
    self=Get WBP_PlayerInfo() -> (WBP_PlayerInfo:WBP Player Info Object Reference)
    InVisibility=Select!(Option 0:ESlateVisibility Enum=Collapsed, Option 1:ESlateVisibility Enum=HitTestInvisible, Index:bool=AND Boolean(A=@n2, B=@n1)) -> (ReturnValue:ESlateVisibility Enum)
   )
   > SetBottomButtonVisibility!(
    execute
    self:WBP Bottom Bar Decoy Button Object Reference=Get DecoyButton_Back() -> (DecoyButton_Back:WBP Bottom Bar Decoy Button Object Reference)
    ShouldButtonBeVisible:bool=Get bIsStandalone()
   ) -> (then)
  ]}
]
variables [
 bIsStandalone {type=boolean}
 CachedLoadingState {type=EFortMtxStoreLoadingState, default=Uninitialized}
 CampaingPadding {type=Margin, default=<ltrb 0 336 0 0>}
 CurrentContentWidget {type=WBP_VbucksStore_Content_Base_C, default=null}
 DefaultPadding {type=Margin, default=<ltrb 0 400 0 0>}
 ExactAmountOffer {type=FortMtxStoreOfferVM, default=null}
 Is Purchasing {type=boolean}
 LastFocusedOfferId {type=string, default=None}
 OnStoreClosed {}
 OverrideOfferMtxPrice {type=integer}
]
widgets {
 info {inheritedWidgetCount=0, parentClass=/Script/FortniteUI.FortVbucksStoreScreen, rootWidgetClass=&Overlay, widgetCount=17}
 widgets [
  Overlay_0 {widgetClassPath=&Overlay, widget=Overlay_0}
  Background_HitBlocker {
   widgetClassPath=&Image
   parent=Overlay_0
   widget=Background_HitBlocker
   properties {brush {
     bIsDynamicallyLoaded=false
     drawAs=Image
     imageType=FullColor
     mirroring=NoMirror
     resourceName=None
     resourceObject=/FortniteCoreUI/Frontend/StoreMain/Materials/MI_UI_Screen_Vbucks_Background
     tiling=NoTile
     imageSize=<32 32>
     margin=<ltrb 0 0 0 0>
     outlineSettings {
      bUseBrushTransparency=false
      roundingType=HalfHeightRadius
      width=0
      color {colorUseRule=UseColor_Specified, specifiedColor=<rgba 0 0 0 0>}
      cornerRadii=<0 0 0 0>
     }
     tintColor {colorUseRule=UseColor_Specified, specifiedColor=<rgba 1 1 1 1>}
     uVRegion {bIsValid=false, max=<0 0>, min=<0 0>}
    }}
   slot {class=OverlaySlot, properties {horizontalAlignment=HAlign_Fill, verticalAlignment=VAlign_Fill}}
  }
  SafeZone_Content {
   widgetClassPath=&SafeZone
   parent=Overlay_0
   widget=SafeZone_Content
   slot {class=OverlaySlot, properties {horizontalAlignment=HAlign_Fill, verticalAlignment=VAlign_Fill}}
  }
  Overlay_Content {widgetClassPath=&Overlay, parent=SafeZone_Content, widget=Overlay_Content, slot {class=SafeZoneSlot}}
  WBP_CaptureForPostBufferUpdate {
   widgetClassPath=&WBP_CaptureForPostBufferUpdate
   bIsVariable=true
   parent=Overlay_Content
   slot {class=OverlaySlot, properties {horizontalAlignment=HAlign_Fill, verticalAlignment=VAlign_Fill}}
  }
  WBP_PlayerInfo {
   widgetClassPath=/Game/UI/Frontend/ItemShop/MultiProduct/WBP_PlayerInfo
   bIsVariable=true
   parent=Overlay_Content
   widget=WBP_PlayerInfo
   slot {class=OverlaySlot, properties {horizontalAlignment=HAlign_Right, padding=<ltrb 0 40 80 0>}}
  }
  Switcher_MainContent {
   widgetClassPath=&CommonVisibilitySwitcher
   bIsVariable=true
   parent=Overlay_Content
   widget=Switcher_MainContent
   slot {class=OverlaySlot, properties {horizontalAlignment=HAlign_Fill, verticalAlignment=VAlign_Fill}}
  }
  Overlay_Offers {
   widgetClassPath=&Overlay
   bIsVariable=true
   parent=Switcher_MainContent
   widget=Overlay_Offers
   slot {class=CommonVisibilitySwitcherSlot}
  }
  NamedSlot_Content {
   widgetClassPath=&NamedSlot
   bIsVariable=true
   parent=Overlay_Offers
   widget=NamedSlot_Content
   slot {class=OverlaySlot, properties {horizontalAlignment=HAlign_Fill, verticalAlignment=VAlign_Fill}}
  }
  WBP_VbucksStore_Content {
   widgetClassPath=&/WBP_VbucksStore_Content
   bIsVariable=true
   parent=NamedSlot_Content
   widget=WBP_VbucksStore_Content
   slot {class=PanelSlot}
  }
  Overlay_EpicRewards {
   widgetClassPath=&Overlay
   bIsVariable=true
   parent=Overlay_Offers
   widget=Overlay_EpicRewards
   properties {visibility=Collapsed}
   slot {class=OverlaySlot, properties {horizontalAlignment=HAlign_Center, verticalAlignment=VAlign_Center, padding=<ltrb 0 400 0 0>}}
  }
  WBP_UIKit_Backplate {
   widgetClassPath=&WBP_UIKit_Backplate
   bIsVariable=true
   parent=Overlay_EpicRewards
   widget=WBP_UIKit_Backplate
   properties {brightness=Dark}
   slot {class=OverlaySlot, properties {horizontalAlignment=HAlign_Fill, verticalAlignment=VAlign_Fill}}
  }
  RichText_EpicRewards {
   widgetClassPath=&CommonRichTextBlock
   bIsVariable=true
   parent=Overlay_EpicRewards
   widget=RichText_EpicRewards
   properties {
    bIsScrollingEnabled=false
    bOverrideDefaultStyle=true
    justification=Center
    mobileTextBlockScale=1.25
    text="[ph] Earn <text color=\"38FF84\">20% back</> in Epic Rewards on purchases of V-Bucks packs"
    textTransformPolicy=ToUpper
    defaultTextStyleOverride {
     overflowPolicy=Clip
     transformPolicy=None
     colorAndOpacity {colorUseRule=UseColor_Specified, specifiedColor=<rgba 1 1 1 1>}
     font {
      bForceMonospaced=false
      bMaterialIsStencil=false
      fontMaterial=null
      fontObject=&HeadingNow
      letterSpacing=0
      monospacedWidth=1
      size=15
      skewAmount=0
      typefaceFontName=75 Medium Italic
      outlineSettings {
       bApplyOutlineToDropShadows=false
       bMiteredCorners=false
       bSeparateFillAlpha=false
       outlineMaterial=null
       outlineSize=0
       outlineColor=<rgba 0 0 0 1>
      }
     }
     highlightColor {colorUseRule=UseColor_Specified, specifiedColor=<rgba 0 0 0 1>}
     highlightShape {
      bIsDynamicallyLoaded=false
      drawAs=Image
      imageType=NoImage
      mirroring=NoMirror
      resourceName=None
      resourceObject=null
      tiling=NoTile
      imageSize=<32 32>
      margin=<ltrb 0 0 0 0>
      outlineSettings {
       bUseBrushTransparency=false
       roundingType=HalfHeightRadius
       width=0
       color {colorUseRule=UseColor_Specified, specifiedColor=<rgba 0 0 0 0>}
       cornerRadii=<0 0 0 0>
      }
      tintColor {colorUseRule=UseColor_Specified, specifiedColor=<rgba 1 1 1 1>}
      uVRegion {bIsValid=false, max=<0 0>, min=<0 0>}
     }
     selectedBackgroundColor {colorUseRule=UseColor_Foreground, specifiedColor=<rgba 1 0 1 1>}
     shadowColorAndOpacity=<rgba 0 0 0 1>
     shadowOffset=<0 0>
     strikeBrush {
      bIsDynamicallyLoaded=false
      drawAs=Image
      imageType=NoImage
      mirroring=NoMirror
      resourceName=None
      resourceObject=null
      tiling=NoTile
      imageSize=<32 32>
      margin=<ltrb 0 0 0 0>
      outlineSettings {
       bUseBrushTransparency=false
       roundingType=HalfHeightRadius
       width=0
       color {colorUseRule=UseColor_Specified, specifiedColor=<rgba 0 0 0 0>}
       cornerRadii=<0 0 0 0>
      }
      tintColor {colorUseRule=UseColor_Specified, specifiedColor=<rgba 1 1 1 1>}
      uVRegion {bIsValid=false, max=<0 0>, min=<0 0>}
     }
     underlineBrush {
      bIsDynamicallyLoaded=false
      drawAs=Image
      imageType=NoImage
      mirroring=NoMirror
      resourceName=None
      resourceObject=null
      tiling=NoTile
      imageSize=<32 32>
      margin=<ltrb 0 0 0 0>
      outlineSettings {
       bUseBrushTransparency=false
       roundingType=HalfHeightRadius
       width=0
       color {colorUseRule=UseColor_Specified, specifiedColor=<rgba 0 0 0 0>}
       cornerRadii=<0 0 0 0>
      }
      tintColor {colorUseRule=UseColor_Specified, specifiedColor=<rgba 1 1 1 1>}
      uVRegion {bIsValid=false, max=<0 0>, min=<0 0>}
     }
    }
   }
   slot {class=OverlaySlot, properties {horizontalAlignment=HAlign_Fill, verticalAlignment=VAlign_Fill, padding=<ltrb 18 16 18 16>}}
  }
  Loading {
   widgetClassPath=&Image
   bIsVariable=true
   parent=Switcher_MainContent
   widget=Loading
   properties {
    visibility=SelfHitTestInvisible
    brush {
     bIsDynamicallyLoaded=false
     drawAs=Image
     imageType=NoImage
     mirroring=NoMirror
     resourceName=None
     resourceObject=/Game/UI/ContentOnDemand/ItemDownloadIndicator/MI_ItemDownloadIndicator_SpinnerMaterial
     tiling=NoTile
     imageSize=<100 100>
     margin=<ltrb 0 0 0 0>
     outlineSettings {
      bUseBrushTransparency=false
      roundingType=HalfHeightRadius
      width=0
      color {colorUseRule=UseColor_Specified, specifiedColor=<rgba 0 0 0 0>}
      cornerRadii=<0 0 0 0>
     }
     tintColor {colorUseRule=UseColor_Specified, specifiedColor=<rgba 1 1 1 1>}
     uVRegion {bIsValid=false, max=<0 0>, min=<0 0>}
    }
   }
   slot {class=CommonVisibilitySwitcherSlot, properties {horizontalAlignment=HAlign_Center, verticalAlignment=VAlign_Center}}
  }
  Text_NoOffers {
   widgetClassPath=&CommonTextBlock
   bIsVariable=true
   parent=Switcher_MainContent
   widget=Text_NoOffers
   properties {
    style=/Game/Athena/Settings/SettingEntries/Styles/TextStyle_HeadingNow_75_18_BoldGray.TextStyle_HeadingNow_75_18_BoldGray_C
    text=No Offers Available
    colorAndOpacity {colorUseRule=UseColor_Specified, specifiedColor=<rgba 0.75 0.75 0.75 1>}
    font {
     bForceMonospaced=false
     bMaterialIsStencil=false
     fontMaterial=null
     fontObject=&HeadingNow
     letterSpacing=0
     monospacedWidth=1
     size=16.5
     skewAmount=0
     typefaceFontName=76 Bold
     outlineSettings {
      bApplyOutlineToDropShadows=false
      bMiteredCorners=false
      bSeparateFillAlpha=false
      outlineMaterial=null
      outlineSize=0
      outlineColor=<rgba 0 0 0 1>
     }
    }
    shadowOffset=<0 0>
   }
   slot {class=CommonVisibilitySwitcherSlot, properties {horizontalAlignment=HAlign_Center, verticalAlignment=VAlign_Center}}
  }
  Text_PurchasingDisabled {
   widgetClassPath=&CommonTextBlock
   bIsVariable=true
   parent=Switcher_MainContent
   widget=Text_PurchasingDisabled
   properties {
    style=&TextStyle_HeadingNow_75_18_BoldGray_C
    text=Purchasing Temporarily Unavailable
    colorAndOpacity {colorUseRule=UseColor_Specified, specifiedColor=<rgba 0.75 0.75 0.75 1>}
    font {
     bForceMonospaced=false
     bMaterialIsStencil=false
     fontMaterial=null
     fontObject=&HeadingNow
     letterSpacing=0
     monospacedWidth=1
     size=16.5
     skewAmount=0
     typefaceFontName=76 Bold
     outlineSettings {
      bApplyOutlineToDropShadows=false
      bMiteredCorners=false
      bSeparateFillAlpha=false
      outlineMaterial=null
      outlineSize=0
      outlineColor=<rgba 0 0 0 1>
     }
    }
    shadowOffset=<0 0>
   }
   slot {class=CommonVisibilitySwitcherSlot, properties {horizontalAlignment=HAlign_Center, verticalAlignment=VAlign_Center}}
  }
  DecoyButton_Back {
   widgetClassPath=/Game/UI/Frontend/BottomBar/WBP_BottomBarDecoyButton
   bIsVariable=true
   parent=Overlay_Content
   widget=DecoyButton_Back
   properties {triggeringInputAction {dataTable=&InputActionDataTable, rowName=Back}}
   slot {class=OverlaySlot}
  }
 ]
}
