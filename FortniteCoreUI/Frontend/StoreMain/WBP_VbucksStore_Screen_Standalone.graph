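# Standalone variant of the frontend V-Bucks store screen, a child Widget Blueprint of WBP_VbucksStore_Screen.
# It adds no widgets, variables or event logic of its own.
# Its only logic is BP_GetDesiredInputConfig, which returns a CommonUI Menu-mode input config with mouse capture while a button is held and no cursor lock to the viewport.
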WidgetBlueprint /FortniteCoreUI/Frontend/StoreMain/WBP_VbucksStore_Screen_Standalone
graphs [
 BP_GetDesiredInputConfig {nodes [
   Get Desired Input Config() {is=title K2Node_FunctionEntry}
   > Return Node!(
    execute
    ReturnValue:UIInput Config Structure=Make UIInput Config!(
     bIgnoreMoveInput:bool=False
     bIgnoreLookInput:bool=False
     InputMode:ECommonInputMode Enum=Menu
     MouseCaptureMode:EMouseCaptureMode Enum=CaptureDuringMouseDown
     MouseLockMode:EMouseLockMode Enum=DoNotLock
     bHideCursorDuringViewportCapture:bool=True
    ) -> (UIInputConfig:UIInput Config Structure) {is=title K2Node_MakeStruct}
   )
  ]}
 EventGraph {links [], nodes []}
]
variables []
widgets {
 info {
  inheritedWidgetCount=0
  parentClass=&/WBP_VbucksStore_Screen.WBP_VbucksStore_Screen_C
  rootWidgetClass=null
  widgetCount=0
 }
 widgets []
}
