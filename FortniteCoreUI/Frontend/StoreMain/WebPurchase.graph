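# Store screen widget, based on FortWebPurchaseScreen, that shows the Epic web purchase display in the frontend.
# An overlay stacks a black backdrop, a circular throbber, the inherited transparent WebContent border, a post-buffer capture widget and a safe-zone Close button.
# The Close button, bound to the Back input action, calls CancelEpicPurchasingDisplay on the Fort MCP context.
# PreConstruct sets that button's padding from a BackButtonPadding variable.
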WidgetBlueprint /FortniteCoreUI/Frontend/StoreMain/WebPurchase
graphs [EventGraph {nodes [
   "On Clicked (Button_Back)"!() -> (OutputDelegate, then, Button:Common Button Base Object Reference) {is=title K2Node_ComponentBoundEvent}
   > CancelEpicPurchasingDisplay(self:Fort Mcp Context Object Reference=Get FortMcpContext!() -> (ReturnValue:Fort Mcp Context Object Reference) {is=title K2Node_GetBlueprintContextEx})
   Event PreConstruct()
   > SetPadding!(
    execute
    self:User Widget Object Reference=Get Button_Back() -> (Button_Back:WBP UIKit Button Quiet Object Reference)
    InPadding:Margin Structure=Get BackButtonPadding() -> (BackButtonPadding:Margin Structure)
   ) -> (then)
   Event Construct()
  ]}]
variables []
widgets {
 info {inheritedWidgetCount=1, parentClass=/Script/FortniteUI.FortWebPurchaseScreen, rootWidgetClass=&VerticalBox, widgetCount=8}
 widgets [
  VerticalBox_0 {widgetClassPath=&VerticalBox, widget=VerticalBox_0}
  Overlay_2891 {
   widgetClassPath=&Overlay
   parent=VerticalBox_0
   widget=Overlay_2891
   slot {class=VerticalBoxSlot, properties {size {sizeRule=Fill, value=1}}}
  }
  Border_12957 {
   widgetClassPath=&Border
   parent=Overlay_2891
   widget=Border_12957
   properties {brushColor=<rgba 0 0 0 1>, padding=<ltrb 0 15 0 10>}
   slot {class=OverlaySlot, properties {horizontalAlignment=HAlign_Fill, verticalAlignment=VAlign_Fill}}
  }
  Throbber {
   widgetClassPath=/Script/UMG.CircularThrobber
   bIsVariable=true
   parent=Overlay_2891
   widget=Throbber
   properties {numberOfPieces=20, period=1.5, radius=100}
   slot {class=OverlaySlot, properties {horizontalAlignment=HAlign_Center, verticalAlignment=VAlign_Center}}
  }
  WebContent {
   widgetClassPath=&Border
   bInherited=true
   bIsVariable=true
   parent=Overlay_2891
   widget=WebContent
   properties {brushColor=<rgba 1 1 1 0>, padding=<ltrb 0 0 0 0>}
   slot {class=OverlaySlot, properties {horizontalAlignment=HAlign_Fill, verticalAlignment=VAlign_Fill}}
  }
  WBP_CaptureForPostBufferUpdate {
   widgetClassPath=&WBP_CaptureForPostBufferUpdate
   bIsVariable=true
   parent=Overlay_2891
   slot {class=OverlaySlot, properties {horizontalAlignment=HAlign_Fill, verticalAlignment=VAlign_Fill}}
  }
  SafeZone_175 {
   widgetClassPath=&SafeZone
   bIsVariable=true
   parent=Overlay_2891
   widget=SafeZone_175
   slot {class=OverlaySlot, properties {horizontalAlignment=HAlign_Fill, verticalAlignment=VAlign_Fill}}
  }
  Button_Back {
   widgetClassPath=/FortUILibrary/UIKit/Buttons/WBP_UIKit_Button_Quiet
   bIsVariable=true
   parent=SafeZone_175
   widget=Button_Back
   properties {
    text=Close
    colorAndOpacity=<rgba 0.6771 0.6771 0.6771 1>
    triggeringInputAction {dataTable=&InputActionDataTable, rowName=Back}
   }
   slot {class=SafeZoneSlot, properties {hAlign=HAlign_Right, vAlign=VAlign_Bottom}}
  }
 ]
}
