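# InfoFlag is a store front-end banner widget showing wrapped rich text over a gradient beside a colored trim strip.
# PreConstruct applies the exposed colors and text, stretches the gradient by horizontal render scale, and derives trim width and text padding from TrimWidth.
# A non-zero MaxWidth caps the SizeBox width.
# Construct collapses the widget unless the bShowRMTPurchaseSpecialInfo runtime option is enabled.
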WidgetBlueprint /FortniteCoreUI/Frontend/StoreMain/InfoFlag
graphs [
 EventGraph {nodes [
   Event PreConstruct()
   > UpdateTrimColor!(execute, self, Color:Linear Color Structure=Get TrimColor() -> (TrimColor:Linear Color Structure)) -> (then)
   > UpdateGradientColor!(execute, self, Color:Linear Color Structure=Get GradientColor() -> (GradientColor:Linear Color Structure)) -> (then)
   > UpdateText!(execute, self, Text:"Text (by ref)"=Get InfoText() -> (InfoText:Text)) -> (then)
   > SetRenderScale!(
    execute
    self:Widget Object Reference=Get BG_Gradient() -> (BG_Gradient:Image Object Reference)
    Scale_X:double=Get GradientDistance() -> (GradientDistance:double)
    Scale_Y:double=1
   ) -> (then)
   n1 = Get TrimWidth() -> (TrimWidth:double)
   > SetBrush(
    self=Get Trim() -> (Trim:Image Object Reference)
    InBrush=Make Slate Brush!(ImageSize:Vector 2D Structure=MakeVector2D(X=@n1, Y=32)) -> (SlateBrush:Slate Brush Structure) {is=title K2Node_MakeStruct}
   )
   n2 = Get TrimWidth()
   n3 = float * float(A=@n2, B=0.5) {is=title K2Node_PromotableOperator}
   > SetPadding!(
    execute
    self:Overlay Slot Object Reference=SlotAsOverlaySlot(Widget=Get CommonRichTextBlock_114() -> (CommonRichTextBlock_114:Common Rich Text Block Object Reference))
    InPadding_Left:float=@n2
    InPadding_Top:float=@n3
    InPadding_Right:float=@n3
    InPadding_Bottom:float=@n3
   ) -> (then)
   > n4 = Branch(Condition="Nearly Equal (Float)"(A=Get MaxWidth() -> (MaxWidth:double), B=0, ErrorTolerance=0))
   SetMaxDesiredWidth!(
    execute=@n4.else
    self:Size Box Object Reference=Get SizeBox_0() -> (SizeBox_0:Size Box Object Reference)
    InMaxDesiredWidth:float=Get MaxWidth()
   ) -> (then)
   Event Construct()
   n5 = GetRuntimeOptions()
   > SetVisibility(InVisibility=Select(
     Option 0=Collapsed
     Option 1=SelfHitTestInvisible
     Index=Get bShowRMTPurchaseSpecialInfo!(self:Fort Runtime Options Object Reference=@n5) -> (bShowRMTPurchaseSpecialInfo:bool)
    ))
  ]}
 UpdateGradientColor {nodes [
   n1 = UpdateGradientColor!() -> (then, Color:Linear Color Structure) {is=title K2Node_FunctionEntry}
   > SetColorAndOpacity(self:Image Object Reference=Get BG_Gradient(), InColorAndOpacity:Linear Color Structure=@n1)
  ]}
 UpdateText {nodes [
   n1 = UpdateText!() -> (then, Text:"Text (by ref)") {is=title K2Node_FunctionEntry}
   > SetText(self=Get CommonRichTextBlock_114(), InText=@n1)
  ]}
 UpdateTrimColor {nodes [
   n1 = UpdateTrimColor!() -> (then, Color:Linear Color Structure) {is=title K2Node_FunctionEntry}
   > SetColorAndOpacity(self=Get Trim(), InColorAndOpacity=@n1)
  ]}
]
variables [
 GradientColor {type=LinearColor, default=<rgba 0 0.0976 0.4793 1>}
 GradientDistance {type=number, default=0.3}
 InfoText {type=string, default="This is debug text for you to see\r\n\r\nLearn more at <text color=\"FFD700FF\"> fn.gg/debugtext</>."}
 MaxWidth {type=number, default=3000}
 TrimColor {type=LinearColor, default=<rgba 1 0.6795 0 1>}
 TrimWidth {type=number, default=16}
]
widgets {
 info {inheritedWidgetCount=0, parentClass=&UserWidget, rootWidgetClass=&SizeBox, widgetCount=6}
 widgets [
  SizeBox_0 {widgetClassPath=&SizeBox, bIsVariable=true, widget=SizeBox_0}
  HorizontalBox_46 {widgetClassPath=&HorizontalBox, parent=SizeBox_0, widget=HorizontalBox_46, slot {class=SizeBoxSlot}}
  Trim {
   widgetClassPath=&Image
   bIsVariable=true
   parent=HorizontalBox_46
   widget=Trim
   properties {
    brush {
     bIsDynamicallyLoaded=false
     drawAs=Image
     imageType=NoImage
     mirroring=NoMirror
     resourceName=None
     resourceObject=null
     tiling=NoTile
     imageSize=<4 32>
     margin=<ltrb 0 0 0 0>
     outlineSettings {
      bUseBrushTransparency=false
      roundingType=HalfHeightRadius
      width=0
      color {colorUseRule=UseColor_Specified, specifiedColor=<rgba 0 0 0 0>}
      cornerRadii=<0 0 0 0>
     }
     tintColor {colorUseRule=UseColor_Specified, specifiedColor=<rgba 1 1 1 1>}
     uVRegion {bIsValid=false, max=<0 0>, min=<0 0>}
    }
    colorAndOpacity=<rgba 0.7991 0.624 0.0137 1>
   }
   slot {class=HorizontalBoxSlot, properties {horizontalAlignment=HAlign_Left}}
  }
  Overlay_411 {
   widgetClassPath=&Overlay
   parent=HorizontalBox_46
   widget=Overlay_411
   slot {class=HorizontalBoxSlot, properties {horizontalAlignment=HAlign_Left, size {sizeRule=Fill, value=1}}}
  }
  BG_Gradient {
   widgetClassPath=&Image
   bIsVariable=true
   parent=Overlay_411
   widget=BG_Gradient
   properties {
    brush {
     bIsDynamicallyLoaded=false
     drawAs=Image
     imageType=NoImage
     mirroring=NoMirror
     resourceName=None
     resourceObject=/Game/UI/Inventory/Art/BasicLinearUIGradient_Cr_ContentBrowserTileEdge_Horz
     tiling=NoTile
     imageSize=<32 32>
     margin=<ltrb 0 0 0 0>
     outlineSettings {
      bUseBrushTransparency=false
      roundingType=HalfHeightRadius
      width=0
      color {colorUseRule=UseColor_Specified, specifiedColor=<rgba 0 0 0 0>}
      cornerRadii=<0 0 0 0>
     }
     tintColor {colorUseRule=UseColor_Specified, specifiedColor=<rgba 1 1 1 1>}
     uVRegion {bIsValid=false, max=<0 0>, min=<0 0>}
    }
    colorAndOpacity=<rgba 0.004 0.1221 0.5149 1>
    renderTransformPivot=<0 0.5>
   }
   slot {class=OverlaySlot, properties {horizontalAlignment=HAlign_Fill, verticalAlignment=VAlign_Fill}}
  }
  CommonRichTextBlock_114 {
   widgetClassPath=&CommonRichTextBlock
   bIsVariable=true
   parent=Overlay_411
   widget=CommonRichTextBlock_114
   properties {
    autoWrapText=true
    bOverrideDefaultStyle=true
    text="This is debug text for you to see\r\n\r\nLearn more at <text color=\"FFD700FF\"> fn.gg/debugtext</>."
    defaultTextStyleOverride {
     overflowPolicy=Clip
     transformPolicy=None
     colorAndOpacity {colorUseRule=UseColor_Specified, specifiedColor=<rgba 1 1 1 1>}
     font {
      bForceMonospaced=false
      bMaterialIsStencil=false
      fontMaterial=null
      fontObject=&BurbankSmall
      letterSpacing=0
      monospacedWidth=1
      size=16
      skewAmount=0
      typefaceFontName=Bold
      outlineSettings {
       bApplyOutlineToDropShadows=false
       bMiteredCorners=false
       bSeparateFillAlpha=false
       outlineMaterial=null
       outlineSize=0
       outlineColor=<rgba 0 0 0 1>
      }
     }
     highlightColor {colorUseRule=UseColor_Specified, specifiedColor=<rgba 0 0 0 1>}
     highlightShape {
      bIsDynamicallyLoaded=false
      drawAs=Image
      imageType=NoImage
      mirroring=NoMirror
      resourceName=None
      resourceObject=null
      tiling=NoTile
      imageSize=<32 32>
      margin=<ltrb 0 0 0 0>
      outlineSettings {
       bUseBrushTransparency=false
       roundingType=HalfHeightRadius
       width=0
       color {colorUseRule=UseColor_Specified, specifiedColor=<rgba 0 0 0 0>}
       cornerRadii=<0 0 0 0>
      }
      tintColor {colorUseRule=UseColor_Specified, specifiedColor=<rgba 1 1 1 1>}
      uVRegion {bIsValid=false, max=<0 0>, min=<0 0>}
     }
     selectedBackgroundColor {colorUseRule=UseColor_Foreground, specifiedColor=<rgba 1 0 1 1>}
     shadowColorAndOpacity=<rgba 0 0 0 1>
     shadowOffset=<0 0>
     strikeBrush {
      bIsDynamicallyLoaded=false
      drawAs=Image
      imageType=NoImage
      mirroring=NoMirror
      resourceName=None
      resourceObject=null
      tiling=NoTile
      imageSize=<32 32>
      margin=<ltrb 0 0 0 0>
      outlineSettings {
       bUseBrushTransparency=false
       roundingType=HalfHeightRadius
       width=0
       color {colorUseRule=UseColor_Specified, specifiedColor=<rgba 0 0 0 0>}
       cornerRadii=<0 0 0 0>
      }
      tintColor {colorUseRule=UseColor_Specified, specifiedColor=<rgba 1 1 1 1>}
      uVRegion {bIsValid=false, max=<0 0>, min=<0 0>}
     }
     underlineBrush {
      bIsDynamicallyLoaded=false
      drawAs=Image
      imageType=NoImage
      mirroring=NoMirror
      resourceName=None
      resourceObject=null
      tiling=NoTile
      imageSize=<32 32>
      margin=<ltrb 0 0 0 0>
      outlineSettings {
       bUseBrushTransparency=false
       roundingType=HalfHeightRadius
       width=0
       color {colorUseRule=UseColor_Specified, specifiedColor=<rgba 0 0 0 0>}
       cornerRadii=<0 0 0 0>
      }
      tintColor {colorUseRule=UseColor_Specified, specifiedColor=<rgba 1 1 1 1>}
      uVRegion {bIsValid=false, max=<0 0>, min=<0 0>}
     }
    }
   }
   slot {class=OverlaySlot, properties {verticalAlignment=VAlign_Fill, padding=<ltrb 16 16 16 16>}}
  }
 ]
}
