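# Store offer banner widget showing an uppercased sale tag and a Battle Pass star bonus on MTX offers.
# SetupSpecialOfferBanner looks up the override tag in CatalogMessaging's banner map, preferring it over the passed sale text and falling back to "Sale!" when empty.
# It collapses unused rows and adjusts the star row's top padding by banner visibility.
# The star row shows "+ count", or "Bonus!" when the count is not positive.
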WidgetBlueprint /FortniteCoreUI/Frontend/StoreMain/MTXOffer_SpecialBanner
graphs [EventGraph {nodes [
   n1 = SetupSpecialOfferBanner!() -> (OutputDelegate, then, bHasSale:bool=~, SaleText:Text, BannerOverrideTag:String, bHasBattlePassStars:bool=~, iNumOfBattlePassStars:int) {is=title K2Node_CustomEvent}
   n2 = Find(
    TargetMap:Map of Strings to Texts=Get Banners!(self:Catalog Messaging Object Reference=Get CatalogMessaging() -> (CatalogMessaging:Catalog Messaging Object Reference)) -> (Banners:Map of Strings to Texts)
    Key=@n1.BannerOverrideTag
   ) -> (Value:Text)
   > n3 = Set SalesInfoText(SalesInfoText:Text=Select(Option 0:Text=@n1.SaleText, Option 1:Text=@n2, Index=@n2.ReturnValue) -> (ReturnValue:Text)) -> (Output_Get:Text)
   n4 = AND Boolean!(A=@n1.bHasSale, B=@n2.ReturnValue, C:bool=TextIsEmpty(InText=@n3)) -> (ReturnValue)
   > Set bShowBanner(bShowBanner:bool=AND Boolean!(A=OR Boolean(A=@n1.bHasSale, B=@n2.ReturnValue), B=NOT Boolean(A=@n4)) -> (ReturnValue)) -> (Output_Get:bool)
   > Set bHasBattlePassStars(bHasBattlePassStars:bool=@n1.bHasBattlePassStars) -> (Output_Get:bool)
   > Set NumOfBattlePassStars(NumOfBattlePassStars:int=@n1.iNumOfBattlePassStars) -> (Output_Get:int=0)
   n5 = Get bHasBattlePassStars() -> (bHasBattlePassStars:bool)
   n6 = Get bShowBanner() -> (bShowBanner:bool)
   > n7 = Branch(Condition=AND Boolean(A="Equal (Boolean)"!(A:bool=@n6, B:bool=false) -> (ReturnValue:bool), B="Equal (Boolean)"(A=@n5, B=false)))
   > SetVisibility(
    self=[
     Get Border_SaleInfo() -> (Border_SaleInfo:Common Border Object Reference)
     Get HB_BattlePassStarsSupplemental() -> (HB_BattlePassStarsSupplemental:Horizontal Box Object Reference)
    ]
    InVisibility=Collapsed
   )
   n8 = Branch(execute=@n7.else, Condition=Get bShowBanner())
   > SetVisibility(self=Get Border_SaleInfo(), InVisibility=SelfHitTestInvisible)
   n9 = TextIsEmpty(InText=Get SalesInfoText() -> (SalesInfoText:Text))
   n10 = Get SalesInfoText()
   > n11 = "SetText (Text)"(
    self=Get T_SaleInfo() -> (T_SaleInfo:Common Text Object Reference)
    InText=TextToUpper!(InText:"Text (by ref)"=Select(Option 0=@n10, Option 1="LOCTEXT(\"Sale!\")", Index=@n9)) -> (ReturnValue:Text)
   )
   n12 = SlotAsVerticalBoxSlot(Widget=Get HB_BattlePassStarsSupplemental())
   n13 = Break Margin!(Margin:"Margin Structure (by ref)"=Get Padding!(self:Vertical Box Slot Object Reference=@n12) -> (Padding:Margin Structure)) -> (Left:float=~, Top:float=~, Right:float=~, Bottom:float=~) {is=title K2Node_BreakStruct}
   n14 = Get bShowBanner()
   SetPadding!(
    execute=[@n18.then, @n11.then]
    self:Vertical Box Slot Object Reference=@n12
    InPadding:Margin Structure=Make Margin(Left=@n13, Top=Select(Option 0:double=20, Option 1:double=-2, Index=@n14) -> (ReturnValue:double), Right=@n13.Right, Bottom=@n13.Bottom)
   ) -> (then)
   n15 = Get bHasBattlePassStars()
   > SetVisibility(
    self=Get HB_BattlePassStarsSupplemental()
    InVisibility=Select(Option 0:ESlateVisibility Enum=Collapsed, Option 1:ESlateVisibility Enum=SelfHitTestInvisible, Index=@n15) -> (ReturnValue:ESlateVisibility Enum)
   )
   > Branch(Condition=@n15)
   n16 = Get NumOfBattlePassStars() -> (NumOfBattlePassStars:int=0)
   n17 = Get NumOfBattlePassStars() -> (NumOfBattlePassStars=0)
   > "SetText (Text)"(
    self=Get T_BattleStarInfoSupplemental() -> (T_BattleStarInfoSupplemental:Common Text Object Reference)
    InText=Select(
     Option 0:Text="LOCTEXT(\"Bonus!\")"
     Option 1:Text=Format Text!(Format="LOCTEXT(\"+ {x}\")", x:int=@n17) -> (Result)
     Index="integer > integer"(A=@n16, B=0) {is=title K2Node_CallFunction}
    ) -> (ReturnValue:Text)
   )
   n18 = SetVisibility(
    execute=@n8.else
    self=Get Border_SaleInfo()
    InVisibility=Select(Option 0:ESlateVisibility Enum=SelfHitTestInvisible, Option 1:ESlateVisibility Enum=Collapsed, Index=Get bHasBattlePassStars()) -> (ReturnValue:ESlateVisibility Enum)
   )
  ]}]
variables [
 bHasBattlePassStars {type=boolean}
 bShowBanner {type=boolean}
 CatalogMessaging {type=CatalogMessaging, default=/Game/Athena/UI/Frontend/CatalogMessages}
 NumOfBattlePassStars {type=integer}
 SalesInfoText {type=string, default=""}
]
widgets {
 info {inheritedWidgetCount=0, parentClass=&UserWidget, rootWidgetClass=&WidgetSwitcher, widgetCount=8}
 widgets [
  WidgetSwitcher_Main {widgetClassPath=&WidgetSwitcher, bIsVariable=true, widget=WidgetSwitcher_Main}
  VerticalBox_0 {
   widgetClassPath=&VerticalBox
   parent=WidgetSwitcher_Main
   widget=VerticalBox_0
   slot {class=WidgetSwitcherSlot, properties {horizontalAlignment=HAlign_Left, verticalAlignment=VAlign_Top}}
  }
  Border_SaleInfo {
   widgetClassPath=&CommonBorder
   bIsVariable=true
   parent=VerticalBox_0
   widget=Border_SaleInfo
   properties {
    bFlipForRightToLeftFlowDirection=true
    background {
     bIsDynamicallyLoaded=false
     drawAs=Box
     imageType=NoImage
     mirroring=NoMirror
     resourceName=None
     resourceObject=/Game/UI/Login/SubGameSelect/Materials/M_Notification_Right
     tiling=NoTile
     imageSize=<96 32>
     margin=<ltrb 0.1 0 0.25 0>
     outlineSettings {
      bUseBrushTransparency=false
      roundingType=HalfHeightRadius
      width=0
      color {colorUseRule=UseColor_Specified, specifiedColor=<rgba 0 0 0 0>}
      cornerRadii=<0 0 0 0>
     }
     tintColor {colorUseRule=UseColor_Specified, specifiedColor=<rgba 1 1 1 1>}
     uVRegion {bIsValid=false, max=<0 0>, min=<0 0>}
    }
    padding=<ltrb 16 8 25 9>
   }
   slot {class=VerticalBoxSlot}
  }
  SizeBox_0 {
   widgetClassPath=&SizeBox
   parent=Border_SaleInfo
   widget=SizeBox_0
   properties {bOverride_MaxDesiredWidth=true, maxDesiredWidth=255}
   slot {class=BorderSlot, properties {padding=<ltrb 16 8 25 9>}}
  }
  T_SaleInfo {
   widgetClassPath=&CommonTextBlock
   bIsVariable=true
   parent=SizeBox_0
   widget=T_SaleInfo
   properties {
    scrollStyle=&TextScrollStyle-Base_C
    text="SALE!"
    font {
     bForceMonospaced=false
     bMaterialIsStencil=false
     fontMaterial=null
     fontObject=&Burbank
     letterSpacing=0
     monospacedWidth=1
     size=22
     skewAmount=0
     typefaceFontName=Burbank-Black
     outlineSettings {
      bApplyOutlineToDropShadows=false
      bMiteredCorners=false
      bSeparateFillAlpha=false
      outlineMaterial=null
      outlineSize=0
      outlineColor=<rgba 0 0 0 1>
     }
    }
    shadowOffset=<0 0>
   }
   slot {class=SizeBoxSlot}
  }
  HB_BattlePassStarsSupplemental {
   widgetClassPath=&HorizontalBox
   bIsVariable=true
   parent=VerticalBox_0
   widget=HB_BattlePassStarsSupplemental
   properties {visibility=Collapsed}
   slot {class=VerticalBoxSlot, properties {padding=<ltrb 25 -2 28 0>}}
  }
  Image_BattleStarSupplemental {
   widgetClassPath=&Image
   bIsVariable=true
   parent=HB_BattlePassStarsSupplemental
   widget=Image_BattleStarSupplemental
   properties {
    visibility=SelfHitTestInvisible
    brush {
     bIsDynamicallyLoaded=false
     drawAs=Image
     imageType=NoImage
     mirroring=NoMirror
     resourceName=None
     resourceObject=/Game/UI/Foundation/Materials/M-BattlePassStarStoreDisplay
     tiling=NoTile
     imageSize=<35 35>
     margin=<ltrb 0 0 0 0>
     outlineSettings {
      bUseBrushTransparency=false
      roundingType=HalfHeightRadius
      width=0
      color {colorUseRule=UseColor_Specified, specifiedColor=<rgba 0 0 0 0>}
      cornerRadii=<0 0 0 0>
     }
     tintColor {colorUseRule=UseColor_Specified, specifiedColor=<rgba 1 1 1 1>}
     uVRegion {bIsValid=false, max=<0 0>, min=<0 0>}
    }
    renderTransform {angle=0, scale=<1 1>, shear=<0 0>, translation=<0 -1>}
   }
   slot {class=HorizontalBoxSlot, properties {horizontalAlignment=HAlign_Left, padding=<ltrb 0 2 0 4>}}
  }
  T_BattleStarInfoSupplemental {
   widgetClassPath=&CommonTextBlock
   bIsVariable=true
   parent=HB_BattlePassStarsSupplemental
   widget=T_BattleStarInfoSupplemental
   properties {
    style=/FortUILibrary/Styles/Typography/Foundation/TextStyle-Header-S.TextStyle-Header-S_C
    text=+ 10
    font {
     bForceMonospaced=false
     bMaterialIsStencil=false
     fontMaterial=null
     fontObject=&Burbank
     letterSpacing=0
     monospacedWidth=1
     size=24
     skewAmount=0
     typefaceFontName=Burbank-Black
     outlineSettings {
      bApplyOutlineToDropShadows=false
      bMiteredCorners=false
      bSeparateFillAlpha=false
      outlineMaterial=null
      outlineSize=0
      outlineColor=<rgba 0 0 0 1>
     }
    }
    shadowOffset=<0 0>
   }
   slot {class=HorizontalBoxSlot, properties {horizontalAlignment=HAlign_Left, padding=<ltrb 4 0 0 0>, size {sizeRule=Fill, value=1}}}
  }
 ]
}
