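# Store tile widget, built on FortRealMoneyOffer, that shows a real-money V-Bucks pack or bundle offer in the frontend store.
# InitOffer switches between currency and bundle layouts and fills quantity, prices, sale and banners.
# Display size sets tile dimensions and border material, while the offer's display asset colors background, gradient and text outlines through dynamic materials.
# Hover plays one animation forward or reversed, and callable functions trigger or reset the intro.
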WidgetBlueprint /FortniteCoreUI/Frontend/StoreMain/StoreMain_MTXOffer
animations [
 OnHover {
  binding_count=1
  length_seconds=0.5
  object=&self:OnHover
  sequence {bindings [ScaleBoxKeyArt {
     id=#1
     display_name=ScaleBox_2
     possessed_class=ScaleBox
     tracks [{
       type=MovieScene2DTransformTrack
       sections [{
         type=MovieScene2DTransformSection
         channels [
          Translation.X {type=MovieSceneScriptingFloatChannel}
          Translation.Y {type=MovieSceneScriptingFloatChannel}
          Angle {type=MovieSceneScriptingFloatChannel}
          Scale.X {type=MovieSceneScriptingFloatChannel, keys [[0,1],[10,1.125]]}
          Scale.Y {type=MovieSceneScriptingFloatChannel, keys [[0,1],[10,1.125]]}
          Shear.X {type=MovieSceneScriptingFloatChannel}
          Shear.Y {type=MovieSceneScriptingFloatChannel}
         ]
        }]
      }]
    }]}
 }
 Intro {
  binding_count=3
  display_label=""
  length_seconds=0.4
  object=&self:Intro
  sequence {
   bindings [
    RMTCurrency {
     id=#2
     possessed_class=VerticalBox
     tracks [
      {
       type=MovieSceneFloatTrack
       display_name=Render Opacity
       sections [{
         type=MovieSceneFloatSection
         end_frame=40
         start_frame=0
         channels [None {type=MovieSceneScriptingFloatChannel, keys [[8.8883,0],[18.8883,1]]}]
        }]
      }
      {
       type=MovieScene2DTransformTrack
       sections [{
         type=MovieScene2DTransformSection
         end_frame=40
         start_frame=0
         channels [
          Translation.X {type=MovieSceneScriptingFloatChannel}
          Translation.Y {type=MovieSceneScriptingFloatChannel}
          Angle {type=MovieSceneScriptingFloatChannel}
          Scale.X {type=MovieSceneScriptingFloatChannel}
          Scale.Y {type=MovieSceneScriptingFloatChannel}
          Shear.X {type=MovieSceneScriptingFloatChannel}
          Shear.Y {type=MovieSceneScriptingFloatChannel}
         ]
        }]
      }
     ]
    }
    VerticalBox_1 {
     id=#3
     possessed_class=VerticalBox
     tracks [
      {
       type=MovieSceneFloatTrack
       display_name=Render Opacity
       sections [{type=MovieSceneFloatSection, end_frame=40, start_frame=0, channels [None {type=MovieSceneScriptingFloatChannel, keys [[9.9983,0],[25,1]]}]}]
      }
      {
       type=MovieScene2DTransformTrack
       sections [{
         type=MovieScene2DTransformSection
         end_frame=40
         start_frame=0
         channels [
          Translation.X {type=MovieSceneScriptingFloatChannel}
          Translation.Y {type=MovieSceneScriptingFloatChannel}
          Angle {type=MovieSceneScriptingFloatChannel}
          Scale.X {type=MovieSceneScriptingFloatChannel, keys [[9.9983,1.1],[39.9983,1]]}
          Scale.Y {type=MovieSceneScriptingFloatChannel, keys [[9.9983,0.8],[29.9983,1]]}
          Shear.X {type=MovieSceneScriptingFloatChannel}
          Shear.Y {type=MovieSceneScriptingFloatChannel}
         ]
        }]
      }
     ]
    }
    BorderAll {
     id=#4
     possessed_class=Border
     tracks [
      {
       type=MovieScene2DTransformTrack
       sections [{
         type=MovieScene2DTransformSection
         end_frame=30
         start_frame=0
         channels [
          Translation.X {type=MovieSceneScriptingFloatChannel, keys [[0,-16],[21.4283,0]]}
          Translation.Y {type=MovieSceneScriptingFloatChannel, keys [[0,8],[21.4283,0]]}
          Angle {type=MovieSceneScriptingFloatChannel, const [0,1]}
          Scale.X {type=MovieSceneScriptingFloatChannel, keys [[0,0.8],[21.4283,1]]}
          Scale.Y {type=MovieSceneScriptingFloatChannel, keys [[0,0.975],[21.4283,1]]}
          Shear.X {type=MovieSceneScriptingFloatChannel}
          Shear.Y {type=MovieSceneScriptingFloatChannel, keys [[0,4],[21.4283,0]]}
         ]
        }]
      }
      {
       type=MovieSceneFloatTrack
       display_name=Render Opacity
       sections [{type=MovieSceneFloatSection, end_frame=30, start_frame=0, channels [None {type=MovieSceneScriptingFloatChannel, keys [[0,0],[14.9983,1]]}]}]
      }
     ]
    }
   ]
   display_rate {numerator=100}
  }
 }
 HoverNew {
  binding_count=2
  length_seconds=0.2
  object=&self:HoverNew
  sequence {
   bindings [
    OverlayKeyImage {
     id=#5
     possessed_class=Overlay
     tracks [{
       type=MovieScene2DTransformTrack
       sections [{
         type=MovieScene2DTransformSection
         end_frame=12
         start_frame=0
         channels [
          Translation.X {type=MovieSceneScriptingFloatChannel}
          Translation.Y {type=MovieSceneScriptingFloatChannel, keys [[0,-24],[11.999,-48]]}
          Angle {type=MovieSceneScriptingFloatChannel}
          Scale.X {type=MovieSceneScriptingFloatChannel}
          Scale.Y {type=MovieSceneScriptingFloatChannel}
          Shear.X {type=MovieSceneScriptingFloatChannel}
          Shear.Y {type=MovieSceneScriptingFloatChannel}
         ]
        }]
      }]
    }
    KeyArt {
     id=#6
     possessed_class=Image
     tracks [{
       type=MovieScene2DTransformTrack
       sections [{
         type=MovieScene2DTransformSection
         end_frame=12
         start_frame=0
         channels [
          Translation.X {type=MovieSceneScriptingFloatChannel}
          Translation.Y {type=MovieSceneScriptingFloatChannel}
          Angle {type=MovieSceneScriptingFloatChannel}
          Scale.X {type=MovieSceneScriptingFloatChannel, keys [[0,1],[11.999,1.075]]}
          Scale.Y {type=MovieSceneScriptingFloatChannel, keys [[0,1],[11.999,1.075]]}
          Shear.X {type=MovieSceneScriptingFloatChannel}
          Shear.Y {type=MovieSceneScriptingFloatChannel}
         ]
        }]
      }]
    }
   ]
   display_rate {numerator=60}
  }
 }
 UnhoverNew {
  binding_count=2
  length_seconds=0.2
  object=&self:UnhoverNew
  sequence {
   bindings [
    OverlayKeyImage {
     id=#7
     possessed_class=Overlay
     tracks [{
       type=MovieScene2DTransformTrack
       sections [{
         type=MovieScene2DTransformSection
         end_frame=12
         start_frame=0
         channels [
          Translation.X {type=MovieSceneScriptingFloatChannel}
          Translation.Y {type=MovieSceneScriptingFloatChannel, keys [[0,-48],[11.999,-24]]}
          Angle {type=MovieSceneScriptingFloatChannel}
          Scale.X {type=MovieSceneScriptingFloatChannel}
          Scale.Y {type=MovieSceneScriptingFloatChannel}
          Shear.X {type=MovieSceneScriptingFloatChannel}
          Shear.Y {type=MovieSceneScriptingFloatChannel}
         ]
        }]
      }]
    }
    KeyArt {
     id=#8
     possessed_class=Image
     tracks [{
       type=MovieScene2DTransformTrack
       sections [{
         type=MovieScene2DTransformSection
         end_frame=12
         start_frame=0
         channels [
          Translation.X {type=MovieSceneScriptingFloatChannel}
          Translation.Y {type=MovieSceneScriptingFloatChannel}
          Angle {type=MovieSceneScriptingFloatChannel}
          Scale.X {type=MovieSceneScriptingFloatChannel, keys [[0,1.075],[12,1]]}
          Scale.Y {type=MovieSceneScriptingFloatChannel, keys [[0,1.075],[12,1]]}
          Shear.X {type=MovieSceneScriptingFloatChannel}
          Shear.Y {type=MovieSceneScriptingFloatChannel}
         ]
        }]
      }]
    }
   ]
   display_rate {numerator=60}
  }
 }
 Hover_Linear {
  binding_count=6
  length_seconds=0.2
  object=&self:Hover_Linear
  sequence {
   bindings [
    OverlayKeyImage {
     id=#5
     possessed_class=Overlay
     tracks [{
       type=MovieScene2DTransformTrack
       sections [{
         type=MovieScene2DTransformSection
         end_frame=12
         start_frame=0
         channels [
          Translation.X {type=MovieSceneScriptingFloatChannel}
          Translation.Y {type=MovieSceneScriptingFloatChannel, keys [[0,-24],[11.999,-48]]}
          Angle {type=MovieSceneScriptingFloatChannel}
          Scale.X {type=MovieSceneScriptingFloatChannel}
          Scale.Y {type=MovieSceneScriptingFloatChannel}
          Shear.X {type=MovieSceneScriptingFloatChannel}
          Shear.Y {type=MovieSceneScriptingFloatChannel}
         ]
        }]
      }]
    }
    Border_Footer {
     id=#9
     possessed_class=Border
     tracks [
      {
       type=MovieSceneWidgetMaterialTrack
       sections [{
         type=MovieSceneParameterSection
         end_frame=12
         start_frame=0
         channels [TopCenterOffset {type=MovieSceneScriptingFloatChannel, keys [[0,-0.3],[12,-0.25]]}]
        }]
      }
      {
       type=MovieSceneColorTrack
       display_name=Brush Color
       sections [{
         type=MovieSceneColorSection
         end_frame=12
         start_frame=0
         channels [
          Color.R {type=MovieSceneScriptingFloatChannel, const [1,1]}
          Color.G {type=MovieSceneScriptingFloatChannel, keys [[0,0.9216],[12,1]]}
          Color.B {type=MovieSceneScriptingFloatChannel, keys [[0,0],[12,1]]}
          Color.A {type=MovieSceneScriptingFloatChannel, const [1,1]}
         ]
        }]
      }
     ]
    }
    OverlayAll {
     id=#10
     possessed_class=Overlay
     tracks [{
       type=MovieScene2DTransformTrack
       sections [{
         type=MovieScene2DTransformSection
         end_frame=12
         start_frame=0
         channels [
          Translation.X {type=MovieSceneScriptingFloatChannel}
          Translation.Y {type=MovieSceneScriptingFloatChannel, keys [[0,0],[12,-8]]}
          Angle {type=MovieSceneScriptingFloatChannel}
          Scale.X {type=MovieSceneScriptingFloatChannel}
          Scale.Y {type=MovieSceneScriptingFloatChannel}
          Shear.X {type=MovieSceneScriptingFloatChannel}
          Shear.Y {type=MovieSceneScriptingFloatChannel}
         ]
        }]
      }]
    }
    HoverBorder {
     id=#11
     possessed_class=Image
     tracks [{
       type=MovieSceneWidgetMaterialTrack
       sections [{
         type=MovieSceneParameterSection
         end_frame=12
         start_frame=0
         channels [
          AnimateAngle {type=MovieSceneScriptingFloatChannel, keys [[0,0.5],[12,1]]}
          AnimateFill {type=MovieSceneScriptingFloatChannel, keys [[0,0],[12,1]]}
          OverallAlpha {type=MovieSceneScriptingFloatChannel, keys [[0,0.2],[12,1]]}
         ]
        }]
      }]
    }
    KeyArt {
     id=#6
     possessed_class=Image
     tracks [{
       type=MovieScene2DTransformTrack
       sections [{
         type=MovieScene2DTransformSection
         end_frame=12
         start_frame=0
         channels [
          Translation.X {type=MovieSceneScriptingFloatChannel}
          Translation.Y {type=MovieSceneScriptingFloatChannel}
          Angle {type=MovieSceneScriptingFloatChannel}
          Scale.X {type=MovieSceneScriptingFloatChannel, keys [[0,1],[11.999,1.075]]}
          Scale.Y {type=MovieSceneScriptingFloatChannel, keys [[0,1],[11.999,1.075]]}
          Shear.X {type=MovieSceneScriptingFloatChannel}
          Shear.Y {type=MovieSceneScriptingFloatChannel}
         ]
        }]
      }]
    }
    SizeBox_2 {
     id=#12
     possessed_class=SizeBox
     tracks [
      {
       type=MovieScene2DTransformTrack
       sections [{
         type=MovieScene2DTransformSection
         end_frame=12
         start_frame=0
         channels [
          Translation.X {type=MovieSceneScriptingFloatChannel}
          Translation.Y {type=MovieSceneScriptingFloatChannel, const [28,1]}
          Angle {type=MovieSceneScriptingFloatChannel}
          Scale.X {type=MovieSceneScriptingFloatChannel, keys [[0,0.75],[12,1]]}
          Scale.Y {type=MovieSceneScriptingFloatChannel, keys [[0,0.75],[12,1]]}
          Shear.X {type=MovieSceneScriptingFloatChannel}
          Shear.Y {type=MovieSceneScriptingFloatChannel}
         ]
        }]
      }
      {
       type=MovieSceneFloatTrack
       display_name=Render Opacity
       sections [{type=MovieSceneFloatSection, end_frame=12, start_frame=0, channels [None {type=MovieSceneScriptingFloatChannel, keys [[0,0],[12,1]]}]}]
      }
     ]
    }
   ]
   display_rate {numerator=60}
  }
 }
]
graphs [
 AboutToShowDetails {nodes [AboutToShowDetails() {is=title K2Node_FunctionEntry}]}
 EventGraph {nodes [
   "On Input Icon Updated (InputActionWidget)"() {is=title K2Node_ComponentBoundEvent}
   > n1 = GetIconDynamicMaterial!(execute, self:Common Action Widget Object Reference=Get InputActionWidget() -> (InputActionWidget:Common Action Widget Object Reference)) -> (then, ReturnValue:Material Instance Dynamic Object Reference)
   > SetScalarParameterValue(self=@n1, ParameterName=Use Dark, Value=1)
   > SetScalarParameterValue(self=@n1, ParameterName=Outline Opacity, Value=0)
   Event PreConstruct()
   > SetTIleSize!(execute, self, DisplaySize:EFortMtxOfferDisplaySize Enum=Medium) -> (then)
   Event OnOfferSet() {is=title K2Node_Event}
   > Set bCanSetPrice(bCanSetPrice:bool=true) -> (Output_Get:bool)
   > InitOffer()
   > SetTIleSize(DisplaySize=GetOfferDisplaySize!(self:Fort Store Front Offer Info Object Reference=GetOfferInfo!(self:Fort Store Front Offer Widget Base Object Reference) -> (ReturnValue:Fort Store Front Offer Info Object Reference)) -> (ReturnValue:EFortMtxOfferDisplaySize Enum=Small))
   n2 = Event OnUpdateStorePrice!() -> (OutputDelegate, then, InStorePrice:"Text (by ref)") {is=title K2Node_Event}
   > "SetText (Text)"(self=Get TextOfferDisplayPrice() -> (TextOfferDisplayPrice:Common Text Object Reference), InText=@n2.InStorePrice)
   > Set bCanSetPrice(bCanSetPrice=false)
   Event On Hovered()
   > Parent: On Hovered() {is=title K2Node_CallParentFunction}
   n3 = Get Hover_Linear() -> (Hover_Linear:Widget Animation Object Reference)
   > PlayAnimationForward(InAnimation=@n3, PlaybackSpeed=1.5)
   Event OnStoreSelectionOfferSet() {is=title K2Node_Event}
   > SetTIleSize(DisplaySize=Medium)
   n4 = Event OnOfferTileBrushLoaded!() -> (OutputDelegate, then, OfferBrush:"Slate Brush Structure (by ref)") {is=title K2Node_Event}
   > SetBrush(self=Get KeyArt() -> (KeyArt:Image Object Reference), InBrush=@n4.OfferBrush)
   > InvalidateLayoutAndVolatility(self:Widget Object Reference=Get KeyArt())
   Event On Unhovered()
   > PlayAnimationReverse(InAnimation=@n3, PlaybackSpeed=2)
   Event On Pressed() {is=title K2Node_Event}
   > Parent: On Pressed() {is=title K2Node_CallParentFunction}
  ]}
 GetBonusText {nodes [
   n1 = GetBonusText!() -> (then, OfferInfo:Fort Store Front Offer Info Object Reference) {is=title K2Node_FunctionEntry}
   n2 = GetBaseQuantity!(self:Fort Store Front Offer Info Object Reference=@n1) -> (ReturnValue:int=0)
   n3 = GetBonusQuantity!(self:Fort Store Front Offer Info Object Reference=@n1) -> (ReturnValue:int=0)
   n4 = float * float(A=float / float(A="To Float (Integer)"(InInt=@n3), B="To Float (Integer)"(InInt=@n2)) {is=title K2Node_CallFunction}, B=100)
   > Return Node!(execute, BonusText:Text=Format Text!(Format="LOCTEXT(\"{extra}% Extra!\")", extra:int=Truncate(A=@n4)) -> (Result))
  ]}
 InitOffer {nodes [
   InitOffer() {is=title K2Node_FunctionEntry}
   n1 = GetOfferInfo()
   > n2 = Set OfferInfo(OfferInfo:Fort Store Front Offer Info Object Reference=@n1) -> (Output_Get:Fort Store Front Offer Info Object Reference)
   > Set TotalQuantity(TotalQuantity:int=GetTotalQuantity!(self:Fort Store Front Offer Info Object Reference=@n2) -> (ReturnValue:int=0)) -> (Output_Get:int=0)
   > Set DisplayAsset(DisplayAsset:Fort Mtx Offer Data Object Reference=GetOfferDisplayAsset!(self:Fort Store Front Offer Info Object Reference=@n1) -> (ReturnValue:Fort Mtx Offer Data Object Reference)) -> (Output_Get:Fort Mtx Offer Data Object Reference)
   > UpdateBannerText()
   n3 = Get OfferInfo() -> (OfferInfo:Fort Store Front Offer Info Object Reference)
   > n4 = Branch(Condition="Equal (Enum)"(
     A:EFortMtxStoreOfferType Enum=GetFortStoreOfferType!(self:Fort Store Front Offer Info Object Reference=@n3) -> (ReturnValue:EFortMtxStoreOfferType Enum=FoundersPack)
     B:EFortMtxStoreOfferType Enum=CurrencyPack
    ))
   > SetBrushFromMaterial(self=Get OverlayGradient() -> (OverlayGradient:Image Object Reference), Material=/Game/UI/Foundation/Materials/M-Store-GradientOverlay)
   > SetVisibility(self=Get TextOfferName() -> (TextOfferName:Common Text Object Reference), InVisibility=Collapsed)
   > SetActiveWidget(
    self:Widget Switcher Object Reference=Get RMTTypeSwitcher() -> (RMTTypeSwitcher:Widget Switcher Object Reference)
    Widget=Get RMTCurrency() -> (RMTCurrency:Vertical Box Object Reference)
   )
   > SetVisibility(
    self=Get AthenaDirectAcquisitionOffer_SpecialBanner() -> (AthenaDirectAcquisitionOffer_SpecialBanner:MTXOffer Special Banner Object Reference)
    InVisibility=Collapsed
   )
   > n5 = "SetText (Text)"(
    self=Get CurrencyTotalQuantity() -> (CurrencyTotalQuantity:Common Text Object Reference)
    InText="To Text (Integer)"(Value=Get TotalQuantity() -> (TotalQuantity:int=0))
   )
   n6 = Branch(execute=[@n18.then, @n5.then], Condition=Get bCanSetPrice() -> (bCanSetPrice:bool))
   > n7 = "SetText (Text)"(
    self=Get TextOriginalPrice() -> (TextOriginalPrice:Common Text Object Reference)
    InText=GetRegularPrice!(self:Fort Store Front Offer Info Object Reference=Get OfferInfo(), PriceIndex:int=0, ItemQuantity:int=1) -> (ReturnValue:Text)
   )
   n8 = GetSalePriceFromLastAppStore!(self:Fort Real Money Offer Object Reference, PriceIndex:int=0, ItemQuantity:int=1) -> (SalePrice:Text, ReturnValue:bool)
   "SetText (Text)"(execute=[@n7.then, @n6.else], self=Get TextOfferDisplayPrice(), InText=@n8)
   > SetVisibility(
    self=Get OverlaySalePrice() -> (OverlaySalePrice:Overlay Object Reference)
    InVisibility=Select(Option 0=Collapsed, Option 1=SelfHitTestInvisible, Index=@n8.ReturnValue)
   )
   n9 = Get OfferInfo()
   > SetActiveWidget(
    self=Get Switcher_Price() -> (Switcher_Price:Common Widget Switcher Legacy Object Reference)
    Widget=Select(
     Option 0:Widget Object Reference=Get TextOfferPurchaseUnvailable() -> (TextOfferPurchaseUnvailable:Common Text Object Reference)
     Option 1:Widget Object Reference=Get HBox_Price() -> (HBox_Price:Horizontal Box Object Reference)
     Index=IsPurchasingAvailable!(self:Fort Store Front Offer Info Object Reference=@n9) -> (ReturnValue:bool)
    ) -> (ReturnValue:Widget Object Reference)
   )
   > Set Gradient(Gradient:Fort Mtx Gradient Structure=Get Gradient!(self:Fort Mtx Offer Data Object Reference=Get DisplayAsset() -> (DisplayAsset:Fort Mtx Offer Data Object Reference)) -> (Gradient:Fort Mtx Gradient Structure)) -> (Output_Get:Fort Mtx Gradient Structure)
   > SetColorAndOpacity(
    self:Image Object Reference=Get BaseColor() -> (BaseColor:Image Object Reference)
    InColorAndOpacity:Linear Color Structure=Get Background!(self:Fort Mtx Offer Data Object Reference=Get DisplayAsset()) -> (Background:Linear Color Structure)
   )
   > n10 = GetDynamicMaterial(self=Get OverlayGradient())
   n11 = Break Fort Mtx Gradient!(FortMtxGradient:"Fort Mtx Gradient Structure (by ref)"=Get Gradient!() -> (Gradient)) -> (Start:Linear Color Structure, Stop:Linear Color Structure) {is=title K2Node_BreakStruct}
   > SetVectorParameterValue(self=@n10, ParameterName=RGBA1, Value=@n11)
   > SetVectorParameterValue(self=@n10, ParameterName=RGBA0, Value=@n11.Stop)
   > SetTextOutlineColor!(execute, self, Text:Text Object Reference=Get CurrencyTotalQuantity(), Gradient:Fort Mtx Gradient Structure=Get Gradient()) -> (then)
   > SetTextOutlineColor(Text=Get VBucksText() -> (VBucksText:Common Text Object Reference), Gradient=Get Gradient())
   > SetTextOutlineColor(Text=Get BundleVBucksText() -> (BundleVBucksText:Common Text Object Reference), Gradient=Get Gradient())
   > SetTextOutlineColor(Text=Get BundleTotalQuantity() -> (BundleTotalQuantity:Common Text Object Reference), Gradient=Get Gradient())
   > SetTextOutlineColor(Text=Get TextOfferName(), Gradient=Get Gradient())
   > SetTextOutlineColor(Text=Get Plus() -> (Plus:Common Text Object Reference), Gradient=Get Gradient())
   n12 = GetOfferInfo()
   n13 = Find(
    TargetMap:Map of Strings to Texts=Get Banners!(self:Catalog Messaging Object Reference=Get CatalogMessaging() -> (CatalogMessaging:Catalog Messaging Object Reference)) -> (Banners:Map of Strings to Texts)
    Key=GetMetaValue!(self:Fort Store Front Offer Info Object Reference=@n12, MetaTag:String=ViolatorTag) -> (ReturnValue:String)
   ) -> (Value:Text)
   > SetupSpecialOfferBanner!(
    execute
    self:MTXOffer Special Banner Object Reference=Get AthenaDirectAcquisitionOffer_SpecialBanner()
    bHasSale:bool=@n13.ReturnValue
    SaleText:Text=@n13
    BannerOverrideTag:String=GetBannerOverrideTag!(self:Fort Store Front Offer Info Object Reference=@n12) -> (ReturnValue:String)
    bHasBattlePassStars:bool=false
    iNumOfBattlePassStars:int=0
   ) -> (then)
   > SetTIleSize(DisplaySize=GetOfferDisplaySize(self=GetOfferInfo()) -> (ReturnValue=Small))
   SetVisibility(execute=@n4.else, self=Get TextOfferName(), InVisibility=SelfHitTestInvisible)
   > SetActiveWidget(self=Get RMTTypeSwitcher(), Widget=Get RMTBundle() -> (RMTBundle:Vertical Box Object Reference))
   > SetVisibility(self=Get AthenaDirectAcquisitionOffer_SpecialBanner(), InVisibility=SelfHitTestInvisible)
   n14 = Get OfferInfo()
   > n15 = Branch(Condition="Not Equal (Name)"!(A:Name=GetStorefrontName!(self:Fort Store Front Offer Info Object Reference=@n14) -> (ReturnValue:Name=None), B:Name=BPGifting) -> (ReturnValue:bool) {is=title K2Node_CallFunction})
   > n16 = "SetText (Text)"(self=Get TextOfferName(), InText=GetName!(self:Fort Store Front Offer Info Object Reference=Get OfferInfo()) -> (ReturnValue:Text))
   n17 = Get TotalQuantity() -> (TotalQuantity=0)
   SetVisibility(
    execute=[@n19.then, @n16.then]
    self=Get RMTVBucksHB() -> (RMTVBucksHB:Horizontal Box Object Reference)
    InVisibility=Select(
     Option 0:ESlateVisibility Enum=Collapsed
     Option 1:ESlateVisibility Enum=SelfHitTestInvisible
     Index="integer > integer"(A=@n17, B=0) {is=title K2Node_CallFunction}
    ) -> (ReturnValue:ESlateVisibility Enum)
   )
   > n18 = "SetText (Text)"(self=Get BundleTotalQuantity(), InText="To Text (Integer)"(Value=Get TotalQuantity() -> (TotalQuantity=0)))
   n19 = "SetText (Text)"(
    execute=@n15.else
    self=Get TextOfferName()
    InText=Get ShortDescription!(self:Fort Mtx Offer Data Object Reference=Get DisplayAsset()) -> (ShortDescription:Text)
   )
  ]}
 InitialReset {nodes [
   InitialReset() {is=title K2Node_FunctionEntry}
   > PlayAnimation(InAnimation=Get Intro(), PlayMode=Reverse, PlaybackSpeed=10)
  ]}
 SetOverrideBannerText {nodes [
   n1 = SetOverrideBannerText!() -> (then, InOverrideBannerText:Text) {is=title K2Node_FunctionEntry}
   > Set OverrideBannerText(OverrideBannerText:Text=@n1) -> (Output_Get:Text)
   > UpdateBannerText()
  ]}
 SetTIleSize {nodes [
   n1 = SetTIleSize!() -> (then, DisplaySize:EFortMtxOfferDisplaySize Enum=Medium) {is=title K2Node_FunctionEntry}
   > n2 = Switch on EFortMtxOfferDisplaySize!(execute, Selection:EFortMtxOfferDisplaySize Enum=@n1) -> (Small:Exec, Medium:Exec, Large:Exec) {is=title K2Node_SwitchEnum}
   n3 = Get TileSIze() -> (TileSIze:Size Box Object Reference)
   SetHeightOverride(execute=@n2.Large, self=@n3, InHeightOverride=600)
   > SetWidthOverride(self=@n3, InWidthOverride=690)
   n4 = Get HoverBorder() -> (HoverBorder:Image Object Reference)
   > SetBrushFromMaterial(self=@n4, Material=/FortniteCoreUI/Frontend/StoreMain/Art/M_UI_StoreTile_Border_Large)
   > SetStretch!(
    execute
    self:Scale Box Object Reference=Get ScaleBoxKeyArt() -> (ScaleBoxKeyArt:Scale Box Object Reference)
    InStretch:EStretch Enum=ScaleToFit
   ) -> (then)
   > n5 = GetDynamicMaterial(self:Border Object Reference=Get Border_Footer() -> (Border_Footer:Border Object Reference))
   > SetScalarParameterValue(self=@n5, ParameterName=TopCenterAngle, Value=-18)
   > SetScalarParameterValue(self=@n5, ParameterName=TopCenterEdgeSharpness, Value=32)
   n6 = Get TileSIze()
   SetHeightOverride(execute=@n2.Medium, self=@n6, InHeightOverride=600)
   > SetWidthOverride(self=@n6, InWidthOverride=345)
   > SetBrushFromMaterial(self=@n4, Material=/FortniteCoreUI/Frontend/StoreMain/Art/M_UI_StoreTile_Border_Medium)
   n7 = Get TileSIze()
   SetHeightOverride(execute=@n2.Small, self=@n7, InHeightOverride=296)
   > SetWidthOverride(self=@n7, InWidthOverride=256)
   > SetBrushFromMaterial(self=@n4, Material=/FortniteCoreUI/Frontend/StoreMain/Art/M_UI_StoreTile_Border_Small)
   n8 = Get BonusText() -> (BonusText:Common Text Object Reference)
   n9 = Break Slate Font Info!(SlateFontInfo:"Slate Font Info Structure (by ref)"=Get Font!(self:Text Object Reference=@n8) -> (Font:Slate Font Info Structure)) -> (
    FontObject:Object Reference
    FontMaterial:Object Reference
    OutlineSettings:Font Outline Settings Structure
    TypefaceFontName:Name
    Size:float=~
    LetterSpacing:int
    SkewAmount:float=~
    bForceMonospaced:bool=~
    bMaterialIsStencil:bool=~
    MonospacedWidth:float=~
   ) {is=title K2Node_BreakStruct}
   n10 = MakeLiteralInt!(Value:int=18) -> (ReturnValue:int=0)
   > SetFont(
    self=@n8
    InFontInfo=MakeSlateFontInfo!(
     FontObject:Object Reference=@n9
     FontMaterial:Object Reference=@n9.FontMaterial
     OutlineSettings:Font Outline Settings Structure=@n9.OutlineSettings
     TypefaceFontName:Name=@n9.TypefaceFontName
     Size:float="To Float (Integer)"(InInt=@n10)
     LetterSpacing:int=0
     SkewAmount:float=0
     bForceMonospaced:bool=false
     bMaterialIsStencil:bool=false
     MonospacedWidth:float=1
    ) -> (ReturnValue:Slate Font Info Structure)
   )
  ]}
 SetTextOutlineColor {nodes [
   n1 = SetTextOutlineColor!() -> (then, Text:Text Object Reference, Gradient:Fort Mtx Gradient Structure) {is=title K2Node_FunctionEntry}
   > n2 = GetDynamicOutlineMaterial!(execute, self:Text Object Reference=@n1) -> (then, ReturnValue:Material Instance Dynamic Object Reference)
   > SetVectorParameterValue(self=@n2, ParameterName=Primary, Value=Break Fort Mtx Gradient(FortMtxGradient=@n1.Gradient))
  ]}
 TriggerIntro {nodes [
   TriggerIntro() {is=title K2Node_FunctionEntry}
   > PlayAnimation(InAnimation=Get Intro())
   > PlaySound2D(Sound=Get MTXOfferTileIntroWhoosh() -> (MTXOfferTileIntroWhoosh:Sound Base Object Reference))
  ]}
 TriggerReset {nodes [
   TriggerReset() {is=title K2Node_FunctionEntry}
   > StopAnimation(InAnimation=Get Intro())
   > InitialReset() {is=title K2Node_CallFunction}
  ]}
 UpdateBannerText {nodes [
   UpdateBannerText() {is=title K2Node_FunctionEntry}
   > Set OfferInfo(OfferInfo=GetOfferInfo())
   n1 = Get OfferInfo()
   > n2 = Branch(Condition="Equal (Enum)"(A=GetFortStoreOfferType(self=@n1) -> (ReturnValue=FoundersPack), B=CurrencyPack))
   > Set BonusQuantity(BonusQuantity:int=GetBonusQuantity(self=Get OfferInfo()) -> (ReturnValue=0)) -> (Output_Get:int=0)
   n3 = NOT Boolean(A=TextIsEmpty(InText=Get OverrideBannerText() -> (OverrideBannerText:Text)))
   n4 = Find(TargetMap=Get Banners(self=Get CatalogMessaging()), Key=GetBannerOverrideTag(self=Get OfferInfo()))
   n5 = AND Boolean(
    A="integer > integer"(A=Get BonusQuantity() -> (BonusQuantity:int=0), B=0)
    B=Get bEnableCurrencyBonusBanner!(self:Fort Runtime Options Object Reference=GetRuntimeOptions()) -> (bEnableCurrencyBonusBanner:bool)
   )
   > SetVisibility(
    self=Get BonusTag() -> (BonusTag:Border Object Reference)
    InVisibility=Select(Option 0=Collapsed, Option 1=SelfHitTestInvisible, Index=OR Boolean!(A=@n4.ReturnValue, B=@n3, C:bool=@n5) -> (ReturnValue))
   )
   n6 = Get OverrideBannerText()
   n7 = GetBonusText!(OfferInfo:Fort Store Front Offer Info Object Reference=Get OfferInfo()) -> (BonusText:Text) {is=title K2Node_CallFunction}
   > "SetText (Text)"(
    self=Get BonusText()
    InText=Select(
     Option 0:Text=Get OverrideBannerText()
     Option 1:Text=Select(Option 0=@n7, Option 1=@n4, Index=@n4.ReturnValue)
     Index=TextIsEmpty(InText=@n6)
    ) -> (ReturnValue:Text)
   )
   SetVisibility(execute=@n2.else, self=Get BonusTag(), InVisibility=Collapsed)
  ]}
]
variables [
 AboutToShowDetails {}
 bCanSetPrice {type=boolean, default=true}
 CatalogMessaging {type=CatalogMessaging, default=/Game/Athena/UI/Frontend/CatalogMessages}
 MTXOfferTileIntroWhoosh {type=SoundBase, default=null}
 OverrideBannerText {type=string, default=""}
]
widgets {
 info {inheritedWidgetCount=2, parentClass=/Script/FortniteUI.FortRealMoneyOffer, rootWidgetClass=&Overlay, widgetCount=40}
 widgets [
  OverlayAll {widgetClassPath=&Overlay, widget=OverlayAll}
  BorderAll {
   widgetClassPath=&Border
   bIsVariable=true
   parent=OverlayAll
   widget=BorderAll
   properties {
    background {
     bIsDynamicallyLoaded=false
     drawAs=NoDrawType
     imageType=NoImage
     mirroring=NoMirror
     resourceName=None
     resourceObject=null
     tiling=NoTile
     imageSize=<32 32>
     margin=<ltrb 0 0 0 0>
     outlineSettings {
      bUseBrushTransparency=false
      roundingType=HalfHeightRadius
      width=0
      color {colorUseRule=UseColor_Specified, specifiedColor=<rgba 0 0 0 0>}
      cornerRadii=<0 0 0 0>
     }
     tintColor {colorUseRule=UseColor_Specified, specifiedColor=<rgba 1 1 1 1>}
     uVRegion {bIsValid=false, max=<0 0>, min=<0 0>}
    }
    padding=<ltrb 0 0 0 0>
   }
   slot {class=OverlaySlot}
  }
  TileSIze {
   widgetClassPath=&SizeBox
   bIsVariable=true
   parent=BorderAll
   widget=TileSIze
   properties {bOverride_HeightOverride=true, bOverride_WidthOverride=true, clipping=ClipToBounds, heightOverride=608, widthOverride=345}
   slot {class=BorderSlot, properties {padding=<ltrb 0 0 0 0>}}
  }
  Overlay_0 {
   widgetClassPath=&Overlay
   parent=TileSIze
   widget=Overlay_0
   properties {clipping=ClipToBoundsWithoutIntersecting}
   slot {class=SizeBoxSlot}
  }
  BaseColor {
   widgetClassPath=&Image
   bIsVariable=true
   parent=Overlay_0
   widget=BaseColor
   properties {visibility=SelfHitTestInvisible, colorAndOpacity=<rgba 0.5149 0.5333 0.6445 1>}
   slot {class=OverlaySlot, properties {horizontalAlignment=HAlign_Fill, verticalAlignment=VAlign_Fill}}
  }
  OverlayGradient {
   widgetClassPath=&Image
   bIsVariable=true
   parent=Overlay_0
   widget=OverlayGradient
   properties {
    visibility=SelfHitTestInvisible
    brush {
     bIsDynamicallyLoaded=false
     drawAs=Image
     imageType=NoImage
     mirroring=NoMirror
     resourceName=None
     resourceObject=&M-Store-GradientOverlay
     tiling=NoTile
     imageSize=<32 32>
     margin=<ltrb 0 0 0 0>
     outlineSettings {
      bUseBrushTransparency=false
      roundingType=HalfHeightRadius
      width=0
      color {colorUseRule=UseColor_Specified, specifiedColor=<rgba 0 0 0 0>}
      cornerRadii=<0 0 0 0>
     }
     tintColor {colorUseRule=UseColor_Specified, specifiedColor=<rgba 1 1 1 1>}
     uVRegion {bIsValid=false, max=<0 0>, min=<0 0>}
    }
   }
   slot {class=OverlaySlot, properties {horizontalAlignment=HAlign_Fill, verticalAlignment=VAlign_Fill, padding=<ltrb 0 0 0 56>}}
  }
  HoverBorder {
   widgetClassPath=&Image
   bIsVariable=true
   parent=Overlay_0
   widget=HoverBorder
   properties {brush {
     bIsDynamicallyLoaded=false
     drawAs=Image
     imageType=NoImage
     mirroring=NoMirror
     resourceName=None
     resourceObject=&M_UI_StoreTile_Border_Large
     tiling=NoTile
     imageSize=<32 32>
     margin=<ltrb 0 0 0 0>
     outlineSettings {
      bUseBrushTransparency=false
      roundingType=HalfHeightRadius
      width=0
      color {colorUseRule=UseColor_Specified, specifiedColor=<rgba 0 0 0 0>}
      cornerRadii=<0 0 0 0>
     }
     tintColor {colorUseRule=UseColor_Specified, specifiedColor=<rgba 1 1 1 1>}
     uVRegion {bIsValid=false, max=<0 0>, min=<0 0>}
    }}
   slot {class=OverlaySlot, properties {horizontalAlignment=HAlign_Fill, verticalAlignment=VAlign_Fill}}
  }
  OverlayKeyImage {
   widgetClassPath=&Overlay
   parent=Overlay_0
   widget=OverlayKeyImage
   properties {renderTransform {angle=0, scale=<1 1>, shear=<0 0>, translation=<0 -24>}}
   slot {class=OverlaySlot, properties {horizontalAlignment=HAlign_Fill, verticalAlignment=VAlign_Fill}}
  }
  ScaleBoxKeyArt {
   widgetClassPath=&ScaleBox
   bIsVariable=true
   parent=OverlayKeyImage
   widget=ScaleBoxKeyArt
   properties {clipping=ClipToBoundsWithoutIntersecting, stretch=ScaleToFitX}
   slot {class=OverlaySlot, properties {horizontalAlignment=HAlign_Fill, verticalAlignment=VAlign_Fill}}
  }
  KeyArt {
   widgetClassPath=&Image
   bIsVariable=true
   parent=ScaleBoxKeyArt
   widget=KeyArt
   properties {
    brush {
     bIsDynamicallyLoaded=false
     drawAs=Image
     imageType=NoImage
     mirroring=NoMirror
     resourceName=None
     resourceObject=/Game/UI/Foundation/Textures/Founders/T_VBucks_Tile_13500_New
     tiling=NoTile
     imageSize=<512 512>
     margin=<ltrb 0 0 0 0>
     outlineSettings {
      bUseBrushTransparency=false
      roundingType=HalfHeightRadius
      width=0
      color {colorUseRule=UseColor_Specified, specifiedColor=<rgba 0 0 0 0>}
      cornerRadii=<0 0 0 0>
     }
     tintColor {colorUseRule=UseColor_Specified, specifiedColor=<rgba 1 1 1 1>}
     uVRegion {bIsValid=false, max=<0 0>, min=<0 0>}
    }
    renderTransform {angle=0, scale=<1 1>, shear=<0 0>, translation=<0 24>}
   }
   slot {class=ScaleBoxSlot}
  }
  OverlayBottomText {
   widgetClassPath=&Overlay
   parent=Overlay_0
   widget=OverlayBottomText
   slot {class=OverlaySlot, properties {horizontalAlignment=HAlign_Fill, verticalAlignment=VAlign_Bottom}}
  }
  RMTTypeSwitcher {
   widgetClassPath=&WidgetSwitcher
   bIsVariable=true
   parent=OverlayBottomText
   widget=RMTTypeSwitcher
   slot {class=OverlaySlot, properties {horizontalAlignment=HAlign_Fill, verticalAlignment=VAlign_Fill, padding=<ltrb 0 0 0 64>}}
  }
  RMTCurrency {
   widgetClassPath=&VerticalBox
   bIsVariable=true
   parent=RMTTypeSwitcher
   widget=RMTCurrency
   slot {class=WidgetSwitcherSlot, properties {padding=<ltrb 8 8 8 8>}}
  }
  CurrencyTotalQuantity {
   widgetClassPath=&CommonTextBlock
   bIsVariable=true
   parent=RMTCurrency
   widget=CurrencyTotalQuantity
   properties {
    justification=Center
    text=9999
    font {
     bForceMonospaced=false
     bMaterialIsStencil=false
     fontMaterial=null
     fontObject=&Burbank
     letterSpacing=0
     monospacedWidth=1
     size=32
     skewAmount=0
     typefaceFontName=BurbankRegular-Black
     outlineSettings {
      bApplyOutlineToDropShadows=false
      bMiteredCorners=false
      bSeparateFillAlpha=false
      outlineMaterial=/FortniteCoreUI/Athena/UI/Challenges/Art/M_UI_SimpleBurn_DynamicColor
      outlineSize=3
      outlineColor=<rgba 1 1 1 1>
     }
    }
    renderTransform {angle=0, scale=<1 1>, shear=<-8 0>, translation=<0 0>}
    shadowOffset=<0 0>
   }
   slot {class=VerticalBoxSlot, properties {verticalAlignment=VAlign_Bottom, padding=<ltrb 0 -8 0 0>}}
  }
  VBucksText {
   widgetClassPath=&CommonTextBlock
   bIsVariable=true
   parent=RMTCurrency
   widget=VBucksText
   properties {
    justification=Center
    text=V-Bucks
    textTransformPolicy=ToUpper
    font {
     bForceMonospaced=false
     bMaterialIsStencil=false
     fontMaterial=null
     fontObject=&Burbank
     letterSpacing=0
     monospacedWidth=1
     size=24
     skewAmount=0
     typefaceFontName=BurbankRegular-Black
     outlineSettings {
      bApplyOutlineToDropShadows=false
      bMiteredCorners=false
      bSeparateFillAlpha=false
      outlineMaterial=&M_UI_SimpleBurn_DynamicColor
      outlineSize=3
      outlineColor=<rgba 1 1 1 1>
     }
    }
    renderTransform {angle=0, scale=<1 1>, shear=<-8 0>, translation=<0 -4>}
    shadowOffset=<0 0>
   }
   slot {class=VerticalBoxSlot, properties {verticalAlignment=VAlign_Top, padding=<ltrb 0 -4 0 0>}}
  }
  RMTBundle {
   widgetClassPath=&VerticalBox
   bIsVariable=true
   parent=RMTTypeSwitcher
   widget=RMTBundle
   slot {class=WidgetSwitcherSlot, properties {padding=<ltrb 8 8 8 8>}}
  }
  TextOfferName {
   widgetClassPath=&CommonTextBlock
   bIsVariable=true
   parent=RMTBundle
   widget=TextOfferName
   properties {
    autoWrapText=true
    justification=Center
    lineHeightPercentage=0.75
    text=SUPER LONG OFFER NAME
    textTransformPolicy=ToUpper
    font {
     bForceMonospaced=false
     bMaterialIsStencil=false
     fontMaterial=null
     fontObject=&Burbank
     letterSpacing=0
     monospacedWidth=1
     size=32
     skewAmount=0
     typefaceFontName=Burbank-Black
     outlineSettings {
      bApplyOutlineToDropShadows=false
      bMiteredCorners=false
      bSeparateFillAlpha=false
      outlineMaterial=&M_UI_SimpleBurn_DynamicColor
      outlineSize=3
      outlineColor=<rgba 1 1 1 1>
     }
    }
    shadowOffset=<0 0>
   }
   slot {class=VerticalBoxSlot, properties {size {sizeRule=Fill, value=1}}}
  }
  RMTVBucksHB {
   widgetClassPath=&HorizontalBox
   bIsVariable=true
   parent=RMTBundle
   widget=RMTVBucksHB
   slot {class=VerticalBoxSlot, properties {horizontalAlignment=HAlign_Center, padding=<ltrb 0 8 0 0>}}
  }
  Plus {
   widgetClassPath=&CommonTextBlock
   bIsVariable=true
   parent=RMTVBucksHB
   widget=Plus
   properties {
    text=+
    font {
     bForceMonospaced=false
     bMaterialIsStencil=false
     fontMaterial=null
     fontObject=&Burbank
     letterSpacing=0
     monospacedWidth=1
     size=48
     skewAmount=0
     typefaceFontName=Burbank-Black
     outlineSettings {
      bApplyOutlineToDropShadows=false
      bMiteredCorners=false
      bSeparateFillAlpha=false
      outlineMaterial=&M_UI_SimpleBurn_DynamicColor
      outlineSize=2
      outlineColor=<rgba 1 1 1 1>
     }
    }
    renderTransform {angle=0, scale=<1 1>, shear=<-8 0>, translation=<0 0>}
    shadowOffset=<0 0>
   }
   slot {class=HorizontalBoxSlot, properties {verticalAlignment=VAlign_Center, padding=<ltrb 0 -32 0 -32>}}
  }
  BundleTotalQuantity {
   widgetClassPath=&CommonTextBlock
   bIsVariable=true
   parent=RMTVBucksHB
   widget=BundleTotalQuantity
   properties {
    justification=Center
    text=9999
    font {
     bForceMonospaced=false
     bMaterialIsStencil=false
     fontMaterial=null
     fontObject=&Burbank
     letterSpacing=0
     monospacedWidth=1
     size=24
     skewAmount=0
     typefaceFontName=BurbankRegular-Black
     outlineSettings {
      bApplyOutlineToDropShadows=false
      bMiteredCorners=false
      bSeparateFillAlpha=false
      outlineMaterial=&M_UI_SimpleBurn_DynamicColor
      outlineSize=2
      outlineColor=<rgba 1 1 1 1>
     }
    }
    renderTransform {angle=0, scale=<1 1>, shear=<-8 0>, translation=<0 0>}
    shadowOffset=<0 0>
   }
   slot {class=HorizontalBoxSlot, properties {verticalAlignment=VAlign_Center, padding=<ltrb 8 0 8 0>, size {sizeRule=Fill, value=1}}}
  }
  BundleVBucksText {
   widgetClassPath=&CommonTextBlock
   bIsVariable=true
   parent=RMTVBucksHB
   widget=BundleVBucksText
   properties {
    justification=Center
    text=V-Bucks
    textTransformPolicy=ToUpper
    font {
     bForceMonospaced=false
     bMaterialIsStencil=false
     fontMaterial=null
     fontObject=&Burbank
     letterSpacing=0
     monospacedWidth=1
     size=18
     skewAmount=0
     typefaceFontName=BurbankRegular-Black
     outlineSettings {
      bApplyOutlineToDropShadows=false
      bMiteredCorners=false
      bSeparateFillAlpha=false
      outlineMaterial=&M_UI_SimpleBurn_DynamicColor
      outlineSize=2
      outlineColor=<rgba 1 1 1 1>
     }
    }
    renderTransform {angle=0, scale=<1 1>, shear=<-8 0>, translation=<0 -4>}
    shadowOffset=<0 0>
   }
   slot {class=HorizontalBoxSlot, properties {verticalAlignment=VAlign_Center, padding=<ltrb 0 2 0 0>}}
  }
  SizeBox_0 {
   widgetClassPath=&SizeBox
   parent=OverlayBottomText
   widget=SizeBox_0
   properties {bOverride_MinDesiredHeight=true, heightOverride=112, minDesiredHeight=72}
   slot {class=OverlaySlot, properties {horizontalAlignment=HAlign_Fill, verticalAlignment=VAlign_Bottom}}
  }
  Border_Footer {
   widgetClassPath=&Border
   bIsVariable=true
   parent=SizeBox_0
   widget=Border_Footer
   properties {
    horizontalAlignment=HAlign_Center
    verticalAlignment=VAlign_Center
    background {
     bIsDynamicallyLoaded=false
     drawAs=Image
     imageType=NoImage
     mirroring=NoMirror
     resourceName=None
     resourceObject=/FortniteCoreUI/UI/Foundation/Materials/M_BoltBorder_Top_White
     tiling=NoTile
     imageSize=<32 32>
     margin=<ltrb 0 0 0 0>
     outlineSettings {
      bUseBrushTransparency=false
      roundingType=HalfHeightRadius
      width=0
      color {colorUseRule=UseColor_Specified, specifiedColor=<rgba 0 0 0 0>}
      cornerRadii=<0 0 0 0>
     }
     tintColor {colorUseRule=UseColor_Specified, specifiedColor=<rgba 1 1 1 1>}
     uVRegion {bIsValid=false, max=<0 0>, min=<0 0>}
    }
    brushColor=<rgba 1 0.9216 0 1>
    padding=<ltrb 0 4 0 0>
   }
   slot {class=SizeBoxSlot}
  }
  VerticalBox_1 {
   widgetClassPath=&VerticalBox
   parent=Border_Footer
   widget=VerticalBox_1
   slot {class=BorderSlot, properties {horizontalAlignment=HAlign_Center, verticalAlignment=VAlign_Center, padding=<ltrb 0 4 0 0>}}
  }
  Switcher_Price {
   widgetClassPath=&CommonWidgetSwitcherLegacy
   bIsVariable=true
   parent=VerticalBox_1
   widget=Switcher_Price
   properties {bOutroPanelBelow=false, bWidgetActivationEnabled=false, transitionDuration=0}
   slot {class=VerticalBoxSlot}
  }
  HBox_Price {
   widgetClassPath=&HorizontalBox
   bIsVariable=true
   parent=Switcher_Price
   widget=HBox_Price
   properties {flowDirectionPreference=Culture}
   slot {class=WidgetSwitcherSlot, properties {horizontalAlignment=HAlign_Center, verticalAlignment=VAlign_Bottom}}
  }
  OverlaySalePrice {
   widgetClassPath=&Overlay
   bIsVariable=true
   parent=HBox_Price
   widget=OverlaySalePrice
   slot {class=HorizontalBoxSlot, properties {verticalAlignment=VAlign_Center, size {sizeRule=Fill, value=1}}}
  }
  TextOriginalPrice {
   widgetClassPath=&CommonTextBlock
   bIsVariable=true
   parent=OverlaySalePrice
   widget=TextOriginalPrice
   properties {
    justification=Center
    text=$888.88
    colorAndOpacity {colorUseRule=UseColor_Specified, specifiedColor=<rgba 0.0075 0.0091 0.0242 0.5>}
    font {
     bForceMonospaced=false
     bMaterialIsStencil=false
     fontMaterial=null
     fontObject=&Burbank
     letterSpacing=0
     monospacedWidth=1
     size=18
     skewAmount=0
     typefaceFontName=BurbankRegular-Black
     outlineSettings {
      bApplyOutlineToDropShadows=false
      bMiteredCorners=false
      bSeparateFillAlpha=false
      outlineMaterial=null
      outlineSize=0
      outlineColor=<rgba 0 0 0 1>
     }
    }
    renderTransform {angle=0, scale=<1 1>, shear=<-8 0>, translation=<0 0>}
    shadowOffset=<0 0>
   }
   slot {class=OverlaySlot, properties {horizontalAlignment=HAlign_Center, verticalAlignment=VAlign_Center, padding=<ltrb 16 0 16 0>}}
  }
  Image_1 {
   widgetClassPath=&Image
   parent=OverlaySalePrice
   widget=Image_1
   properties {
    brush {
     bIsDynamicallyLoaded=false
     drawAs=Image
     imageType=NoImage
     mirroring=NoMirror
     resourceName=None
     resourceObject=null
     tiling=NoTile
     imageSize=<32 4>
     margin=<ltrb 0 0 0 0>
     outlineSettings {
      bUseBrushTransparency=false
      roundingType=HalfHeightRadius
      width=0
      color {colorUseRule=UseColor_Specified, specifiedColor=<rgba 0 0 0 0>}
      cornerRadii=<0 0 0 0>
     }
     tintColor {colorUseRule=UseColor_Specified, specifiedColor=<rgba 1 1 1 1>}
     uVRegion {bIsValid=false, max=<0 0>, min=<0 0>}
    }
    colorAndOpacity=<rgba 0.0075 0.0091 0.0242 1>
    renderTransform {angle=-5, scale=<1 1>, shear=<0 0>, translation=<0 0>}
   }
   slot {class=OverlaySlot, properties {horizontalAlignment=HAlign_Fill, verticalAlignment=VAlign_Center, padding=<ltrb 8 0 8 0>}}
  }
  TextOfferDisplayPrice {
   widgetClassPath=&CommonTextBlock
   bIsVariable=true
   parent=HBox_Price
   widget=TextOfferDisplayPrice
   properties {
    justification=Center
    text=$888.88
    colorAndOpacity {colorUseRule=UseColor_Specified, specifiedColor=<rgba 0.0075 0.0091 0.0242 1>}
    font {
     bForceMonospaced=false
     bMaterialIsStencil=false
     fontMaterial=null
     fontObject=&Burbank
     letterSpacing=0
     monospacedWidth=1
     size=24
     skewAmount=0
     typefaceFontName=BurbankRegular-Black
     outlineSettings {
      bApplyOutlineToDropShadows=false
      bMiteredCorners=false
      bSeparateFillAlpha=false
      outlineMaterial=null
      outlineSize=0
      outlineColor=<rgba 0 0 0 1>
     }
    }
    renderTransform {angle=0, scale=<1 1>, shear=<-8 0>, translation=<0 0>}
    shadowOffset=<0 0>
   }
   slot {class=HorizontalBoxSlot, properties {verticalAlignment=VAlign_Center}}
  }
  WBP_Price_PlatformStoreIcon {
   widgetClassPath=/FortUILibrary/Systems/ItemShopTile/WBP_Price_PlatformStoreIcon
   bIsVariable=true
   parent=HBox_Price
   widget=WBP_Price_PlatformStoreIcon
   properties {imageSize=<40 40>}
   slot {class=HorizontalBoxSlot, properties {horizontalAlignment=HAlign_Center, verticalAlignment=VAlign_Center}}
  }
  TextOfferPurchaseUnvailable {
   widgetClassPath=&CommonTextBlock
   bIsVariable=true
   parent=Switcher_Price
   widget=TextOfferPurchaseUnvailable
   properties {
    justification=Center
    text=Learn more
    textTransformPolicy=ToUpper
    colorAndOpacity {colorUseRule=UseColor_Specified, specifiedColor=<rgba 0.0075 0.0091 0.0242 1>}
    font {
     bForceMonospaced=false
     bMaterialIsStencil=false
     fontMaterial=null
     fontObject=&Burbank
     letterSpacing=0
     monospacedWidth=1
     size=24
     skewAmount=0
     typefaceFontName=BurbankRegular-Black
     outlineSettings {
      bApplyOutlineToDropShadows=false
      bMiteredCorners=false
      bSeparateFillAlpha=false
      outlineMaterial=null
      outlineSize=0
      outlineColor=<rgba 0 0 0 1>
     }
    }
    renderTransform {angle=0, scale=<1 1>, shear=<-8 0>, translation=<0 0>}
    shadowOffset=<0 0>
   }
   slot {class=WidgetSwitcherSlot, properties {verticalAlignment=VAlign_Center}}
  }
  AthenaDirectAcquisitionOffer_SpecialBanner {
   widgetClassPath=&/MTXOffer_SpecialBanner
   bIsVariable=true
   parent=OverlayAll
   widget=AthenaDirectAcquisitionOffer_SpecialBanner
   properties {flowDirectionPreference=Inherit, renderTransform {angle=0, scale=<1 1>, shear=<0 0>, translation=<-8 -8>}}
   slot {class=OverlaySlot}
  }
  BonusTag {
   widgetClassPath=&Border
   bIsVariable=true
   parent=OverlayAll
   widget=BonusTag
   properties {
    background {
     bIsDynamicallyLoaded=false
     drawAs=Box
     imageType=NoImage
     mirroring=NoMirror
     resourceName=None
     resourceObject=/FortniteCoreUI/Frontend/StoreMain/Art/M_UI_BonusTag
     tiling=NoTile
     imageSize=<32 32>
     margin=<ltrb 0.5 0 0.5 0>
     outlineSettings {
      bUseBrushTransparency=false
      roundingType=HalfHeightRadius
      width=0
      color {colorUseRule=UseColor_Specified, specifiedColor=<rgba 0 0 0 0>}
      cornerRadii=<0 0 0 0>
     }
     tintColor {colorUseRule=UseColor_Specified, specifiedColor=<rgba 1 1 1 1>}
     uVRegion {bIsValid=false, max=<0 0>, min=<0 0>}
    }
    padding=<ltrb 24 8 16 8>
    renderTransform {angle=0, scale=<1 1>, shear=<0 0>, translation=<8 -12>}
   }
   slot {class=OverlaySlot, properties {horizontalAlignment=HAlign_Right}}
  }
  SizeBox_1 {
   widgetClassPath=&SizeBox
   parent=BonusTag
   widget=SizeBox_1
   properties {bOverride_MaxDesiredWidth=true, maxDesiredWidth=256}
   slot {class=BorderSlot, properties {padding=<ltrb 24 8 16 8>}}
  }
  ScaleBox_0 {widgetClassPath=&ScaleBox, parent=SizeBox_1, widget=ScaleBox_0, slot {class=SizeBoxSlot}}
  BonusText {
   widgetClassPath=&CommonTextBlock
   bIsVariable=true
   parent=ScaleBox_0
   widget=BonusText
   properties {
    text="20% Extra!"
    textTransformPolicy=ToUpper
    font {
     bForceMonospaced=false
     bMaterialIsStencil=false
     fontMaterial=null
     fontObject=&BurbankSmall
     letterSpacing=0
     monospacedWidth=1
     size=24
     skewAmount=0
     typefaceFontName=Black
     outlineSettings {
      bApplyOutlineToDropShadows=false
      bMiteredCorners=false
      bSeparateFillAlpha=false
      outlineMaterial=null
      outlineSize=0
      outlineColor=<rgba 0 0 0 1>
     }
    }
    renderTransform {angle=0, scale=<1 1>, shear=<-8 0>, translation=<0 0>}
   }
   slot {class=ScaleBoxSlot}
  }
  SizeBox_2 {
   widgetClassPath=&SizeBox
   parent=OverlayAll
   widget=SizeBox_2
   properties {
    bOverride_HeightOverride=true
    bOverride_WidthOverride=true
    heightOverride=44
    widthOverride=44
    renderTransform {angle=0, scale=<1 1>, shear=<0 0>, translation=<0 28>}
   }
   slot {class=OverlaySlot, properties {horizontalAlignment=HAlign_Center, verticalAlignment=VAlign_Bottom}}
  }
  InputActionWidget {
   widgetClassPath=&CommonActionWidget
   bInherited=true
   bIsVariable=true
   parent=SizeBox_2
   widget=InputActionWidget
   properties {iconRimBrush {
     bIsDynamicallyLoaded=false
     drawAs=Image
     imageType=NoImage
     mirroring=NoMirror
     resourceName=None
     resourceObject=/FortUILibrary/Foundation/Buttons/Materials/M-Button-Console-YellowOutline
     tiling=NoTile
     imageSize=<44 44>
     margin=<ltrb 0 0 0 0>
     outlineSettings {
      bUseBrushTransparency=false
      roundingType=HalfHeightRadius
      width=0
      color {colorUseRule=UseColor_Specified, specifiedColor=<rgba 0 0 0 0>}
      cornerRadii=<0 0 0 0>
     }
     tintColor {colorUseRule=UseColor_Specified, specifiedColor=<rgba 1 1 1 1>}
     uVRegion {bIsValid=false, max=<0 0>, min=<0 0>}
    }}
   slot {class=SizeBoxSlot}
  }
  MainImageLoadGuard {widgetClassPath=/Script/CommonUI.CommonLoadGuard, bInherited=true, widget=null}
 ]
}
