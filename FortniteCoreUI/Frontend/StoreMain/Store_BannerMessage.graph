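# Store banner widget based on FortStoreBanner that shows a short promotional message, defaulting to "NEW LOWER PRICES!", in a yellow skewed frame.
# A material-drawn Border holds the inherited, sheared Text_BannerMessage, with a duplicate AnimatingBorder image overlaid on it.
# The Pulse function plays Ani_Intro, which slides and fades the banner in, then flashes the overlay border as it scales up and fades out.
# The event graph has no logic.
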WidgetBlueprint /FortniteCoreUI/Frontend/StoreMain/Store_BannerMessage
animations [Ani_Intro {
  binding_count=2
  length_seconds=0.6
  object=&self:Ani_Intro
  sequence {bindings [
    AnimatingBorder {
     id=#1
     possessed_class=Image
     tracks [
      {
       type=MovieScene2DTransformTrack
       sections [{
         type=MovieScene2DTransformSection
         channels [
          Translation.X {type=MovieSceneScriptingFloatChannel, const [0,1]}
          Translation.Y {type=MovieSceneScriptingFloatChannel, const [0,1]}
          Angle {type=MovieSceneScriptingFloatChannel, const [0,1]}
          Scale.X {type=MovieSceneScriptingFloatChannel, keys [[6,1],[12,1.5]]}
          Scale.Y {type=MovieSceneScriptingFloatChannel, keys [[6,1],[12,1.5]]}
          Shear.X {type=MovieSceneScriptingFloatChannel, const [0,1]}
          Shear.Y {type=MovieSceneScriptingFloatChannel, const [0,1]}
         ]
        }]
      }
      {
       type=MovieSceneFloatTrack
       display_name=Render Opacity
       sections [{type=MovieSceneFloatSection, channels [None {type=MovieSceneScriptingFloatChannel, keys [[0,0],[5,0],[6,0.5],[12,0]]}]}]
      }
     ]
    }
    SizeBox_0 {
     id=#2
     possessed_class=SizeBox
     tracks [
      {
       type=MovieScene2DTransformTrack
       sections [{
         type=MovieScene2DTransformSection
         channels [
          Translation.X {type=MovieSceneScriptingFloatChannel, keys [[0,120],[6,0]]}
          Translation.Y {type=MovieSceneScriptingFloatChannel, const [0,1]}
          Angle {type=MovieSceneScriptingFloatChannel, const [0,1]}
          Scale.X {type=MovieSceneScriptingFloatChannel, const [1,1]}
          Scale.Y {type=MovieSceneScriptingFloatChannel, const [1,1]}
          Shear.X {type=MovieSceneScriptingFloatChannel, const [0,1]}
          Shear.Y {type=MovieSceneScriptingFloatChannel, const [0,1]}
         ]
        }]
      }
      {
       type=MovieSceneFloatTrack
       display_name=Render Opacity
       sections [{type=MovieSceneFloatSection, channels [None {type=MovieSceneScriptingFloatChannel, keys [[0,0],[6,1]]}]}]
      }
     ]
    }
   ]}
 }]
graphs [
 EventGraph {nodes [
   Event PreConstruct()
   Event Construct()
   Event Tick()
  ]}
 Pulse {nodes [
   Pulse() {is=title K2Node_FunctionEntry}
   > PlayAnimation(InAnimation=Get Ani_Intro() -> (Ani_Intro:Widget Animation Object Reference))
  ]}
]
variables []
widgets {
 info {inheritedWidgetCount=1, parentClass=/Script/FortniteUI.FortStoreBanner, rootWidgetClass=&SizeBox, widgetCount=5}
 widgets [
  SizeBox_0 {widgetClassPath=&SizeBox, widget=SizeBox_0, properties {heightOverride=98.6233, widthOverride=618.8726}}
  Overlay_3 {widgetClassPath=&Overlay, parent=SizeBox_0, widget=Overlay_3, slot {class=SizeBoxSlot}}
  Border_0 {
   widgetClassPath=&Border
   parent=Overlay_3
   widget=Border_0
   properties {
    verticalAlignment=VAlign_Center
    background {
     bIsDynamicallyLoaded=false
     drawAs=Image
     imageType=NoImage
     mirroring=NoMirror
     resourceName=None
     resourceObject=&/M_UI_SkewBoxBorder_VBucks
     tiling=NoTile
     imageSize=<32 32>
     margin=<ltrb 0 0 0 0>
     outlineSettings {
      bUseBrushTransparency=false
      roundingType=HalfHeightRadius
      width=0
      color {colorUseRule=UseColor_Specified, specifiedColor=<rgba 0 0 0 0>}
      cornerRadii=<0 0 0 0>
     }
     tintColor {colorUseRule=UseColor_Specified, specifiedColor=<rgba 1 0.9216 0 1>}
     uVRegion {bIsValid=false, max=<0 0>, min=<0 0>}
    }
    padding=<ltrb 25 16 20 15>
   }
   slot {class=OverlaySlot}
  }
  Text_BannerMessage {
   widgetClassPath=&CommonTextBlock
   bInherited=true
   bIsVariable=true
   parent=Border_0
   widget=Text_BannerMessage
   properties {
    justification=Right
    text="NEW LOWER PRICES!"
    textTransformPolicy=ToUpper
    colorAndOpacity {colorUseRule=UseColor_Specified, specifiedColor=<rgba 1 0.9216 0 1>}
    font {
     bForceMonospaced=false
     bMaterialIsStencil=false
     fontMaterial=null
     fontObject=&BurbankSmall
     letterSpacing=0
     monospacedWidth=1
     size=20
     skewAmount=0
     typefaceFontName=Black
     outlineSettings {
      bApplyOutlineToDropShadows=false
      bMiteredCorners=false
      bSeparateFillAlpha=false
      outlineMaterial=null
      outlineSize=0
      outlineColor=<rgba 0 0 0 1>
     }
    }
    renderTransform {angle=0, scale=<1 1>, shear=<-8 0>, translation=<0 0>}
   }
   slot {class=BorderSlot, properties {verticalAlignment=VAlign_Center, padding=<ltrb 25 16 20 15>}}
  }
  AnimatingBorder {
   widgetClassPath=&Image
   bIsVariable=true
   parent=Overlay_3
   widget=AnimatingBorder
   properties {brush {
     bIsDynamicallyLoaded=false
     drawAs=Image
     imageType=NoImage
     mirroring=NoMirror
     resourceName=None
     resourceObject=&M_UI_SkewBoxBorder_VBucks
     tiling=NoTile
     imageSize=<32 32>
     margin=<ltrb 0 0 0 0>
     outlineSettings {
      bUseBrushTransparency=false
      roundingType=HalfHeightRadius
      width=0
      color {colorUseRule=UseColor_Specified, specifiedColor=<rgba 0 0 0 0>}
      cornerRadii=<0 0 0 0>
     }
     tintColor {colorUseRule=UseColor_Specified, specifiedColor=<rgba 1 0.9216 0 1>}
     uVRegion {bIsValid=false, max=<0 0>, min=<0 0>}
    }}
   slot {class=OverlaySlot, properties {horizontalAlignment=HAlign_Fill, verticalAlignment=VAlign_Fill}}
  }
 ]
}
