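# Modal widget on FortRedeemCodeBase where the player enters and redeems a friend code.
# Redeem rejects empty input or calls RedeemFriendCode and switches EntryProgressSwitcher to a throbber view.
# Success closes the modal and broadcasts the MCP context's real-money-purchase-complete delegate with a default item; failure targets the entry field and shows reason-specific error text, hidden on typing.
# Cancel and a touch-only close button pop the modal panel.
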WidgetBlueprint /FortniteCoreUI/UI/Foundation/Modals/RedeemCodeWindow/RedeemFriendCodeWindow
graphs [
 Close {nodes [
   Close() {is=title K2Node_FunctionEntry}
   > PopActivatablePanelInModalLayer!(execute, Panel:Common Activatable Panel Legacy Object Reference=Self-Reference()) -> (then)
  ]}
 EventGraph {nodes [
   "OnTextChanged (Editable Text) (EditableText_FriendCodeEntry)"!() -> (OutputDelegate, then, Text:"Text (by ref)") {is=title K2Node_ComponentBoundEvent}
   > Branch(Condition=IsVisible(self=Get ErrorText() -> (ErrorText:Common Text Object Reference)))
   > SetVisibility(self=Get ErrorText(), InVisibility=Hidden)
   n1 = "On Button Clicked (Button_Cancel)"!() -> (OutputDelegate, then, Button:Common Button Legacy Object Reference) {is=title K2Node_ComponentBoundEvent}
   Close(execute=[@n1.then, @n2.then]) {is=title K2Node_CallFunction}
   n2 = "On Button Clicked (CancelButton)"!() -> (OutputDelegate, then, Button:Common Button Legacy Object Reference) {is=title K2Node_ComponentBoundEvent}
   "On Button Clicked (RedeemButton)"!() -> (OutputDelegate, then, Button:Common Button Legacy Object Reference) {is=title K2Node_ComponentBoundEvent}
   n3 = "GetText (Editable Text)"!(self:Editable Text Object Reference=Get EditableText_FriendCodeEntry() -> (EditableText_FriendCodeEntry:Editable Text Object Reference)) -> (ReturnValue:Text)
   > n4 = Branch(Condition=TextIsEmpty(InText=@n3))
   n5 = Get ErrorText()
   > "SetText (Text)"(self=@n5, InText="LOCTEXT(\"Code cannot be empty!\")")
   > SetVisibility(self=@n5, InVisibility=SelfHitTestInvisible)
   RedeemFriendCode!(execute=@n4.else, self:Fort Redeem Code Base Object Reference, FriendCode:String="To String (Text)"(InText=@n3)) -> (then)
   > SetActiveWidget(
    self:Widget Switcher Object Reference=Get EntryProgressSwitcher() -> (EntryProgressSwitcher:Widget Switcher Object Reference)
    Widget=Get ProgressSizeBox() -> (ProgressSizeBox:Size Box Object Reference)
   )
   n6 = Event OnRedeemFriendCodeComplete!() -> (OutputDelegate, then, bSuccess:bool, FailureReason:ERedeemCodeFailureReason Enum=InvalidCode) {is=title K2Node_Event}
   > HandleRedeemCodeComplete!(execute, self, Success:bool=@n6.bSuccess, FailureReason:ERedeemCodeFailureReason Enum=@n6.FailureReason) -> (then)
   Event Construct()
   > SetActiveWidget(self=Get EntryProgressSwitcher(), Widget=Get EditableText_FriendCodeEntry())
  ]}
 HandleRedeemCodeComplete {nodes [
   n1 = HandleRedeemCodeComplete!() -> (then, Success:bool=~, FailureReason:ERedeemCodeFailureReason Enum=InvalidCode) {is=title K2Node_FunctionEntry}
   > n2 = Branch(Condition=@n1)
   > Close()
   > Call On Real Money Purchase Complete!(
    execute:Exec
    self:Fort Mcp Context Object Reference=Get FortMcpContext!() -> (ReturnValue:Fort Mcp Context Object Reference) {is=title K2Node_GetBlueprintContextEx}
    bSuccess:bool=true
    PurchasedItems:Array of Purchased Item Info Structures=Make Array("[0]":Purchased Item Info Structure) -> (Array:Array of Purchased Item Info Structures)
   ) -> (then:Exec) {is=title K2Node_CallDelegate}
   SetActiveWidget(execute=@n2.else, self=Get EntryProgressSwitcher(), Widget=Get EditableText_FriendCodeEntry())
   n3 = Get ErrorText()
   > "SetText (Text)"(
    self=@n3
    InText=Select!(
     InvalidCode:Text="LOCTEXT(\"Invalid code entered.\")"
     CodeAlreadyUsed:Text="LOCTEXT(\"This code is already in use.\")"
     AlreadyHasAccess:Text="LOCTEXT(\"You already have access.\")"
     FailedToGetItem:Text="LOCTEXT(\"Error trying to redeem code. Contact Epic Customer Support.\")"
     Unknown:Text="LOCTEXT(\"Error trying to redeem code. Contact Epic Customer Support.\")"
     Index:ERedeemCodeFailureReason Enum=Get FailureReason() -> (FailureReason:ERedeemCodeFailureReason Enum=InvalidCode)
    ) -> (ReturnValue:Text)
   )
   > SetVisibility(self=@n3, InVisibility=SelfHitTestInvisible)
  ]}
]
variables []
widgets {
 info {
  inheritedWidgetCount=1
  namedSlotCount=1
  parentClass=/Script/FortniteUI.FortRedeemCodeBase
  rootWidgetClass=&Overlay
  widgetCount=28
 }
 widgets [
  Overlay_182 {widgetClassPath=&Overlay, widget=Overlay_182}
  BackgroundBlur_0 {
   widgetClassPath=&BackgroundBlur
   parent=Overlay_182
   widget=BackgroundBlur_0
   slot {class=OverlaySlot, properties {horizontalAlignment=HAlign_Fill, verticalAlignment=VAlign_Fill}}
  }
  Lightbox {
   widgetClassPath=/Game/UI/Foundation/Lightbox/Lightbox
   bIsVariable=true
   parent=BackgroundBlur_0
   widget=Lightbox
   slot {class=BackgroundBlurSlot, properties {padding=<ltrb 0 0 0 0>}}
  }
  Overlay_183 {widgetClassPath=&Overlay, namedSlotHost=Lightbox, widget=Overlay_183}
  SizeBox_0 {
   widgetClassPath=&SizeBox
   parent=Overlay_183
   widget=SizeBox_0
   properties {bOverride_HeightOverride=true, bOverride_WidthOverride=true, heightOverride=350, widthOverride=1000}
   slot {class=OverlaySlot, properties {horizontalAlignment=HAlign_Center, verticalAlignment=VAlign_Center}}
  }
  Overlay_0 {widgetClassPath=&Overlay, parent=SizeBox_0, widget=Overlay_0, slot {class=SizeBoxSlot}}
  CommonBorder_0 {
   widgetClassPath=&CommonBorder
   parent=Overlay_0
   widget=CommonBorder_0
   properties {style=/FortUILibrary/Styles/Border/Border_Navy_VGrad.Border_Navy_VGrad_C, padding=<ltrb 0 0 0 0>}
   slot {class=OverlaySlot, properties {horizontalAlignment=HAlign_Fill, verticalAlignment=VAlign_Fill}}
  }
  VerticalBox_0 {
   widgetClassPath=&VerticalBox
   parent=CommonBorder_0
   widget=VerticalBox_0
   slot {class=BorderSlot, properties {padding=<ltrb 0 0 0 0>}}
  }
  CommonBorder_Modal {
   widgetClassPath=&CommonBorder
   parent=VerticalBox_0
   widget=CommonBorder_Modal
   properties {style=/FortUILibrary/Styles/Border/Border-ModalHeader.Border-ModalHeader_C}
   slot {class=VerticalBoxSlot}
  }
  HorizontalBox_1 {widgetClassPath=&HorizontalBox, parent=CommonBorder_Modal, widget=HorizontalBox_1, slot {class=BorderSlot}}
  Spacer_0 {widgetClassPath=&Spacer, parent=HorizontalBox_1, widget=Spacer_0, properties {size=<44 100>}, slot {class=HorizontalBoxSlot}}
  Title {
   widgetClassPath=&CommonTextBlock
   bIsVariable=true
   parent=HorizontalBox_1
   widget=Title
   properties {
    style=/FortUILibrary/Styles/Typography/Foundation/TextStyle-Header-L.TextStyle-Header-L_C
    text=REDEEM CODE
    font {
     bForceMonospaced=false
     bMaterialIsStencil=false
     fontMaterial=null
     fontObject=&Burbank
     letterSpacing=0
     monospacedWidth=1
     size=48
     skewAmount=0
     typefaceFontName=Burbank-Black
     outlineSettings {
      bApplyOutlineToDropShadows=false
      bMiteredCorners=false
      bSeparateFillAlpha=false
      outlineMaterial=null
      outlineSize=0
      outlineColor=<rgba 0 0 0 1>
     }
    }
    shadowOffset=<0 0>
   }
   slot {class=HorizontalBoxSlot, properties {horizontalAlignment=HAlign_Left, verticalAlignment=VAlign_Center, padding=<ltrb 8 0 0 0>}}
  }
  EntryProgressSwitcher {
   widgetClassPath=&WidgetSwitcher
   bIsVariable=true
   parent=VerticalBox_0
   widget=EntryProgressSwitcher
   properties {navigation=&self:WidgetTree.EntryProgressSwitcher.WidgetNavigation_0}
   slot {class=VerticalBoxSlot, properties {size {sizeRule=Fill, value=1}}}
  }
  EntryVBox {
   widgetClassPath=&VerticalBox
   bIsVariable=true
   parent=EntryProgressSwitcher
   widget=EntryVBox
   properties {navigation=&self:WidgetTree.EntryVBox.WidgetNavigation_1}
   slot {class=WidgetSwitcherSlot}
  }
  VerticalBox_1 {
   widgetClassPath=&VerticalBox
   parent=EntryVBox
   widget=VerticalBox_1
   properties {navigation=&self:WidgetTree.VerticalBox_1.WidgetNavigation_0}
   slot {class=VerticalBoxSlot, properties {verticalAlignment=VAlign_Center, padding=<ltrb 50 0 20 0>, size {sizeRule=Fill, value=1}}}
  }
  CommonTextBlock_0 {
   widgetClassPath=&CommonTextBlock
   parent=VerticalBox_1
   widget=CommonTextBlock_0
   properties {
    style=/FortUILibrary/Styles/Typography/Foundation/TextStyle-Base-S.TextStyle-Base-S_C
    text=Enter your code to redeem
    font {
     bForceMonospaced=false
     bMaterialIsStencil=false
     fontMaterial=null
     fontObject=&NotoSans
     letterSpacing=0
     monospacedWidth=1
     size=18
     skewAmount=0
     typefaceFontName=NotoSans-Regular
     outlineSettings {
      bApplyOutlineToDropShadows=false
      bMiteredCorners=false
      bSeparateFillAlpha=false
      outlineMaterial=null
      outlineSize=0
      outlineColor=<rgba 0 0 0 1>
     }
    }
    shadowOffset=<0 0>
   }
   slot {class=VerticalBoxSlot, properties {verticalAlignment=VAlign_Center}}
  }
  CommonBorder_1 {
   widgetClassPath=&CommonBorder
   parent=VerticalBox_1
   widget=CommonBorder_1
   properties {style=/FortUILibrary/Styles/Border/Border-SolidBG-DkBlue.Border-SolidBG-DkBlue_C, padding=<ltrb 0 0 0 0>}
   slot {class=VerticalBoxSlot, properties {verticalAlignment=VAlign_Center, padding=<ltrb 0 16 0 16>}}
  }
  EditableText_FriendCodeEntry {
   widgetClassPath=/Script/UMG.EditableText
   bInherited=true
   bIsVariable=true
   parent=CommonBorder_1
   widget=EditableText_FriendCodeEntry
   properties {
    hintText=Code
    widgetStyle {
     backgroundImageComposing {
      bIsDynamicallyLoaded=false
      drawAs=Border
      imageType=FullColor
      mirroring=NoMirror
      resourceName=../../../Engine/Content/Slate/Old/HyperlinkDotted.png
      resourceObject=null
      tiling=Both
      imageSize=<0 0>
      margin=<ltrb 0 0 0 0.1875>
      outlineSettings {
       bUseBrushTransparency=false
       roundingType=HalfHeightRadius
       width=0
       color {colorUseRule=UseColor_Specified, specifiedColor=<rgba 0 0 0 0>}
       cornerRadii=<0 0 0 0>
      }
      tintColor {colorUseRule=UseColor_Specified, specifiedColor=<rgba 1 1 1 1>}
      uVRegion {bIsValid=false, max=<0 0>, min=<0 0>}
     }
     backgroundImageSelected {
      bIsDynamicallyLoaded=false
      drawAs=Box
      imageType=FullColor
      mirroring=NoMirror
      resourceName=../../../Engine/Content/Slate/Common/EditableTextSelectionBackground.png
      resourceObject=null
      tiling=NoTile
      imageSize=<0 0>
      margin=<ltrb 0.25 0.25 0.25 0.25>
      outlineSettings {
       bUseBrushTransparency=false
       roundingType=HalfHeightRadius
       width=0
       color {colorUseRule=UseColor_Specified, specifiedColor=<rgba 0 0 0 0>}
       cornerRadii=<0 0 0 0>
      }
      tintColor {colorUseRule=UseColor_Specified, specifiedColor=<rgba 1 1 1 1>}
      uVRegion {bIsValid=false, max=<0 0>, min=<0 0>}
     }
     caretImage {
      bIsDynamicallyLoaded=false
      drawAs=Image
      imageType=FullColor
      mirroring=NoMirror
      resourceName=../../../Engine/Content/Slate/Old/White.png
      resourceObject=null
      tiling=NoTile
      imageSize=<16 16>
      margin=<ltrb 0 0 0 0>
      outlineSettings {
       bUseBrushTransparency=false
       roundingType=HalfHeightRadius
       width=0
       color {colorUseRule=UseColor_Specified, specifiedColor=<rgba 0 0 0 0>}
       cornerRadii=<0 0 0 0>
      }
      tintColor {colorUseRule=UseColor_Specified, specifiedColor=<rgba 1 1 1 1>}
      uVRegion {bIsValid=false, max=<0 0>, min=<0 0>}
     }
     colorAndOpacity {colorUseRule=UseColor_Specified, specifiedColor=<rgba 1 1 1 1>}
     font {
      bForceMonospaced=false
      bMaterialIsStencil=false
      fontMaterial=null
      fontObject=&NotoSans
      letterSpacing=0
      monospacedWidth=1
      size=16
      skewAmount=0
      typefaceFontName=NotoSans-Regular
      outlineSettings {
       bApplyOutlineToDropShadows=false
       bMiteredCorners=false
       bSeparateFillAlpha=false
       outlineMaterial=null
       outlineSize=0
       outlineColor=<rgba 0 0 0 1>
      }
     }
     hintColorAndOpacity {colorUseRule=UseColor_Foreground, specifiedColor=<rgba 1 0 1 1>}
     selectedBackgroundColor {colorUseRule=UseColor_Foreground, specifiedColor=<rgba 1 0 1 1>}
    }
   }
   slot {class=BorderSlot, properties {padding=<ltrb 0 0 0 0>}}
  }
  ErrorText {
   widgetClassPath=&CommonTextBlock
   bIsVariable=true
   parent=VerticalBox_1
   widget=ErrorText
   properties {
    style=/FortUILibrary/Styles/Typography/Foundation/TextStyle-Base-XS-B-S-Red.TextStyle-Base-XS-B-S-Red_C
    text=Text Block
    visibility=Hidden
    colorAndOpacity {colorUseRule=UseColor_Specified, specifiedColor=<rgba 1 0.0561 0.0648 1>}
    font {
     bForceMonospaced=false
     bMaterialIsStencil=false
     fontMaterial=null
     fontObject=&NotoSans
     letterSpacing=0
     monospacedWidth=1
     size=16
     skewAmount=0
     typefaceFontName=NotoSans-Bold
     outlineSettings {
      bApplyOutlineToDropShadows=true
      bMiteredCorners=false
      bSeparateFillAlpha=true
      outlineMaterial=&M-FontShadow
      outlineSize=2
      outlineColor=<rgba 0.085 0.0146 0.0177 0.5>
     }
    }
    shadowOffset=<0 0>
   }
   slot {class=VerticalBoxSlot}
  }
  HorizontalBox_0 {
   widgetClassPath=&HorizontalBox
   parent=EntryVBox
   widget=HorizontalBox_0
   slot {class=VerticalBoxSlot, properties {horizontalAlignment=HAlign_Right, padding=<ltrb 10 10 20 10>}}
  }
  RedeemButton {
   widgetClassPath=&IconTextButton
   bIsVariable=true
   parent=HorizontalBox_0
   widget=RedeemButton
   properties {
    buttonText=""
    onButtonBaseClicked=(&self:WidgetTree.RedeemButton.HandleOnButtonClicked)
    onButtonBaseDoubleClicked=(&self:WidgetTree.RedeemButton.HandleOnButtonDoubleClicked)
    onButtonBaseHovered=(&self:WidgetTree.RedeemButton.HandleOnButtonHovered)
    onButtonBaseUnhovered=(&self:WidgetTree.RedeemButton.HandleOnButtonUnhovered)
    onSelectedChangedBase=(&self:WidgetTree.RedeemButton.HandleOnSelectedChanged)
    triggeringInputAction {dataTable=&InputActionDataTable, rowName=Redeem}
   }
   slot {class=HorizontalBoxSlot, properties {padding=<ltrb 0 0 24 0>}}
  }
  CancelButton {
   widgetClassPath=&IconTextButton
   bIsVariable=true
   parent=HorizontalBox_0
   widget=CancelButton
   properties {
    buttonText=""
    onButtonBaseClicked=(&self:WidgetTree.CancelButton.HandleOnButtonClicked)
    onButtonBaseDoubleClicked=(&self:WidgetTree.CancelButton.HandleOnButtonDoubleClicked)
    onButtonBaseHovered=(&self:WidgetTree.CancelButton.HandleOnButtonHovered)
    onButtonBaseUnhovered=(&self:WidgetTree.CancelButton.HandleOnButtonUnhovered)
    onSelectedChangedBase=(&self:WidgetTree.CancelButton.HandleOnSelectedChanged)
    triggeringInputAction {dataTable=&InputActionDataTable, rowName=Cancel}
   }
   slot {class=HorizontalBoxSlot}
  }
  ProgressSizeBox {widgetClassPath=&SizeBox, bIsVariable=true, parent=EntryProgressSwitcher, widget=ProgressSizeBox, slot {class=WidgetSwitcherSlot}}
  ProgressVBox {
   widgetClassPath=&VerticalBox
   bIsVariable=true
   parent=ProgressSizeBox
   widget=ProgressVBox
   slot {class=SizeBoxSlot, properties {horizontalAlignment=HAlign_Center, verticalAlignment=VAlign_Center}}
  }
  Image_0 {
   widgetClassPath=&Image
   bIsVariable=true
   parent=ProgressVBox
   widget=Image_0
   properties {brush {
     bIsDynamicallyLoaded=false
     drawAs=Image
     imageType=NoImage
     mirroring=NoMirror
     resourceName=None
     resourceObject=/FortUILibrary/UIKit/Throbber/MI_UIKit_Monochrome_Throbber
     tiling=NoTile
     imageSize=<64 64>
     margin=<ltrb 0 0 0 0>
     outlineSettings {
      bUseBrushTransparency=false
      roundingType=HalfHeightRadius
      width=0
      color {colorUseRule=UseColor_Specified, specifiedColor=<rgba 0 0 0 0>}
      cornerRadii=<0 0 0 0>
     }
     tintColor {colorUseRule=UseColor_Specified, specifiedColor=<rgba 1 1 1 1>}
     uVRegion {bIsValid=false, max=<0 0>, min=<0 0>}
    }}
   slot {class=VerticalBoxSlot, properties {horizontalAlignment=HAlign_Center, verticalAlignment=VAlign_Center}}
  }
  ProgressText {
   widgetClassPath=&CommonTextBlock
   bIsVariable=true
   parent=ProgressVBox
   widget=ProgressText
   properties {
    style=/FortUILibrary/Styles/Typography/Foundation/TextStyle-Base-M-B_Blue.TextStyle-Base-M-B_Blue_C
    text=Redeeming...
    colorAndOpacity {colorUseRule=UseColor_Specified, specifiedColor=<rgba 0.008 0.011 0.0262 1>}
    font {
     bForceMonospaced=false
     bMaterialIsStencil=false
     fontMaterial=null
     fontObject=&NotoSans
     letterSpacing=0
     monospacedWidth=1
     size=24
     skewAmount=0
     typefaceFontName=NotoSans-Bold
     outlineSettings {
      bApplyOutlineToDropShadows=false
      bMiteredCorners=false
      bSeparateFillAlpha=false
      outlineMaterial=null
      outlineSize=0
      outlineColor=<rgba 0 0 0 1>
     }
    }
    shadowOffset=<0 0>
   }
   slot {class=VerticalBoxSlot}
  }
  VisibilityBorder_Touch {
   widgetClassPath=&CommonHardwareVisibilityBorder
   parent=Overlay_0
   widget=VisibilityBorder_Touch
   properties {
    horizontalAlignment=HAlign_Right
    verticalAlignment=VAlign_Top
    visibilityQuery {
     autoDescription=" ANY( Input.Touch )"
     tokenStreamVersion=0
     userDescription=if using touch
     queryTokenStream [0,1,1,1,0]
     tagDictionary [{tagName=Input.Touch}]
    }
   }
   slot {class=OverlaySlot, properties {horizontalAlignment=HAlign_Right}}
  }
  Button_Cancel {
   widgetClassPath=/FortUILibrary/Foundation/Buttons/CloseButton
   bIsVariable=true
   parent=VisibilityBorder_Touch
   widget=Button_Cancel
   properties {
    onButtonBaseClicked=(&self:WidgetTree.Button_Cancel.HandleOnButtonClicked)
    onButtonBaseDoubleClicked=(&self:WidgetTree.Button_Cancel.HandleOnButtonDoubleClicked)
    onButtonBaseHovered=(&self:WidgetTree.Button_Cancel.HandleOnButtonHovered)
    onButtonBaseUnhovered=(&self:WidgetTree.Button_Cancel.HandleOnButtonUnhovered)
    onSelectedChangedBase=(&self:WidgetTree.Button_Cancel.HandleOnSelectedChanged)
    renderTransform {angle=0, scale=<0.8 0.8>, shear=<0 0>, translation=<0 0>}
   }
   slot {class=BorderSlot, properties {horizontalAlignment=HAlign_Right, verticalAlignment=VAlign_Top, padding=<ltrb 0 0 0 0>}}
  }
 ]
}
