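# Sound cue for the whoosh on the V-Bucks tab of the front-end top bar, in the Fort_UI sound class.
# A single wave player feeds a modulator that randomizes pitch slightly downward and fixes volume at a reduced level.
# The cue's volume multiplier is zero, so it is silent as configured.
# It references the shared TopBarTabConcurrency set and carries concurrency override values with a stop-oldest rule.
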SoundCue /Game/Sounds/Fort_UI_Sounds/FrontEnd/Tabs/UI_Whoosh_TopBar_Tab_VBucks_Cue
root=Modulator_0
nodes [
 {id=Modulator_0, children [WavePlayer_0], properties {pitchMax=1, pitchMin=0.9, volumeMax=0.6, volumeMin=0.6}}
 {id=WavePlayer_0, properties {soundWaveAssetPtr=/Game/Sounds/Fort_UI_Sounds/FrontEnd/Tabs/wav/UI_Whoosh_TopBar_Tab_VBucks}}
]
properties {
 maxDistance=2097152
 soundClassObject=&Fort_UI
 volumeMultiplier=0
 concurrencyOverrides {bVolumeScaleCanRelease=true, maxCount=2, resolutionRule=StopOldest, retriggerTime=0.1}
 concurrencySet [&/TopBarTabConcurrency]
}
