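# V-Bucks store screen widget (child of FortVbucksStoreScreen_Legacy) listing purchasable V-Bucks offers in the frontend.
# Offer events switch a widget switcher between spinner, offer entry box, no-offers and purchasing-disabled panels, then stamp "% Extra!" banners from computed bonus percentages.
# Activation sets focus, camera and sound, toggles the pulsing store banner via runtime options, and refreshes the other-platform balance notice.
# PreConstruct applies a mobile layout.
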WidgetBlueprint /FortniteCoreUI/Frontend/StoreMain/VBucksStoreScreen
animations [Intro {
  binding_count=1
  length_seconds=0.4
  object=&self:Intro
  sequence {bindings [NewVBuckSection {
     id=#1
     display_name=CommonTextBlock_1
     possessed_class=VerticalBox
     tracks [
      {
       type=MovieSceneFloatTrack
       display_name=Render Opacity
       sections [{type=MovieSceneFloatSection, end_frame=8, start_frame=0, channels [None {type=MovieSceneScriptingFloatChannel, keys [[0,0],[7,1]]}]}]
      }
      {
       type=MovieScene2DTransformTrack
       sections [{
         type=MovieScene2DTransformSection
         end_frame=8
         start_frame=0
         channels [
          Translation.X {type=MovieSceneScriptingFloatChannel, keys [[0,-64],[7,0]]}
          Translation.Y {type=MovieSceneScriptingFloatChannel}
          Angle {type=MovieSceneScriptingFloatChannel}
          Scale.X {type=MovieSceneScriptingFloatChannel}
          Scale.Y {type=MovieSceneScriptingFloatChannel}
          Shear.X {type=MovieSceneScriptingFloatChannel}
          Shear.Y {type=MovieSceneScriptingFloatChannel}
         ]
        }]
      }
     ]
    }]}
 }]
graphs [
 ApplyMobileLayout {nodes [
   ApplyMobileLayout() {is=title K2Node_FunctionEntry}
   > Branch(Condition=IsMobileGame())
   n1 = SlotAsOverlaySlot(Widget=Get CurrencyNotificationBorder() -> (CurrencyNotificationBorder:Common Border Object Reference))
   > SetHorizontalAlignment(self=@n1, InHorizontalAlignment=HAlign_Center)
   > SetPadding!(
    execute
    self:Overlay Slot Object Reference=@n1
    InPadding_Left:float=0
    InPadding_Top:float=128
    InPadding_Right:float=464
    InPadding_Bottom:float=0
   ) -> (then)
   > SetVisibility(self=Get VBuckTitle() -> (VBuckTitle:Size Box Object Reference), InVisibility=Collapsed)
  ]}
 BP_GetDesiredFocusTarget {nodes [
   Get Desired Focus Target()
   > Return Node!(execute, ReturnValue:Widget Object Reference=Get InitiallyFocusedOffer() -> (InitiallyFocusedOffer:Widget Object Reference))
  ]}
 EventGraph {nodes [
   K2Node_Event_0 = OnTransitionedBySwitcher_1() {is=title K2Node_CustomEvent}
   > Branch(Condition=IsActivated!(self:Common Activatable Widget Object Reference) -> (ReturnValue:bool))
   > SetFocus(self=Get InitiallyFocusedOffer())
   Event OnPurchasingDisabled() {is=title K2Node_Event}
   > SetActiveWidget(
    self:Widget Switcher Object Reference=Get NewSwitcherItemsProgress() -> (NewSwitcherItemsProgress:Common Widget Switcher Legacy Object Reference)
    Widget=Get NewPurchasingDisabledBox() -> (NewPurchasingDisabledBox:Common Border Object Reference)
   )
   Event OnOffersGenerated() {is=title K2Node_Event}
   > SetActiveWidget(self=Get NewSwitcherItemsProgress(), Widget=Get NewVbuckOffers() -> (NewVbuckOffers:Vertical Box Object Reference))
   > SetOverrideBannerTexts()
   Event OnStartReadingOffers() {is=title K2Node_Event}
   > Parent: OnStartReadingOffers() {is=title K2Node_CallParentFunction}
   > SetActiveWidget(self=Get NewSwitcherItemsProgress(), Widget=Get NewSpinner() -> (NewSpinner:Image Object Reference))
   Event NoOffersAvailable() {is=title K2Node_Event}
   > SetActiveWidget(self=Get NewSwitcherItemsProgress(), Widget=Get VB_NoOffers() -> (VB_NoOffers:Vertical Box Object Reference))
   Event PreConstruct()
   > ApplyMobileLayout() {is=title K2Node_CallFunction}
   Event On Activated()
   > SetFocus(self=Get InitiallyFocusedOffer())
   > SpawnSound2D!(
    execute
    Sound:Sound Base Object Reference=Get Whoosh_ScreenAppear() -> (Whoosh_ScreenAppear:Sound Base Object Reference)
    VolumeMultiplier:float=1
    PitchMultiplier:float=1
    StartTime:float=0
    ConcurrencySettings:Sound Concurrency Object Reference
    bPersistAcrossLevelTransition:bool=false
    bAutoDestroy:bool=true
   ) -> (then, ReturnValue:Audio Component Object Reference)
   > SetFrontEndCamera!(
    execute
    self:Fort Front End Context Object Reference=Get FortFrontEndContext!() -> (ReturnValue:Fort Front End Context Object Reference) {is=title K2Node_GetBlueprintContextEx}
    NewState:EFrontEndCamera Enum=FrontendDefault
    TransitionParams:View Target Transition Params Structure=(BlendTime=0,BlendFunction=VTBlend_Cubic,BlendExp=2,bLockOutgoing=False)
   ) -> (then)
   > UpdateCurrencyNotification()
   n1 = Get bShowStoreBanner!(self:Fort Runtime Options Object Reference=GetRuntimeOptions()) -> (bShowStoreBanner:bool)
   > SetVisibility(
    self=Get StoreBanner() -> (StoreBanner:Store Banner Message Object Reference)
    InVisibility=Select(Option 0=Collapsed, Option 1=SelfHitTestInvisible, Index=@n1)
   )
   > Branch(Condition=@n1)
   > Pulse(self:Store Banner Message Object Reference=Cast To Store_BannerMessage!(Object=Get StoreBanner()) -> (AsStore Banner Message:Store Banner Message Object Reference, bSuccess:bool=~) {is=title K2Node_DynamicCast})
   Event OnUpdateOtherPlatformMTXMessage!() -> (OutputDelegate, then, HasOtherPlatformCurrency:bool, CurrencyMessageLocText:"Text (by ref)") {is=title K2Node_Event}
   > UpdateCurrencyNotification()
  ]}
 GetVisibilityByChildrenPresence {nodes [
   n1 = GetVisibilityByChildrenPresence!() -> (then, Panel:Panel Widget Object Reference) {is=title K2Node_FunctionEntry}
   > Return Node!(
    execute
    NewParam:ESlateVisibility Enum=Select(Option 0=Collapsed, Option 1=SelfHitTestInvisible, Index=HasAnyChildren!(self:Panel Widget Object Reference=@n1) -> (ReturnValue:bool))
   )
  ]}
 HandleAllOffersGenerated {nodes [HandleAllOffersGenerated() {is=title K2Node_FunctionEntry}]}
 HandleBack {nodes [
   HandleBack() {is=title K2Node_FunctionEntry}
   > PopContentWidget!(execute) -> (then)
   > Return Node!(execute, PassThrough:bool=false)
  ]}
 HandleRedeem {nodes [
   HandleRedeem() {is=title K2Node_FunctionEntry}
   > n1 = Create Widget!(
    execute:Exec
    Class:User Widget Class Reference=/FortniteCoreUI/UI/Foundation/Modals/RedeemCodeWindow/RedeemFriendCodeWindow.RedeemFriendCodeWindow_C
    OwningPlayer:Player Controller Object Reference=GetOwningPlayer()
   ) -> (then:Exec, ReturnValue:Redeem Friend Code Window Object Reference)
   > TryPushWidgetAsModalPanel!(execute, PushWidget:User Widget Object Reference=@n1) -> (then)
   > Return Node!(execute, bPassThrough:bool=false)
  ]}
 SetOverrideBannerTexts {nodes [
   SetOverrideBannerTexts() {is=title K2Node_FunctionEntry}
   > n1 = For Each Loop(Array:Array of User Widget Object References=GetAllEntries!(self:Dynamic Entry Box Base Object Reference=Get EntryBox_VBuckOffers() -> (EntryBox_VBuckOffers:Dynamic Entry Box Object Reference)) -> (ReturnValue:Array of User Widget Object References)) -> (Array Element:User Widget Object Reference)
   > n2 = Cast To StoreMain_MTXOffer!(execute, Object=@n1) -> (then, CastFailed, AsStore Main MTXOffer:Store Main MTXOffer Object Reference) {is=title K2Node_DynamicCast}
   n3 = GetOfferInfo!(self:Fort Store Front Offer Widget Base Object Reference=@n2) -> (ReturnValue:Fort Store Front Offer Info Object Reference)
   n4 = Round(A=float * float(
     A=CalculateBonusPercent!(self:Fort Vbucks Store Screen Legacy Object Reference, OfferData:Fort Store Front Offer Info Object Reference=@n3) -> (ReturnValue:float)
     B=100
    ) {is=title K2Node_PromotableOperator}) {is=title K2Node_CallFunction}
   > SetOverrideBannerText!(
    execute
    self:Store Main MTXOffer Object Reference=@n2
    InOverrideBannerText:Text=Select(Option 0:Text, Option 1:Text=Format Text!(Format="LOCTEXT(\"{0}% Extra!\")", 0:int=@n4) -> (Result), Index="integer > integer"(A=@n4)) -> (ReturnValue:Text)
   ) -> (then)
  ]}
 UpdateCurrencyNotification {nodes [
   UpdateCurrencyNotification() {is=title K2Node_FunctionEntry}
   n1 = Break Mtx Breakdown!(MtxBreakdown:"Mtx Breakdown Structure (by ref)"=GetMtxBreakdown!(self:Fort Mtx Store Root Base Object Reference) -> (ReturnValue:Mtx Breakdown Structure)) -> (AvailableTotalMtx:int, AvailablePremiumMtx:int, UnavailableTotalMtx:int, UnavailablePremiumMtx:int) {is=title K2Node_BreakStruct}
   > SetVisibility(
    self=Get CurrencyNotificationBorder()
    InVisibility=Select(
     Option 0:ESlateVisibility Enum=Collapsed
     Option 1:ESlateVisibility Enum=SelfHitTestInvisible
     Index="integer > integer"(A=@n1.UnavailableTotalMtx, B=0) {is=title K2Node_CallFunction}
    ) -> (ReturnValue:ESlateVisibility Enum)
   )
   > "SetText (Text)"(
    self=Get CurrencyNotificationText() -> (CurrencyNotificationText:Common Text Object Reference)
    InText=Format Text!(Format="LOCTEXT(\"You have {Amount} V-Bucks on other platforms.\")", Amount:int=@n1.UnavailableTotalMtx) -> (Result)
   )
  ]}
]
variables [
 Ambient Sound {type=AudioComponent, default=null}
 bShowingCampaignUpsell {type=boolean}
 InitiallyFocusedOffer {type=Widget, default=null}
 Whoosh_ScreenAppear {type=SoundBase, default=/Game/Sounds/Fort_UI_Sounds/FrontEnd/Tabs/UI_Whoosh_TopBar_Tab_VBucks_Cue}
]
widgets {
 info {inheritedWidgetCount=3, parentClass=/Script/FortniteUI.FortVbucksStoreScreen_Legacy, rootWidgetClass=&SafeZone, widgetCount=36}
 widgets [
  SafeZone_0 {widgetClassPath=&SafeZone, widget=SafeZone_0, properties {padBottom=false, padLeft=false, padRight=false}}
  Overlay_1 {widgetClassPath=&Overlay, parent=SafeZone_0, widget=Overlay_1, slot {class=SafeZoneSlot}}
  CurrencyNotificationBorder {
   widgetClassPath=&CommonBorder
   bIsVariable=true
   parent=Overlay_1
   widget=CurrencyNotificationBorder
   properties {
    style=/FortUILibrary/Styles/Border/Border_Solid_DkBlue.Border_Solid_DkBlue_C
    visibility=Collapsed
    brushColor=<rgba 1 1 1 0.2>
    padding=<ltrb 4 4 4 4>
   }
   slot {class=OverlaySlot, properties {horizontalAlignment=HAlign_Right}}
  }
  VerticalBox_1 {
   widgetClassPath=&VerticalBox
   parent=CurrencyNotificationBorder
   widget=VerticalBox_1
   slot {class=BorderSlot, properties {padding=<ltrb 4 4 4 4>}}
  }
  CurrencyNotificationText {
   widgetClassPath=&CommonTextBlock
   bIsVariable=true
   parent=VerticalBox_1
   widget=CurrencyNotificationText
   properties {
    style=/FortUILibrary/Styles/Typography/Foundation/TextStyle-Base-S-B.TextStyle-Base-S-B_C
    text="You have ###### V-Bucks on other platforms."
    font {
     bForceMonospaced=false
     bMaterialIsStencil=false
     fontMaterial=null
     fontObject=&NotoSans
     letterSpacing=0
     monospacedWidth=1
     size=18
     skewAmount=0
     typefaceFontName=NotoSans-Bold
     outlineSettings {
      bApplyOutlineToDropShadows=false
      bMiteredCorners=false
      bSeparateFillAlpha=false
      outlineMaterial=null
      outlineSize=0
      outlineColor=<rgba 0 0 0 1>
     }
    }
    shadowOffset=<0 0>
   }
   slot {class=VerticalBoxSlot, properties {horizontalAlignment=HAlign_Right}}
  }
  StoreInvalidationBox {
   widgetClassPath=/Script/UMG.InvalidationBox
   parent=Overlay_1
   widget=StoreInvalidationBox
   slot {class=OverlaySlot, properties {horizontalAlignment=HAlign_Fill, verticalAlignment=VAlign_Fill}}
  }
  OfferScaleBoxRoot {
   widgetClassPath=&ScaleBox
   bIsVariable=true
   parent=StoreInvalidationBox
   widget=OfferScaleBoxRoot
   properties {stretchDirection=DownOnly}
   slot {class=PanelSlot}
  }
  Overlay_2 {
   widgetClassPath=&Overlay
   parent=OfferScaleBoxRoot
   widget=Overlay_2
   slot {class=ScaleBoxSlot, properties {horizontalAlignment=HAlign_Fill}}
  }
  VerticalBox_0 {
   widgetClassPath=&VerticalBox
   parent=Overlay_2
   widget=VerticalBox_0
   slot {class=OverlaySlot, properties {horizontalAlignment=HAlign_Fill, verticalAlignment=VAlign_Fill}}
  }
  NewSwitcherItemsProgress {
   widgetClassPath=&CommonWidgetSwitcherLegacy
   bIsVariable=true
   parent=VerticalBox_0
   widget=NewSwitcherItemsProgress
   properties {transitionDuration=0}
   slot {
    class=VerticalBoxSlot
    properties {horizontalAlignment=HAlign_Center, verticalAlignment=VAlign_Top, padding=<ltrb 32 0 32 32>, size {sizeRule=Fill, value=1}}
   }
  }
  NewVbuckOffers {
   widgetClassPath=&VerticalBox
   bIsVariable=true
   parent=NewSwitcherItemsProgress
   widget=NewVbuckOffers
   slot {class=WidgetSwitcherSlot, properties {horizontalAlignment=HAlign_Center, verticalAlignment=VAlign_Center}}
  }
  SizeBox_3 {
   widgetClassPath=&SizeBox
   parent=NewVbuckOffers
   widget=SizeBox_3
   properties {bOverride_HeightOverride=true, heightOverride=32}
   slot {class=VerticalBoxSlot}
  }
  HorizontalBox_0 {
   widgetClassPath=&HorizontalBox
   bIsVariable=true
   parent=NewVbuckOffers
   widget=HorizontalBox_0
   properties {
    flowDirectionPreference=LeftToRight
    navigation=&self:WidgetTree.HorizontalBox_0.WidgetNavigation_0
   }
   slot {class=VerticalBoxSlot}
  }
  NewVBuckSection {
   widgetClassPath=&VerticalBox
   bIsVariable=true
   parent=HorizontalBox_0
   widget=NewVBuckSection
   slot {class=HorizontalBoxSlot, properties {horizontalAlignment=HAlign_Left, verticalAlignment=VAlign_Center, size {sizeRule=Fill, value=1}}}
  }
  VBuckTitle {
   widgetClassPath=&SizeBox
   bIsVariable=true
   parent=NewVBuckSection
   widget=VBuckTitle
   properties {flowDirectionPreference=Culture}
   slot {class=VerticalBoxSlot, properties {padding=<ltrb 8 0 8 0>}}
  }
  VerticalBox_3 {widgetClassPath=&VerticalBox, parent=VBuckTitle, widget=VerticalBox_3, slot {class=SizeBoxSlot}}
  MirrorSubHeaderHeightOverride {
   widgetClassPath=&Spacer
   parent=VerticalBox_3
   widget=MirrorSubHeaderHeightOverride
   properties {size=<1 42>}
   slot {class=VerticalBoxSlot, properties {horizontalAlignment=HAlign_Left}}
  }
  HorizontalBox_3 {
   widgetClassPath=&HorizontalBox
   parent=VerticalBox_3
   widget=HorizontalBox_3
   slot {class=VerticalBoxSlot, properties {padding=<ltrb 0 -12 0 0>}}
  }
  CommonTextBlock_1 {
   widgetClassPath=&CommonTextBlock
   parent=HorizontalBox_3
   widget=CommonTextBlock_1
   properties {
    style=/FortUILibrary/Styles/Typography/_TextStyle_PrimaryH2._TextStyle_PrimaryH2_C
    text=V-BUCKS
    textTransformPolicy=ToUpper
    font {
     bForceMonospaced=false
     bMaterialIsStencil=false
     fontMaterial=null
     fontObject=&Burbank
     letterSpacing=0
     monospacedWidth=1
     size=48
     skewAmount=0
     typefaceFontName=BurbankRegular-Black
     outlineSettings {
      bApplyOutlineToDropShadows=false
      bMiteredCorners=false
      bSeparateFillAlpha=false
      outlineMaterial=null
      outlineSize=0
      outlineColor=<rgba 0 0 0 1>
     }
    }
    shadowOffset=<0 0>
    strikeBrush {
     bIsDynamicallyLoaded=false
     drawAs=NoDrawType
     imageType=NoImage
     mirroring=NoMirror
     resourceName=None
     resourceObject=null
     tiling=NoTile
     imageSize=<32 32>
     margin=<ltrb 0 0 0 0>
     outlineSettings {
      bUseBrushTransparency=false
      roundingType=HalfHeightRadius
      width=0
      color {colorUseRule=UseColor_Specified, specifiedColor=<rgba 0 0 0 0>}
      cornerRadii=<0 0 0 0>
     }
     tintColor {colorUseRule=UseColor_Specified, specifiedColor=<rgba 1 1 1 1>}
     uVRegion {bIsValid=false, max=<0 0>, min=<0 0>}
    }
   }
   slot {class=HorizontalBoxSlot, properties {horizontalAlignment=HAlign_Left, verticalAlignment=VAlign_Center}}
  }
  Spacer_2 {
   widgetClassPath=&Spacer
   parent=HorizontalBox_3
   widget=Spacer_2
   slot {
    class=HorizontalBoxSlot
    properties {horizontalAlignment=HAlign_Right, verticalAlignment=VAlign_Center, padding=<ltrb 16 0 16 0>, size {sizeRule=Fill, value=1}}
   }
  }
  InfoFlag {
   widgetClassPath=&/InfoFlag
   bIsVariable=true
   parent=NewVBuckSection
   widget=InfoFlag
   properties {gradientDistance=0.35, infoText=TBD_Blurb of text to learn more, maxWidth=1275, trimWidth=10}
   slot {class=VerticalBoxSlot, properties {horizontalAlignment=HAlign_Left, padding=<ltrb 9 0 0 0>}}
  }
  Border_0 {
   widgetClassPath=&Border
   parent=NewVBuckSection
   widget=Border_0
   properties {
    flowDirectionPreference=LeftToRight
    background {
     bIsDynamicallyLoaded=false
     drawAs=NoDrawType
     imageType=NoImage
     mirroring=NoMirror
     resourceName=None
     resourceObject=null
     tiling=NoTile
     imageSize=<32 32>
     margin=<ltrb 0 0 0 0>
     outlineSettings {
      bUseBrushTransparency=false
      roundingType=HalfHeightRadius
      width=0
      color {colorUseRule=UseColor_Specified, specifiedColor=<rgba 0 0 0 0>}
      cornerRadii=<0 0 0 0>
     }
     tintColor {colorUseRule=UseColor_Specified, specifiedColor=<rgba 1 1 1 1>}
     uVRegion {bIsValid=false, max=<0 0>, min=<0 0>}
    }
    padding=<ltrb 0 0 0 0>
   }
   slot {class=VerticalBoxSlot, properties {padding=<ltrb 0 12 0 12>}}
  }
  EntryBox_VBuckOffers {
   widgetClassPath=&DynamicEntryBox
   bInherited=true
   bIsVariable=true
   parent=Border_0
   widget=EntryBox_VBuckOffers
   properties {entryWidgetClass=&/StoreMain_MTXOffer.StoreMain_MTXOffer_C, numDesignerPreviewEntries=4}
   slot {class=BorderSlot, properties {padding=<ltrb 0 0 0 0>}}
  }
  Text_OfferMessage {
   widgetClassPath=&CommonRichTextBlock
   bInherited=true
   bIsVariable=true
   parent=NewVbuckOffers
   widget=Text_OfferMessage
   properties {
    bOverrideDefaultStyle=true
    justification=Center
    text=Offer Message
    visibility=Hidden
    wrapTextAt=1400
    defaultTextStyleOverride {
     overflowPolicy=Clip
     transformPolicy=None
     colorAndOpacity {colorUseRule=UseColor_Specified, specifiedColor=<rgba 1 1 1 1>}
     font {
      bForceMonospaced=false
      bMaterialIsStencil=false
      fontMaterial=null
      fontObject=&BurbankSmall
      letterSpacing=0
      monospacedWidth=1
      size=16
      skewAmount=0
      typefaceFontName=Bold
      outlineSettings {
       bApplyOutlineToDropShadows=false
       bMiteredCorners=false
       bSeparateFillAlpha=false
       outlineMaterial=null
       outlineSize=0
       outlineColor=<rgba 0 0 0 1>
      }
     }
     highlightColor {colorUseRule=UseColor_Specified, specifiedColor=<rgba 0 0 0 1>}
     highlightShape {
      bIsDynamicallyLoaded=false
      drawAs=Image
      imageType=NoImage
      mirroring=NoMirror
      resourceName=None
      resourceObject=null
      tiling=NoTile
      imageSize=<32 32>
      margin=<ltrb 0 0 0 0>
      outlineSettings {
       bUseBrushTransparency=false
       roundingType=HalfHeightRadius
       width=0
       color {colorUseRule=UseColor_Specified, specifiedColor=<rgba 0 0 0 0>}
       cornerRadii=<0 0 0 0>
      }
      tintColor {colorUseRule=UseColor_Specified, specifiedColor=<rgba 1 1 1 1>}
      uVRegion {bIsValid=false, max=<0 0>, min=<0 0>}
     }
     selectedBackgroundColor {colorUseRule=UseColor_Foreground, specifiedColor=<rgba 1 0 1 1>}
     shadowColorAndOpacity=<rgba 0 0 0 1>
     shadowOffset=<0 0>
     strikeBrush {
      bIsDynamicallyLoaded=false
      drawAs=Image
      imageType=NoImage
      mirroring=NoMirror
      resourceName=None
      resourceObject=null
      tiling=NoTile
      imageSize=<32 32>
      margin=<ltrb 0 0 0 0>
      outlineSettings {
       bUseBrushTransparency=false
       roundingType=HalfHeightRadius
       width=0
       color {colorUseRule=UseColor_Specified, specifiedColor=<rgba 0 0 0 0>}
       cornerRadii=<0 0 0 0>
      }
      tintColor {colorUseRule=UseColor_Specified, specifiedColor=<rgba 1 1 1 1>}
      uVRegion {bIsValid=false, max=<0 0>, min=<0 0>}
     }
     underlineBrush {
      bIsDynamicallyLoaded=false
      drawAs=Image
      imageType=NoImage
      mirroring=NoMirror
      resourceName=None
      resourceObject=null
      tiling=NoTile
      imageSize=<32 32>
      margin=<ltrb 0 0 0 0>
      outlineSettings {
       bUseBrushTransparency=false
       roundingType=HalfHeightRadius
       width=0
       color {colorUseRule=UseColor_Specified, specifiedColor=<rgba 0 0 0 0>}
       cornerRadii=<0 0 0 0>
      }
      tintColor {colorUseRule=UseColor_Specified, specifiedColor=<rgba 1 1 1 1>}
      uVRegion {bIsValid=false, max=<0 0>, min=<0 0>}
     }
    }
   }
   slot {class=VerticalBoxSlot, properties {horizontalAlignment=HAlign_Center, padding=<ltrb 4 2 4 2>}}
  }
  Overlay_0 {
   widgetClassPath=&Overlay
   bIsVariable=true
   parent=NewSwitcherItemsProgress
   widget=Overlay_0
   slot {class=WidgetSwitcherSlot, properties {horizontalAlignment=HAlign_Center, verticalAlignment=VAlign_Center}}
  }
  NewSpinner {
   widgetClassPath=&Image
   bIsVariable=true
   parent=Overlay_0
   widget=NewSpinner
   properties {
    brush {
     bIsDynamicallyLoaded=false
     drawAs=Image
     imageType=NoImage
     mirroring=NoMirror
     resourceName=None
     resourceObject=/FortUILibrary/UIKit/Throbber/MI_UIKit_Monochrome_Throbber
     tiling=NoTile
     imageSize=<128 128>
     margin=<ltrb 0 0 0 0>
     outlineSettings {
      bUseBrushTransparency=false
      roundingType=HalfHeightRadius
      width=0
      color {colorUseRule=UseColor_Specified, specifiedColor=<rgba 0 0 0 0>}
      cornerRadii=<0 0 0 0>
     }
     tintColor {colorUseRule=UseColor_Specified, specifiedColor=<rgba 1 1 1 1>}
     uVRegion {bIsValid=false, max=<0 0>, min=<0 0>}
    }
    colorAndOpacity=<rgba 1 1 1 0.7>
   }
   slot {class=OverlaySlot, properties {padding=<ltrb 256 64 256 64>}}
  }
  VB_NoOffers {
   widgetClassPath=&VerticalBox
   bIsVariable=true
   parent=NewSwitcherItemsProgress
   widget=VB_NoOffers
   slot {class=WidgetSwitcherSlot, properties {horizontalAlignment=HAlign_Center, verticalAlignment=VAlign_Center}}
  }
  NewNoOffersBox {
   widgetClassPath=&CommonBorder
   bIsVariable=true
   parent=VB_NoOffers
   widget=NewNoOffersBox
   properties {
    horizontalAlignment=HAlign_Center
    style=/FortUILibrary/Styles/Border/Border_StatRow_WhiteLowOpacity.Border_StatRow_WhiteLowOpacity_C
    padding=<ltrb 0 0 0 0>
   }
   slot {class=VerticalBoxSlot, properties {horizontalAlignment=HAlign_Center}}
  }
  SizeBox_10 {
   widgetClassPath=&SizeBox
   parent=NewNoOffersBox
   widget=SizeBox_10
   properties {bOverride_HeightOverride=true, heightOverride=48}
   slot {class=BorderSlot, properties {horizontalAlignment=HAlign_Center, padding=<ltrb 0 0 0 0>}}
  }
  CommonTextBlock_5 {
   widgetClassPath=&CommonTextBlock
   parent=SizeBox_10
   widget=CommonTextBlock_5
   properties {
    style=/FortUILibrary/Styles/Typography/Foundation/TextStyle-Base-XS-B-70pc.TextStyle-Base-XS-B-70pc_C
    text=No Offers Available
    colorAndOpacity {colorUseRule=UseColor_Specified, specifiedColor=<rgba 1 1 1 0.7>}
    font {
     bForceMonospaced=false
     bMaterialIsStencil=false
     fontMaterial=null
     fontObject=&NotoSans
     letterSpacing=0
     monospacedWidth=1
     size=16
     skewAmount=0
     typefaceFontName=NotoSans-Bold
     outlineSettings {
      bApplyOutlineToDropShadows=false
      bMiteredCorners=false
      bSeparateFillAlpha=false
      outlineMaterial=null
      outlineSize=0
      outlineColor=<rgba 0 0 0 1>
     }
    }
    shadowOffset=<0 0>
   }
   slot {class=SizeBoxSlot, properties {padding=<ltrb 32 8 32 8>}}
  }
  NewPurchasingDisabledBox {
   widgetClassPath=&CommonBorder
   bIsVariable=true
   parent=NewSwitcherItemsProgress
   widget=NewPurchasingDisabledBox
   properties {
    horizontalAlignment=HAlign_Center
    style=&Border_StatRow_WhiteLowOpacity_C
    padding=<ltrb 0 0 0 0>
   }
   slot {class=WidgetSwitcherSlot, properties {horizontalAlignment=HAlign_Center, verticalAlignment=VAlign_Center}}
  }
  SizeBox_11 {
   widgetClassPath=&SizeBox
   parent=NewPurchasingDisabledBox
   widget=SizeBox_11
   properties {bOverride_HeightOverride=true, heightOverride=48}
   slot {class=BorderSlot, properties {horizontalAlignment=HAlign_Center, padding=<ltrb 0 0 0 0>}}
  }
  CommonTextBlock_6 {
   widgetClassPath=&CommonTextBlock
   parent=SizeBox_11
   widget=CommonTextBlock_6
   properties {
    style=&TextStyle-Base-XS-B-70pc_C
    text=Purchasing Temporarily Unavailable
    colorAndOpacity {colorUseRule=UseColor_Specified, specifiedColor=<rgba 1 1 1 0.7>}
    font {
     bForceMonospaced=false
     bMaterialIsStencil=false
     fontMaterial=null
     fontObject=&NotoSans
     letterSpacing=0
     monospacedWidth=1
     size=16
     skewAmount=0
     typefaceFontName=NotoSans-Bold
     outlineSettings {
      bApplyOutlineToDropShadows=false
      bMiteredCorners=false
      bSeparateFillAlpha=false
      outlineMaterial=null
      outlineSize=0
      outlineColor=<rgba 0 0 0 1>
     }
    }
    shadowOffset=<0 0>
   }
   slot {class=SizeBoxSlot, properties {padding=<ltrb 32 8 32 8>}}
  }
  NamedSlot_ExtraContentBottom {
   widgetClassPath=&NamedSlot
   bIsVariable=true
   parent=VerticalBox_0
   widget=NamedSlot_ExtraContentBottom
   properties {bExposeOnInstanceOnly=true}
   slot {class=VerticalBoxSlot}
  }
  StoreBanner {
   widgetClassPath=&/Store_BannerMessage
   bIsVariable=true
   parent=Overlay_2
   widget=StoreBanner
   properties {visibility=Collapsed}
   slot {class=OverlaySlot, properties {horizontalAlignment=HAlign_Right, padding=<ltrb 0 20 70 0>}}
  }
  LoadingSpinner {
   widgetClassPath=&Image
   bInherited=true
   parent=Overlay_1
   widget=LoadingSpinner
   properties {
    visibility=SelfHitTestInvisible
    brush {
     bIsDynamicallyLoaded=false
     drawAs=Image
     imageType=NoImage
     mirroring=NoMirror
     resourceName=None
     resourceObject=/Game/UI/ContentOnDemand/ItemDownloadIndicator/MI_ItemDownloadIndicator_SpinnerMaterial
     tiling=NoTile
     imageSize=<100 100>
     margin=<ltrb 0 0 0 0>
     outlineSettings {
      bUseBrushTransparency=false
      roundingType=HalfHeightRadius
      width=0
      color {colorUseRule=UseColor_Specified, specifiedColor=<rgba 0 0 0 0>}
      cornerRadii=<0 0 0 0>
     }
     tintColor {colorUseRule=UseColor_Specified, specifiedColor=<rgba 1 1 1 1>}
     uVRegion {bIsValid=false, max=<0 0>, min=<0 0>}
    }
   }
   slot {class=OverlaySlot, properties {horizontalAlignment=HAlign_Center, verticalAlignment=VAlign_Center}}
  }
 ]
}
